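; Blueprint for the Akimbo Envoy pickaxe cosmetic, adding a reactive material effect on weak spot hits.
; The impact FX event calls the parent, then triggers reactivity on weak-spot surface types, or on every frontend or preview hit.
; A server-to-client event chain shares the effect with other clients, restarting a timeline that drives the weapon mesh's "Reactivity Alpha" material parameter.
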(event Custom|CE_FX_Server
  ;[1 0.2033 0.1435] Making sure the reactivity can be seen by other clients
  (CallFunction|CEFXClient))

(event FortWeapon|EventOnPlayImpactFX
  (params (HitResult :type Hit%20Result%20Structure%20%28by%20ref%29) (ImpactPhysicalSurface :type EPhysicalSurface%20Enum :default "SurfaceType_Default") (SpawnedPSC :type FXSystem%20Component%20Object%20Reference))
  (|CallParentFunction HitResult ImpactPhysicalSurface SpawnedPSC)
  (Utilities|FlowControl|Sequence
    (:then_0
      ;[1 0.2263 0.8277] Weak Spot Reactivity Setup
      ;;[1 0.7738 0.1235] Checking for Weak Spot Hits
      (if (Math|Boolean|ORBoolean (Utilities|Enum|Equal(Enum) ImpactPhysicalSurface "SurfaceType12") (Utilities|Enum|Equal(Enum) ImpactPhysicalSurface "SurfaceType14") (Utilities|Enum|Equal(Enum) ImpactPhysicalSurface "SurfaceType15") (Utilities|Enum|Equal(Enum) ImpactPhysicalSurface "SurfaceType16"))
        (CallFunction|CEFXServer)))
    (:then_1
      ;;[0.3228 0.31 1] Calls the reactivity on FrontEnd. Every hit there should be reactive.
      (if (or (Reactive|Frontend|IsCosmeticObjectDisplayedasPreview 0) (Reactive|Frontend|IsCosmeticObjectDisplayedinFrontendContext 0))
        (CallFunction|CEFXServer)))))

(event Custom|CE_FX_Client
  ;[1 0.2033 0.1435] Making sure the reactivity can be seen by other clients
  (CallFunction|ReactivityEvent))

(event Custom|ReactivityEvent
  (bind (_direction _reactivityalpha) (Timeline|WeakSpotAnimation PlayFromStart)
    (:Update
      (Rendering|Material|SetScalarParameterValue (Class|BTechArtGlobalLibrary|GetDynamicMaterialInstance (Variables|Visual|GetWeaponMesh)) "Reactivity Alpha" _reactivityalpha :!class /Script/Engine.MaterialInstanceDynamic))))

(fn ConstructionScript
  (|CallParentFunction))
