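; Arctic Ice Blue pickaxe Blueprint that adds special impact effects for weak-point hits.
; After parent impact logic, it flags weak points by checking the physical surface against a list, always true in frontend display, and feeds the impact FX.
; In game, the surface also passes through server-to-client custom events to update the flag.
; Weak-point hits pulse the idle VFX's PlayTimeline briefly, then spawn a Niagara effect on the weapon mesh.
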(event FortWeapon|EventOnPlayImpactFX
  (params (HitResult :type Hit%20Result%20Structure%20%28by%20ref%29) (ImpactPhysicalSurface :type EPhysicalSurface%20Enum :default "SurfaceType_Default") (SpawnedPSC :type FXSystem%20Component%20Object%20Reference))
  (bind _returnvalue (CallFunction|GetIdleVFXComponent))
  (|CallParentFunction HitResult ImpactPhysicalSurface SpawnedPSC)
  (if (Reactive|Frontend|IsCosmeticObjectDisplayedinFrontendContext 0)
    ;[1 1 1] WeakPoint impact
    (bind _output_get (Variables|Default|SetIs%20hit%20weak%20point true))
    (Effects|Components|ParticleSystem|SetBoolParameter (Variables|Default|GetImpact%20FX) "Is WeakPoint" _output_get)
    (else
      ;[1 1 1] WeakPoint impact
      (bind _output_get_2 (Variables|Default|SetCurrent%20Impact%20Physical%20Surface ImpactPhysicalSurface))
      (bind _output_get_3 (Variables|Default|SetIs%20hit%20weak%20point (Utilities|Array|ContainsItem (Variables|Default|GetWeakPoint%20Physical%20Surface) _output_get_2)))
      (Effects|Components|ParticleSystem|SetBoolParameter (Variables|Default|GetImpact%20FX) "Is WeakPoint" _output_get_3)
      (CallFunction|CEFXServer ImpactPhysicalSurface)))
  (Effects|Components|ParticleSystem|SetBoolParameter _returnvalue "PlayTimeline" (Variables|Default|GetIs%20hit%20weak%20point))
  (Utilities|FlowControl|Delay 0.1)
  (Effects|Components|ParticleSystem|SetBoolParameter _returnvalue "PlayTimeline" false)
  (if (Variables|Default|GetIs%20hit%20weak%20point)
    (Effects|SpawnFXSystemAttached :FXSystemAsset "/BRCosmetics/Weapons/FORT_Melee/PIckaxe_ArcticIceBlue/FX/NS_ArcticIceBlue_IdleImpact.NS_ArcticIceBlue_IdleImpact" :AttachToComponent (Variables|Visual|GetWeaponMesh) :AttachPointName "b_AR_Root" :Rotation "0.000000,-90.000000,0.000000" :Scale "1.000000,1.000000,1.000000" :LocationType "SnapToTarget" :bAutoDestroy true :PoolingMethod "None" :bAutoActivate true)))

(event Custom|CE_FX_Server
  (params (Impact%20Physical%20Surface :type EPhysicalSurface%20Enum :default "SurfaceType_Default"))
  (CallFunction|CEFXClient Impact%20Physical%20Surface))

(event Custom|CE_FX_Client
  (params (Impact%20Physical%20Surface :type EPhysicalSurface%20Enum :default "SurfaceType_Default"))
  (bind _output_get (Variables|Default|SetCurrent%20Impact%20Physical%20Surface Impact%20Physical%20Surface))
  (Variables|Default|SetIs%20hit%20weak%20point (Utilities|Array|ContainsItem (Variables|Default|GetWeakPoint%20Physical%20Surface) _output_get)))

(fn OnRep_CurrentImpactPhysicalSurface)

(fn ConstructionScript
  (|CallParentFunction))
