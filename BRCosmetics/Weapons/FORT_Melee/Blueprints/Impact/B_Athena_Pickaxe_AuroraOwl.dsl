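; Reactive AuroraOwl pickaxe cosmetic that flares its idle and impact Niagara effects on kills or weak-spot hits.
; Init, BeginPlay and visibility events cache and deactivate idle FX, zero ReactivityTime, and trigger a delayed reaction in frontend.
; Kills and certain surface hits outside frontend call a server-to-client event chain so other players see the reaction.
; ReactivityEvent sets IsReactive on impact FX, then through DoOnce reinitializes idle FX and drives ReactivityTime from a timeline.
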(event EventOnInitCosmeticAlterations
  (params (CosmeticMod :type Fort%20Cosmetic%20Modification%20Structure))
  (|CallParentFunction CosmeticMod)
  (bind _asniagara_particle_system_component (Utilities|Casting|CastToNiagaraComponent (CallFunction|GetIdleVFXComponent)))
  (bind _output_get (Variables|Default|SetSpawned%20IdleFX _asniagara_particle_system_component))
  (Components|Activation|Deactivate _output_get)
  (Effects|Components|ParticleSystem|SetFloatParameter (CallFunction|GetIdleVFXComponent) "ReactivityTime" 0.0)
  (if (Reactive|Frontend|IsCosmeticObjectDisplayedinFrontendContext 0)
    (Utilities|FlowControl|Delay 5.0)
    (Utilities|FlowControl|DoOnce false
      (:Completed
        (CallFunction|ReactivityEvent)
        (Utilities|FlowControl|Delay 1e+01)
        (Utilities|FlowControl|Sequence
          (:then_0
            _)
          (:then_1
            _))))))

(event EventBeginPlay
  (|CallParentFunction)
  (bind _asniagara_particle_system_component (Utilities|Casting|CastToNiagaraComponent (CallFunction|GetIdleVFXComponent)))
  (bind _output_get (Variables|Default|SetSpawned%20IdleFX _asniagara_particle_system_component))
  (Components|Activation|Deactivate _output_get)
  (Effects|Components|ParticleSystem|SetFloatParameter (CallFunction|GetIdleVFXComponent) "ReactivityTime" 0.0)
  (if (Reactive|Frontend|IsCosmeticObjectDisplayedinFrontendContext 0)
    (Utilities|FlowControl|Delay 5.0)
    (Utilities|FlowControl|DoOnce false
      (:Completed
        (CallFunction|ReactivityEvent)
        (Utilities|FlowControl|Delay 1e+01)
        (Utilities|FlowControl|Sequence
          (:then_0
            _)
          (:then_1
            _))))))

(event EventOnWeaponVisibilityChanged
  (params (bVisible :type Boolean) (bSetForLocalControllerOnly :type Boolean))
  (|CallParentFunction bVisible bSetForLocalControllerOnly)
  (bind _asniagara_particle_system_component (Utilities|Casting|CastToNiagaraComponent (CallFunction|GetIdleVFXComponent)))
  (bind _output_get (Variables|Default|SetSpawned%20IdleFX _asniagara_particle_system_component))
  (Components|Activation|Deactivate _output_get)
  (Effects|Components|ParticleSystem|SetFloatParameter (CallFunction|GetIdleVFXComponent) "ReactivityTime" 0.0)
  (if (Reactive|Frontend|IsCosmeticObjectDisplayedinFrontendContext 0)
    (Utilities|FlowControl|Delay 5.0)
    (Utilities|FlowControl|DoOnce false
      (:Completed
        (CallFunction|ReactivityEvent)
        (Utilities|FlowControl|Delay 1e+01)
        (Utilities|FlowControl|Sequence
          (:then_0
            _)
          (:then_1
            _))))))

(event EventOnInitWeaponCosmetics
  (|CallParentFunction)
  (bind _asniagara_particle_system_component (Utilities|Casting|CastToNiagaraComponent (CallFunction|GetIdleVFXComponent)))
  (bind _output_get (Variables|Default|SetSpawned%20IdleFX _asniagara_particle_system_component))
  (Components|Activation|Deactivate _output_get)
  (Effects|Components|ParticleSystem|SetFloatParameter (CallFunction|GetIdleVFXComponent) "ReactivityTime" 0.0)
  (if (Reactive|Frontend|IsCosmeticObjectDisplayedinFrontendContext 0)
    (Utilities|FlowControl|Delay 5.0)
    (Utilities|FlowControl|DoOnce false
      (:Completed
        (CallFunction|ReactivityEvent)
        (Utilities|FlowControl|Delay 1e+01)
        (Utilities|FlowControl|Sequence
          (:then_0
            _)
          (:then_1
            _))))))

(event EventUpdateBasedOnKills
  (params (Watched%20Kills :type Integer))
  ; Reactivity On kills
  (|CallParentFunction Watched%20Kills)
  (if (> Watched%20Kills 0)
    (CallFunction|CEFXServer) ; It calls the Server Event, that calls the Client Event, \nthat finally calls the reactivity event
  ))

(event FortWeapon|EventOnPlayImpactFX
  (params (HitResult :type Hit%20Result%20Structure%20%28by%20ref%29) (ImpactPhysicalSurface :type EPhysicalSurface%20Enum :default "SurfaceType_Default") (SpawnedPSC :type FXSystem%20Component%20Object%20Reference))
  ; Reactivity On WeakSpot
  (|CallParentFunction HitResult ImpactPhysicalSurface SpawnedPSC)
  (Variables|Default|SetSpawned%20Impact%20FX SpawnedPSC)
  (if (Math|Boolean|ORBoolean (Utilities|Enum|Equal(Enum) (Collision|GetSurfaceType HitResult) "SurfaceType12") (Utilities|Enum|Equal(Enum) (Collision|GetSurfaceType HitResult) "SurfaceType14") (Utilities|Enum|Equal(Enum) (Collision|GetSurfaceType HitResult) "SurfaceType15") (Utilities|Enum|Equal(Enum) (Collision|GetSurfaceType HitResult) "SurfaceType16") (or (Reactive|Frontend|IsCosmeticObjectDisplayedasPreview 0) (Reactive|Frontend|IsCosmeticObjectDisplayedinFrontendContext 0)))
    (Effects|Components|ParticleSystem|SetBoolParameter (Variables|Default|GetSpawned%20Impact%20FX) "IsReactive" true)
    (if (not (Reactive|Frontend|IsCosmeticObjectDisplayedinFrontendContext 0))
      (CallFunction|CEFXServer) ; It calls the Server Event, that calls the Client Event, \nthat finally calls the reactivity event
    )))

(event Custom|CE_FX_Server
  ;[1 0.2356 0.1858] "Server Logic". It makes sure the reactivity can be seen by other clients
  (CallFunction|CEFXClient))

(event Custom|ReactivityEvent
  (bind _spawned_idlefx (Variables|Default|GetSpawned%20IdleFX))
  ; Reactivity Event
  (Effects|Components|ParticleSystem|SetBoolParameter (Variables|Default|GetSpawned%20Impact%20FX) "IsReactive" true)
  (Utilities|FlowControl|DoOnce false
    (:Completed
      (Niagara|ReinitializeSystem _spawned_idlefx)
      (bind (_direction _reactivitytrack) (Timeline|ReactivityTimeline PlayFromStart)
        (:Update
          _)
        (:Finished
          _))
      (Effects|Components|ParticleSystem|SetFloatParameter _spawned_idlefx "ReactivityTime" _reactivitytrack))))

(event Custom|CE_FX_Client
  ;[1 0.2356 0.1858] "Server Logic". It makes sure the reactivity can be seen by other clients
  (CallFunction|ReactivityEvent))

(fn ConstructionScript
  (|CallParentFunction))
