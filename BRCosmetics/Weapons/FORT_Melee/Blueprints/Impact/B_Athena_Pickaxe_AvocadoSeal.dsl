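; Cosmetic pickaxe weapon Blueprint (AvocadoSeal) with a glow and VFX reaction to weak-point hits.
; On instigator set or equip, it binds the pawn's mesh to the idle Niagara effect, with a short delay if that component is not yet valid.
; Weak-point impacts (via FX_Server) and frontend-display impacts call FX_Client, which flags the impact VFX reactive and runs a timeline driving weapon emissive brightness and idle/swing VFX reactivity.
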(event FortWeapon|EventOnInstigatorSet
  (bind _returnvalue (CallFunction|GetIdleVFXComponent))
  (|CallParentFunction)
  (bind _asfort_pawn (Utilities|Casting|CastToFortPawn (Game|GetInstigator)))
  (Variables|Default|SetInstigatorSkeletalMesh (Variables|/Script/Engine.Character|GetMesh _asfort_pawn))
  (Utilities|IsValid _returnvalue
    (:Is%20Valid
      _)
    (:Is%20Not%20Valid
      (Utilities|FlowControl|Delay 0.2)))
  (bind _asniagara_particle_system_component (Utilities|Casting|CastToNiagaraComponent _returnvalue)
    (:then
      (Niagara|SetNiagaraSkeletalMeshComponent _asniagara_particle_system_component "User.InstigatorMesh" (Variables|Default|GetInstigatorSkeletalMesh)))
    (:CastFailed
      (Development|PrintString "Niagara Cast Failed"))))

(event FortWeapon|EventOnEquippedAndAttached
  (bind _returnvalue (CallFunction|GetIdleVFXComponent))
  (|CallParentFunction)
  (bind _asfort_pawn (Utilities|Casting|CastToFortPawn (Game|GetInstigator)))
  (Variables|Default|SetInstigatorSkeletalMesh (Variables|/Script/Engine.Character|GetMesh _asfort_pawn))
  (Utilities|IsValid _returnvalue
    (:Is%20Valid
      _)
    (:Is%20Not%20Valid
      (Utilities|FlowControl|Delay 0.2)))
  (bind _asniagara_particle_system_component (Utilities|Casting|CastToNiagaraComponent _returnvalue)
    (:then
      (Niagara|SetNiagaraSkeletalMeshComponent _asniagara_particle_system_component "User.InstigatorMesh" (Variables|Default|GetInstigatorSkeletalMesh)))
    (:CastFailed
      (Development|PrintString "Niagara Cast Failed"))))

(event FortWeapon|EventOnLocalUnEquip
  (|CallParentFunction))

(event FortWeapon|EventOnUnEquippedAndDetached
  (|CallParentFunction))

(event EventSwingRight
  (|CallParentFunction)
  (Variables|Default|SetRight%20? true))

(event EventSwingLeft
  (|CallParentFunction)
  (Variables|Default|SetRight%20? false))

(event FortWeapon|EventOnPlayImpactFX
  (params (HitResult :type Hit%20Result%20Structure%20%28by%20ref%29) (ImpactPhysicalSurface :type EPhysicalSurface%20Enum :default "SurfaceType_Default") (SpawnedPSC :type FXSystem%20Component%20Object%20Reference))
  (Variables|Default|SetImpact%20VFX SpawnedPSC)
  (|CallParentFunction HitResult ImpactPhysicalSurface SpawnedPSC)
  (Utilities|FlowControl|Sequence
    (:then_0
      (if (Reactive|Frontend|IsCosmeticObjectDisplayedinFrontendContext self)
        (CallFunction|FXClient)))
    (:then_1
      (if (Utilities|Array|ContainsItem (Variables|Default|GetWeakPoint%20Physical%20Surface%20Array_0) ImpactPhysicalSurface)
        (CallFunction|FXServer)))))

(event Custom|FX_Server
  (CallFunction|FXClient))

(event Custom|FX_Client
  (Effects|Components|ParticleSystem|SetBoolParameter (Variables|Default|GetImpact%20VFX) "IsReactive" true)
  (bind (_direction _weak_point_reactivity _reactivity) (Timeline|Timeline PlayFromStart)
    (:Update
      (Rendering|Material|SetScalarParameterValueonMaterials (Variables|Visual|GetWeaponMesh) "EmissiveBrightness" _weak_point_reactivity)
      (Effects|Components|ParticleSystem|SetFloatParameter (CallFunction|GetIdleVFXComponent) "Reactivity" _reactivity)
      (Effects|Components|ParticleSystem|SetFloatParameter (CallFunction|GetSwingVFXComponent) "Reactivity" _reactivity))))

(fn ConstructionScript
  (|CallParentFunction)
  (Components|Activation|SetAutoActivate (CallFunction|GetIdlePSC) false))
