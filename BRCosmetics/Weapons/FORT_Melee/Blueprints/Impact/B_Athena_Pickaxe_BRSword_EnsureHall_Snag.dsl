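; Impact Blueprint for the EnsureHall Snag sword pickaxe that flags its impact effect as a weakpoint hit.
; On impact, it calls the parent and stores the spawned FX component.
; SetBlsWeakpoint sets that component's bIsWeakpoint parameter to true, called directly and through ReplicateReactivity, with no test for an actual weakpoint.
; The frontend-preview and physical-surface checks end in an empty inner branch.
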(event EventBeginPlay
  (|CallParentFunction) ; This node is disabled and will not be called.\nDrag off pins to build functionality.
)

(event EventTick
  (params (DeltaSeconds :type Float%20%28single-precision%29))
  (|CallParentFunction DeltaSeconds) ; This node is disabled and will not be called.\nDrag off pins to build functionality.
)

(event FortWeapon|EventOnPlayImpactFX
  (params (HitResult :type Hit%20Result%20Structure%20%28by%20ref%29) (ImpactPhysicalSurface :type EPhysicalSurface%20Enum :default "SurfaceType_Default") (SpawnedPSC :type FXSystem%20Component%20Object%20Reference))
  (|CallParentFunction HitResult ImpactPhysicalSurface SpawnedPSC)
  (Variables|Default|SetImpact%20FX SpawnedPSC)
  ;[0 0.6724 1 0.5] CHECK FRONTEND
  (if (not (or (Reactive|Frontend|IsCosmeticObjectDisplayedasPreview 0) (Reactive|Frontend|IsCosmeticObjectDisplayedinFrontendContext 0)))
    (if (Utilities|Array|ContainsItem (Variables|Default|GetImpactPhysicalSurface) ImpactPhysicalSurface)
      _))
  (CallFunction|SetBlsWeakpoint)
  (CallFunction|ReplicateReactivity))

(event Custom|SetBlsWeakpoint
  (Effects|Components|ParticleSystem|SetBoolParameter (Variables|Default|GetImpact%20FX) "bIsWeakpoint" true))

(event Custom|ReplicateReactivity
  (CallFunction|SetBlsWeakpoint))

(fn ConstructionScript
  (|CallParentFunction))
