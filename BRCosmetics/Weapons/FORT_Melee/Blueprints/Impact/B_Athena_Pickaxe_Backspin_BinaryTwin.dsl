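; Cosmetic Blueprint for the Backspin BinaryTwin pickaxe, driving reactive emissive glow and trail effects.
; Timelines lerp the mesh's "emissive mult" parameter from GlowBase on kills, swings and impacts on three special surface types or in frontend display.
; Init, alteration, BeginPlay and visibility events repeat one setup: swing particles off, glow reset, pawn delegates bound outside frontend.
; Those delegates deactivate Trail through DoOnce nodes on jump and landing, then reactivate it on landing.
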(event Custom|UpdateBasedOnKills_1
  (params (Watched%20Kills :type Integer))
  ;[1 1 1] GlowOnKill
  (bind _output_get (Variables|Default|SetKillCount (Math|Conversions|ToFloat(Integer) Watched%20Kills)))
  (if (> _output_get 0.0)
    (bind (_direction _lerp_alpha) (Timeline|Timeline_0 PlayFromStart)
      (:Update
        (Rendering|Material|SetVectorParameterValueonMaterials (Variables|Visual|GetWeaponMesh) "emissive mult" (Math|Conversions|ToVector(LinearColor) (Math|Color|LinearColor*Float "(R=1.000000,G=0.072205,B=0.000000,A=1.000000)" (Math|Float|Lerp (Variables|Default|GetGlowBase) (Variables|Default|GetGlowIntensity) _lerp_alpha))))))))

(event FortWeapon|EventOnPlayImpactFX
  (params (HitResult :type Hit%20Result%20Structure%20%28by%20ref%29) (ImpactPhysicalSurface :type EPhysicalSurface%20Enum :default "SurfaceType_Default") (SpawnedPSC :type FXSystem%20Component%20Object%20Reference))
  ;[1 1 1] GlowOnWeakSpotHit
  (if (Math|Boolean|ORBoolean (Utilities|Enum|Equal(Enum) (Collision|GetSurfaceType HitResult) "SurfaceType14") (Utilities|Enum|Equal(Enum) (Collision|GetSurfaceType HitResult) "SurfaceType15") (Utilities|Enum|Equal(Enum) (Collision|GetSurfaceType HitResult) "SurfaceType16") (Reactive|Frontend|IsCosmeticObjectDisplayedinFrontendContext 0))
    (bind (_direction _lerp_alpha) (Timeline|Timeline_1 PlayFromStart)
      (:Update
        (Rendering|Material|SetScalarParameterValueonMaterials (Variables|Visual|GetWeaponMesh) "emissive mult" (Math|Float|Lerp (Variables|Default|GetGlowBase) (* (Variables|Default|GetGlowIntensity) (Math|Float|Lerp 0.5 6e+01 (Math|Conversions|ToFloat(Boolean) (Reactive|Frontend|IsCosmeticObjectDisplayedinFrontendContext 0)))) _lerp_alpha))))))

(event EventSwingLeft
  (|CallParentFunction)
  (Variables|Default|SetRight%20? false)
  (bind (_direction _lerp_alpha) (Timeline|Timeline_2 PlayFromStart)
    (:Update
      (Rendering|Material|SetScalarParameterValueonMaterials (Variables|Visual|GetWeaponMesh) "emissive mult" (Math|Float|Lerp (Variables|Default|GetGlowBase) (* (Variables|Default|GetGlowIntensity) 15.0) _lerp_alpha)))))

(event EventSwingRight
  (|CallParentFunction)
  (Variables|Default|SetRight%20? true)
  (Timeline|Timeline_2 PlayFromStart))

(event EventOnInitWeaponCosmetics
  (|CallParentFunction)
  (Components|Activation|Deactivate (CallFunction|GetSwingPSC))
  (if (Reactive|Frontend|IsCosmeticObjectDisplayedinFrontendContext 0)
    (Rendering|Material|SetScalarParameterValueonMaterials (Variables|Visual|GetWeaponMesh) "emissive mult" (Variables|Default|GetGlowBase))
    (else
      (Rendering|Material|SetScalarParameterValueonMaterials (Variables|Visual|GetWeaponMesh) "emissive mult" (Variables|Default|GetGlowBase))
      (bind _asfort_player_pawn_athena (Utilities|Casting|CastToFortPlayerPawnAthena (Game|GetInstigator)))
      (Variables|Default|SetAs%20Fort%20Player%20Pawn%20Athena _asfort_player_pawn_athena)
      (Character|BindEventtoOnCharacterMovementUpdated _asfort_player_pawn_athena (EventDispatchers|CreateEvent :!event Custom|On%20Begin%20Jump))
      (EventDispatchers|BindEventtoOnPawnLanded _asfort_player_pawn_athena (EventDispatchers|CreateEvent :!event Custom|On%20End%20Jump)))))

(event EventOnInitCosmeticAlterations
  (params (CosmeticMod :type Fort%20Cosmetic%20Modification%20Structure))
  (|CallParentFunction CosmeticMod)
  (Components|Activation|Deactivate (CallFunction|GetSwingPSC))
  (if (Reactive|Frontend|IsCosmeticObjectDisplayedinFrontendContext 0)
    (Rendering|Material|SetScalarParameterValueonMaterials (Variables|Visual|GetWeaponMesh) "emissive mult" (Variables|Default|GetGlowBase))
    (else
      (Rendering|Material|SetScalarParameterValueonMaterials (Variables|Visual|GetWeaponMesh) "emissive mult" (Variables|Default|GetGlowBase))
      (bind _asfort_player_pawn_athena (Utilities|Casting|CastToFortPlayerPawnAthena (Game|GetInstigator)))
      (Variables|Default|SetAs%20Fort%20Player%20Pawn%20Athena _asfort_player_pawn_athena)
      (Character|BindEventtoOnCharacterMovementUpdated _asfort_player_pawn_athena (EventDispatchers|CreateEvent :!event Custom|On%20Begin%20Jump))
      (EventDispatchers|BindEventtoOnPawnLanded _asfort_player_pawn_athena (EventDispatchers|CreateEvent :!event Custom|On%20End%20Jump)))))

(event EventBeginPlay
  (|CallParentFunction)
  (Components|Activation|Deactivate (CallFunction|GetSwingPSC))
  (if (Reactive|Frontend|IsCosmeticObjectDisplayedinFrontendContext 0)
    (Rendering|Material|SetScalarParameterValueonMaterials (Variables|Visual|GetWeaponMesh) "emissive mult" (Variables|Default|GetGlowBase))
    (else
      (Rendering|Material|SetScalarParameterValueonMaterials (Variables|Visual|GetWeaponMesh) "emissive mult" (Variables|Default|GetGlowBase))
      (bind _asfort_player_pawn_athena (Utilities|Casting|CastToFortPlayerPawnAthena (Game|GetInstigator)))
      (Variables|Default|SetAs%20Fort%20Player%20Pawn%20Athena _asfort_player_pawn_athena)
      (Character|BindEventtoOnCharacterMovementUpdated _asfort_player_pawn_athena (EventDispatchers|CreateEvent :!event Custom|On%20Begin%20Jump))
      (EventDispatchers|BindEventtoOnPawnLanded _asfort_player_pawn_athena (EventDispatchers|CreateEvent :!event Custom|On%20End%20Jump)))))

(event EventOnWeaponVisibilityChanged
  (params (bVisible :type Boolean) (bSetForLocalControllerOnly :type Boolean))
  (|CallParentFunction bVisible bSetForLocalControllerOnly)
  (Components|Activation|Deactivate (CallFunction|GetSwingPSC))
  (if (Reactive|Frontend|IsCosmeticObjectDisplayedinFrontendContext 0)
    (Rendering|Material|SetScalarParameterValueonMaterials (Variables|Visual|GetWeaponMesh) "emissive mult" (Variables|Default|GetGlowBase))
    (else
      (Rendering|Material|SetScalarParameterValueonMaterials (Variables|Visual|GetWeaponMesh) "emissive mult" (Variables|Default|GetGlowBase))
      (bind _asfort_player_pawn_athena (Utilities|Casting|CastToFortPlayerPawnAthena (Game|GetInstigator)))
      (Variables|Default|SetAs%20Fort%20Player%20Pawn%20Athena _asfort_player_pawn_athena)
      (Character|BindEventtoOnCharacterMovementUpdated _asfort_player_pawn_athena (EventDispatchers|CreateEvent :!event Custom|On%20Begin%20Jump))
      (EventDispatchers|BindEventtoOnPawnLanded _asfort_player_pawn_athena (EventDispatchers|CreateEvent :!event Custom|On%20End%20Jump)))))

(event Custom|On%20Begin%20Jump
  (params (DeltaSeconds :type Float%20%28single-precision%29) (OldLocation :type Vector) (OldVelocity :type Vector))
  (if (Character|IsJumpProvidingForce (Variables|Default|GetAs%20Fort%20Player%20Pawn%20Athena))
    (Utilities|FlowControl|DoOnce false
      (:Completed
        (Components|Activation|Deactivate (Variables|Default|GetTrail))))))

(event Custom|On%20End%20Jump
  (params (Hit :type Hit%20Result%20Structure%20%28by%20ref%29))
  (Utilities|FlowControl|Sequence
    (:then_0
      (Utilities|FlowControl|DoOnce false
        (:Completed
          (Components|Activation|Deactivate (Variables|Default|GetTrail)))))
    (:then_1
      (Components|Activation|Activate (Variables|Default|GetTrail) false))))

(fn ConstructionScript
  (|CallParentFunction))
