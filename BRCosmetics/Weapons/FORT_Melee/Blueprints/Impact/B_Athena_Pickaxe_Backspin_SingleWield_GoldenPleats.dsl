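; Pickaxe cosmetic Blueprint for the single-wield Backspin GoldenPleats melee weapon, which makes its idle particle effect react to swinging.
; Left and right swing events call the parent, then set the "bisSwinging" bool parameter true on the idle particle system component.
; Both swing-end events call the parent and clear that parameter after a short fixed delay, so the swing state lingers briefly.
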(event EventSwingRight
  (|CallParentFunction)
  (Effects|Components|ParticleSystem|SetBoolParameter (CallFunction|GetIdlePSC) "bisSwinging" true))

(event EventBeginPlay)

(event EventSwingLeft
  (|CallParentFunction)
  (Effects|Components|ParticleSystem|SetBoolParameter (CallFunction|GetIdlePSC) "bisSwinging" true))

(event EventTick
  (params (DeltaSeconds :type Float%20%28single-precision%29)))

(event EventSwingLeftEnd
  (|CallParentFunction)
  (Utilities|FlowControl|Delay 2.0)
  (Effects|Components|ParticleSystem|SetBoolParameter (CallFunction|GetIdlePSC) "bisSwinging" false))

(event EventSwingRightEnd
  (|CallParentFunction)
  (Utilities|FlowControl|Delay 2.0)
  (Effects|Components|ParticleSystem|SetBoolParameter (CallFunction|GetIdlePSC) "bisSwinging" false))

(fn ConstructionScript
  (|CallParentFunction))
