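; Cosmetic Blueprint for the MysticMale single-wield Backspin pickaxe that adds reactive VFX on top of its parent melee weapon logic.
; Swing start and end events toggle an IsSwingingMystic bool on the idle VFX.
; On impact it spawns the on-hit Niagara effect attached to the weapon mesh, deactivates the swing VFX, and sets HasHitThing on the anim trail VFX.
; Either swing end clears HasHitThing again.
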(event FortWeapon|EventOnPlayImpactFX
  (params (HitResult :type Hit%20Result%20Structure%20%28by%20ref%29) (ImpactPhysicalSurface :type EPhysicalSurface%20Enum :default "SurfaceType_Default") (SpawnedPSC :type FXSystem%20Component%20Object%20Reference))
  (|CallParentFunction HitResult ImpactPhysicalSurface SpawnedPSC)
  (Effects|SpawnFXSystemAttached (Variables|Default|GetOn%20Hit%20Niagara%20FX) (Variables|Visual|GetWeaponMesh))
  (Components|Activation|Deactivate (CallFunction|GetSwingVFXComponent))
  (Effects|Components|ParticleSystem|SetBoolParameter (CallFunction|GetAnimTrailVFXComponent) "HasHitThing" true))

(event EventBeginPlay)

(event EventTick
  (params (DeltaSeconds :type Float%20%28single-precision%29)))

(event EventSwingRight
  (|CallParentFunction)
  (Effects|Components|ParticleSystem|SetBoolParameter (CallFunction|GetIdleVFXComponent) "IsSwingingMystic" true))

(event EventSwingRightEnd
  (|CallParentFunction)
  (Effects|Components|ParticleSystem|SetBoolParameter (CallFunction|GetIdleVFXComponent) "IsSwingingMystic" false)
  (Effects|Components|ParticleSystem|SetBoolParameter (CallFunction|GetAnimTrailVFXComponent) "HasHitThing" false))

(event EventSwingLeft
  (|CallParentFunction)
  (Effects|Components|ParticleSystem|SetBoolParameter (CallFunction|GetIdleVFXComponent) "IsSwingingMystic" true))

(event EventSwingLeftEnd
  (|CallParentFunction)
  (Effects|Components|ParticleSystem|SetBoolParameter (CallFunction|GetIdleVFXComponent) "IsSwingingMystic" false)
  (Effects|Components|ParticleSystem|SetBoolParameter (CallFunction|GetAnimTrailVFXComponent) "HasHitThing" false))

(fn ConstructionScript
  (|CallParentFunction))
