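; Blueprint for the Backspin pickaxe cosmetic that toggles its idle particle effect's blade motion in step with melee swings.
; Left and right swing events set the "BladeSwing" bool parameter on the idle VFX particle component to true.
; The matching swing-end events wait half a second, then set it back to false.
; All swing events call their parent implementation first, and both directions share the OnToggleBladeMovement function.
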(event EventBeginPlay)

(event EventTick
  (params (DeltaSeconds :type Float%20%28single-precision%29)))

(event EventSwingLeftEnd
  (|CallParentFunction)
  (Utilities|FlowControl|Delay 0.5)
  (CallFunction|OnToggleBladeMovement false))

(event EventSwingRightEnd
  (|CallParentFunction)
  (Utilities|FlowControl|Delay 0.5)
  (CallFunction|OnToggleBladeMovement false))

(event EventSwingLeft
  (|CallParentFunction)
  (CallFunction|OnToggleBladeMovement true))

(event EventSwingRight
  (|CallParentFunction)
  (CallFunction|OnToggleBladeMovement true))

(fn OnToggleBladeMovement
  (params (InValue :type Boolean))
  (Effects|Components|ParticleSystem|SetBoolParameter (CallFunction|GetIdleVFXComponent) "BladeSwing" InValue))

(fn ConstructionScript
  (|CallParentFunction))
