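; Fortnite pickaxe weapon Blueprint (BadCat) whose mesh glows orange-red in response to kills and special-surface impacts.
; A kill update stores the count and, above zero, plays a timeline driving the EmissiveMultiplier material vector by GlowIntensity.
; Impacts on SurfaceType14, 15 or 16 play a second timeline at half that intensity, for weak-spot hits.
; Swing events record left or right, and cosmetic init deactivates the swing particle component.
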(event Custom|UpdateBasedOnKills_1
  (params (Watched%20Kills :type Integer))
  ;[1 1 1] GlowOnKill
  (bind _output_get (Variables|Default|SetKillCount (Math|Conversions|ToFloat(Integer) Watched%20Kills)))
  (if (> _output_get 0.0)
    (bind (_direction _lerp_alpha) (Timeline|Timeline_0 PlayFromStart)
      (:Update
        (Rendering|Material|SetVectorParameterValueonMaterials (Variables|Visual|GetWeaponMesh) "EmissiveMultiplier" (Math|Conversions|ToVector(LinearColor) (Math|Color|LinearColor*Float "(R=1.000000,G=0.072205,B=0.000000,A=1.000000)" (Math|Float|Lerp 0.0 (Variables|Default|GetGlowIntensity) _lerp_alpha))))))))

(event FortWeapon|EventOnPlayImpactFX
  (params (HitResult :type Hit%20Result%20Structure%20%28by%20ref%29) (ImpactPhysicalSurface :type EPhysicalSurface%20Enum :default "SurfaceType_Default") (SpawnedPSC :type FXSystem%20Component%20Object%20Reference))
  ;[1 1 1] GlowOnWeakSpotHit
  (|CallParentFunction HitResult ImpactPhysicalSurface SpawnedPSC)
  (if (Math|Boolean|ORBoolean (Utilities|Enum|Equal(Enum) (Collision|GetSurfaceType HitResult) "SurfaceType14") (Utilities|Enum|Equal(Enum) (Collision|GetSurfaceType HitResult) "SurfaceType15") (Utilities|Enum|Equal(Enum) (Collision|GetSurfaceType HitResult) "SurfaceType16"))
    (bind (_direction _lerp_alpha) (Timeline|Timeline_1 PlayFromStart)
      (:Update
        (Rendering|Material|SetVectorParameterValueonMaterials (Variables|Visual|GetWeaponMesh) "EmissiveMultiplier" (Math|Conversions|ToVector(LinearColor) (Math|Color|LinearColor*Float "(R=1.000000,G=0.072205,B=0.000000,A=1.000000)" (Math|Float|Lerp 0.0 (* (Variables|Default|GetGlowIntensity) 0.5) _lerp_alpha))))))))

(event EventSwingLeft
  (|CallParentFunction)
  (Variables|Default|SetRight%20? false))

(event EventSwingRight
  (|CallParentFunction)
  (Variables|Default|SetRight%20? true))

(event EventOnInitWeaponCosmetics
  (|CallParentFunction)
  (Components|Activation|Deactivate (CallFunction|GetSwingPSC)))

(fn ConstructionScript
  (|CallParentFunction))
