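; Cosmetic Blueprint for the BitFight pickaxe, a melee weapon with main and offhand meshes, adding variant-aware swing effects and impact-reactive materials.
; It caches the loadout's mesh variant stage on init, instigator set and equip, feeding it to a Niagara trail spawned on the second left swing.
; Swings pass montage section indices to VFX.
; Matching-surface impacts, or frontend previews, timestamp WeakSpotHit material parameters according to swing side.
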(event EventOnInitWeaponCosmetics
  (|CallParentFunction)
  (bind _asfort_player_pawn (Utilities|Casting|CastToFortPlayerPawn (Game|GetInstigator)))
  (bind (_index _bindexisvalid) (CosmeticLoadout|GetVariantChannelActiveIndexforSlotActor "(TagName=\"LoadoutSlot.Pickaxe\")" _asfort_player_pawn "(TagName=\"Cosmetics.Variant.Channel.Mesh\")" (Utilities|Array|MakeArray "(TagName=\"Cosmetics.Variant.Property.Stage1\")" "(TagName=\"Cosmetics.Variant.Property.Stage2\")")))
  (Variables|Default|SetActiveVariantIndex (Math|Integer|SelectInt _index 0 _bindexisvalid)))

(event FortWeapon|EventOnInstigatorSet
  (|CallParentFunction)
  (bind _asfort_player_pawn (Utilities|Casting|CastToFortPlayerPawn (Game|GetInstigator)))
  (bind (_index _bindexisvalid) (CosmeticLoadout|GetVariantChannelActiveIndexforSlotActor "(TagName=\"LoadoutSlot.Pickaxe\")" _asfort_player_pawn "(TagName=\"Cosmetics.Variant.Channel.Mesh\")" (Utilities|Array|MakeArray "(TagName=\"Cosmetics.Variant.Property.Stage1\")" "(TagName=\"Cosmetics.Variant.Property.Stage2\")")))
  (Variables|Default|SetActiveVariantIndex (Math|Integer|SelectInt _index 0 _bindexisvalid)))

(event FortWeapon|EventOnEquippedAndAttached
  (|CallParentFunction)
  (bind _asfort_player_pawn (Utilities|Casting|CastToFortPlayerPawn (Game|GetInstigator)))
  (bind (_index _bindexisvalid) (CosmeticLoadout|GetVariantChannelActiveIndexforSlotActor "(TagName=\"LoadoutSlot.Pickaxe\")" _asfort_player_pawn "(TagName=\"Cosmetics.Variant.Channel.Mesh\")" (Utilities|Array|MakeArray "(TagName=\"Cosmetics.Variant.Property.Stage1\")" "(TagName=\"Cosmetics.Variant.Property.Stage2\")")))
  (Variables|Default|SetActiveVariantIndex (Math|Integer|SelectInt _index 0 _bindexisvalid)))

(event EventSwingRight
  (bind _returnvalue (CallFunction|GetSwingVFXComponent))
  ;[1 1 1] Final Swing
  (|CallParentFunction)
  (bind _section_index (Utilities|SetupSwingMontageSectionIndex (Game|GetInstigator))
    (:Out
      (Effects|Components|ParticleSystem|SetIntParameter _returnvalue "SwingSectionIndex" _section_index))
    (:Cast%20Failed%20Out
      (Effects|Components|ParticleSystem|SetIntParameter _returnvalue "SwingSectionIndex" -1))))

(event EventSwingLeft2
  (bind _returnvalue (CallFunction|GetSwingVFXComponent))
  ;[1 1 1] Final Swing
  (|CallParentFunction)
  (bind _returnvalue_2 (Niagara|SpawnSystemAttached "/BRCosmetics/Weapons/FORT_Melee/Pickaxe_BitFight/FX/NS_Pickaxe_BitFight_Trail.NS_Pickaxe_BitFight_Trail" (Variables|Visual|GetWeaponMesh) "None" "0, 0, 0" "0, 0, 0" "KeepRelativeOffset" true))
  (Effects|Components|ParticleSystem|SetFloatParameter _returnvalue_2 "Variant" (Math|Conversions|ToFloat(Integer) (Variables|Default|GetActiveVariantIndex)))
  (bind _section_index (Utilities|SetupSwingMontageSectionIndex (Game|GetInstigator))
    (:Out
      (Effects|Components|ParticleSystem|SetIntParameter _returnvalue "SwingSectionIndex" _section_index))
    (:Cast%20Failed%20Out
      (Effects|Components|ParticleSystem|SetIntParameter _returnvalue "SwingSectionIndex" -1))))

(event EventSwingLeft
  (bind _returnvalue (CallFunction|GetSwingVFXComponent))
  ;[1 1 1] Final Swing
  (|CallParentFunction)
  (bind _section_index (Utilities|SetupSwingMontageSectionIndex (Game|GetInstigator))
    (:Out
      (Effects|Components|ParticleSystem|SetIntParameter _returnvalue "SwingSectionIndex" _section_index))
    (:Cast%20Failed%20Out
      (Effects|Components|ParticleSystem|SetIntParameter _returnvalue "SwingSectionIndex" -1))))

(event EventSwingRight2
  (bind _returnvalue (CallFunction|GetSwingVFXComponent))
  ;[1 1 1] Final Swing
  (|CallParentFunction)
  (bind _section_index (Utilities|SetupSwingMontageSectionIndex (Game|GetInstigator))
    (:Out
      (Effects|Components|ParticleSystem|SetIntParameter _returnvalue "SwingSectionIndex" _section_index))
    (:Cast%20Failed%20Out
      (Effects|Components|ParticleSystem|SetIntParameter _returnvalue "SwingSectionIndex" -1))))

(event FortWeapon|EventOnPlayImpactFX
  (params (HitResult :type Hit%20Result%20Structure%20%28by%20ref%29) (ImpactPhysicalSurface :type EPhysicalSurface%20Enum :default "SurfaceType_Default") (SpawnedPSC :type FXSystem%20Component%20Object%20Reference))
  (bind _returnvalue (Utilities|Time|GetGameTimeinSeconds))
  ; On Weakpoint shader reactivity
  (|CallParentFunction HitResult ImpactPhysicalSurface SpawnedPSC)
  (if (or (Reactive|Frontend|IsCosmeticObjectDisplayedasPreview 0) (Reactive|Frontend|IsCosmeticObjectDisplayedinFrontendContext 0))
    (CallFunction|WeakpointHit)
    (elif (Utilities|Array|ContainsItem (Variables|Default|GetImpactPhysicalSurfaces) ImpactPhysicalSurface)
      (if (not (Variables|Default|GetRight?))
        (Rendering|Material|SetScalarParameterValueonMaterials (Variables|Visual|GetWeaponMeshOffhand) "WeakSpotHit" _returnvalue)
        (if (Variables|Default|GetLeft2%20Swing?)
          _))
      (Rendering|Material|SetScalarParameterValueonMaterials (Variables|Visual|GetWeaponMesh) "WeakSpotHit" _returnvalue))))

(event Custom|WeakpointHit
  (bind _returnvalue (Utilities|Time|GetGameTimeinSeconds))
  ; On Weakpoint shader reactivity
  (if (not (Variables|Default|GetRight?))
    (Rendering|Material|SetScalarParameterValueonMaterials (Variables|Visual|GetWeaponMeshOffhand) "WeakSpotHit" _returnvalue)
    (if (Variables|Default|GetLeft2%20Swing?)
      _))
  (Rendering|Material|SetScalarParameterValueonMaterials (Variables|Visual|GetWeaponMesh) "WeakSpotHit" _returnvalue))

(event EventSwingRight_Common
  ; On Weakpoint shader reactivity
  (|CallParentFunction)
  (Variables|Default|SetRight? true))

(event EventSwingLeft_Common
  ; On Weakpoint shader reactivity
  (|CallParentFunction)
  (Variables|Default|SetRight? false))

(fn GetFXParamNames
  (return "LoopDuration"))

(fn SetParamsOnTrails
  (params (LoopDuration :type Float%20%28single-precision%29) (Lifetime :type Float%20%28single-precision%29))
  (for _array_element (Utilities|Array|MakeArray (CallFunction|GetAnimTrailsPSC) (CallFunction|GetAnimTrailsOffhandPSC))
    (bind _asniagara_particle_system_component (Utilities|Casting|CastToNiagaraComponent _array_element))
    (Niagara|SetNiagaraVariable(Float) _asniagara_particle_system_component "LoopDuration" LoopDuration)
    (Niagara|SetNiagaraVariable(Float) _asniagara_particle_system_component "Lifetime" Lifetime)))

(fn ConstructionScript
  (|CallParentFunction))
