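; Cosmetic Blueprint for the ButtonCase pickaxe, adding reactive orange glow and section-specific swing effects.
; Timelines drive the weapon mesh's EmissiveMultiplier when kill count is above zero, or at half intensity on weak-spot surface hits.
; Both swing events read the instigator's montage section; on "Swing2" they spawn an attached jab effect and hide the normal swing VFX.
; It also deactivates the swing particles on init and exposes a SwingIdx audio parameter setter.
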(event Custom|UpdateBasedOnKills_1
  (params (Watched%20Kills :type Integer))
  ;[1 1 1] GlowOnKill
  (bind _output_get (Variables|Default|SetKillCount (Math|Conversions|ToFloat(Integer) Watched%20Kills)))
  (if (> _output_get 0.0)
    (bind (_direction _lerp_alpha) (Timeline|Timeline_0 PlayFromStart)
      (:Update
        (Rendering|Material|SetVectorParameterValueonMaterials (Variables|Visual|GetWeaponMesh) "EmissiveMultiplier" (Math|Conversions|ToVector(LinearColor) (Math|Color|LinearColor*Float "(R=1.000000,G=0.072205,B=0.000000,A=1.000000)" (Math|Float|Lerp 0.0 (Variables|Default|GetGlowIntensity) _lerp_alpha))))))))

(event FortWeapon|EventOnPlayImpactFX
  (params (HitResult :type Hit%20Result%20Structure%20%28by%20ref%29) (ImpactPhysicalSurface :type EPhysicalSurface%20Enum :default "SurfaceType_Default") (SpawnedPSC :type FXSystem%20Component%20Object%20Reference))
  ;[1 1 1] GlowOnWeakSpotHit
  (|CallParentFunction HitResult ImpactPhysicalSurface SpawnedPSC)
  (Utilities|FlowControl|Sequence
    (:then_0
      (if (Math|Boolean|ORBoolean (Utilities|Enum|Equal(Enum) (Collision|GetSurfaceType HitResult) "SurfaceType14") (Utilities|Enum|Equal(Enum) (Collision|GetSurfaceType HitResult) "SurfaceType15") (Utilities|Enum|Equal(Enum) (Collision|GetSurfaceType HitResult) "SurfaceType16"))
        (bind (_direction _lerp_alpha) (Timeline|Timeline_1 PlayFromStart)
          (:Update
            (Rendering|Material|SetVectorParameterValueonMaterials (Variables|Visual|GetWeaponMesh) "EmissiveMultiplier" (Math|Conversions|ToVector(LinearColor) (Math|Color|LinearColor*Float "(R=1.000000,G=0.072205,B=0.000000,A=1.000000)" (Math|Float|Lerp 0.0 (* (Variables|Default|GetGlowIntensity) 0.5) _lerp_alpha))))))))))

(event EventOnInitWeaponCosmetics
  (|CallParentFunction)
  (Components|Activation|Deactivate (CallFunction|GetSwingPSC)))

(event EventSwingLeft
  (bind _returnvalue (Game|GetInstigator))
  (bind _asfort_pawn (Utilities|Casting|CastToFortPawn _returnvalue))
  (bind _mesh (Variables|/Script/Engine.Character|GetMesh _asfort_pawn))
  (bind _returnvalue_2 (Components|SkeletalMesh|GetAnimInstance _mesh))
  (bind _returnvalue_3 (Animation|Montage|GetCurrentActiveMontage _returnvalue_2))
  (bind _returnvalue_4 (Animation|Montage|MontageGetCurrentSection _returnvalue_2 _returnvalue_3))
  (bind _returnvalue_5 (== _returnvalue_4 "Swing2"))
  (|CallParentFunction)
  (Variables|Default|SetRight%20? false)
  (Networking|CanExecuteCosmeticEvents
    (:True
      (Utilities|FlowControl|Sequence
        (:then_0
          (if _returnvalue_5
            (Effects|SpawnFXSystemAttached (Variables|Default|GetFXSystemJab) (Variables|Visual|GetWeaponMesh) "Melee_TwoHanded_Handle")))
        (:then_1
          (Development|SetHiddeninGame (CallFunction|GetSwingVFXComponent) _returnvalue_5 false))))))

(event EventSwingRight
  (bind _returnvalue (Game|GetInstigator))
  (bind _asfort_pawn (Utilities|Casting|CastToFortPawn _returnvalue))
  (bind _mesh (Variables|/Script/Engine.Character|GetMesh _asfort_pawn))
  (bind _returnvalue_2 (Components|SkeletalMesh|GetAnimInstance _mesh))
  (bind _returnvalue_3 (Animation|Montage|GetCurrentActiveMontage _returnvalue_2))
  (bind _returnvalue_4 (Animation|Montage|MontageGetCurrentSection _returnvalue_2 _returnvalue_3))
  (bind _returnvalue_5 (== _returnvalue_4 "Swing2"))
  (|CallParentFunction)
  (Variables|Default|SetRight%20? true)
  (Networking|CanExecuteCosmeticEvents
    (:True
      (Utilities|FlowControl|Sequence
        (:then_0
          (if _returnvalue_5
            (Effects|SpawnFXSystemAttached (Variables|Default|GetFXSystemJab) (Variables|Visual|GetWeaponMesh) "Melee_TwoHanded_Handle")))
        (:then_1
          (Development|SetHiddeninGame (CallFunction|GetSwingVFXComponent) _returnvalue_5 false))))))

(fn SetSwingAudioParam
  (params (InInt :type Integer))
  (bind _returnvalue (AudioParameter|GetAudioParameterComponent))
  (Utilities|IsValid _returnvalue
    (:Is%20Valid
      (Audio|Parameter|SetIntegerParameter _returnvalue "SwingIdx" InInt :!class /Script/AudioExtensions.AudioParameterControllerInterface))))

(fn ConstructionScript
  (|CallParentFunction))
