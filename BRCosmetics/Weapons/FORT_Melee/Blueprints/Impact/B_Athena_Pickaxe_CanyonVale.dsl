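; CanyonVale pickaxe cosmetic Blueprint whose materials and VFX react to hits.
; Impacts, and weapon init in frontend display, stamp game time into the mesh's "Timestamp" parameter, then call ReactivityOnServer, which chains to ReactivityUpdate to set "ReactivityTime" on idle and swing VFX.
; Instigator-set and equip events feed the pawn's mesh to the idle Niagara effect, delaying briefly if that component is invalid.
; Swings record their side; construction disables idle-effect auto-activation.
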(event EventOnInitWeaponCosmetics
  (if (Reactive|Frontend|IsCosmeticObjectDisplayedinFrontendContext 0)
    (CallFunction|HitReactivity)))

(event Custom|HitReactivity
  (Rendering|Material|SetScalarParameterValueonMaterials (Variables|Visual|GetWeaponMesh) "Timestamp" (Utilities|Time|GetGameTimeinSeconds))
  (CallFunction|ReactivityonServer))

(event FortWeapon|EventOnPlayImpactFX
  (params (HitResult :type Hit%20Result%20Structure%20%28by%20ref%29) (ImpactPhysicalSurface :type EPhysicalSurface%20Enum :default "SurfaceType_Default") (SpawnedPSC :type FXSystem%20Component%20Object%20Reference))
  (|CallParentFunction HitResult ImpactPhysicalSurface SpawnedPSC)
  (Rendering|Material|SetScalarParameterValueonMaterials (Variables|Visual|GetWeaponMesh) "Timestamp" (Utilities|Time|GetGameTimeinSeconds))
  (CallFunction|ReactivityonServer))

(event FortWeapon|EventOnInstigatorSet
  (bind _returnvalue (CallFunction|GetIdleVFXComponent))
  (|CallParentFunction)
  (bind _asfort_pawn (Utilities|Casting|CastToFortPawn (Game|GetInstigator)))
  (Variables|Default|SetInstigatorSkeletalMesh (Variables|/Script/Engine.Character|GetMesh _asfort_pawn))
  (Utilities|IsValid _returnvalue
    (:Is%20Valid
      _)
    (:Is%20Not%20Valid
      (Utilities|FlowControl|Delay 0.2)))
  (bind _asniagara_particle_system_component (Utilities|Casting|CastToNiagaraComponent _returnvalue)
    (:then
      (Niagara|SetNiagaraSkeletalMeshComponent _asniagara_particle_system_component "User.InstigatorMesh" (Variables|Default|GetInstigatorSkeletalMesh)))
    (:CastFailed
      (Development|PrintString "Niagara Cast Failed"))))

(event FortWeapon|EventOnEquippedAndAttached
  (bind _returnvalue (CallFunction|GetIdleVFXComponent))
  (|CallParentFunction)
  (bind _asfort_pawn (Utilities|Casting|CastToFortPawn (Game|GetInstigator)))
  (Variables|Default|SetInstigatorSkeletalMesh (Variables|/Script/Engine.Character|GetMesh _asfort_pawn))
  (Utilities|IsValid _returnvalue
    (:Is%20Valid
      _)
    (:Is%20Not%20Valid
      (Utilities|FlowControl|Delay 0.2)))
  (bind _asniagara_particle_system_component (Utilities|Casting|CastToNiagaraComponent _returnvalue)
    (:then
      (Niagara|SetNiagaraSkeletalMeshComponent _asniagara_particle_system_component "User.InstigatorMesh" (Variables|Default|GetInstigatorSkeletalMesh)))
    (:CastFailed
      (Development|PrintString "Niagara Cast Failed"))))

(event FortWeapon|EventOnLocalUnEquip
  (|CallParentFunction))

(event FortWeapon|EventOnUnEquippedAndDetached
  (|CallParentFunction))

(event EventSwingRight
  (|CallParentFunction)
  (Variables|Default|SetRight%20? true))

(event EventSwingLeft
  (|CallParentFunction)
  (Variables|Default|SetRight%20? false))

(event Custom|ReactivityOnServer
  (CallFunction|ReactivityUpdate))

(event Custom|ReactivityUpdate
  (bind _returnvalue (Utilities|Time|GetGameTimeinSeconds))
  (Effects|Components|ParticleSystem|SetFloatParameter (CallFunction|GetIdleVFXComponent) "ReactivityTime" _returnvalue)
  (Effects|Components|ParticleSystem|SetFloatParameter (CallFunction|GetSwingVFXComponent) "ReactivityTime" _returnvalue))

(fn ConstructionScript
  (|CallParentFunction)
  (Components|Activation|SetAutoActivate (CallFunction|GetIdlePSC) false))
