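; Pickaxe weapon Blueprint for the CiderPetalClover cosmetic that adds reactive visual effects when it plays impact FX.
; After the parent impact handling, it checks frontend display context and the struck surface against the weapon's impact surface list.
; It stamps game time into the weapon mesh materials' Timestamp parameter, then restarts a color-switch timeline that drives the impact effect's ColorProgress parameter.
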(event FortWeapon|EventOnPlayImpactFX
  (params (HitResult :type Hit%20Result%20Structure%20%28by%20ref%29) (ImpactPhysicalSurface :type EPhysicalSurface%20Enum :default "SurfaceType_Default") (SpawnedPSC :type FXSystem%20Component%20Object%20Reference))
  (|CallParentFunction HitResult ImpactPhysicalSurface SpawnedPSC)
  (if (Reactive|Frontend|IsCosmeticObjectDisplayedinFrontendContext 0)
    _
    (elif (Utilities|Array|ContainsItem (Variables|Default|GetImpactPhysicalSurfaces) ImpactPhysicalSurface)
      _))
  (Utilities|FlowControl|Sequence
    (:then_0
      (Rendering|Material|SetScalarParameterValueonMaterials (Variables|Visual|GetWeaponMesh) "Timestamp" (Utilities|Time|GetGameTimeinSeconds)))
    (:then_1
      ; FX Color Switch
      (bind (_direction _fx_colorswitch) (Timeline|Timeline PlayFromStart)
        (:Update
          (Effects|Components|ParticleSystem|SetFloatParameter (Variables|Default|GetImpact%20FX) "ColorProgress" _fx_colorswitch))))))

(fn ConstructionScript
  (|CallParentFunction))
