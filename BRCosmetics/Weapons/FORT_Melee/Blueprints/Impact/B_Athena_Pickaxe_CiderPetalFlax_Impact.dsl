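; Melee weapon Blueprint for the CiderPetalFlax pickaxe that adds weak-point hit effects.
; Each impact forwards its physical surface through a server event to a client event, which stores it and checks it against the weak-point surface list.
; Display in a frontend context also counts as a weak-point hit.
; Such hits drive a Niagara effect on the handle socket and a game-time timestamp parameter on the weapon mesh materials.
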(event FortWeapon|EventOnPlayImpactFX
  (params (HitResult :type Hit%20Result%20Structure%20%28by%20ref%29) (ImpactPhysicalSurface :type EPhysicalSurface%20Enum :default "SurfaceType_Default") (SpawnedPSC :type FXSystem%20Component%20Object%20Reference))
  (|CallParentFunction HitResult ImpactPhysicalSurface SpawnedPSC)
  (CallFunction|CEFXServer ImpactPhysicalSurface))

(event EventBeginPlay)

(event Custom|CE_FX_Server
  (params (Impact%20Physical%20Surface :type EPhysicalSurface%20Enum :default "SurfaceType_Default"))
  (CallFunction|CEFXClient Impact%20Physical%20Surface))

(event EventTick
  (params (DeltaSeconds :type Float%20%28single-precision%29)))

(event Custom|CE_FX_Client
  (params (Impact%20Physical%20Surface :type EPhysicalSurface%20Enum :default "SurfaceType_Default"))
  (bind _output_get (Variables|Default|SetCurrent%20Impact%20Physical%20Surface Impact%20Physical%20Surface))
  (Variables|Default|SetIs%20hit%20weak%20point (or (Utilities|Array|ContainsItem (Variables|Default|GetWeakPoint%20Physical%20Surface) _output_get) (Reactive|Frontend|IsCosmeticObjectDisplayedinFrontendContext 0)))
  (if (Variables|Default|GetIs%20hit%20weak%20point)
    (Effects|SpawnFXSystemAttached "/BRCosmetics/Weapons/FORT_Melee/Pickaxe_CiderPetalFlax/FX/NS_Pickaxe_CiderPetalFlax_Idle_Weakpoint.NS_Pickaxe_CiderPetalFlax_Idle_Weakpoint" (Variables|Visual|GetWeaponMesh) "Melee_TwoHanded_Handle")
    (Rendering|Material|SetScalarParameterValueonMaterials (Variables|Visual|GetWeaponMesh) "Timestamp_WeakPoint" (Utilities|Time|GetGameTimeinSeconds))))

(fn OnRep_CurrentImpactPhysicalSurface)

(fn ConstructionScript
  (|CallParentFunction))
