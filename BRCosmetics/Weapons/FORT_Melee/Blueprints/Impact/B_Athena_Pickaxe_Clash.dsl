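; Melee pickaxe weapon Blueprint for the Clash pickaxe that links its anim trail VFX to the wielder's MiscOrTail cosmetic part.
; On instigator set and on equip/attach, it runs parent logic, then waits for that part to be applied on the player pawn.
; Identical logic in both events feeds a valid part mesh into the trail Niagara component's "SkeletalMesh" parameter.
; The construction script disables auto-activation of the idle particle component.
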(event FortWeapon|EventOnInstigatorSet
  (|CallParentFunction)
  (bind _asfort_player_pawn (Utilities|Casting|CastToFortPlayerPawn (Game|GetInstigator)))
  (bind (_pawn _part _meshpart) (Cosmetics|WaitForCosmeticSpecificPartApplied _asfort_player_pawn true "MiscOrTail")
    (:OnPartApplied
      (Utilities|IsValid _meshpart
        (:Is%20Valid
          (bind _asniagara_particle_system_component (Utilities|Casting|CastToNiagaraComponent (CallFunction|GetAnimTrailVFXComponent)))
          (Niagara|SetNiagaraSkeletalMeshComponent _asniagara_particle_system_component "SkeletalMesh" _meshpart))))))

(event FortWeapon|EventOnEquippedAndAttached
  (|CallParentFunction)
  (bind _asfort_player_pawn (Utilities|Casting|CastToFortPlayerPawn (Game|GetInstigator)))
  (bind (_pawn _part _meshpart) (Cosmetics|WaitForCosmeticSpecificPartApplied _asfort_player_pawn true "MiscOrTail")
    (:OnPartApplied
      (Utilities|IsValid _meshpart
        (:Is%20Valid
          (bind _asniagara_particle_system_component (Utilities|Casting|CastToNiagaraComponent (CallFunction|GetAnimTrailVFXComponent)))
          (Niagara|SetNiagaraSkeletalMeshComponent _asniagara_particle_system_component "SkeletalMesh" _meshpart))))))

(fn ConstructionScript
  (|CallParentFunction)
  (Components|Activation|SetAutoActivate (CallFunction|GetIdlePSC) false))
