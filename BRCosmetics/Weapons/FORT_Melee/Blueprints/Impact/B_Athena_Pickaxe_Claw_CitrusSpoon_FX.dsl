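; Melee pickaxe weapon Blueprint for the CitrusSpoon claw pickaxe that links its swing VFX to the body of the wielding character.
; On equip and attach, it runs the parent logic, then casts the instigator to a Fort pawn and validates its mesh.
; It casts the swing VFX component to Niagara and feeds that skeletal mesh into the "Pawn" parameter, printing a debug message if the Niagara cast fails.
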(event FortWeapon|EventOnEquippedAndAttached
  (|CallParentFunction)
  (bind _asfort_pawn (Utilities|Casting|CastToFortPawn (Game|GetInstigator)))
  (Utilities|IsValid (Variables|/Script/Engine.Character|GetMesh _asfort_pawn)
    (:Is%20Valid
      (bind _asniagara_particle_system_component (Utilities|Casting|CastToNiagaraComponent (CallFunction|GetSwingVFXComponent))
        (:then
          (Niagara|SetNiagaraSkeletalMeshComponent _asniagara_particle_system_component "Pawn" (Variables|/Script/Engine.Character|GetMesh _asfort_pawn)))
        (:CastFailed
          (Development|PrintString "CAST FAILED - Niagara Component"))))))

(fn ConstructionScript
  (|CallParentFunction))
