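; Fortnite Coyote Trail Dark pickaxe blueprint that customizes the weapon's cosmetic swing and idle effects.
; On equip, it binds the pawn mesh to the swing Niagara component's "Pawn" parameter, and construction disables idle effect auto-activation.
; Left and right swings run identical cosmetic-gated logic keyed on the active montage section.
; "Swing1" spawns an attached Niagara swing effect and hides the anim trail, while other sections show the trail.
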(event FortWeapon|EventOnEquippedAndAttached
  (|CallParentFunction)
  (bind _asfort_pawn (Utilities|Casting|CastToFortPawn (Game|GetInstigator)))
  (Utilities|IsValid (Variables|/Script/Engine.Character|GetMesh _asfort_pawn)
    (:Is%20Valid
      (bind _asniagara_particle_system_component (Utilities|Casting|CastToNiagaraComponent (CallFunction|GetSwingVFXComponent))
        (:then
          (Niagara|SetNiagaraSkeletalMeshComponent _asniagara_particle_system_component "Pawn" (Variables|/Script/Engine.Character|GetMesh _asfort_pawn)))
        (:CastFailed
          (Development|PrintString "CAST FAILED - Niagara Component"))))))

(event EventSwingRight
  (bind _returnvalue (Game|GetInstigator))
  (bind _asfort_pawn (Utilities|Casting|CastToFortPawn _returnvalue))
  (bind _mesh (Variables|/Script/Engine.Character|GetMesh _asfort_pawn))
  (bind _returnvalue_2 (Components|SkeletalMesh|GetAnimInstance _mesh))
  (bind _returnvalue_3 (Animation|Montage|GetCurrentActiveMontage _returnvalue_2))
  (bind _returnvalue_4 (Animation|Montage|MontageGetCurrentSection _returnvalue_2 _returnvalue_3))
  (bind _returnvalue_5 (== _returnvalue_4 "Swing1"))
  (|CallParentFunction)
  (Networking|CanExecuteCosmeticEvents
    (:True
      (Utilities|FlowControl|Sequence
        (:then_0
          (if _returnvalue_5
            (Effects|SpawnFXSystemAttached "/BRCosmetics/Weapons/FORT_Melee/PIckaxe_CoyoteTrailDark/FX/NS_Pickaxe_CoyoteTrailDark_Swing.NS_Pickaxe_CoyoteTrailDark_Swing" (Variables|Visual|GetWeaponMesh) "Melee_TwoHanded_Handle")))
        (:then_1
          (Development|SetHiddeninGame (CallFunction|GetAnimTrailVFXComponent) _returnvalue_5 false))))))

(event EventSwingLeft
  (bind _returnvalue (Game|GetInstigator))
  (bind _asfort_pawn (Utilities|Casting|CastToFortPawn _returnvalue))
  (bind _mesh (Variables|/Script/Engine.Character|GetMesh _asfort_pawn))
  (bind _returnvalue_2 (Components|SkeletalMesh|GetAnimInstance _mesh))
  (bind _returnvalue_3 (Animation|Montage|GetCurrentActiveMontage _returnvalue_2))
  (bind _returnvalue_4 (Animation|Montage|MontageGetCurrentSection _returnvalue_2 _returnvalue_3))
  (bind _returnvalue_5 (== _returnvalue_4 "Swing1"))
  (|CallParentFunction)
  (Networking|CanExecuteCosmeticEvents
    (:True
      (Utilities|FlowControl|Sequence
        (:then_0
          (if _returnvalue_5
            (Effects|SpawnFXSystemAttached "/BRCosmetics/Weapons/FORT_Melee/PIckaxe_CoyoteTrailDark/FX/NS_Pickaxe_CoyoteTrailDark_Swing.NS_Pickaxe_CoyoteTrailDark_Swing" (Variables|Visual|GetWeaponMesh) "Melee_TwoHanded_Handle")))
        (:then_1
          (Development|SetHiddeninGame (CallFunction|GetAnimTrailVFXComponent) _returnvalue_5 false))))))

(fn ConstructionScript
  (|CallParentFunction)
  (Components|Activation|SetAutoActivate (CallFunction|GetIdlePSC) false))
