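; Weapon Blueprint for the CubeQueen pickaxe cosmetic that drives its idle Niagara VFX during swings.
; Left and right swing events call the parent, cast the idle VFX component to a Niagara component and set its IsSwing bool to true.
; A one-second retriggerable delay then sets IsSwing back to false, so each event's repeated swings extend its own timer.
; The construction script only calls the parent.
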(event EventSwingRight
  (|CallParentFunction)
  (bind _asniagara_particle_system_component (Utilities|Casting|CastToNiagaraComponent (CallFunction|GetIdleVFXComponent)))
  (Niagara|SetNiagaraVariable(Bool) _asniagara_particle_system_component "IsSwing" true)
  (Utilities|FlowControl|RetriggerableDelay 1.0)
  (Niagara|SetNiagaraVariable(Bool) _asniagara_particle_system_component "IsSwing" false))

(event EventSwingLeft
  (|CallParentFunction)
  (bind _asniagara_particle_system_component (Utilities|Casting|CastToNiagaraComponent (CallFunction|GetIdleVFXComponent)))
  (Niagara|SetNiagaraVariable(Bool) _asniagara_particle_system_component "IsSwing" true)
  (Utilities|FlowControl|RetriggerableDelay 1.0)
  (Niagara|SetNiagaraVariable(Bool) _asniagara_particle_system_component "IsSwing" false))

(fn ConstructionScript
  (|CallParentFunction))
