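; Fortnite pickaxe weapon Blueprint for the CyberArmor Female one-handed melee cosmetic, linking its idle Niagara effect to the wielder's body.
; On instigator set and on equip, it caches the pawn's mesh and feeds it to the idle VFX component's User.InstigatorMesh parameter.
; An invalid VFX component triggers a short delay.
; Swing events record left or right in a flag, and construction disables auto-activation on the idle particle component.
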(event FortWeapon|EventOnInstigatorSet
  (bind _returnvalue (CallFunction|GetIdleVFXComponent))
  (|CallParentFunction)
  (bind _asfort_pawn (Utilities|Casting|CastToFortPawn (Game|GetInstigator)))
  (Variables|Default|SetInstigatorSkeletalMesh (Variables|/Script/Engine.Character|GetMesh _asfort_pawn))
  (Utilities|IsValid _returnvalue
    (:Is%20Valid
      _)
    (:Is%20Not%20Valid
      (Utilities|FlowControl|Delay 0.2)))
  (bind _asniagara_particle_system_component (Utilities|Casting|CastToNiagaraComponent _returnvalue)
    (:then
      (Niagara|SetNiagaraSkeletalMeshComponent _asniagara_particle_system_component "User.InstigatorMesh" (Variables|Default|GetInstigatorSkeletalMesh)))
    (:CastFailed
      (Development|PrintString "Niagara Cast Failed"))))

(event FortWeapon|EventOnEquippedAndAttached
  (bind _returnvalue (CallFunction|GetIdleVFXComponent))
  (|CallParentFunction)
  (bind _asfort_pawn (Utilities|Casting|CastToFortPawn (Game|GetInstigator)))
  (Variables|Default|SetInstigatorSkeletalMesh (Variables|/Script/Engine.Character|GetMesh _asfort_pawn))
  (Utilities|IsValid _returnvalue
    (:Is%20Valid
      _)
    (:Is%20Not%20Valid
      (Utilities|FlowControl|Delay 0.2)))
  (bind _asniagara_particle_system_component (Utilities|Casting|CastToNiagaraComponent _returnvalue)
    (:then
      (Niagara|SetNiagaraSkeletalMeshComponent _asniagara_particle_system_component "User.InstigatorMesh" (Variables|Default|GetInstigatorSkeletalMesh)))
    (:CastFailed
      (Development|PrintString "Niagara Cast Failed"))))

(event FortWeapon|EventOnLocalUnEquip
  (|CallParentFunction))

(event FortWeapon|EventOnUnEquippedAndDetached
  (|CallParentFunction))

(event EventSwingRight
  (|CallParentFunction)
  (Variables|Default|SetRight%20? true))

(event EventSwingLeft
  (|CallParentFunction)
  (Variables|Default|SetRight%20? false))

(fn ConstructionScript
  (|CallParentFunction)
  (Components|Activation|SetAutoActivate (CallFunction|GetIdlePSC) false))
