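; Cosmetic Blueprint for the DerangedMile axe pickaxe, whose look reacts to the player's elimination count.
; In matches, it watches the AthenaKills stat, prints it, and relays it through ReactivityServerEvent to Reactivity.
; Reactivity sets CurrentIndex on mesh materials and idle VFX, stamping game time, only for counts up to three.
; Frontend display instead cycles a preview count on looping timers; impact VFX receive the current count clamped to 0–3.
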(event Custom|CustomEvent
  (bind _result (Math|Integer|IncrementInt (Variables|Default|GetFEKillCount))
    (:%20%20
      (CallFunction|Reactivity _result)
      (Utilities|Time|ClearandInvalidateTimerbyHandle (Variables|Default|GetNewVar))
      (bind _returnvalue (Utilities|Time|SetTimerbyEvent (delegate Custom|CustomEvent_0) 5.0))
      (Variables|Default|SetNewVar _returnvalue))))

(event Custom|CustomEvent_0
  (if (<= (Variables|Default|GetFEKillCount) 3)
    (bind _result (Math|Integer|IncrementInt (Variables|Default|GetFEKillCount))
      (:%20%20
        (CallFunction|Reactivity _result)
        (Utilities|Time|ClearandInvalidateTimerbyHandle (Variables|Default|GetNewVar))
        (bind _returnvalue (Utilities|Time|SetTimerbyEvent (delegate Custom|CustomEvent_0) 5.0))
        (Variables|Default|SetNewVar _returnvalue)))
    (else
      (bind _output_get (Variables|Default|SetFEKillCount 0))
      (CallFunction|Reactivity _output_get)
      (bind _returnvalue_2 (Utilities|Time|SetTimerbyEvent (delegate Custom|CustomEvent) 5.0))
      (Variables|Default|SetNewVar _returnvalue_2))))

(event EventOnInitWeaponCosmetics
  (|CallParentFunction)
  (if (Reactive|Frontend|IsCosmeticObjectDisplayedinFrontendContext 0)
    (bind _output_get (Variables|Default|SetFEKillCount 0))
    (CallFunction|Reactivity _output_get)
    (Utilities|Time|ClearandInvalidateTimerbyHandle (Variables|Default|GetNewVar))
    (bind _returnvalue (Utilities|Time|SetTimerbyEvent (delegate Custom|CustomEvent) 5.0))
    (Variables|Default|SetNewVar _returnvalue)
    (else
      (bind _asfort_player_pawn (Utilities|Casting|CastToFortPlayerPawn (Variables|Actor|GetInstigator)))
      (bind (_statname _statvalue) (Cosmetics|WatchCosmeticStat _asfort_player_pawn "(TagName=\"StatManager.AthenaKills\")")
        (:OnStatChanged
          (bind _output_get_2 (Variables|Default|SetKillCount _statvalue))
          (Development|PrintString (Utilities|String|Append "Elimination Kill Count: " (Utilities|String|ToString(Integer) _output_get_2)) true true "(R=0.940123,G=1.000000,B=0.000009,A=1.000000)" 1e+01)
          (CallFunction|ReactivityServerEvent _output_get_2))))))

(event Custom|Reactivity
  (params (NewParam :type Integer))
  (bind _weaponmesh (Variables|Visual|GetWeaponMesh))
  (if (<= NewParam 3)
    (Rendering|Material|SetScalarParameterValueonMaterials _weaponmesh "CurrentIndex" (Math|Conversions|ToFloat(Integer) NewParam))
    (Rendering|Material|SetScalarParameterValueonMaterials _weaponmesh "Timestamp_1" (Utilities|Time|GetGameTimeinSeconds))
    (Effects|Components|ParticleSystem|SetIntParameter (CallFunction|GetIdleVFXComponent) "CurrentIndex" NewParam)))

(event Custom|ReactivityServerEvent
  (params (KillCount :type Integer))
  (CallFunction|Reactivity KillCount))

(event Custom|ImpactReactivity
  (params (KillCount :type Integer))
  (Effects|Components|ParticleSystem|SetIntParameter (Variables|Default|GetImpactVFX) "CurrentIndex" KillCount))

(event Custom|ImpactReactivityServerEvent
  (params (KillCount :type Integer))
  (CallFunction|ImpactReactivity KillCount))

(event FortWeapon|EventOnPlayImpactFX
  (params (HitResult :type Hit%20Result%20Structure%20%28by%20ref%29) (ImpactPhysicalSurface :type EPhysicalSurface%20Enum :default "SurfaceType_Default") (SpawnedPSC :type FXSystem%20Component%20Object%20Reference))
  (bind _killcount (Variables|Default|GetKillCount))
  (bind _returnvalue (Math|Integer|Clamp(Integer) _killcount 0 3))
  (Variables|Default|SetImpactVFX SpawnedPSC)
  (|CallParentFunction HitResult ImpactPhysicalSurface SpawnedPSC)
  (if (Reactive|Frontend|IsCosmeticObjectDisplayedinFrontendContext 0)
    (Effects|Components|ParticleSystem|SetIntParameter (Variables|Default|GetImpactVFX) "CurrentIndex" (Math|Integer|Clamp(Integer) (Variables|Default|GetFEKillCount) 0 3))
    (else
      (CallFunction|ImpactReactivity _returnvalue)
      (CallFunction|ImpactReactivityServerEvent _returnvalue))))

(fn ConstructionScript
  (|CallParentFunction))
