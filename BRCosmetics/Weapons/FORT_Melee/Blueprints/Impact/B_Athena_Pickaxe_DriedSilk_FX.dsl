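; Impact effects Blueprint for the DriedSilk pickaxe, extending the parent's on-play impact FX event.
; After the parent call, the event includes a switch on physical surface type and a short delay before TweakHumanImpactFX.
; That function rotates the Impact FX component in world space to face the first player's camera, using a look-at rotation.
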(event FortWeapon|EventOnPlayImpactFX
  (params (HitResult :type Hit%20Result%20Structure%20%28by%20ref%29) (ImpactPhysicalSurface :type EPhysicalSurface%20Enum :default "SurfaceType_Default") (SpawnedPSC :type FXSystem%20Component%20Object%20Reference))
  (|CallParentFunction HitResult ImpactPhysicalSurface SpawnedPSC)
  (Utilities|FlowControl|Sequence
    (:then_1
      (switch Utilities|FlowControl|Switch|SwitchonEPhysicalSurface ImpactPhysicalSurface
        (:SurfaceType4
          _)
        (:SurfaceType5
          _))
      (Utilities|FlowControl|Delay 0.05)
      (CallFunction|TweakHumanImpactFX))))

(fn TweakHumanImpactFX
  (Transformation|SetWorldRotation (Variables|Default|GetImpact%20FX) (Math|Rotator|FindLookatRotation (Transformation|GetWorldLocation (Variables|Default|GetImpact%20FX)) (Camera|GetCameraLocation (Game|GetPlayerCameraManager 0))) false false))

(fn ConstructionScript
  (|CallParentFunction))
