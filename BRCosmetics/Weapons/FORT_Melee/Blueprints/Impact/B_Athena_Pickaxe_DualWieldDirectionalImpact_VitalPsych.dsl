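; Dual-wield pickaxe Blueprint for the VitalPsych cosmetic that orients impact FX by swinging hand and adds a double-strike final impact.
; Swing events track the hand and second left swing, driving a "SecondImpact" parameter on both hands' trails and impact effects.
; Impacts rotate from the hit normal, with the double strike spawning an extra Niagara system in gameplay.
; In frontend display, trail notifies read swing state and impacts use offset positions and separate rotations.
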(event FortWeapon|EventOnEquippedAndAttached
  ;[1 1 1 1 200] USEFUL / SITUATIONAL
  ;;[1 1 1] Equip trail FX
  (|CallParentFunction)
  (bind _asplayer_pawn_athena (Utilities|Casting|CastToPlayerPawn_Athena (Game|GetPlayerCharacter 0))))

(event EventSwingLeft2
  ;[1 1 1] AnimTrail Last Slash Parameter
  (|CallParentFunction)
  (Variables|Default|SetLeft2%20Swing? true)
  (if (Variables|Default|GetLeft2%20Swing?)
    (Effects|Components|ParticleSystem|SetBoolParameter (CallFunction|GetAnimTrailsPSC) "SecondImpact" true)
    (Effects|Components|ParticleSystem|SetBoolParameter (CallFunction|GetAnimTrailsOffhandPSC) "SecondImpact" true)))

(event FortWeapon|EventOnPlayImpactFX
  (params (HitResult :type Hit%20Result%20Structure%20%28by%20ref%29) (ImpactPhysicalSurface :type EPhysicalSurface%20Enum :default "SurfaceType_Default") (SpawnedPSC :type FXSystem%20Component%20Object%20Reference))
  (bind (_bblockinghit _binitialoverlap _time _distance _location _impactpoint _normal) (Collision|BreakHitResult HitResult))
  (bind _returnvalue (Math|Rotator|MakeRotfromX _normal))
  (bind _returnvalue_2 (Reactive|Frontend|IsCosmeticObjectDisplayedinFrontendContext 0))
  (bind _finalimpactoffset (Variables|Default|GetFinalImpactOffset))
  (bind _returnvalue_3 (Transformation|GetActorTransform))
  (bind _returnvalue_4 (Math|Transform|TransformLocation _returnvalue_3 _finalimpactoffset))
  (bind _impactfx (Variables|Default|GetImpactFX))
  ;[0.425 0.536 0.3211 1 200] FinalImpact(DoubleStrike)
  (|CallParentFunction HitResult ImpactPhysicalSurface SpawnedPSC)
  ;;[1 1 1] StoreImpact Ref
  (bind _output_get (Variables|Default|SetImpactFX SpawnedPSC))
  (bind _asniagara_particle_system_component (Utilities|Casting|CastToNiagaraComponent _output_get))
  (Utilities|FlowControl|Sequence
    (:then_0
      (if _returnvalue_2
        ;;[1 1 1] Fix Frontend Weirdness with Spawning Impacts for Offhand
        (Transformation|SetWorldLocation (Variables|Default|GetImpactFX) (Math|Vector|SelectVector (Transformation|GetWorldLocation (Variables|Default|GetImpactFX)) (Math|Transform|TransformLocation (Transformation|GetActorTransform) (Variables|Default|GetImpactOffset)) (Variables|Default|GetRight?)) false false)
        (Transformation|AddLocalRotation (Variables|Default|GetImpactFX) (Math|Rotator|SelectRotator (Variables|Default|GetLeft%20Rotation_FRONTEND) (Variables|Default|GetRightRotation_FRONTEND) (Variables|Default|GetRight?)) false false)
        (else
          ;;[1 1 1] GameplayRotation
          (Transformation|SetWorldRotation SpawnedPSC (Math|Rotator|CombineRotators _returnvalue (select (Variables|Default|GetRight?) (Variables|Default|GetRight%20Rotation) (Variables|Default|GetLeft%20Rotation))) false false)))
      (if (Variables|Default|GetLeft2%20Swing?)
        _
        (elif (Variables|Default|GetbFrontEndBothHands)
          _))
      (bind _asniagara_system (Utilities|Casting|CastToNiagaraSystem (Effects|Components|ParticleSystem|GetFXSystemAsset _impactfx)))
      (Effects|Components|ParticleSystem|SetBoolParameter (Variables|Default|GetImpactFX) "SecondImpact" true)
      (if (Reactive|Frontend|IsCosmeticObjectDisplayedinFrontendContext 0)
        ;;[1 1 1] FrontendLocation
        (Transformation|SetWorldLocation (Variables|Default|GetImpactFX) _returnvalue_4 false false)
        (else
          (bind _returnvalue_5 (Niagara|SpawnSystemAttached _asniagara_system _impactfx "None" "0, 0, 0" "0, 0, 0" "KeepRelativeOffset" true))
          ;;[1 1 1] GameplayBoolSet
          (Effects|Components|ParticleSystem|SetBoolParameter _returnvalue_5 "SecondImpact" true)
          (Transformation|SetWorldRotation _returnvalue_5 (Math|Rotator|CombineRotators _returnvalue "0,60.000000, 0") false false)
          (Transformation|SetWorldLocation _returnvalue_5 (Math|Vector|SelectVector _returnvalue_4 (Transformation|GetWorldLocation (Variables|Default|GetImpactFX)) (Reactive|Frontend|IsCosmeticObjectDisplayedinFrontendContext 0)) false false))))
    (:then_1
      ;;[1 1 1] Set GameplayBool
      (if (not _returnvalue_2)
        (Niagara|SetNiagaraVariable(Bool) _asniagara_particle_system_component "SecondImpact" (Variables|Default|GetLeft2%20Swing?))))))

(event EventSwingLeft_Common
  ;[1 1 1] Basic DualWield Settings
  (|CallParentFunction)
  (Variables|Default|SetRight? false))

(event EventSwingRight_Common
  ;[1 1 1] Basic DualWield Settings
  (|CallParentFunction)
  (Variables|Default|SetRight? true))

(event EventSwingLeft
  ;[1 1 1] Basic DualWield Settings
  (|CallParentFunction)
  (Variables|Default|SetLeft2%20Swing? false))

(event EventSwingRight
  (bind _left2_swing (Variables|Default|GetLeft2%20Swing?))
  ;[1 1 1] Basic DualWield Settings
  (|CallParentFunction)
  (Variables|Default|SetLeft2%20Swing? false)
  ;[1 1 1] ResetSecondImpactTrail
  (Effects|Components|ParticleSystem|SetBoolParameter (CallFunction|GetAnimTrailsPSC) "SecondImpact" _left2_swing)
  (Effects|Components|ParticleSystem|SetBoolParameter (CallFunction|GetAnimTrailsOffhandPSC) "SecondImpact" _left2_swing))

(event EventAnimTrailsNotify
  (params (bActive :type Boolean))
  (bind _impactfx (Variables|Default|GetImpactFX))
  ;[1 1 1] Basic DualWield Settings
  (|CallParentFunction bActive)
  (if bActive
    (if (Reactive|Frontend|IsCosmeticObjectDisplayedinFrontendContext 0)
      (switch Utilities|FlowControl|Switch|SwitchonEFortDualWieldSwingState (CallFunction|GetCurrentSwingState)
        (:MainHand
          (Variables|Default|SetRight? true)
          (Effects|Components|ParticleSystem|SetBoolParameter _impactfx "SecondImpact" false))
        (:OffHand
          (Variables|Default|SetRight? false)
          (Effects|Components|ParticleSystem|SetBoolParameter _impactfx "SecondImpact" false))
        (:BothHands
          (Effects|Components|ParticleSystem|SetBoolParameter (Variables|Default|GetImpactFX) "SecondImpact" true)
          (Variables|Default|SetbFrontEndBothHands true)))
      (Variables|Default|SetbFrontEndBothHands false))))

(fn ConstructionScript
  (|CallParentFunction))
