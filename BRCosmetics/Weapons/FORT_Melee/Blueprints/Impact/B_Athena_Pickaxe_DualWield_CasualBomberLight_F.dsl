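; Dual-wield pickaxe cosmetic Blueprint (CasualBomberLight) with a variant-dependent eyeball and a kill-reactive horn glow.
; On equip, it reads the pickaxe loadout slot's active Mesh variant from the instigating player pawn and makes Eye_R1 visible for index zero or lower.
; On kill updates, it maps the watched kill count between min and max thresholds onto the HornGlowINT material scalar of the weapon mesh.
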(event EventOnInitWeaponCosmetics
  (|CallParentFunction))

(event FortWeapon|EventOnEquippedAndAttached
  (|CallParentFunction)
  (bind _asfort_player_pawn (Utilities|Casting|CastToFortPlayerPawn (Variables|Actor|GetInstigator)))
  (bind _index (CosmeticLoadout|GetVariantChannelActiveIndexforSlotActor "(TagName=\"LoadoutSlot.Pickaxe\")" _asfort_player_pawn "(TagName=\"Cosmetics.Variant.Channel.Mesh\")" (Utilities|Array|MakeArray "(TagName=\"Cosmetics.Variant.Property.Stage1\")" "(TagName=\"Cosmetics.Variant.Property.Stage2\")")))
  (Variables|Default|SetCurrentVariant _index)
  ;[1 1 1] Eyeball visibility based on variant
  (if (<= (Variables|Default|GetCurrentVariant) 0)
    (Rendering|SetVisibility (Variables|Default|GetEye_R1) true false)))

(event EventUpdateBasedOnKills
  (params (Watched%20Kills :type Integer))
  ;[1 1 1] Horn GLow
  (|CallParentFunction Watched%20Kills)
  (Variables|Default|SetKillCount (Math|Conversions|ToFloat(Integer) Watched%20Kills))
  (Rendering|Material|SetScalarParameterValueonMaterials (Variables|Visual|GetWeaponMesh) "HornGlowINT" (Math|Float|MapRangeClamped (Variables|Default|GetKillCount) (Variables|Default|GetGlowKillThresholdMin) (Variables|Default|GetGlowKillThresholdMax) 0.0 6.0)))

(fn ConstructionScript
  (|CallParentFunction))
