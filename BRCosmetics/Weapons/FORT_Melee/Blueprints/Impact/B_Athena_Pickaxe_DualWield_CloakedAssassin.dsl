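; Reactive cosmetic Blueprint for the CloakedAssassin dual-wield pickaxe, which powers up visually as it hits certain surface types.
; Qualifying impacts increment replicated Stage; OnRep_Stage, for material variant 0 only, plays a sound, fires burst FX and updates Niagara damage level.
; Stages 1-3 raise glow and opacity on the weapon mesh and a dynamic particle-cube material.
; Frontend displays preview stages, detecting non-reactive styles by a NULL burst-FX name.
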(event FortWeapon|EventOnInstigatorSet
  (bind _returnvalue (Game|GetInstigator))
  (|CallParentFunction)
  (Utilities|IsValid _returnvalue
    (:Is%20Valid
      (bind _asfort_player_pawn (Utilities|Casting|CastToFortPlayerPawn _returnvalue))
      (bind _index (CosmeticLoadout|GetVariantChannelActiveIndexforSlotActor "(TagName=\"LoadoutSlot.Pickaxe\")" _asfort_player_pawn "(TagName=\"Cosmetics.Variant.Channel.Material\")" (Utilities|Array|MakeArray "(TagName=\"Cosmetics.Variant.Property.Mat1\")" "(TagName=\"Cosmetics.Variant.Property.Mat2\")")))
      (Variables|Default|SetVariantIndex _index))))

(event EventOnInitWeaponCosmetics
  (bind _weaponburstfx (Variables|Default|GetWeaponBurstFX))
  (bind _returnvalue (Utilities|GetDisplayName _weaponburstfx))
  (bind (_lefts _rights) (Utilities|String|Split _returnvalue "."))
  (|CallParentFunction)
  (Utilities|FlowControl|Sequence
    (:then_0
      (bind _asfort_player_pawn_athena (Utilities|Casting|CastToFortPlayerPawnAthena (Game|GetInstigator)))
      (Class|FortPlayerPawnAthena|AddComponenttoPawnHighlightSet _asfort_player_pawn_athena (Variables|Default|GetWeaponBurstFX))
      (Class|FortPlayerPawnAthena|AddComponenttoPawnHighlightSet _asfort_player_pawn_athena (Variables|Default|GetWeaponBurstFXOffhand))
      (Class|FortPlayerPawnAthena|AddComponenttoPawnHighlightSet _asfort_player_pawn_athena (CallFunction|GetIdleVFXComponent))
      (Class|FortPlayerPawnAthena|AddComponenttoPawnHighlightSet _asfort_player_pawn_athena (CallFunction|GetIdleOffhandVFXComponent)))
    (:then_1
      (if (Reactive|Frontend|IsCosmeticObjectDisplayedinFrontendContext self)
        (Utilities|FlowControl|DelayUntilNextTick)
        (Development|PrintString _rights)
        (if (Utilities|String|EqualExactly(String) _rights "WeaponBurstFX NS_Pickaxe_CloakedAssassin_NULL")
          (Development|PrintString "Non-Reactive")
          (CallFunction|Stage0MatSetup)
          (else
            (Utilities|FlowControl|DoOnce false
              (:Completed
                (CallFunction|ParticleSetup)
                (switch int (Variables|Default|GetStateInt)
                  (:0
                    (CallFunction|Stage0MatSetup)
                    (Variables|Default|SetStateInt 1)
                    (Utilities|FlowControl|RetriggerableDelay 3.0))
                  (:1
                    _)
                  (:2
                    (CallFunction|Stage2MatSetup)
                    (CallFunction|FireVFX)
                    (Variables|Default|SetStateInt 3))
                  (:3
                    (CallFunction|Stage3MatSetup)
                    (CallFunction|FireVFX)
                    (Variables|Default|SetStateInt 0)))
                (CallFunction|Stage1MatSetup)
                (CallFunction|FireVFX)
                (Variables|Default|SetStateInt 2)))))
        (else
          (Utilities|FlowControl|Delay 0.5)
          (CallFunction|ParticleSetup))))))

(event Custom|Fire_VFX
  (Components|Activation|Activate (Variables|Default|GetWeaponBurstFX) false))

(event EventBeginPlay
  (|CallParentFunction)
  (Transformation|AttachComponentToComponent (Variables|Default|GetWeaponBurstFX) (Variables|Visual|GetWeaponMesh) "None" "KeepRelative" "KeepRelative" "KeepRelative" true)
  (Transformation|AttachComponentToComponent (Variables|Default|GetWeaponBurstFXOffhand) (Variables|Visual|GetWeaponMeshOffhand) "None" "KeepRelative" "KeepRelative" "KeepRelative" true))

(event FortWeapon|EventOnEquippedAndAttached
  (bind _idlefx (Variables|Default|GetIdleFX))
  (|CallParentFunction)
  (Utilities|Casting|CastToFortGameStateFrontEnd (Game|GetGameState)
    (:then
      (Niagara|SetNiagaraVariable(Bool) _idlefx "bIsFrontEndPreview" true)
      (Niagara|SetNiagaraVariable(Int32) _idlefx "User.DamageLevel" 3)
      (CallFunction|Stage3MatSetup))
    (:CastFailed
      (Niagara|SetNiagaraVariable(Bool) _idlefx "bIsFrontEndPreview" false)
      (CallFunction|Stage0MatSetup))))

(event Custom|ParticleSetup
  (bind _idlefx (Variables|Default|GetIdleFX))
  (bind _asniagara_particle_system_component (Utilities|Casting|CastToNiagaraComponent (CallFunction|GetSwingVFXComponent))
    (:then
      (Variables|Default|SetSwingFX _asniagara_particle_system_component))
    (:CastFailed
      _))
  (bind _asniagara_particle_system_component_2 (Utilities|Casting|CastToNiagaraComponent (CallFunction|GetSwingOffhandVFXComponent))
    (:then
      (Variables|Default|SetSwingOffhandFX _asniagara_particle_system_component_2))
    (:CastFailed
      _))
  (bind _asniagara_particle_system_component_3 (Utilities|Casting|CastToNiagaraComponent (CallFunction|GetIdleVFXComponent))
    (:then
      (Variables|Default|SetIdleFX _asniagara_particle_system_component_3))
    (:CastFailed
      _))
  (bind _asniagara_particle_system_component_4 (Utilities|Casting|CastToNiagaraComponent (CallFunction|GetIdleOffhandVFXComponent))
    (:then
      (Variables|Default|SetIdleOffhandFX _asniagara_particle_system_component_4))
    (:CastFailed
      _))
  (bind _returnvalue (Rendering|Material|CreateDynamicMaterialInstance (Variables|Default|GetMaterialBase) "ReactiveCubeDynamicMaterial" "None" :!class /Script/Engine.KismetMaterialLibrary))
  (bind _output_get (Variables|Default|SetParticleCubeMaterial _returnvalue))
  (Niagara|SetNiagaraVariable(Material) (Variables|Default|GetIdleFX) "User.CubeMatInterface" _output_get)
  (Utilities|Casting|CastToFortGameStateFrontEnd (Game|GetGameState)
    (:then
      (Niagara|SetNiagaraVariable(Bool) _idlefx "bIsFrontEndPreview" true)
      (Niagara|SetNiagaraVariable(Int32) _idlefx "User.DamageLevel" 3)
      (CallFunction|Stage3MatSetup))
    (:CastFailed
      (Niagara|SetNiagaraVariable(Bool) _idlefx "bIsFrontEndPreview" false)
      (CallFunction|Stage0MatSetup))))

(event EventOnPlayWeaponFireFX
  (params (bPersistentFire :type Boolean) (bSecondaryFire :type Boolean))
  (Niagara|SetNiagaraVariable(Bool) (Variables|Default|GetSwingFX) "User.FinalStage" (Variables|Default|GetFinalStage)))

(event FortWeapon|EventOnPlayImpactFX
  (params (HitResult :type Hit%20Result%20Structure%20%28by%20ref%29) (ImpactPhysicalSurface :type EPhysicalSurface%20Enum :default "SurfaceType_Default") (SpawnedPSC :type FXSystem%20Component%20Object%20Reference))
  (|CallParentFunction HitResult ImpactPhysicalSurface SpawnedPSC)
  (Utilities|FlowControl|Delay 0.2)
  (if (Math|Boolean|ORBoolean (Utilities|Enum|Equal(Enum) ImpactPhysicalSurface "SurfaceType14") (Utilities|Enum|Equal(Enum) ImpactPhysicalSurface "SurfaceType15") (Utilities|Enum|Equal(Enum) ImpactPhysicalSurface "SurfaceType16") (Utilities|Enum|Equal(Enum) ImpactPhysicalSurface "SurfaceType12"))
    (CallFunction|ReplicateImpact)))

(event Custom|Play%20Weakspot%20Impact%20Sound
  (params (Stage :type Integer))
  (Audio|PlaySoundatLocation (Utilities|Array|Get(acopy) (Variables|Default|GetWeakspotImpactSounds) (Math|Integer|Clamp(Integer) Stage 0 (Utilities|Array|LastIndex (Variables|Default|GetWeakspotImpactSounds)))) (Transformation|GetWorldLocation (Variables|Visual|GetWeaponMesh))))

(event Custom|ReplicateImpact
  ;[1 1 1] Replication
  (Variables|Default|SetStage (+ (Variables|Default|GetStage) 1)))

(event Custom|Stage0MatSetup
  (bind _particlecubematerial (Variables|Default|GetParticleCubeMaterial))
  (bind _weaponmesh (Variables|Visual|GetWeaponMesh))
  ;[1 1 1] Stage Materials set
  (Rendering|Material|SetScalarParameterValueonMaterials (Variables|Visual|GetWeaponMeshOffhand) "Emissive_BrightnessMax" 0.0)
  (Rendering|Material|SetScalarParameterValueonMaterials _weaponmesh "Emissive_BrightnessMin" 0.0)
  (Rendering|Material|SetScalarParameterValueonMaterials _weaponmesh "Emissive mult" 0.05)
  (Rendering|Material|SetScalarParameterValueonMaterials _weaponmesh "glass opacity" 0.12)
  (Rendering|Material|SetScalarParameterValueonMaterials _weaponmesh "Center Opacity" 0.0)
  (Rendering|Material|SetScalarParameterValue _particlecubematerial "Emissive_BrightnessMax" 0.0 :!class /Script/Engine.MaterialInstanceDynamic)
  (Rendering|Material|SetScalarParameterValue _particlecubematerial "Emissive_BrightnessMin" 0.0 :!class /Script/Engine.MaterialInstanceDynamic)
  (Rendering|Material|SetScalarParameterValue _particlecubematerial "Emissive mult" 0.1 :!class /Script/Engine.MaterialInstanceDynamic)
  (Rendering|Material|SetScalarParameterValue _particlecubematerial "Center Opacity" 0.0 :!class /Script/Engine.MaterialInstanceDynamic)
  (Niagara|SetNiagaraVariable(Int32) (Variables|Default|GetIdleFX) "User.DamageLevel" 0)
  (Variables|Default|SetStage 0)
  (Variables|Default|SetFinalStage false))

(event Custom|Stage1MatSetup
  (bind _particlecubematerial (Variables|Default|GetParticleCubeMaterial))
  (bind _weaponmesh (Variables|Visual|GetWeaponMesh))
  ;[1 1 1] Stage Materials set
  (Rendering|Material|SetScalarParameterValueonMaterials _weaponmesh "Emissive_BrightnessMax" 1.0)
  (Rendering|Material|SetScalarParameterValueonMaterials _weaponmesh "Emissive_BrightnessMin" -0.1)
  (Rendering|Material|SetScalarParameterValueonMaterials _weaponmesh "Emissive mult" 0.15)
  (Rendering|Material|SetScalarParameterValueonMaterials _weaponmesh "glass opacity" 0.25)
  (Rendering|Material|SetScalarParameterValueonMaterials _weaponmesh "Center Opacity" 0.0)
  (Rendering|Material|SetScalarParameterValue _particlecubematerial "Emissive_BrightnessMax" 1.0 :!class /Script/Engine.MaterialInstanceDynamic)
  (Rendering|Material|SetScalarParameterValue _particlecubematerial "Emissive_BrightnessMin" -0.1 :!class /Script/Engine.MaterialInstanceDynamic)
  (Rendering|Material|SetScalarParameterValue _particlecubematerial "Emissive mult" 0.35 :!class /Script/Engine.MaterialInstanceDynamic)
  (Rendering|Material|SetScalarParameterValue _particlecubematerial "Center Opacity" 0.0 :!class /Script/Engine.MaterialInstanceDynamic))

(event Custom|Stage2MatSetup
  (bind _particlecubematerial (Variables|Default|GetParticleCubeMaterial))
  (bind _weaponmesh (Variables|Visual|GetWeaponMesh))
  ;[1 1 1] Stage Materials set
  (Rendering|Material|SetScalarParameterValueonMaterials _weaponmesh "Emissive_BrightnessMax" 3.0)
  (Rendering|Material|SetScalarParameterValueonMaterials _weaponmesh "Emissive_BrightnessMin" -0.1)
  (Rendering|Material|SetScalarParameterValueonMaterials _weaponmesh "Emissive mult" 0.25)
  (Rendering|Material|SetScalarParameterValueonMaterials _weaponmesh "glass opacity" 0.55)
  (Rendering|Material|SetScalarParameterValueonMaterials _weaponmesh "Center Opacity" 0.5)
  (Rendering|Material|SetScalarParameterValue _particlecubematerial "Emissive_BrightnessMax" 3.0 :!class /Script/Engine.MaterialInstanceDynamic)
  (Rendering|Material|SetScalarParameterValue _particlecubematerial "Emissive_BrightnessMin" -0.1 :!class /Script/Engine.MaterialInstanceDynamic)
  (Rendering|Material|SetScalarParameterValue _particlecubematerial "Emissive mult" 0.5 :!class /Script/Engine.MaterialInstanceDynamic)
  (Rendering|Material|SetScalarParameterValue _particlecubematerial "Center Opacity" 0.5 :!class /Script/Engine.MaterialInstanceDynamic))

(event Custom|Stage3MatSetup
  (bind _particlecubematerial (Variables|Default|GetParticleCubeMaterial))
  (bind _weaponmesh (Variables|Visual|GetWeaponMesh))
  ;[1 1 1] Stage Materials set
  (Rendering|Material|SetScalarParameterValueonMaterials _weaponmesh "Emissive_BrightnessMax" 5.0)
  (Rendering|Material|SetScalarParameterValueonMaterials _weaponmesh "Emissive_BrightnessMin" 0.1)
  (Rendering|Material|SetScalarParameterValueonMaterials _weaponmesh "Emissive mult" 0.35)
  (Rendering|Material|SetScalarParameterValueonMaterials _weaponmesh "glass opacity" 1.0)
  (Rendering|Material|SetScalarParameterValueonMaterials _weaponmesh "Center Opacity" 1.0)
  (Rendering|Material|SetScalarParameterValue _particlecubematerial "Emissive_BrightnessMax" 8.0 :!class /Script/Engine.MaterialInstanceDynamic)
  (Rendering|Material|SetScalarParameterValue _particlecubematerial "Emissive_BrightnessMin" 0.1 :!class /Script/Engine.MaterialInstanceDynamic)
  (Rendering|Material|SetScalarParameterValue _particlecubematerial "Emissive mult" 1.0 :!class /Script/Engine.MaterialInstanceDynamic)
  (Rendering|Material|SetScalarParameterValue _particlecubematerial "Center Opacity" 1.0 :!class /Script/Engine.MaterialInstanceDynamic)
  (Variables|Default|SetFinalStage true))

(fn OnRep_Stage
  (bind _stage (Variables|Default|GetStage))
  (Networking|CanExecuteCosmeticEvents
    (:True
      (switch int (Variables|Default|GetVariantIndex)
        (:0
          (CallFunction|PlayWeakspotImpactSound _stage)
          (Niagara|SetNiagaraVariable(Int32) (Variables|Default|GetIdleFX) "User.DamageLevel" (Math|Integer|Clamp(Integer) _stage 0 3))
          (switch int _stage
            (:1
              (CallFunction|Stage1MatSetup))
            (:2
              (CallFunction|Stage2MatSetup))
            (:3
              (CallFunction|Stage3MatSetup)))
          (Components|Activation|Activate (Variables|Default|GetWeaponBurstFX) false))))))

(fn ConstructionScript
  (|CallParentFunction))
