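; Weapon Blueprint for the dual-wield "Defect" cosmetic pickaxe.
; On equip and attach, it calls the parent, then sets a pickaxe-position switch scalar on the weapon mesh materials.
; It checks whether the cosmetic is displayed in a frontend context, but both branches write the same value, so the check has no effect.
; The construction script only calls the parent.
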(event FortWeapon|EventOnEquippedAndAttached
  (bind _weaponmesh (Variables|Visual|GetWeaponMesh))
  (|CallParentFunction)
  (if (Reactive|Frontend|IsCosmeticObjectDisplayedinFrontendContext self)
    (Rendering|Material|SetScalarParameterValueonMaterials _weaponmesh "DO NOT OVERRIDE-Pickaxe-Position-BP-Switch" 1.0)
    (else
      (Rendering|Material|SetScalarParameterValueonMaterials _weaponmesh "DO NOT OVERRIDE-Pickaxe-Position-BP-Switch" 1.0))))

(fn ConstructionScript
  (|CallParentFunction))
