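; Weapon cosmetic Blueprint for the FlameBride dual-wield pickaxe, animating Ember and Burn material glow and offhand idle VFX.
; Tick interpolates glow from weapon angular velocity and delayed targets in frontend previews, or from player movement in game.
; Swings pause idle VFX and nudge glow upward; swing ends restore VFX and run Ember and reversed Burn timelines.
; Frontend impacts add an offhand Niagara effect and reposition impact FX per swing montage section.
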(event EventOnInitWeaponCosmetics
  (bind _weaponmesh (Variables|Visual|GetWeaponMesh))
  (Actor|Tick|SetActorTickEnabled true)
  (if (Reactive|Frontend|IsCosmeticObjectDisplayedinFrontendContext 0)
    (Rendering|Material|SetScalarParameterValueonMaterials _weaponmesh "Velocity Intensity Max" 5.0)
    (Rendering|Material|SetScalarParameterValueonMaterials _weaponmesh "Flame Smoothness Max" 1.5)
    (Utilities|FlowControl|Sequence
      (:then_0
        (Variables|AngularVelocity|SetLast%20Position (Transformation|GetWorldLocation _weaponmesh))
        (Variables|AngularVelocity|SetLast%20Linear%20Velocity "0, 0, 0")
        (Variables|AngularVelocity|SetLast%20Rotation "0, 0, 0"))
      (:then_1
        (CallFunction|Started)))))

(event FortWeapon|EventOnPlayImpactFX
  (params (HitResult :type Hit%20Result%20Structure%20%28by%20ref%29) (ImpactPhysicalSurface :type EPhysicalSurface%20Enum :default "SurfaceType_Default") (SpawnedPSC :type FXSystem%20Component%20Object%20Reference))
  (bind _impact_fx (Variables|Default|GetImpact%20FX))
  (|CallParentFunction HitResult ImpactPhysicalSurface SpawnedPSC)
  (if (Reactive|Frontend|IsCosmeticObjectDisplayedinFrontendContext 0)
    (Utilities|FlowControl|Sequence
      (:then_0
        (bind _returnvalue (Effects|SpawnFXSystemAttached "/BRCosmetics/Weapons/FORT_Melee/Pickaxe_FlameBride_Offhand/FX/NS_Pickaxe_FlameBride_OffHand_Impact.NS_Pickaxe_FlameBride_OffHand_Impact" (Variables|Visual|GetWeaponMeshOffhand) "b_AR_Root" (Variables|/Script/Engine.SceneComponent|GetRelativeLocation (Variables|Visual|GetWeaponMeshOffhand))))
        (Effects|Components|ParticleSystem|SetBoolParameter _returnvalue "bIsFrontEndPreview" true))
      (:then_1
        (bind _asfort_pawn (Utilities|Casting|CastToFortPawn (Game|GetInstigator)))
        (bind _section_index (Utilities|SetupSwingMontageSectionIndex _asfort_pawn)
          (:Out
            (Transformation|SetWorldLocation _impact_fx (Transformation|GetWorldLocation (Variables|/Script/Engine.Character|GetMesh _asfort_pawn)) false false)
            (Transformation|AddRelativeLocation _impact_fx (Math|Vector|RotateVector (Utilities|Array|Get(aref) (Variables|Default|GetFEImpactOnPawnLocations) _section_index) (Transformation|GetWorldRotation (Variables|/Script/Engine.Character|GetMesh _asfort_pawn))) false false)))))))

(event EventSwingLeft_Common
  (Components|Activation|Deactivate (CallFunction|GetIdleOffhandVFXComponent)))

(event EventSwingRight_Common
  (Components|Activation|Deactivate (CallFunction|GetIdleOffhandVFXComponent)))

(event EventSwingLeft2
  (Components|Activation|Deactivate (CallFunction|GetIdleOffhandVFXComponent)))

(event EventSwingRight2
  (Components|Activation|Deactivate (CallFunction|GetIdleOffhandVFXComponent)))

(event EventSwingRight
  (Utilities|FlowControl|Sequence
    (:then_0
      (Components|Activation|Deactivate (CallFunction|GetIdleOffhandVFXComponent)))
    (:then_1
      (Variables|Default|SetIsSwinging true)
      (CallFunction|BurnUp))))

(event EventSwingLeft
  (Utilities|FlowControl|Sequence
    (:then_0
      (Components|Activation|Deactivate (CallFunction|GetIdleOffhandVFXComponent)))
    (:then_1
      (Variables|Default|SetIsSwinging true)
      (CallFunction|BurnUp))))

(event EventSwingLeftEnd
  (bind _weaponmesh (Variables|Visual|GetWeaponMesh))
  (Utilities|FlowControl|Sequence
    (:then_0
      (Utilities|FlowControl|Delay 0.3)
      (Components|Activation|Activate (CallFunction|GetIdleOffhandVFXComponent) true))
    (:then_1
      (Variables|Default|SetTimeline%20Running? true)
      (bind (_direction _velocity_intensity) (Timeline|Ember%20Timeline PlayFromStart)
        (:Update
          (bind _output_get (Variables|FlameBurn|SetCurrent%20Flame%20Intensity _velocity_intensity))
          (Rendering|Material|SetScalarParameterValueonMaterials _weaponmesh "Ember Time" _output_get)
          (bind (_direction_2 _velocity_intensity_2) (Timeline|Burn%20%20Timeline ReverseFromEnd)
            (:Update
              (bind _output_get_2 (Variables|Default|SetCurrent%20Burn%20Intensity%20 _velocity_intensity_2))
              (Rendering|Material|SetScalarParameterValueonMaterials _weaponmesh "Burn Time" _output_get_2))
            (:Finished
              (Variables|Default|SetTimeline%20Running? false)
              (Utilities|FlowControl|RetriggerableDelay 2.0)
              (Variables|Default|SetIsSwinging false)
              (Variables|Default|SetJustSwung? true))))))))

(event EventSwingRightEnd
  (bind _weaponmesh (Variables|Visual|GetWeaponMesh))
  (Utilities|FlowControl|Sequence
    (:then_0
      (Utilities|FlowControl|Delay 0.3)
      (Components|Activation|Activate (CallFunction|GetIdleOffhandVFXComponent) true))
    (:then_1
      (Variables|Default|SetTimeline%20Running? true)
      (Timeline|Ember%20Timeline PlayFromStart))))

(event Custom|Started
  (bind (_target_low_high_x _target_low_high_y) (Math|Vector2D|BreakVector2D (Variables|GetTarget%20Low/High_X)))
  (bind (_target_low_high_x_2 _target_low_high_y_2) (Math|Vector2D|BreakVector2D (Variables|GetTarget%20Low/High_X)))
  (if (== (Variables|FlameBurn|GetTarget) _target_low_high_y)
    (Variables|FlameBurn|SetEmber%20Interp%20Speed 2.0)
    (Variables|FlameBurn|SetBurn%20Interp%20Speed 0.4)
    (Variables|FlameBurn|SetTarget _target_low_high_x_2)
    (Variables|FlameBurn|SetTarget%20Burn _target_low_high_x_2)
    (Utilities|FlowControl|RetriggerableDelay (Variables|FlameBurn|GetLoop%20Time%20Gap))
    (Variables|Default|SetJustSwung? false))
  (Variables|FlameBurn|SetEmber%20Interp%20Speed 1.0)
  (Variables|FlameBurn|SetBurn%20Interp%20Speed 1.0)
  (Variables|FlameBurn|SetTarget _target_low_high_y_2)
  (Variables|FlameBurn|SetTarget%20Burn _target_low_high_y_2))

(event EventTick
  (params (DeltaSeconds :type Float%20%28single-precision%29))
  (bind _weaponmesh (Variables|Visual|GetWeaponMesh))
  (bind (_returnvalue_roll _returnvalue_pitch _returnvalue_yaw) (Math|Rotator|BreakRotator (Transformation|GetWorldRotation _weaponmesh)))
  (bind _last_rotation (Variables|AngularVelocity|GetLast%20Rotation))
  (bind _returnvalue (Utilities|Operators|Subtract (Math|Vector|MakeVector _returnvalue_roll _returnvalue_pitch _returnvalue_yaw) _last_rotation))
  (bind _returnvalue_2 (/ _returnvalue DeltaSeconds))
  (bind (_target_low_high_x _target_low_high_y) (Math|Vector2D|BreakVector2D (Variables|GetTarget%20Low/High_X)))
  (bind (_target_low_high_x_2 _target_low_high_y_2) (Math|Vector2D|BreakVector2D (Variables|GetTarget%20Low/High_X)))
  (bind _current_burn_intensity (Variables|Default|GetCurrent%20Burn%20Intensity%20))
  (bind _weaponmesh_2 (Variables|Visual|GetWeaponMesh))
  (bind _weaponmesh_3 (Variables|Visual|GetWeaponMesh))
  (bind _current_flame_intensity (Variables|FlameBurn|GetCurrent%20Flame%20Intensity))
  (if (Reactive|Frontend|IsCosmeticObjectDisplayedinFrontendContext 0)
    (if (not (Variables|Default|GetIsSwinging))
      (Utilities|FlowControl|Sequence
        (:then_0
          (bind _output_get (Variables|FlameBurn|SetCurrent%20Flame%20Intensity (Math|Interpolation|FInterpTo (Variables|FlameBurn|GetCurrent%20Flame%20Intensity) (Variables|FlameBurn|GetTarget) DeltaSeconds (Variables|FlameBurn|GetEmber%20Interp%20Speed))))
          (bind _output_get_2 (Variables|Default|SetCurrent%20Burn%20Intensity%20 (Math|Interpolation|FInterpTo (Variables|Default|GetCurrent%20Burn%20Intensity%20) (Variables|FlameBurn|GetTarget%20Burn) DeltaSeconds (Variables|FlameBurn|GetBurn%20Interp%20Speed))))
          (Rendering|Material|SetScalarParameterValueonMaterials _weaponmesh_3 "Ember Time" _output_get)
          (Rendering|Material|SetScalarParameterValueonMaterials _weaponmesh_3 "Burn Time" _output_get_2))
        (:then_1
          (Variables|Default|SetAngular%20Acceleration (Math|Vector|VectortoRadians (* (/ (- _returnvalue_2 (Variables|AngularVelocity|GetLastAngularVelocity)) DeltaSeconds) (Variables|Default|GetAngular%20Scale%20Multiplier))))
          (bind _output_get_3 (Variables|AngularVelocity|SetLastAngularVelocity _returnvalue_2))
          (Variables|SetLast%20Rotation_X (Math|Vector|MakeVector _returnvalue_roll _returnvalue_pitch _returnvalue_yaw))
          (if (and (Math|Boolean|ORBoolean (> (.x _output_get_3) 15.0) (> (.y _output_get_3) 15.0) (> (.z _output_get_3) 15.0)) (Variables|FlameBurn|GetVelocity%20Changed))
            (if (Variables|Default|GetJustSwung?)
              (Variables|FlameBurn|SetLoop%20Time%20Gap 12.0)
              (else
                (Variables|FlameBurn|SetVelocity%20Cache _output_get_3)
                (Variables|FlameBurn|SetVelocity%20Changed false)
                (Variables|FlameBurn|SetLoop%20Time%20Gap 1.0)))
            (Variables|FlameBurn|SetEmber%20Interp%20Speed 1.0)
            (Variables|FlameBurn|SetBurn%20Interp%20Speed 1.0)
            (Variables|FlameBurn|SetTarget _target_low_high_y_2)
            (Variables|FlameBurn|SetTarget%20Burn _target_low_high_y_2)
            (Utilities|FlowControl|RetriggerableDelay (Variables|FlameBurn|GetLoop%20Time%20Gap))
            (Variables|Default|SetJustSwung? false)
            (if (== (Variables|FlameBurn|GetTarget) _target_low_high_y)
              (Variables|FlameBurn|SetEmber%20Interp%20Speed 2.0)
              (Variables|FlameBurn|SetBurn%20Interp%20Speed 0.4)
              (Variables|FlameBurn|SetTarget _target_low_high_x_2)
              (Variables|FlameBurn|SetTarget%20Burn _target_low_high_x_2))
            (elif (Utilities|Operators|NotEqual(!=) _output_get_3 (Variables|FlameBurn|GetVelocity%20Cache) 0)
              (Variables|FlameBurn|SetVelocity%20Changed true)
              (Variables|FlameBurn|SetLoop%20Time%20Gap 12.0))))))
    (elif (Math|Conversions|ToBoolean(Integer) (Math|Integer|Clamp(Integer) (Math|Float|Round (Math|Vector|VectorLength (Transformation|GetVelocity (Game|GetPlayerPawn 0)))) 0 1))
      (Components|Timeline|Stop (Variables|B_Athena_Pickaxe_DualWield_FlameBride|GetEmber%20Timeline))
      (Variables|Default|SetTimeline%20Running? false)
      (bind _output_get_4 (Variables|FlameBurn|SetCurrent%20Flame%20Intensity (Math|Interpolation|FInterpTo _current_flame_intensity 1.0 DeltaSeconds 1.0)))
      (bind _output_get_5 (Variables|Default|SetCurrent%20Burn%20Intensity%20 (Math|Interpolation|FInterpTo _current_burn_intensity 1.0 DeltaSeconds 1.0)))
      (Rendering|Material|SetScalarParameterValueonMaterials _weaponmesh_2 "Ember Time" _output_get_4)
      (Rendering|Material|SetScalarParameterValueonMaterials _weaponmesh_2 "Burn Time" _output_get_5)
      (elif (Variables|Default|GetTimeline%20Running?)
        _
        (else
          (bind _output_get_6 (Variables|FlameBurn|SetCurrent%20Flame%20Intensity (Math|Interpolation|FInterpTo _current_flame_intensity 0.0 DeltaSeconds (Variables|FlameBurn|GetEmber%20Interp%20Speed))))
          (bind _output_get_7 (Variables|Default|SetCurrent%20Burn%20Intensity%20 (Math|Interpolation|FInterpTo _current_burn_intensity 0.0 DeltaSeconds (Variables|FlameBurn|GetBurn%20Interp%20Speed))))
          (Rendering|Material|SetScalarParameterValueonMaterials _weaponmesh_2 "Ember Time" _output_get_6)
          (Rendering|Material|SetScalarParameterValueonMaterials _weaponmesh_2 "Burn Time" _output_get_7))))))

(event Custom|Burn%20Up
  (bind _weaponmesh (Variables|Visual|GetWeaponMesh))
  (bind _output_get (Variables|FlameBurn|SetCurrent%20Flame%20Intensity (Math|Interpolation|FInterpTo (Variables|FlameBurn|GetCurrent%20Flame%20Intensity) 1.0 (stale EventTick DeltaSeconds) 1.0)))
  (bind _output_get_2 (Variables|Default|SetCurrent%20Burn%20Intensity%20 (Math|Interpolation|FInterpTo (Variables|Default|GetCurrent%20Burn%20Intensity%20) 1.0 (stale EventTick DeltaSeconds) 1.0)))
  (Rendering|Material|SetScalarParameterValueonMaterials _weaponmesh "Ember Time" _output_get)
  (Rendering|Material|SetScalarParameterValueonMaterials _weaponmesh "Burn Time" _output_get_2))

(fn ConstructionScript
  (|CallParentFunction))
