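; Blueprint for the Frozen Reality dual-wield forearm pickaxe cosmetic, handling its equip and impact visuals.
; On equip it spawns a Niagara equip burst at the weapon mesh's idle FX socket and runs a timeline driving the "Reveal" parameter on the idle VFX component.
; Swing events record which hand is swinging.
; Impact FX are then rotated to the hit normal with an opposite angular offset for left and right swings.
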(event FortWeapon|EventOnEquippedAndAttached
  (bind _returnvalue (CallFunction|GetIdleVFXComponent))
  (|CallParentFunction)
  (Utilities|FlowControl|Sequence
    (:then_0
      (Niagara|SpawnSystemAttached "/BRCosmetics/Weapons/FORT_Melee/Pickaxe_FrozenReality/FX/NS_Pickaxe_FrozenReality_Equip.NS_Pickaxe_FrozenReality_Equip" (Variables|Visual|GetWeaponMesh) "FX_IdleSocket" "0, 0, 0" "0, 0, 0" "SnapToTarget"))
    (:then_1
      (bind (_direction _fx_reveal) (Timeline|OnEquip PlayFromStart)
        (:Update
          (Utilities|IsValid _returnvalue
            (:Is%20Valid
              (bind _asniagara_particle_system_component (Utilities|Casting|CastToNiagaraComponent _returnvalue))
              (Niagara|SetNiagaraVariable(Float) _asniagara_particle_system_component "Reveal" _fx_reveal))))))))

(event EventBeginPlay
  (|CallParentFunction) ; This node is disabled and will not be called.\nDrag off pins to build functionality.
)

(event EventTick
  (params (DeltaSeconds :type Float%20%28single-precision%29))
  (|CallParentFunction DeltaSeconds) ; This node is disabled and will not be called.\nDrag off pins to build functionality.
)

(event EventSwingLeft
  (|CallParentFunction)
  (Variables|Default|SetPickaxeIsSwingingRight false))

(event EventSwingRight
  (|CallParentFunction)
  (Variables|Default|SetPickaxeIsSwingingRight true))

(event EventSwingLeft2
  (|CallParentFunction)
  (Variables|Default|SetPickaxeIsSwingingRight false))

(event EventSwingRight2
  (|CallParentFunction)
  (Variables|Default|SetPickaxeIsSwingingRight true))

(event FortWeapon|EventOnPlayImpactFX
  (params (HitResult :type Hit%20Result%20Structure%20%28by%20ref%29) (ImpactPhysicalSurface :type EPhysicalSurface%20Enum :default "SurfaceType_Default") (SpawnedPSC :type FXSystem%20Component%20Object%20Reference))
  (bind (_bblockinghit _binitialoverlap _time _distance _location _impactpoint _normal) (Collision|BreakHitResult HitResult))
  (|CallParentFunction HitResult ImpactPhysicalSurface SpawnedPSC)
  (Transformation|SetWorldRotation SpawnedPSC (Math|Rotator|CombineRotators (Math|Rotator|MakeRotfromX _normal) (select (Variables|Default|GetPickaxeIsSwingingRight) "0.000000,-60.000000,0.000000" "0.000000,60.000000,0.000000")) false false))

(fn ConstructionScript
  (|CallParentFunction))
