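; Dual-wield forearm pickaxe cosmetic Blueprint that animates its idle particle effect when weapon visibility changes.
; A preview check selects either the FrontEnd or Gameplay timeline, so preview and in-game displays use separate curves.
; On show, that timeline plays from the start, feeding LightScale and Size float parameters to the idle VFX component each update.
; On hide, the selected timeline's time is reset.
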(event EventOnWeaponVisibilityChanged
  (params (bVisible :type Boolean) (bSetForLocalControllerOnly :type Boolean))
  (bind _returnvalue (CallFunction|GetIdleVFXComponent))
  (if (Reactive|Frontend|IsCosmeticObjectDisplayedasPreview 0)
    (if bVisible
      (bind (_direction _lightscalefe _sizefe) (Timeline|FrontEnd PlayFromStart)
        (:Update
          (Effects|Components|ParticleSystem|SetFloatParameter _returnvalue "LightScale" _lightscalefe)
          (Effects|Components|ParticleSystem|SetFloatParameter _returnvalue "Size" _sizefe)))
      (else
        (Timeline|FrontEnd SetNewTime)))
    (elif bVisible
      (bind (_direction_2 _lightstrenght _size) (Timeline|Gameplay PlayFromStart)
        (:Update
          (Effects|Components|ParticleSystem|SetFloatParameter _returnvalue "LightScale" _lightstrenght)
          (Effects|Components|ParticleSystem|SetFloatParameter _returnvalue "Size" _size)))
      (else
        (Timeline|Gameplay SetNewTime)))))

(fn ConstructionScript
  (|CallParentFunction))
