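; Impact Blueprint for the PeacefulPoemPumped dual-wield forearm pickaxe that orients its Niagara hit effects by swing.
; Swing events track right versus left and the second left swing, and equip caches the owning pawn.
; On impact, effects align to the hit normal plus a per-side rotation; second left swings add an attached directional burst and set SecondImpact.
; Cosmetic previews instead deactivate the supplied effect and spawn the directional burst relative to the pawn.
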(event EventBeginPlay
  (|CallParentFunction) ; This node is disabled and will not be called.\nDrag off pins to build functionality.
)

(event EventTick
  (params (DeltaSeconds :type Float%20%28single-precision%29))
  (|CallParentFunction DeltaSeconds) ; This node is disabled and will not be called.\nDrag off pins to build functionality.
)

(event FortWeapon|EventOnPlayImpactFX
  (params (HitResult :type Hit%20Result%20Structure%20%28by%20ref%29) (ImpactPhysicalSurface :type EPhysicalSurface%20Enum :default "SurfaceType_Default") (SpawnedPSC :type FXSystem%20Component%20Object%20Reference))
  (bind (_returnvalue_location _returnvalue_rotation) (Math|Transform|BreakTransform (Transformation|GetActorTransform (Variables|Default|GetAs%20Fort%20Player%20Pawn))))
  (bind (_returnvalue_location_2 _returnvalue_rotation_2 _returnvalue_scale) (Math|Transform|BreakTransform (Transformation|GetActorTransform (Variables|Default|GetAs%20Fort%20Player%20Pawn))))
  (bind (_bblockinghit _binitialoverlap _time _distance _location _impactpoint _normal) (Collision|BreakHitResult HitResult))
  (bind _returnvalue (Math|Rotator|MakeRotfromX _normal))
  (bind _left2_swing (Variables|Default|GetLeft2%20Swing?))
  (|CallParentFunction HitResult ImpactPhysicalSurface SpawnedPSC)
  (if (Reactive|Frontend|IsCosmeticObjectDisplayedasPreview 0)
    (Components|Activation|Deactivate SpawnedPSC)
    (bind _returnvalue_2 (Niagara|SpawnSystematLocation "/BRCosmetics/Weapons/FORT_Melee/Pickaxe_PeacefulPoem_Pumped/FX/NS_Pickaxe_PeacefulPoemPumped_Impact_Directional.NS_Pickaxe_PeacefulPoemPumped_Impact_Directional" (Math|Transform|TransformLocation (Transformation|GetActorTransform (Variables|Default|GetAs%20Fort%20Player%20Pawn)) "100.000000,0.000000,50.000000") (Math|Rotator|CombineRotators _returnvalue_rotation_2 (select (Variables|Default|GetRight?) "0.000000,-60.000000,0.000000" "0.000000,60.000000,0.000000")) _returnvalue_scale))
    (if _left2_swing
      (Niagara|SetNiagaraVariable(Bool) _returnvalue_2 "SecondImpact" _left2_swing)
      (Transformation|SetWorldRotation _returnvalue_2 (Math|Rotator|CombineRotators _returnvalue_rotation "0.000000,180.000000,0.000000") false false))
    (else
      (bind _asniagara_particle_system_component (Utilities|Casting|CastToNiagaraComponent SpawnedPSC))
      (Utilities|FlowControl|Sequence
        (:then_0
          (Variables|Default|SetImpact%20FX_0 _asniagara_particle_system_component)
          (Transformation|SetWorldRotation SpawnedPSC (Math|Rotator|CombineRotators _returnvalue (select (Variables|Default|GetRight?) (Variables|Default|GetRight%20Rotation) (Variables|Default|GetLeft%20Rotation))) false false)
          (if (Variables|Default|GetLeft2%20Swing?)
            (bind _returnvalue_3 (Niagara|SpawnSystemAttached "/BRCosmetics/Weapons/FORT_Melee/Pickaxe_PeacefulPoem_Pumped/FX/NS_Pickaxe_PeacefulPoemPumped_Impact_Directional.NS_Pickaxe_PeacefulPoemPumped_Impact_Directional" (Variables|Default|GetImpact%20FX_0) "None" "0, 0, 0" "0, 0, 0" "KeepRelativeOffset" true))
            (Transformation|SetWorldRotation _returnvalue_3 (Math|Rotator|CombineRotators _returnvalue "0,60.000000, 0") false false)))
        (:then_1
          (Niagara|SetNiagaraVariable(Bool) _asniagara_particle_system_component "SecondImpact" (Variables|Default|GetLeft2%20Swing?)))))))

(event EventSwingLeft_Common
  (|CallParentFunction)
  (Variables|Default|SetRight? false))

(event EventSwingRight_Common
  (|CallParentFunction)
  (Variables|Default|SetRight? true))

(event EventSwingLeft
  (|CallParentFunction)
  (Variables|Default|SetLeft2%20Swing? false))

(event EventSwingRight
  (|CallParentFunction)
  (Variables|Default|SetLeft2%20Swing? false))

(event EventSwingLeft2
  (|CallParentFunction)
  (Variables|Default|SetLeft2%20Swing? true))

(event FortWeapon|EventOnEquippedAndAttached
  (|CallParentFunction)
  (bind _asfort_player_pawn (Utilities|Casting|CastToFortPlayerPawn (Variables|Actor|GetInstigator)))
  (Variables|Default|SetAs%20Fort%20Player%20Pawn _asfort_player_pawn))

(fn ConstructionScript
  (|CallParentFunction))
