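; Dual-wield forearm pickaxe Blueprint for the SteamPower cosmetic, adding Niagara swing trails per hand.
; Left swings attach the blue steam trail to the main weapon mesh, and right swings attach the orange one to the offhand mesh.
; After cosmetic initialization it also calls SetOverlayMaterial on the offhand mesh with no material assigned.
; Swing, initialization and construction logic all call the parent first, and Tick is empty.
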(event EventTick
  (params (DeltaSeconds :type Float%20%28single-precision%29)))

(event EventSwingLeft
  (|CallParentFunction)
  (Niagara|SpawnSystemAttached "/BRCosmetics/Weapons/FORT_Melee/Pickaxe_SteamPower_Blue/FX/NS_Pickaxe_SteamPowerBlue_Trail.NS_Pickaxe_SteamPowerBlue_Trail" (Variables|Visual|GetWeaponMesh) "None" "0, 0, 0" "0, 0, 0" "SnapToTarget"))

(event EventSwingRight
  (|CallParentFunction)
  (Niagara|SpawnSystemAttached "/BRCosmetics/Weapons/FORT_Melee/Pickaxe_SteamPower/NS_Pickaxe_SteamPowerOrange_Trail.NS_Pickaxe_SteamPowerOrange_Trail" (Variables|Visual|GetWeaponMeshOffhand) "None" "0, 0, 0" "0, 0, 0" "SnapToTarget"))

(event EventHandlePostCosmeticInitialization
  (|CallParentFunction)
  (Rendering|Material|SetOverlayMaterial (Variables|Visual|GetWeaponMeshOffhand) 0 false 0 :!class /Script/Engine.MeshComponent))

(fn ConstructionScript
  (|CallParentFunction))
