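; Blueprint for a generic dual-wield pickaxe (Candor) that tells its Niagara impact effect which hand made the swing.
; The right and left swing events call their parents, then set an Is Swing Left flag to false or true.
; On impact FX playback, it calls the parent, casts the spawned FX component to a Niagara component, and passes the flag in as a bool user variable of the same name.
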(event EventSwingRight_Common
  (|CallParentFunction)
  (Variables|Default|SetIs%20Swing%20Left false))

(event EventSwingLeft_Common
  (|CallParentFunction)
  (Variables|Default|SetIs%20Swing%20Left true))

(event FortWeapon|EventOnPlayImpactFX
  (params (HitResult :type Hit%20Result%20Structure%20%28by%20ref%29) (ImpactPhysicalSurface :type EPhysicalSurface%20Enum :default "SurfaceType_Default") (SpawnedPSC :type FXSystem%20Component%20Object%20Reference))
  (|CallParentFunction HitResult ImpactPhysicalSurface SpawnedPSC)
  (bind _asniagara_particle_system_component (Utilities|Casting|CastToNiagaraComponent SpawnedPSC))
  (Niagara|SetNiagaraVariable(Bool) _asniagara_particle_system_component "Is Swing Left" (Variables|Default|GetIs%20Swing%20Left)))

(fn ConstructionScript
  (|CallParentFunction))
