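; Cosmetic Blueprint for the Celestial female dual-wield pickaxe that adds sparkle effects when its wielder lands.
; Both the instigator-set and equipped-and-attached events cast the instigator to a Fort player pawn and bind to its landed dispatcher.
; After a short delay, the landing handler spawns the Niagara sparkle system on the idle FX socket of both main and offhand weapon meshes.
; The unequip events only call their parents, with no explicit unbind.
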(event EventOnInitWeaponCosmetics
  (|CallParentFunction))

(event FortWeapon|EventOnInstigatorSet
  (|CallParentFunction)
  (bind _asfort_player_pawn (Utilities|Casting|CastToFortPlayerPawn (Game|GetInstigator)))
  (EventDispatchers|BindEventtoOnPawnLanded _asfort_player_pawn (EventDispatchers|CreateEvent :!event Custom|OnLanded)))

(event FortWeapon|EventOnEquippedAndAttached
  (|CallParentFunction)
  (bind _asfort_player_pawn (Utilities|Casting|CastToFortPlayerPawn (Game|GetInstigator)))
  (EventDispatchers|BindEventtoOnPawnLanded _asfort_player_pawn (EventDispatchers|CreateEvent :!event Custom|OnLanded)))

(event Custom|OnLanded
  (params (Hit :type Hit%20Result%20Structure%20%28by%20ref%29))
  (Utilities|FlowControl|Delay 0.1)
  (Niagara|SpawnSystemAttached "/BRCosmetics/Weapons/FORT_Melee/Pickaxe_Celestial_Female_1h/FX/NS_Celestial_F_Sparkles.NS_Celestial_F_Sparkles" (Variables|Visual|GetWeaponMesh) "FX_Idle_Start" "0, 0, 0" "0, 0, 0" "KeepRelativeOffset" false true "AutoRelease")
  (Niagara|SpawnSystemAttached "/BRCosmetics/Weapons/FORT_Melee/Pickaxe_Celestial_Female_1h/FX/NS_Celestial_F_Sparkles.NS_Celestial_F_Sparkles" (Variables|Visual|GetWeaponMeshOffhand) "FX_Idle_Start" "0, 0, 0" "0, 0, 0" "KeepRelativeOffset" false true "AutoRelease"))

(event FortWeapon|EventOnLocalUnEquip
  (|CallParentFunction))

(event FortWeapon|EventOnUnEquippedAndDetached
  (|CallParentFunction))

(fn ConstructionScript
  (|CallParentFunction))
