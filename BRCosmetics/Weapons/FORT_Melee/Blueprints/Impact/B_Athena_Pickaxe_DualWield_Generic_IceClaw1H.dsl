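; Cosmetic Blueprint for the IceClaw one-handed variant of a dual-wield pickaxe, extending its parent's behavior.
; On weapon cosmetics initialization it calls the parent, then checks whether it is displayed in a frontend context.
; If so, it sets the FrontendSpawn bool parameter on the idle VFX component, letting that effect distinguish frontend display.
; The construction script only calls the parent.
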(event EventOnInitWeaponCosmetics
  (|CallParentFunction)
  (if (Reactive|Frontend|IsCosmeticObjectDisplayedinFrontendContext self)
    (Effects|Components|ParticleSystem|SetBoolParameter (CallFunction|GetIdleVFXComponent) "FrontendSpawn" true)))

(fn ConstructionScript
  (|CallParentFunction))
