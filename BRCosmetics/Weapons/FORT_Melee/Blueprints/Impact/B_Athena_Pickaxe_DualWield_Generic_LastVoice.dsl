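; Dual-wield pickaxe Blueprint for the LastVoice cosmetic that restarts each hand's idle VFX when that hand swings.
; The right and left swing events call the parent, then activate the main-hand or offhand idle VFX component, respectively, with reset enabled.
; BeginPlay and Tick are empty, and the construction script only calls the parent.
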(event EventBeginPlay)

(event EventTick
  (params (DeltaSeconds :type Float%20%28single-precision%29)))

(event EventSwingRight_Common
  (|CallParentFunction)
  (Components|Activation|Activate (CallFunction|GetIdleVFXComponent) true))

(event EventSwingLeft_Common
  (|CallParentFunction)
  (Components|Activation|Activate (CallFunction|GetIdleOffhandVFXComponent) true))

(fn ConstructionScript
  (|CallParentFunction))
