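; Cosmetic Blueprint for the LunchBox dual-wield pickaxe that flags its idle effect when shown as a front-end preview.
; On weapon cosmetics init, it calls the parent, then casts the game state to FortGameStateFrontEnd.
; It sets the bIsFrontEndPreview bool parameter to true on the idle VFX particle component.
; The construction script only calls the parent.
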(event EventOnInitWeaponCosmetics
  (|CallParentFunction)
  (Utilities|Casting|CastToFortGameStateFrontEnd (Game|GetGameState))
  (Effects|Components|ParticleSystem|SetBoolParameter (CallFunction|GetIdleVFXComponent) "bIsFrontEndPreview" true))

(fn ConstructionScript
  (|CallParentFunction))
