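; Dual-wield Tar Male pickaxe weapon Blueprint that drives the cosmetic's idle and impact Niagara effects.
; On instigator set and on equip, it feeds the pawn's mesh to both idle effects and flags the offhand one, with a short delay if the effect is missing.
; Swing events track the active hand, which picks the impact rotation.
; Impacts also attach an extra effect to that hand's mesh when the particle style variant check passes.
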(event EventOnInitWeaponCosmetics
  (|CallParentFunction)
  (Effects|Components|ParticleSystem|SetBoolParameter (CallFunction|GetIdleVFXComponent) "bIsFrontEndPreview" (Reactive|Frontend|IsCosmeticObjectDisplayedinFrontendContext self)))

(event FortWeapon|EventOnInstigatorSet
  (bind _returnvalue (CallFunction|GetIdleVFXComponent))
  (|CallParentFunction)
  (bind _asfort_pawn (Utilities|Casting|CastToFortPawn (Game|GetInstigator)))
  (Variables|Default|SetInstigatorSkeletalMesh (Variables|/Script/Engine.Character|GetMesh _asfort_pawn))
  (Utilities|IsValid _returnvalue
    (:Is%20Valid
      _)
    (:Is%20Not%20Valid
      (Utilities|FlowControl|Delay 0.2)))
  (bind _asniagara_particle_system_component (Utilities|Casting|CastToNiagaraComponent _returnvalue)
    (:then
      (Niagara|SetNiagaraSkeletalMeshComponent _asniagara_particle_system_component "User.InstigatorMesh" (Variables|Default|GetInstigatorSkeletalMesh))
      (bind _asniagara_particle_system_component_2 (Utilities|Casting|CastToNiagaraComponent (CallFunction|GetIdleOffhandVFXComponent)))
      (Niagara|SetNiagaraSkeletalMeshComponent _asniagara_particle_system_component_2 "User.InstigatorMesh" (Variables|Default|GetInstigatorSkeletalMesh))
      (Niagara|SetNiagaraVariable(Int32) _asniagara_particle_system_component_2 "Offhand_Add" 1))
    (:CastFailed
      (Development|PrintString "Niagara Cast Failed"))))

(event FortWeapon|EventOnEquippedAndAttached
  (bind _returnvalue (CallFunction|GetIdleVFXComponent))
  (|CallParentFunction)
  (bind _asfort_pawn (Utilities|Casting|CastToFortPawn (Game|GetInstigator)))
  (Variables|Default|SetInstigatorSkeletalMesh (Variables|/Script/Engine.Character|GetMesh _asfort_pawn))
  (Utilities|IsValid _returnvalue
    (:Is%20Valid
      _)
    (:Is%20Not%20Valid
      (Utilities|FlowControl|Delay 0.2)))
  (bind _asniagara_particle_system_component (Utilities|Casting|CastToNiagaraComponent _returnvalue)
    (:then
      (Niagara|SetNiagaraSkeletalMeshComponent _asniagara_particle_system_component "User.InstigatorMesh" (Variables|Default|GetInstigatorSkeletalMesh))
      (bind _asniagara_particle_system_component_2 (Utilities|Casting|CastToNiagaraComponent (CallFunction|GetIdleOffhandVFXComponent)))
      (Niagara|SetNiagaraSkeletalMeshComponent _asniagara_particle_system_component_2 "User.InstigatorMesh" (Variables|Default|GetInstigatorSkeletalMesh))
      (Niagara|SetNiagaraVariable(Int32) _asniagara_particle_system_component_2 "Offhand_Add" 1))
    (:CastFailed
      (Development|PrintString "Niagara Cast Failed"))))

(event FortWeapon|EventOnLocalUnEquip
  (|CallParentFunction))

(event FortWeapon|EventOnUnEquippedAndDetached
  (|CallParentFunction))

(event FortWeapon|EventOnPlayImpactFX
  (params (HitResult :type Hit%20Result%20Structure%20%28by%20ref%29) (ImpactPhysicalSurface :type EPhysicalSurface%20Enum :default "SurfaceType_Default") (SpawnedPSC :type FXSystem%20Component%20Object%20Reference))
  (bind (_bblockinghit _binitialoverlap _time _distance _location _impactpoint _normal) (Collision|BreakHitResult HitResult))
  (Transformation|SetWorldRotation SpawnedPSC (Math|Rotator|CombineRotators (Math|Rotator|MakeRotfromX _normal) (select (Variables|Default|GetRight%20?) (Variables|Default|GetRight%20Rotation) (Variables|Default|GetLeft%20Rotation))) false false)
  (bind _asfort_player_pawn (Utilities|Casting|CastToFortPlayerPawn (Game|GetInstigator)))
  (bind (_index _bindexisvalid _bfoundchannel) (CosmeticLoadout|GetVariantChannelActiveIndexforSlotActor "(TagName=\"LoadoutSlot.Pickaxe\")" _asfort_player_pawn "(TagName=\"Cosmetics.Variant.Channel.Particle\")" (Utilities|Array|MakeArray "(TagName=\"Cosmetics.Variant.Property.Particle1\")" "(TagName=\"Cosmetics.Variant.Property.Particle2\")")))
  (if (Math|Boolean|ANDBoolean (> _index 0) _bindexisvalid _bfoundchannel)
    (Niagara|SpawnSystemAttached "/BRCosmetics/Weapons/FORT_Melee/Pickaxe_Tar_Male/FX/NS_Pickaxe_TarMale_Impacts_Idle.NS_Pickaxe_TarMale_Impacts_Idle" (select (Variables|Default|GetRight%20?) (Variables|Visual|GetWeaponMesh) (Variables|Visual|GetWeaponMeshOffhand)) "Melee_TwoHanded_Handle" "0, 0, 0" "0, 0, 0" "KeepRelativeOffset" true)
    (else
      (Development|PrintString "Cosmetic Check Failed"))))

(event EventSwingLeft_Common
  (|CallParentFunction)
  (Variables|Default|SetRight%20? false))

(event EventSwingRight_Common
  (|CallParentFunction)
  (Variables|Default|SetRight%20? true))

(fn ConstructionScript
  (|CallParentFunction))
