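; Cosmetic Blueprint for the Nox dual-wield pickaxe that tells its swing VFX which hand is swinging.
; It extends the parent's common left and right swing events, calling the parent logic in each.
; Left swings use the offhand swing VFX component and right swings use the main one.
; When that component is valid and casts to Niagara, its IsRightHand bool parameter is set to match the hand.
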(event EventSwingLeft_Common
  (bind _returnvalue (CallFunction|GetSwingOffhandVFXComponent))
  (bind (_asniagara_particle_system_component _bsuccess) (Utilities|Casting|CastToNiagaraComponent _returnvalue))
  (|CallParentFunction)
  (Utilities|IsValid _returnvalue
    (:Is%20Valid
      (if _bsuccess
        (Effects|Components|ParticleSystem|SetBoolParameter _asniagara_particle_system_component "IsRightHand" false)))))

(event EventSwingRight_Common
  (bind _returnvalue (CallFunction|GetSwingVFXComponent))
  (bind (_asniagara_particle_system_component _bsuccess) (Utilities|Casting|CastToNiagaraComponent _returnvalue))
  (|CallParentFunction)
  (Utilities|IsValid _returnvalue
    (:Is%20Valid
      (if _bsuccess
        (Effects|Components|ParticleSystem|SetBoolParameter _asniagara_particle_system_component "IsRightHand" true)))))

(fn ConstructionScript
  (|CallParentFunction))
