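; Pickaxe weapon cosmetic Blueprint (Athena "Elastic") that controls when the idle and swing effects play.
; The construction script disables auto-activation of the idle particle component, which BeginPlay activates explicitly.
; BeginPlay also binds the Galileo_Ferry_Axe anim Blueprint's FrontEndIdle event, which, like EndPlay, re-activates that component once after a short delay.
; Visibility changes toggle the idle VFX component, and cosmetic initialization deactivates the swing VFX.
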(event EventBeginPlay
  (|CallParentFunction)
  (Utilities|FlowControl|Sequence
    (:then_0
      (Components|Activation|Activate (CallFunction|GetIdlePSC) false))
    (:then_1
      (bind _asgalileo_ferry_axe_anim_bp (Utilities|Casting|CastToGalileo_Ferry_Axe_AnimBP (Components|SkeletalMesh|GetAnimInstance (Variables|Visual|GetWeaponMesh))))
      (Delegates|BindEventtoFrontEndIdle _asgalileo_ferry_axe_anim_bp (delegate Custom|FrontEndIdle)))))

(event Custom|FrontEndIdle
  (Utilities|FlowControl|DoOnce false
    (:Completed
      (Utilities|FlowControl|Delay 0.3)
      (Components|Activation|Activate (CallFunction|GetIdlePSC) false))))

(event EventEndPlay
  (params (EndPlayReason :type EEndPlayReason%20Enum :default "Destroyed"))
  (|CallParentFunction EndPlayReason)
  (Utilities|FlowControl|DoOnce false
    (:Completed
      (Utilities|FlowControl|Delay 0.3)
      (Components|Activation|Activate (CallFunction|GetIdlePSC) false))))

(event EventOnWeaponVisibilityChanged
  (params (bVisible :type Boolean) (bSetForLocalControllerOnly :type Boolean))
  (if bVisible
    (Components|Activation|Activate (CallFunction|GetIdleVFXComponent) false)
    (else
      (Components|Activation|Deactivate (CallFunction|GetIdleVFXComponent)))))

(event EventOnInitWeaponCosmetics
  (|CallParentFunction)
  (Components|Activation|Deactivate (CallFunction|GetSwingVFXComponent)))

(fn ConstructionScript
  (|CallParentFunction)
  (Components|Activation|SetAutoActivate (CallFunction|GetIdlePSC) false))
