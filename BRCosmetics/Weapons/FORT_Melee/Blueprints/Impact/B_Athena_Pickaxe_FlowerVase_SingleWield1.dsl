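; Cosmetic Blueprint for the FlowerVase single-wield pickaxe, driving reactive glow and impact effects.
; Timelines pulse the mesh's orange EmissiveMultiplier when the kill count is positive, and at half intensity on weak-spot surface hits.
; Impact particles get SpawnDirectionalSparks zeroed and, for swing montage sections 2 and 3, are rotated from the hit normal.
; It also tracks left/right swings, deactivates the swing particle system on init and exposes a swing-index audio parameter setter.
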(event Custom|UpdateBasedOnKills_1
  (params (Watched%20Kills :type Integer))
  ;[1 1 1] GlowOnKill
  (bind _output_get (Variables|Default|SetKillCount (Math|Conversions|ToFloat(Integer) Watched%20Kills)))
  (if (> _output_get 0.0)
    (bind (_direction _lerp_alpha) (Timeline|Timeline_0 PlayFromStart)
      (:Update
        (Rendering|Material|SetVectorParameterValueonMaterials (Variables|Visual|GetWeaponMesh) "EmissiveMultiplier" (Math|Conversions|ToVector(LinearColor) (Math|Color|LinearColor*Float "(R=1.000000,G=0.072205,B=0.000000,A=1.000000)" (Math|Float|Lerp 0.0 (Variables|Default|GetGlowIntensity) _lerp_alpha))))))))

(event EventSwingLeft
  (|CallParentFunction)
  (Variables|Default|SetRight%20? false))

(event FortWeapon|EventOnPlayImpactFX
  (params (HitResult :type Hit%20Result%20Structure%20%28by%20ref%29) (ImpactPhysicalSurface :type EPhysicalSurface%20Enum :default "SurfaceType_Default") (SpawnedPSC :type FXSystem%20Component%20Object%20Reference))
  (bind (_bblockinghit _binitialoverlap _time _distance _location _impactpoint _normal) (Collision|BreakHitResult HitResult))
  ;[1 1 1] GlowOnWeakSpotHit
  (|CallParentFunction HitResult ImpactPhysicalSurface SpawnedPSC)
  (Utilities|FlowControl|Sequence
    (:then_0
      (if (Math|Boolean|ORBoolean (Utilities|Enum|Equal(Enum) (Collision|GetSurfaceType HitResult) "SurfaceType14") (Utilities|Enum|Equal(Enum) (Collision|GetSurfaceType HitResult) "SurfaceType15") (Utilities|Enum|Equal(Enum) (Collision|GetSurfaceType HitResult) "SurfaceType16"))
        (bind (_direction _lerp_alpha) (Timeline|Timeline_1 PlayFromStart)
          (:Update
            (Rendering|Material|SetVectorParameterValueonMaterials (Variables|Visual|GetWeaponMesh) "EmissiveMultiplier" (Math|Conversions|ToVector(LinearColor) (Math|Color|LinearColor*Float "(R=1.000000,G=0.072205,B=0.000000,A=1.000000)" (Math|Float|Lerp 0.0 (* (Variables|Default|GetGlowIntensity) 0.5) _lerp_alpha))))))))
    (:then_1
      ;[0.02 1 0.4447] On Play Impact VFX
      (bind _output_get (Variables|Default|SetImpact%20FX SpawnedPSC))
      (Utilities|IsValid _output_get
        (:Is%20Valid
          (bind _returnvalue (CallFunction|SetupSwingMontageSectionIndex))
          (switch int _returnvalue
            (:Default
              _)
            (:0
              _)
            (:1
              _)
            (:2
              (Transformation|SetWorldRotation (Variables|Default|GetImpact%20FX) (Math|Transform|TransformRotation (Transformation|GetWorldTransform (Variables|Default|GetImpact%20FX)) (Math|Rotator|CombineRotators (Math|Transform|InverseTransformRotation (Transformation|GetWorldTransform _output_get) (Math|Rotator|MakeRotfromX _normal)) "0.000000,90.000000,45.000000")) false false))
            (:3
              (Transformation|SetWorldRotation (Variables|Default|GetImpact%20FX) (Math|Transform|TransformRotation (Transformation|GetWorldTransform (Variables|Default|GetImpact%20FX)) (Math|Rotator|CombineRotators (Math|Transform|InverseTransformRotation (Transformation|GetWorldTransform _output_get) (Math|Rotator|MakeRotfromX _normal)) "0.000000,-90.000000,-20.000000")) false false)))
          (Effects|Components|ParticleSystem|SetIntParameter _output_get "SpawnDirectionalSparks" 0))))))

(event EventSwingRight
  (|CallParentFunction)
  (Variables|Default|SetRight%20? true))

(event EventOnInitWeaponCosmetics
  (|CallParentFunction)
  (Components|Activation|Deactivate (CallFunction|GetSwingPSC)))

(fn SetSwingAudioParam
  (params (InInt :type Integer))
  (bind _returnvalue (AudioParameter|GetAudioParameterComponent))
  (Utilities|IsValid _returnvalue
    (:Is%20Valid
      (Audio|Parameter|SetIntegerParameter _returnvalue "SwingIdx" InInt :!class /Script/AudioExtensions.AudioParameterControllerInterface))))

(fn ConstructionScript
  (|CallParentFunction))
