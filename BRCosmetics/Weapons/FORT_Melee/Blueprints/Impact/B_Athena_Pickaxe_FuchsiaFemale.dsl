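; Fortnite pickaxe Blueprint for the FuchsiaFemale cosmetic, customizing its swing and trail effects.
; On equip it binds the instigator pawn's mesh to the swing Niagara component as "Pawn".
; Swing events record left or right, and trail notifies use that with a flag to set "IsR2" on the trail effect for the notify following a left swing's, marking the third swing.
; Construction disables auto-activation of the idle particle effect.
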(event FortWeapon|EventOnEquippedAndAttached
  (|CallParentFunction)
  (bind _asfort_pawn (Utilities|Casting|CastToFortPawn (Game|GetInstigator)))
  (Utilities|IsValid (Variables|/Script/Engine.Character|GetMesh _asfort_pawn)
    (:Is%20Valid
      (bind _asniagara_particle_system_component (Utilities|Casting|CastToNiagaraComponent (CallFunction|GetSwingVFXComponent))
        (:then
          (Niagara|SetNiagaraSkeletalMeshComponent _asniagara_particle_system_component "Pawn" (Variables|/Script/Engine.Character|GetMesh _asfort_pawn)))
        (:CastFailed
          (Development|PrintString "CAST FAILED - Niagara Component"))))))

(event EventSwingRight
  ;[1 1 1] Set the third swing
  (|CallParentFunction)
  (Variables|Default|SetInt 1))

(event EventSwingLeft
  ;[1 1 1] Set the third swing
  (|CallParentFunction)
  (Variables|Default|SetInt 2))

(event EventAnimTrailsNotify
  (params (bActive :type Boolean))
  ;[1 1 1] Set the third swing
  (|CallParentFunction bActive)
  (Effects|Components|ParticleSystem|SetBoolParameter (CallFunction|GetAnimTrailVFXComponent) "IsR2" false)
  (if (== (Variables|Default|GetInt) 2)
    (Variables|Default|Setbool true)
    (Variables|Default|SetInt 0)
    (elif (Variables|Default|Getbool)
      (Effects|Components|ParticleSystem|SetBoolParameter (CallFunction|GetAnimTrailVFXComponent) "IsR2" true)
      (Variables|Default|Setbool false))))

(fn ConstructionScript
  (|CallParentFunction)
  (Components|Activation|SetAutoActivate (CallFunction|GetIdlePSC) false))
