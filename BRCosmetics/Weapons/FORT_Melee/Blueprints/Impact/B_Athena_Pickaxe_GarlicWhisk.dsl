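; Pickaxe weapon Blueprint for the GarlicWhisk cosmetic that drives a WeakPoint float parameter on its impact effect.
; On each impact it stores the spawned FX component, calls the parent, then branches on whether it is a frontend preview.
; Previews offset the effect from the pawn mesh by montage section and flag one section in four as a weak point.
; Otherwise CE_FX_Client sets the flag from whether the hit physical surface is in the weak-point list.
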(event FortWeapon|EventOnPlayImpactFX
  (params (HitResult :type Hit%20Result%20Structure%20%28by%20ref%29) (ImpactPhysicalSurface :type EPhysicalSurface%20Enum :default "SurfaceType_Default") (SpawnedPSC :type FXSystem%20Component%20Object%20Reference))
  (bind _impact_fx (Variables|Default|GetImpact%20FX))
  (Variables|Default|SetImpactFX SpawnedPSC)
  (|CallParentFunction HitResult ImpactPhysicalSurface SpawnedPSC)
  (if (Reactive|Frontend|IsCosmeticObjectDisplayedasPreview 0)
    ; Setting frontend location and weakspot cycling.
    (bind _asfort_pawn (Utilities|Casting|CastToFortPawn (Game|GetInstigator)))
    (bind _output_get (Variables|Default|SetSectionIndex (Montage|GetSectionIndex (Animation|Montage|GetCurrentActiveMontage (Components|SkeletalMesh|GetAnimInstance (Variables|/Script/Engine.Character|GetMesh _asfort_pawn))) (Animation|Montage|MontageGetCurrentSection (Components|SkeletalMesh|GetAnimInstance (Variables|/Script/Engine.Character|GetMesh _asfort_pawn)) (Animation|Montage|GetCurrentActiveMontage (Components|SkeletalMesh|GetAnimInstance (Variables|/Script/Engine.Character|GetMesh _asfort_pawn)))))))
    (Transformation|SetWorldLocation _impact_fx (Transformation|GetWorldLocation (Variables|/Script/Engine.Character|GetMesh _asfort_pawn)) false false)
    (Transformation|AddRelativeLocation _impact_fx (Math|Vector|RotateVector (Utilities|Array|Get(aref) (Variables|Default|GetFEImpactOnPawnLocations) _output_get) (Transformation|GetWorldRotation (Variables|/Script/Engine.Character|GetMesh _asfort_pawn))) false false)
    (bind _asniagara_particle_system_component (Utilities|Casting|CastToNiagaraComponent (Variables|Default|GetImpactFX)))
    (if (== (Math|Integer|%25%28Integer%29 (Variables|Default|GetSectionIndex) 4) 1)
      (Effects|Components|ParticleSystem|SetFloatParameter _asniagara_particle_system_component "WeakPoint" 1.0)
      (else
        (Effects|Components|ParticleSystem|SetFloatParameter _asniagara_particle_system_component "WeakPoint" 0.0)))
    (else
      (CallFunction|CEFXClient ImpactPhysicalSurface))))

(event Custom|CE_FX_Server
  (params (Impact%20Physical%20Surface :type EPhysicalSurface%20Enum :default "SurfaceType_Default"))
  (CallFunction|CEFXClient Impact%20Physical%20Surface))

(event Custom|CE_FX_Client
  (params (Impact%20Physical%20Surface :type EPhysicalSurface%20Enum :default "SurfaceType_Default"))
  (bind _impactfx (Variables|Default|GetImpactFX))
  (bind _output_get (Variables|Default|SetPhysicalSurface Impact%20Physical%20Surface))
  (Variables|Default|SetIsWeakPoint (Utilities|Array|ContainsItem (Variables|Default|GetWeakPointPhysicalSurface) _output_get))
  (if (Variables|Default|GetIsWeakPoint)
    (Effects|Components|ParticleSystem|SetFloatParameter _impactfx "WeakPoint" 1.0)
    (else
      (Effects|Components|ParticleSystem|SetFloatParameter _impactfx "WeakPoint" 0.0))))

(fn ConstructionScript
  (|CallParentFunction))
