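; Candy cane pickaxe Blueprint whose material reacts to the watched kill count through an override of the kill-update event.
; Each update calls the parent, then writes the count to the KillCounter parameter on the slot 0 dynamic material.
; From four kills, it compares the count with the material's own TotalLights parameter: below that, Lightshow1 switches on for a brief delay; at or above it, Lightshow2 switches on and is not reset here.
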(event EventUpdateBasedOnKills
  (params (Watched%20Kills :type Integer))
  (bind _returnvalue (Math|Conversions|ToFloat(Integer) Watched%20Kills))
  (bind _material0mid (Variables|FortWeaponPickaxeAthena|GetMaterial0MID))
  (|CallParentFunction Watched%20Kills)
  (Rendering|Material|SetScalarParameterValue (Variables|FortWeaponPickaxeAthena|GetMaterial0MID) "KillCounter" _returnvalue :!class /Script/Engine.MaterialInstanceDynamic)
  (if (>= _returnvalue 4.0)
    (bind _returnvalue_2 (Rendering|Material|GetScalarParameterValue (Variables|FortWeaponPickaxeAthena|GetMaterial0MID) "TotalLights" :!class /Script/Engine.MaterialInstanceDynamic))
    (if (>= _returnvalue _returnvalue_2)
      (Rendering|Material|SetScalarParameterValue _material0mid "Lightshow2" 1.0 :!class /Script/Engine.MaterialInstanceDynamic)
      (else
        (Utilities|FlowControl|Sequence
          (:then_0
            (Rendering|Material|SetScalarParameterValue _material0mid "Lightshow1" 1 :!class /Script/Engine.MaterialInstanceDynamic))
          (:then_1
            (Utilities|FlowControl|Delay 2)
            (Rendering|Material|SetScalarParameterValue _material0mid "Lightshow1" 0.0 :!class /Script/Engine.MaterialInstanceDynamic)))))))

(fn ConstructionScript
  (|CallParentFunction))
