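; Lyrical variant of the generic Athena pickaxe Blueprint that tells its swing visual effect which direction the pickaxe is swinging.
; The left and right swing events call the parent, then set the "SwingRight" bool parameter on the swing VFX component to false or true, if that component is valid.
; BeginPlay and Tick are empty, and the construction script only calls the parent.
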(event EventBeginPlay)

(event EventTick
  (params (DeltaSeconds :type Float%20%28single-precision%29)))

(event EventSwingLeft
  (bind _returnvalue (CallFunction|GetSwingVFXComponent))
  (|CallParentFunction)
  (Utilities|IsValid _returnvalue
    (:Is%20Valid
      (Effects|Components|ParticleSystem|SetBoolParameter _returnvalue "SwingRight" false))))

(event EventSwingRight
  (bind _returnvalue (CallFunction|GetSwingVFXComponent))
  (|CallParentFunction)
  (Utilities|IsValid _returnvalue
    (:Is%20Valid
      (Effects|Components|ParticleSystem|SetBoolParameter _returnvalue "SwingRight" true))))

(fn ConstructionScript
  (|CallParentFunction))
