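; Historian Male pickaxe cosmetic Blueprint that builds up frost and ice effects as it hits weak points.
; Impacts on weak-point surfaces increment a counter capped at four, which a retriggerable five-second delay resets, clearing the effect.
; The counter's OnRep sets the mesh material's Time parameter and drives staged Niagara idle FX through parameters and timelines.
; In frontend display, a timeline animates the material and spawns separate Niagara FX.
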(event FortWeapon|EventOnPlayImpactFX
  (params (HitResult :type Hit%20Result%20Structure%20%28by%20ref%29) (ImpactPhysicalSurface :type EPhysicalSurface%20Enum :default "SurfaceType_Default") (SpawnedPSC :type FXSystem%20Component%20Object%20Reference))
  (|CallParentFunction HitResult ImpactPhysicalSurface SpawnedPSC)
  (CallFunction|CEFX ImpactPhysicalSurface))

(event Custom|CE_FX
  (params (Impact%20Physical%20Surface :type EPhysicalSurface%20Enum :default "SurfaceType_Default"))
  (CallFunction|CEFXClient Impact%20Physical%20Surface))

(event Custom|CE_FX_Client
  (params (Impact%20Physical%20Surface :type EPhysicalSurface%20Enum :default "SurfaceType_Default"))
  (bind _output_get (Variables|Default|SetCurrent%20Impact%20Physical%20Surface Impact%20Physical%20Surface))
  (Variables|Default|SetIs%20WeakPoint (Utilities|Array|ContainsItem (Variables|Default|GetWeakPoint%20Physical%20Surface) _output_get))
  (Development|PrintString (Utilities|String|EnumtoString _output_get))
  (if (Variables|Default|GetIs%20WeakPoint)
    (bind _result (Math|Integer|IncrementInt (Variables|Default|GetHitWeakPoint))
      (:%20%20
        (bind _output_get_2 (Variables|Default|SetHitWeakPoint (Math|Integer|Clamp(Integer) _result 0 4)))
        (Development|PrintString (Utilities|String|ToString(Integer) _output_get_2) true true "(R=0.000000,G=0.660000,B=1.000000,A=1.000000)" 4.0)
        (Utilities|FlowControl|RetriggerableDelay 5.0)
        (Variables|Default|SetHitWeakPoint 0)
        (Components|Activation|Deactivate (Variables|Default|GetIdleFX))
        (Variables|Default|SetIdleFX 0)
        (bind (_direction _time) (Timeline|Timeline_0 PlayFromStart)
          (:Update
            (Rendering|Material|SetScalarParameterValueonMaterials (Variables|Visual|GetWeaponMesh) "Time" (Math|Float|Lerp 0.0 1.0 _time))))))))

(event Custom|Stage%203
  (bind _idlefx (Variables|Default|GetIdleFX))
  (bind _idlefx_2 (Variables|Default|GetIdleFX))
  (bind (_direction _frost _icemesh) (Timeline|Timeline_1 PlayFromStart)
    (:Update
      (Effects|Components|ParticleSystem|SetFloatParameter _idlefx "frost/iceAlpha " _frost)
      (Effects|Components|ParticleSystem|SetFloatParameter _idlefx "IceDissolve" _icemesh)
      (Utilities|FlowControl|RetriggerableDelay 5.0)
      (bind (_direction_2 _icemesh_2 _frostalpha) (Timeline|Timeline_2 PlayFromStart)
        (:Update
          (Effects|Components|ParticleSystem|SetFloatParameter _idlefx_2 "IceDissolve" _icemesh_2)
          (Effects|Components|ParticleSystem|SetFloatParameter _idlefx_2 "frost/iceAlpha " _frostalpha))))))

(event Custom|Stage%204
  (bind _idlefx (Variables|Default|GetIdleFX))
  (Utilities|FlowControl|RetriggerableDelay 5.0)
  (Timeline|Timeline_2 PlayFromStart))

(event EventBeginPlay
  (|CallParentFunction)
  (Components|Activation|Deactivate (CallFunction|GetAnimTrailVFXComponent))
  (if (Reactive|Frontend|IsCosmeticObjectDisplayedinFrontendContext (Variables|Visual|GetWeaponMesh))
    (bind (_direction _time) (Timeline|Timeline_3 PlayFromStart)
      (:Update
        (Rendering|Material|SetScalarParameterValueonMaterials (Variables|Visual|GetWeaponMesh) "Time" (Math|Float|Lerp 0.0 1.0 _time)))
      (:Active
        (Niagara|SpawnSystemAttached "/BRCosmetics/Weapons/FORT_Melee/Pickaxe_HistorianMale/FX/NS_Pickaxe_HistorianMale_Frontend_V2.NS_Pickaxe_HistorianMale_Frontend_V2" (Variables|Visual|GetWeaponMesh) "b_AR_Root" "0, 0, 0" "0, 0, 0" "KeepRelativeOffset" true)))))

(event EventOnInitWeaponCosmetics
  (|CallParentFunction)
  (Components|Activation|Deactivate (CallFunction|GetAnimTrailVFXComponent))
  (if (Reactive|Frontend|IsCosmeticObjectDisplayedinFrontendContext (Variables|Visual|GetWeaponMesh))
    (Timeline|Timeline_3 PlayFromStart)))

(event EventOnWeaponVisibilityChanged
  (params (bVisible :type Boolean) (bSetForLocalControllerOnly :type Boolean))
  (|CallParentFunction bVisible bSetForLocalControllerOnly)
  (Components|Activation|Deactivate (CallFunction|GetAnimTrailVFXComponent))
  (if (Reactive|Frontend|IsCosmeticObjectDisplayedinFrontendContext (Variables|Visual|GetWeaponMesh))
    (Timeline|Timeline_3 PlayFromStart)))

(fn OnRep_HitWeakPoint
  (bind _idlefx (Variables|Default|GetIdleFX))
  (bind _weaponmesh (Variables|Visual|GetWeaponMesh))
  (Networking|CanExecuteCosmeticEvents
    (:True
      (switch int (Variables|Default|GetHitWeakPoint)
        (:1
          (Rendering|Material|SetScalarParameterValueonMaterials _weaponmesh "Time" 1.0))
        (:2
          (Rendering|Material|SetScalarParameterValueonMaterials _weaponmesh "Time" 1.0)
          (CallFunction|SpawnIdleFX)
          (Effects|Components|ParticleSystem|SetFloatParameter (Variables|Default|GetIdleFX) "IceDissolve" 1.0))
        (:3
          (Rendering|Material|SetScalarParameterValueonMaterials _weaponmesh "Time" 1.0)
          (CallFunction|SpawnIdleFX)
          (CallFunction|Stage3))
        (:4
          (Rendering|Material|SetScalarParameterValueonMaterials _weaponmesh "Time" 1.0)
          (CallFunction|SpawnIdleFX)
          (Effects|Components|ParticleSystem|SetFloatParameter _idlefx "frost/iceAlpha " 1.0)
          (Effects|Components|ParticleSystem|SetFloatParameter _idlefx "IceDissolve" 0.0)
          (CallFunction|Stage4))))))

(fn SpawnIdleFX
  (Utilities|IsValid (Variables|Default|GetIdleFX)
    (:Is%20Valid
      (if (Components|Activation|IsActive (Variables|Default|GetIdleFX))
        (Development|PrintString "Historian Pickaxe - IdleFX - Already Spawned and Active")
        (else
          (Components|Activation|Activate (Variables|Default|GetIdleFX) true))))
    (:Is%20Not%20Valid
      (bind _returnvalue (Niagara|SpawnSystemAttached "/BRCosmetics/Weapons/FORT_Melee/Pickaxe_HistorianMale/FX/NS_Pickaxe_HistorianMale_Idle.NS_Pickaxe_HistorianMale_Idle" (Variables|Visual|GetWeaponMesh) "b_AR_Root" "0, 0, 0" "0, 0, 0" "KeepRelativeOffset" true true "None" false))
      (Variables|Default|SetIdleFX _returnvalue))))

(fn ConstructionScript
  (|CallParentFunction)
  (Components|Activation|SetAutoActivate (CallFunction|GetIdlePSC) false))
