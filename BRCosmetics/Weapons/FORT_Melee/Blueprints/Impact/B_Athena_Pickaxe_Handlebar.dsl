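; Weapon Blueprint for the Handlebar pickaxe that flags weakpoint hits on its impact effect.
; On impact it calls the parent, pulses the spawned effect's "Weakpoint" parameter true for three specific surface types, then resets it after a short delay.
; PlayAudio, uncalled here and noted as moved to an idle Niagara effect, fades previous impact audio and attaches a new sound to the weapon mesh when bWeakpoint is set.
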(event FortWeapon|EventOnPlayImpactFX
  (params (HitResult :type Hit%20Result%20Structure%20%28by%20ref%29) (ImpactPhysicalSurface :type EPhysicalSurface%20Enum :default "SurfaceType_Default") (SpawnedPSC :type FXSystem%20Component%20Object%20Reference))
  ;[1 1 1] WEAKPOINTS
  (|CallParentFunction HitResult ImpactPhysicalSurface SpawnedPSC)
  (if (Math|Boolean|ORBoolean (Utilities|Enum|Equal(Enum) (Collision|GetSurfaceType HitResult) "SurfaceType14") (Utilities|Enum|Equal(Enum) (Collision|GetSurfaceType HitResult) "SurfaceType15") (Utilities|Enum|Equal(Enum) (Collision|GetSurfaceType HitResult) "SurfaceType16"))
    (Effects|Components|ParticleSystem|SetBoolParameter SpawnedPSC "Weakpoint" true)
    (Utilities|FlowControl|Delay 0.2)
    (else
      _))
  (Effects|Components|ParticleSystem|SetBoolParameter SpawnedPSC "Weakpoint" false))

(fn PlayAudio
  ;[1 1 1] Moved to idle NS effect
  (if (Variables|Default|GetbWeakpoint)
    (bind _impactaudio (Variables|Default|GetImpactAudio)
      (:then
        (Audio|Components|Audio|FadeOut _impactaudio 0.3 0.0 "Linear" :!class /Script/Engine.AudioComponent))
      (:else
        _))
    (bind _returnvalue (Audio|SpawnSoundAttached (Variables|Default|GetWeaponFXSound) (Variables|Visual|GetWeaponMesh)))
    (Variables|Default|SetImpactAudio _returnvalue)))

(fn ConstructionScript
  (|CallParentFunction))
