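; Headset pickaxe cosmetic Blueprint that upgrades its Niagara effects as the player gets kills.
; Once watched kills exceed one, it spawns a one-time kill burst and switches trail and idle effects to stage two; impact effects follow the kill count.
; For frontend preview, a looping timer replays the burst and stage-two look, then a timeline drives the idle fade before reverting.
; Visibility changes also start that timer once, without a frontend check.
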(event FortWeapon|EventOnPlayImpactFX
  (params (HitResult :type Hit%20Result%20Structure%20%28by%20ref%29) (ImpactPhysicalSurface :type EPhysicalSurface%20Enum :default "SurfaceType_Default") (SpawnedPSC :type FXSystem%20Component%20Object%20Reference))
  ;[1 1 1] Impact
  (if (Reactive|Frontend|IsCosmeticObjectDisplayedinFrontendContext self)
    (if (> (Math|Conversions|ToFloat(Integer) (Variables|Default|GetImpact%20FX%20FE%20Stage)) 1.0)
      (Effects|Components|ParticleSystem|SetIntParameter SpawnedPSC "FXStage" 2)
      (else
        (Effects|Components|ParticleSystem|SetIntParameter SpawnedPSC "FXStage" 1)))
    (elif (> (Variables|Default|GetKillCount) 1.0)
      (Effects|Components|ParticleSystem|SetIntParameter SpawnedPSC "FXStage" 2)
      (else
        (Effects|Components|ParticleSystem|SetIntParameter SpawnedPSC "FXStage" 1)))))

(event EventUpdateBasedOnKills
  (params (Watched%20Kills :type Integer))
  (bind _returnvalue (CallFunction|GetAnimTrailsPSC))
  ;[1 1 1] On Kill
  (if (== (Niagara|GetNiagaraSystemAsset (Utilities|Casting|CastToNiagaraComponent (CallFunction|GetIdleVFXComponent))) "/BRCosmetics/Weapons/FORT_Melee/Pickaxe_Headset/FX/NS_Pickaxe_Headset_Idle.NS_Pickaxe_Headset_Idle")
    (bind _output_get (Variables|Default|SetKillCount (Math|Conversions|ToFloat(Integer) Watched%20Kills)))
    (if (> _output_get 1.0)
      (Utilities|FlowControl|DoOnce false
        (:Completed
          (Development|PrintString "kill！" true true "(R=1.000000,G=0.680701,B=0.000000,A=1.000000)" 5.0)
          (Effects|SpawnFXSystemAttached "/BRCosmetics/Weapons/FORT_Melee/Pickaxe_Headset/FX/NS_Pickaxe_Headset_Kill.NS_Pickaxe_Headset_Kill" (Variables|Visual|GetWeaponMesh) "b_AR_Root")
          (Utilities|FlowControl|Delay 0.25)
          (Effects|Components|ParticleSystem|SetIntParameter _returnvalue "FXStage01" 0)
          (Effects|Components|ParticleSystem|SetIntParameter _returnvalue "FXStage02" 1)
          (Effects|Components|ParticleSystem|SetIntParameter (CallFunction|GetIdleVFXComponent) "FXStage" 2))))))

(event Custom|CE_Frontend%20FX%20Loop%20Burst
  ;[1 1 1] BeginPlay
  (if (== (Niagara|GetNiagaraSystemAsset (Utilities|Casting|CastToNiagaraComponent (CallFunction|GetIdleVFXComponent))) "/BRCosmetics/Weapons/FORT_Melee/Pickaxe_Headset/FX/NS_Pickaxe_Headset_Idle.NS_Pickaxe_Headset_Idle")
    (Effects|Components|ParticleSystem|SetFloatParameter (CallFunction|GetIdleVFXComponent) "Global Fade FE" 1.0)
    (bind _returnvalue (Effects|SpawnFXSystemAttached "/BRCosmetics/Weapons/FORT_Melee/Pickaxe_Headset/FX/NS_Pickaxe_Headset_Kill.NS_Pickaxe_Headset_Kill" (Variables|Visual|GetWeaponMesh) "b_AR_Root"))
    (Effects|Components|ParticleSystem|SetBoolParameter _returnvalue "bIsFrontEndPreview" (Reactive|Frontend|IsCosmeticObjectDisplayedinFrontendContext self))
    (Utilities|FlowControl|Delay 0.1)
    (Effects|Components|ParticleSystem|SetIntParameter (CallFunction|GetIdleVFXComponent) "FXStage" 2)
    (Variables|Default|SetImpact%20FX%20FE%20Stage 2)
    (Utilities|FlowControl|Delay 3.5)
    (bind (_direction _fade) (Timeline|Timeline PlayFromStart)
      (:Update
        (Effects|Components|ParticleSystem|SetFloatParameter (CallFunction|GetIdleVFXComponent) "Global Fade FE" _fade))
      (:Finished
        (Effects|Components|ParticleSystem|SetIntParameter (CallFunction|GetIdleVFXComponent) "FXStage" 1)
        (Variables|Default|SetImpact%20FX%20FE%20Stage 1)))))

(event EventOnWeaponVisibilityChanged
  (params (bVisible :type Boolean) (bSetForLocalControllerOnly :type Boolean))
  ;[1 1 1] BeginPlay
  (|CallParentFunction bVisible bSetForLocalControllerOnly)
  (Utilities|FlowControl|DoOnce false
    (:Completed
      (bind _returnvalue (Utilities|Time|SetTimerbyEvent (delegate Custom|CE_Frontend%20FX%20Loop%20Burst) 6.0 true))
      (Variables|Default|SetTimer _returnvalue))))

(event EventOnInitWeaponCosmetics
  (bind _returnvalue (CallFunction|GetAnimTrailsPSC))
  ;[1 1 1] BeginPlay
  (|CallParentFunction)
  (if (Reactive|Frontend|IsCosmeticObjectDisplayedinFrontendContext self)
    (Utilities|FlowControl|DoOnce false
      (:Completed
        (bind _returnvalue_2 (Utilities|Time|SetTimerbyEvent (delegate Custom|CE_Frontend%20FX%20Loop%20Burst) 6.0 true))
        (Variables|Default|SetTimer _returnvalue_2)))
    (else
      (Effects|Components|ParticleSystem|SetIntParameter _returnvalue "FXStage01" 1)
      (Effects|Components|ParticleSystem|SetIntParameter _returnvalue "FXStage02" 0)
      (Effects|Components|ParticleSystem|SetIntParameter (CallFunction|GetIdleVFXComponent) "FXStage" 1))))

(event EventDestroyed
  (if (Reactive|Frontend|IsCosmeticObjectDisplayedinFrontendContext 0)
    (Utilities|Time|ClearandInvalidateTimerbyHandle (Variables|Default|GetTimer))))

(fn ConstructionScript
  (|CallParentFunction))
