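; Pickaxe weapon Blueprint for HeavyRoar that adds a reactive effect to its swing, idle and impact VFX.
; Each impact calls the parent handler, then passes the hit through FX_Server to FX_Client.
; FX_Client restarts the Weak Point Hit timeline for four specific surface types or when the cosmetic is displayed in the frontend.
; The timeline drives WeakPointHitFade on all three effects and WeakPointHitBurst on the idle effect only.
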(event FortWeapon|EventOnPlayImpactFX
  (params (HitResult :type Hit%20Result%20Structure%20%28by%20ref%29) (ImpactPhysicalSurface :type EPhysicalSurface%20Enum :default "SurfaceType_Default") (SpawnedPSC :type FXSystem%20Component%20Object%20Reference))
  (|CallParentFunction HitResult ImpactPhysicalSurface SpawnedPSC)
  (CallFunction|FXServer HitResult))

(event Custom|FX_Server
  (params (Hit%20Result :type Hit%20Result%20Structure))
  (CallFunction|FXClient Hit%20Result))

(event Custom|FX_Client
  (params (Hit%20Result :type Hit%20Result%20Structure))
  (if (Math|Boolean|ORBoolean (Utilities|Enum|Equal(Enum) (Collision|GetSurfaceType Hit%20Result) "SurfaceType12") (Utilities|Enum|Equal(Enum) (Collision|GetSurfaceType Hit%20Result) "SurfaceType14") (Utilities|Enum|Equal(Enum) (Collision|GetSurfaceType Hit%20Result) "SurfaceType15") (Utilities|Enum|Equal(Enum) (Collision|GetSurfaceType Hit%20Result) "SurfaceType16") (Reactive|Frontend|IsCosmeticObjectDisplayedinFrontendContext 0))
    (bind (_direction _weak_point_hit_fade _weak_point_hit_burst) (Timeline|Weak%20Point%20Hit PlayFromStart)
      (:Update
        (Effects|Components|ParticleSystem|SetFloatParameter (CallFunction|GetSwingVFXComponent) "WeakPointHitFade" _weak_point_hit_fade)
        (Effects|Components|ParticleSystem|SetFloatParameter (CallFunction|GetIdleVFXComponent) "WeakPointHitFade" _weak_point_hit_fade)
        (Effects|Components|ParticleSystem|SetFloatParameter (CallFunction|GetIdleVFXComponent) "WeakPointHitBurst" _weak_point_hit_burst)
        (Effects|Components|ParticleSystem|SetFloatParameter (Variables|Default|GetImpact%20FX) "WeakPointHitFade" _weak_point_hit_fade)))))

(fn ConstructionScript
  (|CallParentFunction))
