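; Cosmetic Blueprint for the HedgeSprig pickaxe in Battle Royale, derived from a parent pickaxe weapon class.
; On weapon cosmetic initialization it runs the parent logic, then calls ToonFrontendLUTReplacePickaxe to swap in the toon front-end LUT.
; The construction script only calls the parent.
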(event EventOnInitWeaponCosmetics
  ;[0 0 0 0.5] Front End LUT
  (|CallParentFunction)
  (Utilities|ToonFrontendLUTReplacePickaxe))

(fn ConstructionScript
  (|CallParentFunction))
