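; Impact blueprint for the HiHat pickaxe cosmetic that tells the impact effect whether a hit landed on a weak point.
; On impact it stores the spawned effect component, then runs the parent logic.
; A server-named custom event forwards the surface type to a client-named one, which checks it against the weak-point surface list.
; The client event then sets the effect's WeakPoint float parameter to one or zero.
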(event FortWeapon|EventOnPlayImpactFX
  (params (HitResult :type Hit%20Result%20Structure%20%28by%20ref%29) (ImpactPhysicalSurface :type EPhysicalSurface%20Enum :default "SurfaceType_Default") (SpawnedPSC :type FXSystem%20Component%20Object%20Reference))
  (Variables|Default|SetImpactFX SpawnedPSC)
  (|CallParentFunction HitResult ImpactPhysicalSurface SpawnedPSC))

(event Custom|CE_FX_Server
  (params (Impact%20Physical%20Surface :type EPhysicalSurface%20Enum :default "SurfaceType_Default"))
  (CallFunction|CEFXClient Impact%20Physical%20Surface))

(event Custom|CE_FX_Client
  (params (Impact%20Physical%20Surface :type EPhysicalSurface%20Enum :default "SurfaceType_Default"))
  (bind _impactfx (Variables|Default|GetImpactFX))
  (bind _output_get (Variables|Default|SetPhysicalSurface Impact%20Physical%20Surface))
  (Variables|Default|SetIsWeakPoint (Utilities|Array|ContainsItem (Variables|Default|GetWeakPointPhysicalSurface) _output_get))
  (if (Variables|Default|GetIsWeakPoint)
    (Effects|Components|ParticleSystem|SetFloatParameter _impactfx "WeakPoint" 1.0)
    (else
      (Effects|Components|ParticleSystem|SetFloatParameter _impactfx "WeakPoint" 0.0))))

(fn ConstructionScript
  (|CallParentFunction))
