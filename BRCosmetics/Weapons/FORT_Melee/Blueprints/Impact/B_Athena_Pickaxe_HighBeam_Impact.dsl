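; Cosmetic weapon Blueprint for the HighBeam pickaxe that changes its effects when the wielder wears the HighBeam body mesh.
; A 0.1-second timer polls that mesh until it matches, then turns the swing trail green and activates or attaches a Niagara hand trail.
; Impacts repeat the mesh check for color and map the current montage section to a swing direction vector.
; The hand trail follows weapon visibility and hidden-hand state, deactivating on destruction.
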(event EventSwingLeft
  (|CallParentFunction)
  (Variables|Default|SetRight%20? false))

(event EventSwingRight
  (|CallParentFunction)
  (Variables|Default|SetRight%20? true))

(event EventOnInitWeaponCosmetics
  (|CallParentFunction)
  (Networking|CanExecuteCosmeticEvents
    (:True
      (bind _returnvalue (Utilities|Time|SetTimerbyEvent (delegate Custom|CheckSelf) 0.1 true))
      (Variables|Default|SetCheckTimer _returnvalue))))

(event Custom|CheckSelf
  (bind _asfort_player_pawn (Utilities|Casting|CastToFortPlayerPawn (Game|GetInstigator))
    (:then
      (if (== (Variables|/Script/Engine.SkeletalMeshComponent|GetSkeletalMeshAsset (Customization|GetSkeletalMeshforPartType _asfort_player_pawn "Body" :!class /Script/FortniteGame.FortPlayerPawn)) "/Game/Characters/Player/Female/Medium/Bodies/F_MED_HighBeam/Meshes/F_MED_HighBeam.F_MED_HighBeam")
        (Effects|Components|ParticleSystem|SetBoolParameter (CallFunction|GetAnimTrailVFXComponent) "Is green" true)
        (bind _handtrailfx (Variables|Default|GetHandTrailFX)
          (:then
            (if (not (Components|Activation|IsActive _handtrailfx))
              (Components|Activation|Activate _handtrailfx true)))
          (:else
            (bind _returnvalue (Effects|SpawnFXSystemAttached "/BRCosmetics/Weapons/FORT_Melee/Pickaxe_HighBeam/FX/NS_Pickaxe_HighBeam_HandTrail.NS_Pickaxe_HighBeam_HandTrail" (Customization|GetSkeletalMeshforPartType _asfort_player_pawn "Body" :!class /Script/FortniteGame.FortPlayerPawn)))
            (Variables|Default|SetHandTrailFX _returnvalue)))
        (Utilities|Time|ClearandInvalidateTimerbyHandle (Variables|Default|GetCheckTimer))
        (else
          (Effects|Components|ParticleSystem|SetBoolParameter (CallFunction|GetAnimTrailVFXComponent) "Is green" false))))
    (:CastFailed
      (if (Reactive|Frontend|IsCosmeticObjectDisplayedinFrontendContext 0)
        (Utilities|Time|ClearandInvalidateTimerbyHandle (Variables|Default|GetCheckTimer))))))

(event EventOnWeaponVisibilityChanged
  (params (bVisible :type Boolean) (bSetForLocalControllerOnly :type Boolean))
  (bind _handtrailfx (Variables|Default|GetHandTrailFX))
  (|CallParentFunction bVisible bSetForLocalControllerOnly)
  (Rendering|SetVisibility _handtrailfx bVisible false)
  (Effects|Components|ParticleSystem|SetBoolParameter _handtrailfx "bisHandHidden" (Variables|Default|GetHand%20is%20Hidden)))

(event EventDestroyed
  (|CallParentFunction)
  (Components|Activation|Deactivate (Variables|Default|GetHandTrailFX))
  (Actor|DestroyActor))

(event Custom|Change%20HIdden%20Hand%20Settings
  (Effects|Components|ParticleSystem|SetBoolParameter (Variables|Default|GetHandTrailFX) "bisHandHidden" (Variables|Default|GetHand%20is%20Hidden)))

(event FortWeapon|EventOnPlayImpactFX
  (params (HitResult :type Hit%20Result%20Structure%20%28by%20ref%29) (ImpactPhysicalSurface :type EPhysicalSurface%20Enum :default "SurfaceType_Default") (SpawnedPSC :type FXSystem%20Component%20Object%20Reference))
  (|CallParentFunction HitResult ImpactPhysicalSurface SpawnedPSC)
  (Networking|CanExecuteCosmeticEvents
    (:True
      (Utilities|FlowControl|Sequence
        (:then_0
          (bind _asplayer_pawn_athena (Utilities|Casting|CastToPlayerPawn_Athena (Game|GetInstigator)))
          (if (== (Variables|/Script/Engine.SkeletalMeshComponent|GetSkeletalMeshAsset (Customization|GetSkeletalMeshforPartType _asplayer_pawn_athena "Body" :!class /Script/FortniteGame.FortPlayerPawn)) "/Game/Characters/Player/Female/Medium/Bodies/F_MED_HighBeam/Meshes/F_MED_HighBeam.F_MED_HighBeam")
            (Effects|Components|ParticleSystem|SetBoolParameter SpawnedPSC "Is green" true)
            (else
              (Effects|Components|ParticleSystem|SetBoolParameter SpawnedPSC "Is green" false))))
        (:then_1
          (Variables|Default|SetImpactFX%20Index (Utilities|Array|FindItem (Variables|Default|GetMontageSections) (Animation|Montage|MontageGetCurrentSection (Components|SkeletalMesh|GetAnimInstance (Variables|/Script/Engine.Character|GetMesh (Utilities|Casting|CastToFortPawn (Game|GetInstigator)))) (Animation|Montage|GetCurrentActiveMontage (Components|SkeletalMesh|GetAnimInstance (Variables|/Script/Engine.Character|GetMesh (Utilities|Casting|CastToFortPawn (Game|GetInstigator))))))))
          (Effects|Components|ParticleSystem|SetVectorParameter (Variables|Default|GetImpact%20FX) "Direction" (Utilities|Select :Index (Variables|Default|GetImpactFX%20Index) :Option%200 (Variables|Default|GetSwing1%20Dir) :Option%201 (Variables|Default|GetSwing2%20Dir) :Option%202 (Variables|Default|GetSwing3%20Dir) :Option%203 (Variables|Default|GetSwing4%20Dir))))))))

(fn ConstructionScript
  (|CallParentFunction))
