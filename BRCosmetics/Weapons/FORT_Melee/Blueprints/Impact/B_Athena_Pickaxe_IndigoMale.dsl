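; Blueprint for the IndigoMale pickaxe cosmetic that drives its idle particle effect during melee swings.
; Left and right swing events call the parent, then set the bIsSwinging bool parameter on the idle VFX component to true.
; Each event then waits on a 1.25-second retriggerable delay before setting the parameter back to false.
; BeginPlay and Tick are empty, and the construction script only calls the parent.
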(event EventBeginPlay)

(event EventTick
  (params (DeltaSeconds :type Float%20%28single-precision%29)))

(event EventSwingLeft
  (bind _returnvalue (CallFunction|GetIdleVFXComponent))
  (|CallParentFunction)
  (Effects|Components|ParticleSystem|SetBoolParameter _returnvalue "bIsSwinging" true)
  (Utilities|FlowControl|RetriggerableDelay 1.25)
  (Effects|Components|ParticleSystem|SetBoolParameter _returnvalue "bIsSwinging" false))

(event EventSwingRight
  (bind _returnvalue (CallFunction|GetIdleVFXComponent))
  (|CallParentFunction)
  (Effects|Components|ParticleSystem|SetBoolParameter _returnvalue "bIsSwinging" true)
  (Utilities|FlowControl|RetriggerableDelay 1.25)
  (Effects|Components|ParticleSystem|SetBoolParameter _returnvalue "bIsSwinging" false))

(fn ConstructionScript
  (|CallParentFunction))
