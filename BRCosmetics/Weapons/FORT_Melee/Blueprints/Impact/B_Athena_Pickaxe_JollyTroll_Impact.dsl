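; Melee weapon Blueprint for the JollyTroll pickaxe that makes the weapon mesh blink in reaction to impacts.
; On impact it blinks when displayed in a frontend context, and also passes the surface through CE_FX_Server to CE_FX_Client, which blinks on weak-point surfaces.
; The blink drives the R_blink_pose material scalar closed, then reopens it after a delay; BeginPlay sets it open.
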(event FortWeapon|EventOnPlayImpactFX
  (params (HitResult :type Hit%20Result%20Structure%20%28by%20ref%29) (ImpactPhysicalSurface :type EPhysicalSurface%20Enum :default "SurfaceType_Default") (SpawnedPSC :type FXSystem%20Component%20Object%20Reference))
  (|CallParentFunction HitResult ImpactPhysicalSurface SpawnedPSC)
  (Utilities|FlowControl|Sequence
    (:then_0
      (if (Reactive|Frontend|IsCosmeticObjectDisplayedinFrontendContext self)
        (CallFunction|HitBlink)))
    (:then_1
      (CallFunction|CEFXServer ImpactPhysicalSurface))))

(event Custom|CE_FX_Server
  (params (Impact%20Physical%20Surface :type EPhysicalSurface%20Enum :default "SurfaceType_Default"))
  (CallFunction|CEFXClient Impact%20Physical%20Surface))

(event Custom|CE_FX_Client
  (params (Impact%20Physical%20Surface :type EPhysicalSurface%20Enum :default "SurfaceType_Default"))
  (bind _output_get (Variables|Default|SetPhysicalSurface Impact%20Physical%20Surface))
  (Variables|Default|SetIsWeakPoint (Utilities|Array|ContainsItem (Variables|Default|GetWeakPointSurface) _output_get))
  (if (Variables|Default|GetIsWeakPoint)
    (CallFunction|HitBlink)))

(event Custom|hitBlink
  (bind _weaponmesh (Variables|Visual|GetWeaponMesh))
  (Rendering|Material|SetScalarParameterValueonMaterials _weaponmesh "R_blink_pose" 0.0) ; Close
  (Utilities|FlowControl|Delay 4.0)
  (Rendering|Material|SetScalarParameterValueonMaterials _weaponmesh "R_blink_pose" 0.6) ; open
)

(event EventBeginPlay
  (Rendering|Material|SetScalarParameterValueonMaterials (Variables|Visual|GetWeaponMesh) "R_blink_pose" 0.6))

(fn ConstructionScript
  (|CallParentFunction))
