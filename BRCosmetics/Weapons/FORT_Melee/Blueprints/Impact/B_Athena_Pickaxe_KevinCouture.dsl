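; Kevin Couture pickaxe cosmetic whose glow builds through four stages as it lands hits on specific surface types.
; Those impacts increment a hit counter and play sounds indexed by it; OnRep_HitIncrement checks the material variant to drive stages and bursts.
; Stages set emissive parameters on the mesh and a dynamic material fed to idle Niagara VFX, with the last flagging swing VFX.
; Equipping resets progress, while frontend display previews stages on timed delays.
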(event EventOnInitWeaponCosmetics
  (bind _returnvalue (CallFunction|GetIdleVFXComponent))
  (|CallParentFunction)
  (Utilities|FlowControl|Sequence
    (:then_0
      ;[1 1 1] Custom Depth
      (Rendering|SetRenderCustomDepth _returnvalue true)
      (Rendering|SetCustomDepthStencilValue _returnvalue (Variables|/Script/Engine.PrimitiveComponent|GetCustomDepthStencilValue (Variables|/Script/Engine.Character|GetMesh (Game|GetPlayerCharacter 0))))
      (bind _returnvalue_2 (Rendering|Material|CreateDynamicMaterialInstance (Variables|Default|GetMaterial) "None" "None" :!class /Script/Engine.KismetMaterialLibrary))
      (bind _output_get (Variables|Default|SetMaterialInstance _returnvalue_2))
      (bind _asniagara_particle_system_component (Utilities|Casting|CastToNiagaraComponent (CallFunction|GetIdleVFXComponent)))
      (Niagara|SetNiagaraVariable(Material) _asniagara_particle_system_component "Material" _output_get)
      (Utilities|Casting|CastToFortGameStateFrontEnd (Game|GetGameState)))
    (:then_1
      (if (Reactive|Frontend|IsCosmeticObjectDisplayedinFrontendContext 0)
        (Utilities|FlowControl|Delay 4.0)
        (CallFunction|PickaxeBurst)
        (CallFunction|Stage1)
        (Utilities|FlowControl|Delay 4.0)
        (CallFunction|PickaxeBurst)
        (CallFunction|Stage2)
        (Utilities|FlowControl|Delay 4.0)
        (CallFunction|PickaxeBurst)
        (CallFunction|Stage3)
        (Utilities|FlowControl|Delay 4.0)
        (CallFunction|PickaxeBurst)
        (CallFunction|Stage4)
        (Utilities|FlowControl|Delay 4.0)
        (CallFunction|ResetStages)))))

(event Custom|Reset%20Stages
  (bind _materialinstance (Variables|Default|GetMaterialInstance))
  (bind _weaponmesh (Variables|Visual|GetWeaponMesh))
  (Utilities|FlowControl|Sequence
    (:then_0
      (Rendering|Material|SetScalarParameterValueonMaterials _weaponmesh "Emissive_BrightnessMax" 0.0)
      (Rendering|Material|SetScalarParameterValueonMaterials _weaponmesh "Emissive_BrightnessMin" 0.0)
      (Rendering|Material|SetScalarParameterValueonMaterials _weaponmesh "Emissive mult" 0.1))
    (:then_1
      (Effects|Components|ParticleSystem|SetBoolParameter (CallFunction|GetSwingVFXComponent) "FinalStage" false)
      (Rendering|Material|SetScalarParameterValue _materialinstance "Emissive_BrightnessMax" 0.0 :!class /Script/Engine.MaterialInstanceDynamic)
      (Rendering|Material|SetScalarParameterValue _materialinstance "Emissive_BrightnessMin" 0.0 :!class /Script/Engine.MaterialInstanceDynamic)
      (Rendering|Material|SetScalarParameterValue _materialinstance "Emissive mult" 0.1 :!class /Script/Engine.MaterialInstanceDynamic))))

(event EventBeginPlay
  (|CallParentFunction)
  (Transformation|AttachComponentToComponent (Variables|Default|GetNS_Pickaxe_KevinCouture_Burst) (Variables|Visual|GetWeaponMesh) "None" "KeepRelative" "KeepRelative" "KeepRelative" true)
  (CallFunction|ResetStages))

(event Custom|PickaxeBurst
  (Components|Activation|Activate (Variables|Default|GetNS_Pickaxe_KevinCouture_Burst) true))

(event FortWeapon|EventOnEquippedAndAttached
  (|CallParentFunction)
  (Variables|Default|SetHit%20Increment 0)
  (CallFunction|ResetStages)
  (Components|Activation|Deactivate (Variables|Default|GetNS_Pickaxe_KevinCouture_Burst)))

(event Custom|Stage_1
  (bind _materialinstance (Variables|Default|GetMaterialInstance))
  (bind _weaponmesh (Variables|Visual|GetWeaponMesh))
  ;[1 1 1] Stage Materials set
  (Utilities|FlowControl|Sequence
    (:then_0
      (Rendering|Material|SetScalarParameterValueonMaterials _weaponmesh "Emissive_BrightnessMax" 0.0)
      (Rendering|Material|SetScalarParameterValueonMaterials _weaponmesh "Emissive_BrightnessMin" 0.0)
      (Rendering|Material|SetScalarParameterValueonMaterials _weaponmesh "Emissive mult" 0.1))
    (:then_1
      (Rendering|Material|SetScalarParameterValue _materialinstance "Emissive_BrightnessMax" 0.0 :!class /Script/Engine.MaterialInstanceDynamic)
      (Rendering|Material|SetScalarParameterValue _materialinstance "Emissive_BrightnessMin" 0.0 :!class /Script/Engine.MaterialInstanceDynamic)
      (Rendering|Material|SetScalarParameterValue _materialinstance "Emissive mult" 0.1 :!class /Script/Engine.MaterialInstanceDynamic))))

(event FortWeapon|EventOnPlayImpactFX
  (params (HitResult :type Hit%20Result%20Structure%20%28by%20ref%29) (ImpactPhysicalSurface :type EPhysicalSurface%20Enum :default "SurfaceType_Default") (SpawnedPSC :type FXSystem%20Component%20Object%20Reference))
  (|CallParentFunction HitResult ImpactPhysicalSurface SpawnedPSC)
  (if (Math|Boolean|ORBoolean (Utilities|Enum|Equal(Enum) ImpactPhysicalSurface "SurfaceType14") (Utilities|Enum|Equal(Enum) ImpactPhysicalSurface "SurfaceType15") (Utilities|Enum|Equal(Enum) ImpactPhysicalSurface "SurfaceType16") (Utilities|Enum|Equal(Enum) ImpactPhysicalSurface "SurfaceType12"))
    (CallFunction|ServerHitIncrement)
    (CallFunction|PlayWeakspotImpactSound (Variables|Default|GetHit%20Increment))))

(event Custom|ServerHitIncrement
  ;[1 1 1] Replication
  (Variables|Default|SetHit%20Increment (+ (Variables|Default|GetHit%20Increment) 1)))

(event Custom|Stage_2
  (bind _weaponmesh (Variables|Visual|GetWeaponMesh))
  (bind _materialinstance (Variables|Default|GetMaterialInstance))
  ;[1 1 1] Stage Materials set
  (Utilities|FlowControl|Sequence
    (:then_0
      (Rendering|Material|SetScalarParameterValueonMaterials _weaponmesh "Emissive_BrightnessMax" 3.0)
      (Rendering|Material|SetScalarParameterValueonMaterials _weaponmesh "Emissive_BrightnessMin" 0.0)
      (Rendering|Material|SetScalarParameterValueonMaterials _weaponmesh "Emissive mult" 0.35))
    (:then_1
      (Rendering|Material|SetScalarParameterValue _materialinstance "Emissive_BrightnessMax" 0.5 :!class /Script/Engine.MaterialInstanceDynamic)
      (Rendering|Material|SetScalarParameterValue _materialinstance "Emissive_BrightnessMin" 0.0 :!class /Script/Engine.MaterialInstanceDynamic)
      (Rendering|Material|SetScalarParameterValue _materialinstance "Emissive mult" 0.35 :!class /Script/Engine.MaterialInstanceDynamic))))

(event Custom|Play%20Weakspot%20Impact%20Sound
  (params (Index :type Integer))
  (Audio|PlaySoundatLocation (Utilities|Array|Get(acopy) (Variables|Default|GetWeakspotImpactSounds) (Math|Integer|Clamp(Integer) Index 0 (Utilities|Array|LastIndex (Variables|Default|GetWeakspotImpactSounds)))) (Transformation|GetWorldLocation (Variables|Visual|GetWeaponMesh))))

(event Custom|Stage_3
  (bind _weaponmesh (Variables|Visual|GetWeaponMesh))
  (bind _materialinstance (Variables|Default|GetMaterialInstance))
  ;[1 1 1] Stage Materials set
  (Utilities|FlowControl|Sequence
    (:then_0
      (Rendering|Material|SetScalarParameterValueonMaterials _weaponmesh "Emissive_BrightnessMax" 8.0)
      (Rendering|Material|SetScalarParameterValueonMaterials _weaponmesh "Emissive_BrightnessMin" -0.1)
      (Rendering|Material|SetScalarParameterValueonMaterials _weaponmesh "Emissive mult" 0.5))
    (:then_1
      (Rendering|Material|SetScalarParameterValue _materialinstance "Emissive_BrightnessMax" 8.0 :!class /Script/Engine.MaterialInstanceDynamic)
      (Rendering|Material|SetScalarParameterValue _materialinstance "Emissive_BrightnessMin" 0.0 :!class /Script/Engine.MaterialInstanceDynamic)
      (Rendering|Material|SetScalarParameterValue _materialinstance "Emissive mult" 0.5 :!class /Script/Engine.MaterialInstanceDynamic))))

(event Custom|Stage_4
  (bind _weaponmesh (Variables|Visual|GetWeaponMesh))
  (bind _materialinstance (Variables|Default|GetMaterialInstance))
  ;[1 1 1] Stage Materials set
  (Utilities|FlowControl|Sequence
    (:then_0
      (Rendering|Material|SetScalarParameterValueonMaterials _weaponmesh "Emissive_BrightnessMax" 25.0)
      (Rendering|Material|SetScalarParameterValueonMaterials _weaponmesh "Emissive_BrightnessMin" 0.1)
      (Rendering|Material|SetScalarParameterValueonMaterials _weaponmesh "Emissive mult" 1.0))
    (:then_1
      (Effects|Components|ParticleSystem|SetBoolParameter (CallFunction|GetSwingVFXComponent) "FinalStage" true)
      (Rendering|Material|SetScalarParameterValue _materialinstance "Emissive_BrightnessMax" 25.0 :!class /Script/Engine.MaterialInstanceDynamic)
      (Rendering|Material|SetScalarParameterValue _materialinstance "Emissive_BrightnessMin" 0.1 :!class /Script/Engine.MaterialInstanceDynamic)
      (Rendering|Material|SetScalarParameterValue _materialinstance "Emissive mult" 1.0 :!class /Script/Engine.MaterialInstanceDynamic))))

(fn OnRep_HitIncrement
  (Networking|CanExecuteCosmeticEvents
    (:True
      (bind _asfort_player_pawn (Utilities|Casting|CastToFortPlayerPawn (Variables|Actor|GetInstigator)))
      (bind _index (CosmeticLoadout|GetVariantChannelActiveIndexforSlotActor "(TagName=\"LoadoutSlot.Pickaxe\")" _asfort_player_pawn "(TagName=\"Cosmetics.Variant.Channel.Material\")" (Utilities|Array|MakeArray "(TagName=\"Cosmetics.Variant.Property.Mat1\")" "(TagName=\"Cosmetics.Variant.Property.Mat2\")")))
      (if (== _index 0)
        (switch int (Variables|Default|GetHit%20Increment)
          (:0
            (CallFunction|ResetStages))
          (:1
            (CallFunction|Stage1))
          (:2
            (CallFunction|Stage2))
          (:3
            (CallFunction|Stage3))
          (:4
            (CallFunction|Stage4)))
        (CallFunction|PickaxeBurst)))))

(fn ConstructionScript
  (|CallParentFunction))
