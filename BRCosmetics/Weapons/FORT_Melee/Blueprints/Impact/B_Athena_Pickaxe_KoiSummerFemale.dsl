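; Cosmetic Blueprint for the KoiSummerFemale pickaxe that makes the weapon glow on weak-spot hits.
; Impact FX events trigger reactivity in frontend preview, or through ReplicateReactivity when the hit surface is one of three specific types.
; A shared timeline lerps the mesh's Emissive Color from white to GlowIntensity and drives the trail and idle VFX "Glow" parameter.
; Left and right swing events record the swing side in a boolean.
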(event FortWeapon|EventOnPlayImpactFX
  (params (HitResult :type Hit%20Result%20Structure%20%28by%20ref%29) (ImpactPhysicalSurface :type EPhysicalSurface%20Enum :default "SurfaceType_Default") (SpawnedPSC :type FXSystem%20Component%20Object%20Reference))
  ;[1 1 1] GlowOnWeakSpotHit
  (|CallParentFunction HitResult ImpactPhysicalSurface SpawnedPSC)
  (if (Reactive|Frontend|IsCosmeticObjectDisplayedasPreview 0)
    (CallFunction|FrontendReactivity)
    (elif (Math|Boolean|ORBoolean (Utilities|Enum|Equal(Enum) (Collision|GetSurfaceType HitResult) "SurfaceType14") (Utilities|Enum|Equal(Enum) (Collision|GetSurfaceType HitResult) "SurfaceType15") (Utilities|Enum|Equal(Enum) (Collision|GetSurfaceType HitResult) "SurfaceType16"))
      (CallFunction|ReplicateReactivity))))

(event Custom|WeakpointReactivity
  ;[1 1 1] Weakpoint Reactivity
  (bind (_direction _lerp_alpha) (Timeline|Timeline_1 PlayFromStart)
    (:Update
      (Rendering|Material|SetVectorParameterValueonMaterials (Variables|Visual|GetWeaponMesh) "Emissive Color" (Math|Conversions|ToVector(LinearColor) (Math|Color|Lerp(LinearColor) "(R=1.000000,G=1.000000,B=1.000000,A=1.000000)" (Math|Conversions|ToLinearColor(Float) (Variables|Default|GetGlowIntensity)) _lerp_alpha)))
      (Effects|Components|ParticleSystem|SetFloatParameter (CallFunction|GetAnimTrailVFXComponent) "Glow" _lerp_alpha)
      (Effects|Components|ParticleSystem|SetFloatParameter (CallFunction|GetIdleVFXComponent) "Glow" _lerp_alpha))))

(event Custom|ReplicateReactivity
  (CallFunction|WeakpointReactivity))

(event EventSwingLeft
  (|CallParentFunction)
  (Variables|Default|SetRight%20? false))

(event EventSwingRight
  (|CallParentFunction)
  (Variables|Default|SetRight%20? true))

(event Custom|FrontendReactivity
  ;[1 1 1] Weakpoint Reactivity
  (Timeline|Timeline_1 PlayFromStart))

(fn ConstructionScript
  (|CallParentFunction))
