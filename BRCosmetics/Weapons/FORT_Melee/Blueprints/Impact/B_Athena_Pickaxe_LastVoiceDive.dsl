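; LastVoiceDive pickaxe Blueprint that layers a special cycling sound and VFX changes onto weapon impacts.
; On each impact FX event it stores the hit, runs the parent logic, activates idle VFX and deactivates active swing VFX.
; It then spawns the enemy-impact cue at the impact point, with no actor filter in this graph.
; An impact counter feeds the cue's integer parameter and wraps after eight hits, stepping through sound variations.
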(event FortWeapon|EventOnPlayImpactFX
  (params (HitResult :type Hit%20Result%20Structure%20%28by%20ref%29) (ImpactPhysicalSurface :type EPhysicalSurface%20Enum :default "SurfaceType_Default") (SpawnedPSC :type FXSystem%20Component%20Object%20Reference))
  (bind (_bblockinghit _binitialoverlap _time _distance _location _impactpoint) (Collision|BreakHitResult (Variables|Default|GetPIckaxe%20Impact)))
  (bind _returnvalue (CallFunction|GetSwingVFXComponent))
  (Variables|Default|SetPIckaxe%20Impact HitResult)
  (|CallParentFunction HitResult ImpactPhysicalSurface SpawnedPSC)
  (Utilities|FlowControl|Sequence
    (:then_0
      (Components|Activation|Activate (CallFunction|GetIdleVFXComponent) false))
    (:then_1
      (Utilities|IsValid _returnvalue
        (:Is%20Valid
          (if (Components|Activation|IsActive _returnvalue)
            (Components|Activation|Deactivate _returnvalue)))))
    (:then_2
      ;[1 1 1] Play special sound for pickaxe when hitting actors
      (bind _returnvalue_2 (Audio|SpawnSoundatLocation "/BRCosmetics/Sounds/PickAxes/LastVoiceDive/PickaxeImpactEnemy_Last_VoiceDive_Cue.PickaxeImpactEnemy_Last_VoiceDive_Cue" _impactpoint))
      (Audio|Parameter|SetIntegerParameter _returnvalue_2 "LastVoiceDrive" (Variables|Default|GetPixkaxe%20Impact%20Count) :!class /Script/Engine.AudioComponent)
      (bind _result (Math|Integer|IncrementInt (Variables|Default|GetPixkaxe%20Impact%20Count))
        (:%20%20
          (bind _output_get (Variables|Default|SetPixkaxe%20Impact%20Count _result))
          (if (>= _output_get 8)
            (Variables|Default|SetPixkaxe%20Impact%20Count 0)))))))

(fn ConstructionScript
  (|CallParentFunction))
