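; Lavish Male dual pickaxe weapon Blueprint that adds equip trail FX and orients impact particles to match each swing.
; On equip, it attaches Niagara trails to main and offhand meshes after a short retriggerable delay, unless player 0's pawn is moving fast, falling or jumping.
; Swing events set right/left and primary/secondary flags.
; After another short delay, the impact handler uses those flags to pick a rotation, applied relative to the spawned effect's transform.
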(event FortWeapon|EventOnEquippedAndAttached
  ;[1 1 1] Equip trail FX
  (|CallParentFunction)
  (bind _asplayer_pawn_athena (Utilities|Casting|CastToPlayerPawn_Athena (Game|GetPlayerCharacter 0)))
  (if (> (Variables|/Script/FortniteGame.FortPlayerPawnAthena|GetCurrentPawnSquaredSpeed _asplayer_pawn_athena) 8.1e+05)
    _
    (elif (or (AI|Components|NavMovement|IsFalling (Variables|/Script/Engine.Character|GetCharacterMovement _asplayer_pawn_athena)) (Variables|/Script/Engine.Character|GetbWasJumping _asplayer_pawn_athena))
      _
      (else
        (Utilities|FlowControl|RetriggerableDelay 0.09)
        (bind _returnvalue (Niagara|SpawnSystemAttached "/BRCosmetics/Weapons/FORT_Melee/Pickaxe_Masked_Warrior_Spring_Male_1H/FX/NS_Pickaxe_Lavish_M_Trail.NS_Pickaxe_Lavish_M_Trail" (Variables|Visual|GetWeaponMesh) "None" "0, 0, 0" "0, 0, 0" "KeepRelativeOffset" true false))
        (Components|Activation|SetActive _returnvalue true true)
        (bind _returnvalue_2 (Niagara|SpawnSystemAttached "/BRCosmetics/Weapons/FORT_Melee/Pickaxe_Masked_Warrior_Spring_Male_1H/FX/NS_Pickaxe_Lavish_M_Trail.NS_Pickaxe_Lavish_M_Trail" (Variables|Visual|GetWeaponMeshOffhand) "None" "0, 0, 0" "0, 0, 0" "KeepRelativeOffset" true false))
        (Components|Activation|SetActive _returnvalue_2 true true)))))

(event EventSwingLeft_Common
  ;[1 1 1] Selects the correct swing
  (|CallParentFunction)
  (Variables|Default|SetbIsRightSwing false)
  (Variables|Default|SetbIsSecondarySwing false))

(event EventSwingRight_Common
  ;[1 1 1] Selects the correct swing
  (|CallParentFunction)
  (Variables|Default|SetbIsRightSwing true)
  (Variables|Default|SetbIsSecondarySwing false))

(event EventSwingLeft
  ;[1 1 1] Selects the correct swing
  (|CallParentFunction)
  (Variables|Default|SetbIsRightSwing false)
  (Variables|Default|SetbIsSecondarySwing false))

(event EventSwingRight
  ;[1 1 1] Selects the correct swing
  (|CallParentFunction)
  (Variables|Default|SetbIsRightSwing true)
  (Variables|Default|SetbIsSecondarySwing false))

(event EventSwingLeft2
  ;[1 1 1] Selects the correct swing
  (|CallParentFunction)
  (Variables|Default|SetbIsRightSwing false)
  (Variables|Default|SetbIsSecondarySwing true))

(event EventSwingRight2
  ;[1 1 1] Selects the correct swing
  (|CallParentFunction)
  (Variables|Default|SetbIsRightSwing true)
  (Variables|Default|SetbIsSecondarySwing true))

(event FortWeapon|EventOnPlayImpactFX
  (params (HitResult :type Hit%20Result%20Structure%20%28by%20ref%29) (ImpactPhysicalSurface :type EPhysicalSurface%20Enum :default "SurfaceType_Default") (SpawnedPSC :type FXSystem%20Component%20Object%20Reference))
  ;[1 1 1] Rotates the impact particle to match the swing
  (|CallParentFunction HitResult ImpactPhysicalSurface SpawnedPSC)
  (Utilities|FlowControl|RetriggerableDelay 0.02)
  (Transformation|SetWorldRotation SpawnedPSC (Math|Transform|TransformRotation (Transformation|GetWorldTransform SpawnedPSC) (Math|Rotator|SelectRotator (Math|Rotator|SelectRotator (Variables|Default|GetSecondaryRightRotation) (Variables|Default|GetPrimaryRightRotation) (Variables|Default|GetbIsSecondarySwing)) (Math|Rotator|SelectRotator (Variables|Default|GetSecondaryLeftRotation) (Variables|Default|GetPrimaryLeftRotation) (Variables|Default|GetbIsSecondarySwing)) (Variables|Default|GetbIsRightSwing))) false false))

(fn ConstructionScript
  (|CallParentFunction))
