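; Pickaxe cosmetic Blueprint (Lettuce) that plays a reactive effect on impacts with four specific surface types, or when previewed or shown in the frontend.
; Impact FX events relay through server and client custom events so other players see the reaction.
; Only particle variant index 2 reacts, despite a comment citing index 0.
; A "Weak Point Hit" timeline drives swing, idle and impact particle parameters plus the weapon mesh's FXIntensity.
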(event FortWeapon|EventOnPlayImpactFX
  (params (HitResult :type Hit%20Result%20Structure%20%28by%20ref%29) (ImpactPhysicalSurface :type EPhysicalSurface%20Enum :default "SurfaceType_Default") (SpawnedPSC :type FXSystem%20Component%20Object%20Reference))
  (|CallParentFunction HitResult ImpactPhysicalSurface SpawnedPSC)
  (if (Math|Boolean|ORBoolean (Utilities|Enum|Equal(Enum) (Collision|GetSurfaceType HitResult) "SurfaceType12") (Utilities|Enum|Equal(Enum) (Collision|GetSurfaceType HitResult) "SurfaceType14") (Utilities|Enum|Equal(Enum) (Collision|GetSurfaceType HitResult) "SurfaceType15") (Utilities|Enum|Equal(Enum) (Collision|GetSurfaceType HitResult) "SurfaceType16") (or (Reactive|Frontend|IsCosmeticObjectDisplayedasPreview 0) (Reactive|Frontend|IsCosmeticObjectDisplayedinFrontendContext 0)))
    (CallFunction|CEFXServer) ; It calls the Server Event, that calls the Client Event, \nthat finally calls the reactivity event
  ))

(event Custom|CE_FX_Server
  ;[1 0.2356 0.1858] "Server Logic". It makes sure the reactivity can be seen by other clients
  (CallFunction|CEFXClient))

(event Custom|CE_FX_Client
  ;[1 0.2356 0.1858] "Server Logic". It makes sure the reactivity can be seen by other clients
  (CallFunction|ReactivityEvent))

(event Custom|ReactivityEvent
  ;[1 1 1] Makes sure it will only affect the "reactive" cosmetic variant. The reactive one is usually setup as index 0
  ;[1 0 0] TODO: Implement this on final item setup
  (bind _index (Class|FortCosmeticFunctionLibrary|%28Deprecated%29GetVariantChannelActiveIndexForSlot "(TagName=\"LoadoutSlot.Pickaxe\")" (CallFunction|GetCosmeticData) "(TagName=\"Cosmetics.Variant.Channel.Particle\")" (Utilities|Array|MakeArray "(TagName=\"Cosmetics.Variant.Property.Particle1\")" "(TagName=\"Cosmetics.Variant.Property.Particle2\")" "(TagName=\"Cosmetics.Variant.Property.Particle3\")")))
  (if (== _index 2)
    (bind (_direction _weak_point_hit_fade _weak_point_hit_burst _weakpointmaterialfx) (Timeline|Weak%20Point%20Hit PlayFromStart)
      (:Update
        (Utilities|FlowControl|Sequence
          (:then_0
            (Effects|Components|ParticleSystem|SetFloatParameter (CallFunction|GetSwingVFXComponent) "WeakPointHitFade" _weak_point_hit_fade)
            (Effects|Components|ParticleSystem|SetFloatParameter (CallFunction|GetIdleVFXComponent) "WeakPointHitFade" _weak_point_hit_fade)
            (Effects|Components|ParticleSystem|SetFloatParameter (CallFunction|GetIdleVFXComponent) "WeakPointHitBurst" _weak_point_hit_burst)
            (Effects|Components|ParticleSystem|SetFloatParameter (Variables|Default|GetImpact%20FX) "WeakPointHitFade" _weak_point_hit_fade))
          (:then_1
            (Rendering|Material|SetScalarParameterValueonMaterials (Variables|Visual|GetWeaponMesh) "FXIntensity" _weakpointmaterialfx)))))))

(fn ConstructionScript
  (|CallParentFunction))
