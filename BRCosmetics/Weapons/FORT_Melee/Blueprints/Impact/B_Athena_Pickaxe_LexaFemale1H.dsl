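; Blueprint for the LexaFemale1H Athena pickaxe cosmetic, adjusting its swing trail effect.
; On the second left swing event it runs the parent logic, waits a short retriggerable delay, then activates and resets the anim trail VFX component.
; The construction script only calls its parent.
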(event EventSwingLeft2
  (|CallParentFunction)
  (Utilities|FlowControl|RetriggerableDelay 0.17)
  (Components|Activation|SetActive (CallFunction|GetAnimTrailVFXComponent) true true))

(fn ConstructionScript
  (|CallParentFunction))
