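; Blueprint for the LilacLeather Athena pickaxe cosmetic, a melee harvesting weapon.
; Its construction script runs the parent logic, then turns off auto-activation on the idle particle system component returned by GetIdlePSC.
; All other behavior comes from the parent class.
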(fn ConstructionScript
  (|CallParentFunction)
  (Components|Activation|SetAutoActivate (CallFunction|GetIdlePSC) false))
