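; Lively Domino pickaxe Blueprint that drives its impact and idle particle effects from each hit.
; Impact FX takes a "Direction" vector chosen by the instigator's current swing montage section.
; Server and client events check the surface against weak-point surfaces; matches light the idle VFX, with a retriggerable-delay dissolve and do-once sweep timelines.
; A separate timeline animates the weak-point glow when displayed in the frontend.
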(event FortWeapon|EventOnPlayImpactFX
  (params (HitResult :type Hit%20Result%20Structure%20%28by%20ref%29) (ImpactPhysicalSurface :type EPhysicalSurface%20Enum :default "SurfaceType_Default") (SpawnedPSC :type FXSystem%20Component%20Object%20Reference))
  (|CallParentFunction HitResult ImpactPhysicalSurface SpawnedPSC)
  (Utilities|FlowControl|Sequence
    (:then_0
      ; We can made this function execute on server and client at same time , not need callback from server ,that will let FX execute delay a bit depend on network speed
      (CallFunction|CEFXServer ImpactPhysicalSurface)
      (CallFunction|CEFXClient ImpactPhysicalSurface)
      (Networking|CanExecuteCosmeticEvents
        (:True
          (Variables|Default|SetImpact%20FX%20Index (Utilities|Array|FindItem (Variables|Default|GetMontageSections) (Animation|Montage|MontageGetCurrentSection (Components|SkeletalMesh|GetAnimInstance (Variables|/Script/Engine.Character|GetMesh (Utilities|Casting|CastToFortPawn (Game|GetInstigator)))) (Animation|Montage|GetCurrentActiveMontage (Components|SkeletalMesh|GetAnimInstance (Variables|/Script/Engine.Character|GetMesh (Utilities|Casting|CastToFortPawn (Game|GetInstigator))))))))
          (Effects|Components|ParticleSystem|SetVectorParameter (Variables|Default|GetImpact%20FX) "Direction" (Utilities|Select :Index (Variables|Default|GetImpact%20FX%20Index) :Option%200 (Variables|Default|GetSwing1%20Direction) :Option%201 (Variables|Default|GetSwing2%20Direction) :Option%202 (Variables|Default|GetSwing3%20Direction) :Option%203 (Variables|Default|GetSwing4%20Direction))))))
    (:then_1
      (if (Reactive|Frontend|IsCosmeticObjectDisplayedinFrontendContext 0)
        (bind (_direction _feweakpoint) (Timeline|Timeline PlayFromStart)
          (:Update
            (Effects|Components|ParticleSystem|SetFloatParameter (CallFunction|GetIdleVFXComponent) "WeakPoint" _feweakpoint)))))))

(event Custom|CE_FX_Server
  (params (Impact%20Physical%20Surface :type EPhysicalSurface%20Enum :default "SurfaceType_Default"))
  (CallFunction|CEFXClient Impact%20Physical%20Surface))

(event Custom|CE_FX_Client
  (params (Impact%20Physical%20Surface :type EPhysicalSurface%20Enum :default "SurfaceType_Default"))
  (bind _returnvalue (CallFunction|GetIdleVFXComponent))
  (bind _output_get (Variables|Default|SetImpact%20Physical%20Surface Impact%20Physical%20Surface))
  (bind _output_get_2 (Variables|Default|SetWeakPoint (Utilities|Array|ContainsItem (Variables|Default|GetWeakPointSurface) _output_get)))
  (if _output_get_2
    (Effects|Components|ParticleSystem|SetFloatParameter _returnvalue "WeakPoint" 1.0)
    (Utilities|FlowControl|Sequence
      (:then_0
        (Utilities|FlowControl|Sequence
          (:then_0
            (Utilities|FlowControl|RetriggerableDelay 5.0)
            (Utilities|FlowControl|Sequence
              (:then_0
                (bind (_direction _dissolve) (Timeline|Dissolve PlayFromStart)
                  (:Update
                    (Effects|Components|ParticleSystem|SetFloatParameter _returnvalue "WeakPoint" _dissolve))))
              (:then_1
                _)))
          (:then_1
            _))
        (Utilities|FlowControl|DoOnce false
          (:Completed
            (Utilities|FlowControl|Sequence
              (:then_0
                (bind (_direction_2 _weakpoint_sweep) (Timeline|Sweep PlayFromStart)
                  (:Update
                    (Effects|Components|ParticleSystem|SetFloatParameter _returnvalue "WeakPoint_Sweep" _weakpoint_sweep))))
              (:then_1
                (bind (_direction_3 _size) (Timeline|SweepGlowSize PlayFromStart)
                  (:Update
                    (Effects|Components|ParticleSystem|SetFloatParameter _returnvalue "WeakPoint_SweepSize" _size))))))))
      (:then_1
        _))
    (Timeline|Dissolve Stop)))

(fn ConstructionScript
  (|CallParentFunction))
