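; Cosmetic Blueprint for the Local Talent Flare pickaxe that sequences its deploy and idle visual effects.
; Begin play, cosmetic init and visibility changes deactivate the idle VFX, then DoOnce-gated chains spawn a deploy Niagara system on the mesh root and reactivate idle VFX after a retriggerable delay.
; The idle particle component activates at begin play, and after short delays on end play and the anim Blueprint's FrontEndIdle delegate.
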(event EventBeginPlay
  (bind _returnvalue (CallFunction|GetIdleVFXComponent))
  (|CallParentFunction)
  (Utilities|FlowControl|Sequence
    (:then_0
      (Components|Activation|Activate (CallFunction|GetIdlePSC) false))
    (:then_1
      (bind _asgalileo_ferry_axe_anim_bp (Utilities|Casting|CastToGalileo_Ferry_Axe_AnimBP (Components|SkeletalMesh|GetAnimInstance (Variables|Visual|GetWeaponMesh))))
      (Delegates|BindEventtoFrontEndIdle _asgalileo_ferry_axe_anim_bp (delegate Custom|FrontEndIdle)))
    (:then_2
      (Utilities|FlowControl|DelayUntilNextTick)
      (Components|Activation|Deactivate _returnvalue)
      (Utilities|FlowControl|DoOnce false
        (:Completed
          (Niagara|SpawnSystemAttached "/BRCosmetics/Weapons/FORT_Melee/Pickaxe_LocalTalent_Flare/FX/NS_Pickaxe_LocalTalent_Flare_Deploy.NS_Pickaxe_LocalTalent_Flare_Deploy" (Variables|Visual|GetWeaponMesh) "root" "0, 0, 0" "0, 0, 0" "KeepRelativeOffset" true)
          (Utilities|FlowControl|RetriggerableDelay 0.68)
          (Components|Activation|Activate _returnvalue false))))))

(event Custom|FrontEndIdle
  (Utilities|FlowControl|DoOnce false
    (:Completed
      (Utilities|FlowControl|Delay 0.3)
      (Components|Activation|Activate (CallFunction|GetIdlePSC) false))))

(event EventEndPlay
  (params (EndPlayReason :type EEndPlayReason%20Enum :default "Destroyed"))
  (|CallParentFunction EndPlayReason)
  (Utilities|FlowControl|DoOnce false
    (:Completed
      (Utilities|FlowControl|Delay 0.3)
      (Components|Activation|Activate (CallFunction|GetIdlePSC) false))))

(event EventOnWeaponVisibilityChanged
  (params (bVisible :type Boolean) (bSetForLocalControllerOnly :type Boolean))
  (Utilities|FlowControl|DelayUntilNextTick)
  (Components|Activation|Deactivate (CallFunction|GetIdleVFXComponent))
  (Utilities|FlowControl|DoOnce false
    (:Completed
      (Niagara|SpawnSystemAttached "/BRCosmetics/Weapons/FORT_Melee/Pickaxe_LocalTalent_Flare/FX/NS_Pickaxe_LocalTalent_Flare_Deploy.NS_Pickaxe_LocalTalent_Flare_Deploy" (Variables|Visual|GetWeaponMesh) "root" "0, 0, 0" "0, 0, 0" "KeepRelativeOffset" true)
      (Utilities|FlowControl|RetriggerableDelay 0.68)
      (Components|Activation|Activate (CallFunction|GetIdleVFXComponent) false))))

(event EventOnInitWeaponCosmetics
  (bind _returnvalue (CallFunction|GetIdleVFXComponent))
  (|CallParentFunction)
  (Utilities|FlowControl|DelayUntilNextTick)
  (Components|Activation|Deactivate _returnvalue)
  (Utilities|FlowControl|DoOnce false
    (:Completed
      (Niagara|SpawnSystemAttached "/BRCosmetics/Weapons/FORT_Melee/Pickaxe_LocalTalent_Flare/FX/NS_Pickaxe_LocalTalent_Flare_Deploy.NS_Pickaxe_LocalTalent_Flare_Deploy" (Variables|Visual|GetWeaponMesh) "root" "0, 0, 0" "0, 0, 0" "KeepRelativeOffset" true)
      (Utilities|FlowControl|RetriggerableDelay 0.68)
      (Components|Activation|Activate _returnvalue false))))

(fn ConstructionScript
  (|CallParentFunction))
