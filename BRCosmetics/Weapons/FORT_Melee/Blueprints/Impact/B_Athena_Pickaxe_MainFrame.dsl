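; Pickaxe weapon Blueprint for the Mainframe cosmetic that drives its idle Niagara effect.
; Weapon visibility changes activate or deactivate that Niagara component.
; Cosmetic init detects frontend display, sets a flag on the effect, and in the frontend also overrides its mesh size and scale.
; The construction script turns off auto-activation on the idle particle component returned by GetIdlePSC.
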(event EventOnWeaponVisibilityChanged
  (params (bVisible :type Boolean) (bSetForLocalControllerOnly :type Boolean))
  (bind _ns_pickaxe_mainframe_male_idle (Variables|Default|GetNS_Pickaxe_Mainframe_Male_idle))
  (|CallParentFunction bVisible bSetForLocalControllerOnly)
  (if bVisible
    (Components|Activation|Activate _ns_pickaxe_mainframe_male_idle false)
    (else
      (Components|Activation|Deactivate _ns_pickaxe_mainframe_male_idle))))

(event EventOnInitWeaponCosmetics
  (bind _ns_pickaxe_mainframe_male_idle (Variables|Default|GetNS_Pickaxe_Mainframe_Male_idle))
  (|CallParentFunction)
  (if (Reactive|Frontend|IsCosmeticObjectDisplayedinFrontendContext (Variables|Default|GetNS_Pickaxe_Mainframe_Male_idle))
    (Niagara|SetNiagaraVariable(Vector3) _ns_pickaxe_mainframe_male_idle "Mesh_size" "1.000000,0.700000,1.000000")
    (Niagara|SetNiagaraVariable(Float) _ns_pickaxe_mainframe_male_idle "Size_scale" 1.0)
    (Niagara|SetNiagaraVariable(Bool) _ns_pickaxe_mainframe_male_idle "Is_in_frontend?" true)
    (else
      (Niagara|SetNiagaraVariable(Bool) _ns_pickaxe_mainframe_male_idle "Is_in_frontend?" false))))

(fn ConstructionScript
  (|CallParentFunction)
  (Components|Activation|SetAutoActivate (CallFunction|GetIdlePSC) false))
