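; Athena MicroLeaf pickaxe weapon cosmetic whose material reacts as impacts land.
; A CurrentIndex counter resets on equip, increments per impact up to 24, and drives a same-named scalar on the weapon mesh's dynamic material.
; In frontend display while unequipped, init runs a delayed loop and steps the index with a far higher cap and a debug print.
; The construction script disables auto-activation of the idle particle component.
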(event EventOnInitWeaponCosmetics
  (|CallParentFunction)
  (if (Reactive|Frontend|IsCosmeticObjectDisplayedinFrontendContext 0)
    ; Frontend Off Pawn locker loop
    (if (not (FortWeapon|IsEquipped))
      (Utilities|ForLoopwithDelay 0 (Math|Random|RandomIntegerinRange 1 3) 0.3
        (:LoopBody
          _)
        (:Complete
          (Utilities|FlowControl|Delay (Math|Random|RandomFloatinRange 3.0 5.0))))
      (if (< (Variables|Default|GetCurrentIndex) 1500)
        (bind _result (Math|Integer|IncrementInt (Variables|Default|GetCurrentIndex))
          (:%20%20
            (CallFunction|UpdateReactivity _result)
            (Development|PrintString (Utilities|String|ToString(Integer) (Variables|Default|GetCurrentIndex)))))
        (else
          (bind _output_get (Variables|Default|SetCurrentIndex 0))
          (CallFunction|UpdateReactivity _output_get))))
    (else
      (bind _output_get_2 (Variables|Default|SetCurrentIndex 0))
      (CallFunction|UpdateReactivity _output_get_2))))

(event FortWeapon|EventOnEquippedAndAttached
  (|CallParentFunction)
  (bind _output_get (Variables|Default|SetCurrentIndex 0))
  (CallFunction|UpdateReactivity _output_get))

(event FortWeapon|EventOnPlayImpactFX
  (params (HitResult :type Hit%20Result%20Structure%20%28by%20ref%29) (ImpactPhysicalSurface :type EPhysicalSurface%20Enum :default "SurfaceType_Default") (SpawnedPSC :type FXSystem%20Component%20Object%20Reference))
  (|CallParentFunction HitResult ImpactPhysicalSurface SpawnedPSC)
  (CallFunction|OnSwing))

(event Custom|OnSwing
  (if (< (Variables|Default|GetCurrentIndex) 24)
    (bind _result (Math|Integer|IncrementInt (Variables|Default|GetCurrentIndex))
      (:%20%20
        (CallFunction|UpdateReactivity _result)))))

(event EventSwingLeft
  (|CallParentFunction))

(event EventSwingRight
  (|CallParentFunction))

(event Custom|UpdateReactivity
  (params (Value :type Integer))
  (Rendering|Material|SetScalarParameterValue (Class|BTechArtGlobalLibrary|GetDynamicMaterialInstance (Variables|Visual|GetWeaponMesh)) "CurrentIndex" (Math|Conversions|ToFloat(Integer) Value) :!class /Script/Engine.MaterialInstanceDynamic))

(fn ConstructionScript
  (|CallParentFunction)
  (Components|Activation|SetAutoActivate (CallFunction|GetIdlePSC) false))
