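; Weapon Blueprint for the MistMylar pickaxe cosmetic that tailors its impact and idle trail effects to each swing.
; Swing events track left or right side and whether it is a second swing.
; On impact, it orients the spawned effect to the hit normal plus a per-side rotation and passes second-swing flags to Niagara as SecondImpact and RightSwing2.
; Common swing events switch off the idle trail, then restore it after a retriggerable delay.
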(event FortWeapon|EventOnPlayImpactFX
  (params (HitResult :type Hit%20Result%20Structure%20%28by%20ref%29) (ImpactPhysicalSurface :type EPhysicalSurface%20Enum :default "SurfaceType_Default") (SpawnedPSC :type FXSystem%20Component%20Object%20Reference))
  (bind (_bblockinghit _binitialoverlap _time _distance _location _impactpoint _normal) (Collision|BreakHitResult HitResult))
  (|CallParentFunction HitResult ImpactPhysicalSurface SpawnedPSC)
  (bind _asniagara_particle_system_component (Utilities|Casting|CastToNiagaraComponent SpawnedPSC))
  (Utilities|FlowControl|Sequence
    (:then_0
      (Variables|Default|SetImpact%20FX _asniagara_particle_system_component)
      (Transformation|SetWorldRotation SpawnedPSC (Math|Rotator|CombineRotators (Math|Rotator|MakeRotfromX _normal) (select (Variables|Default|GetRight?) (Variables|Default|GetRight%20Rotation) (Variables|Default|GetLeft%20Rotation))) false false))
    (:then_1
      (Niagara|SetNiagaraVariable(Bool) _asniagara_particle_system_component "SecondImpact" (Variables|Default|GetLeft2%20Swing?)))
    (:then_2
      (Niagara|SetNiagaraVariable(Bool) _asniagara_particle_system_component "RightSwing2" (Variables|Default|GetRight%202%20Swing)))))

(event EventSwingLeft
  (|CallParentFunction)
  (Variables|Default|SetLeft2%20Swing? false)
  (Variables|Default|SetRight%202%20Swing false))

(event EventSwingRight
  (|CallParentFunction)
  (Variables|Default|SetLeft2%20Swing? false)
  (Variables|Default|SetRight%202%20Swing false))

(event EventSwingLeft_Common
  (|CallParentFunction)
  (Variables|Default|SetRight? false)
  (Effects|Components|ParticleSystem|SetBoolParameter (CallFunction|GetIdleVFXComponent) "Idle Trail" false)
  (Utilities|FlowControl|RetriggerableDelay (Variables|Default|GetIdle%20Trail%20Delay))
  (Effects|Components|ParticleSystem|SetBoolParameter (CallFunction|GetIdleVFXComponent) "Idle Trail" true))

(event EventSwingRight_Common
  (|CallParentFunction)
  (Variables|Default|SetRight? true)
  (Effects|Components|ParticleSystem|SetBoolParameter (CallFunction|GetIdleVFXComponent) "Idle Trail" false)
  (Utilities|FlowControl|RetriggerableDelay (Variables|Default|GetIdle%20Trail%20Delay))
  (Effects|Components|ParticleSystem|SetBoolParameter (CallFunction|GetIdleVFXComponent) "Idle Trail" true))

(event EventSwingRight2
  (|CallParentFunction)
  (Variables|Default|SetRight%202%20Swing true))

(event EventSwingLeft2
  (|CallParentFunction)
  (Variables|Default|SetRight%202%20Swing false)
  (Variables|Default|SetLeft2%20Swing? true))

(fn ConstructionScript
  (|CallParentFunction))
