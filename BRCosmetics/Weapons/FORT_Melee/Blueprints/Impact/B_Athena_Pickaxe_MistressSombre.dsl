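; Mistress Sombre pickaxe Blueprint that plays a reactive visual effect on kills and on certain impacts.
; Impacts on specific surface types, or in preview and frontend contexts, relay through server and client events so other clients see it.
; The reactivity event timestamps mesh materials and flags the swing VFX while a timeline drives the idle VFX, clearing the flag when finished.
; Spawned impact particles inherit the current reactive state.
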(event EventUpdateBasedOnKills
  (params (Watched%20Kills :type Integer))
  ; React On kills
  (|CallParentFunction Watched%20Kills)
  (if (> Watched%20Kills 0)
    (CallFunction|ReactivityEvent)))

(event FortWeapon|EventOnPlayImpactFX
  (params (HitResult :type Hit%20Result%20Structure%20%28by%20ref%29) (ImpactPhysicalSurface :type EPhysicalSurface%20Enum :default "SurfaceType_Default") (SpawnedPSC :type FXSystem%20Component%20Object%20Reference))
  ; Reacty On WeakSpot
  (|CallParentFunction HitResult ImpactPhysicalSurface SpawnedPSC)
  (bind _output_get (Variables|Default|SetSpawned%20PSC SpawnedPSC))
  (Effects|Components|ParticleSystem|SetBoolParameter _output_get "IsReactive" (Variables|Default|GetbIsReactive))
  (if (Math|Boolean|ORBoolean (Utilities|Enum|Equal(Enum) (Collision|GetSurfaceType HitResult) "SurfaceType12") (Utilities|Enum|Equal(Enum) (Collision|GetSurfaceType HitResult) "SurfaceType14") (Utilities|Enum|Equal(Enum) (Collision|GetSurfaceType HitResult) "SurfaceType15") (Utilities|Enum|Equal(Enum) (Collision|GetSurfaceType HitResult) "SurfaceType16") (or (Reactive|Frontend|IsCosmeticObjectDisplayedasPreview 0) (Reactive|Frontend|IsCosmeticObjectDisplayedinFrontendContext 0)))
    (CallFunction|CEFXServer) ; It calls the Server Event, that calls the Client Event, \nthat finally calls the reactivity event
  ))

(event Custom|ReactivityEvent
  ; Reactivity Event
  (Rendering|Material|SetScalarParameterValueonMaterials (Variables|Visual|GetWeaponMesh) "Timestamp" (Utilities|Time|GetGameTimeinSeconds))
  (bind _output_get (Variables|Default|SetbIsReactive true))
  (Effects|Components|ParticleSystem|SetBoolParameter (CallFunction|GetSwingVFXComponent) "IsReactive" _output_get)
  (bind (_direction _isreactive) (Timeline|ReactvityOrbTimeline PlayFromStart)
    (:Update
      (Effects|Components|ParticleSystem|SetFloatParameter (CallFunction|GetIdleVFXComponent) "IsReactive" _isreactive))
    (:Finished
      (bind _output_get_2 (Variables|Default|SetbIsReactive false))
      (Effects|Components|ParticleSystem|SetBoolParameter (CallFunction|GetSwingVFXComponent) "IsReactive" _output_get_2))))

(event Custom|CE_FX_Server
  ;[1 0.2356 0.1858] "Server Logic". It makes sure the reactivity can be seen by other clients
  (CallFunction|CEFXClient))

(event Custom|CE_FX_Client
  ;[1 0.2356 0.1858] "Server Logic". It makes sure the reactivity can be seen by other clients
  (CallFunction|ReactivityEvent))

(fn ConstructionScript
  (|CallParentFunction))
