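; Mochi pickaxe cosmetic Blueprint that animates a face on the weapon mesh by toggling three "Hide Element" material parameters.
; Impact FX smile when displayed in the frontend, and always pass the surface through CE_FX_Server to CE_FX_Client, which smiles on weak-point surfaces.
; Cosmetic init and visibility changes, each behind a DoOnce, set a looping blink timer.
; A smile suppresses blinking, then resets after a delay; it also prints "TARGET HIT".
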(event FortWeapon|EventOnPlayImpactFX
  (params (HitResult :type Hit%20Result%20Structure%20%28by%20ref%29) (ImpactPhysicalSurface :type EPhysicalSurface%20Enum :default "SurfaceType_Default") (SpawnedPSC :type FXSystem%20Component%20Object%20Reference))
  (|CallParentFunction HitResult ImpactPhysicalSurface SpawnedPSC)
  (Utilities|FlowControl|Sequence
    (:then_0
      (if (Reactive|Frontend|IsCosmeticObjectDisplayedinFrontendContext self)
        (CallFunction|MochiSmile)))
    (:then_1
      (CallFunction|CEFXServer ImpactPhysicalSurface))))

(event Custom|CE_FX_Server
  (params (Impact%20Physical%20Surface :type EPhysicalSurface%20Enum :default "SurfaceType_Default"))
  (CallFunction|CEFXClient Impact%20Physical%20Surface))

(event Custom|CE_FX_Client
  (params (Impact%20Physical%20Surface :type EPhysicalSurface%20Enum :default "SurfaceType_Default"))
  (bind _output_get (Variables|Default|SetPhysicalSurface Impact%20Physical%20Surface))
  (Variables|Default|SetIsWeakPoint (Utilities|Array|ContainsItem (Variables|Default|GetWeakPointSurface) _output_get))
  (if (Variables|Default|GetIsWeakPoint)
    (CallFunction|MochiSmile)))

(event EventOnInitWeaponCosmetics
  ;[1 1 1] Start blinking
  (|CallParentFunction)
  (Utilities|FlowControl|DoOnce false
    (:Completed
      (Utilities|Time|SetTimerbyEvent (delegate Custom|MochiBlink) 7.2 true))))

(event EventOnWeaponVisibilityChanged
  (params (bVisible :type Boolean) (bSetForLocalControllerOnly :type Boolean))
  ;[1 1 1] Start blinking
  (|CallParentFunction bVisible bSetForLocalControllerOnly)
  (Utilities|FlowControl|DoOnce false
    (:Completed
      (Utilities|Time|SetTimerbyEvent (delegate Custom|MochiBlink) 7.2 true))))

(event Custom|MochiBlink
  (bind _weaponmesh (Variables|Visual|GetWeaponMesh))
  (bind _weaponmesh_2 (Variables|Visual|GetWeaponMesh))
  ;[1 1 1] Start blinking
  (if (not (Variables|Default|GetisSmiling))
    (Rendering|Material|SetScalarParameterValueonMaterials _weaponmesh_2 "Hide Element 01" 1.0)
    (Rendering|Material|SetScalarParameterValueonMaterials _weaponmesh_2 "Hide Element 02" 0.0)
    (Rendering|Material|SetScalarParameterValueonMaterials _weaponmesh_2 "Hide Element 03" 1.0)
    ;[1 1 1] Stop blinking after a blink
    (Utilities|FlowControl|RetriggerableDelay 0.3) ; just incase the timer is missed
    (if (not (Variables|Default|GetisSmiling))
      (Rendering|Material|SetScalarParameterValueonMaterials _weaponmesh "Hide Element 01" 1.0)
      (Rendering|Material|SetScalarParameterValueonMaterials _weaponmesh "Hide Element 02" 1.0)
      (Rendering|Material|SetScalarParameterValueonMaterials _weaponmesh "Hide Element 03" 0.0))))

(event Custom|MochiSmile
  (bind _weaponmesh (Variables|Visual|GetWeaponMesh))
  (bind _weaponmesh_2 (Variables|Visual|GetWeaponMesh))
  (Development|PrintString "TARGET HIT")
  (Rendering|Material|SetScalarParameterValueonMaterials _weaponmesh_2 "Hide Element 01" 0.0)
  (Rendering|Material|SetScalarParameterValueonMaterials _weaponmesh_2 "Hide Element 02" 1.0)
  (Rendering|Material|SetScalarParameterValueonMaterials _weaponmesh_2 "Hide Element 03" 1.0)
  (Variables|Default|SetisSmiling true)
  (Utilities|FlowControl|Delay 5.0)
  ;[1 1 1] Stop blinking after a blink
  (Variables|Default|SetisSmiling false)
  (Rendering|Material|SetScalarParameterValueonMaterials _weaponmesh "Hide Element 01" 1.0)
  (Rendering|Material|SetScalarParameterValueonMaterials _weaponmesh "Hide Element 02" 1.0)
  (Rendering|Material|SetScalarParameterValueonMaterials _weaponmesh "Hide Element 03" 0.0))

(fn ConstructionScript
  (|CallParentFunction))
