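; Cosmetic pickaxe Blueprint (MoleAcorn) that adds reactive weak-point visuals to swing trails, impact FX, idle FX and the weapon material.
; Impacts on four specific surface types fire TriggerReactivity, directly and via ReplicateReactivity, flagging weak-point impact FX and an idle flash burst.
; Frontend previews instead derive the weak-point flag from the swing montage section.
; Weak-point hits timestamp the material and idle FX, fade back in, then clear the flags after a short delay.
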(event FortWeapon|EventOnInstigatorSet
  (|CallParentFunction))

(event FortWeapon|EventOnEquippedAndAttached
  (|CallParentFunction))

(event FortWeapon|EventOnLocalUnEquip
  (|CallParentFunction))

(event FortWeapon|EventOnUnEquippedAndDetached
  (|CallParentFunction))

(event EventSwingRight
  (|CallParentFunction)
  (Variables|Default|SetRight%20? true)
  (Effects|Components|ParticleSystem|SetBoolParameter (CallFunction|GetAnimTrailsPSC) "EnableWeakPoint" true))

(event EventSwingLeft
  (|CallParentFunction)
  (Variables|Default|SetRight%20? false)
  (Effects|Components|ParticleSystem|SetBoolParameter (CallFunction|GetAnimTrailsPSC) "EnableWeakPoint" true))

(event EventOnPlayWeakpointHitFX
  (|CallParentFunction)
  (if (Variables|Default|GetEnableWeakPointEffect)
    (Variables|Default|SetEnableWeakPointEffect false)
    (Rendering|Material|SetScalarParameterValueonMaterials (Variables|Visual|GetWeaponMesh) "BP_GetTime" (Utilities|Time|GetGameTimeinSeconds))
    (Effects|Components|ParticleSystem|SetBoolParameter (CallFunction|GetIdleVFXComponent) "FadeBackIn" true)
    (Effects|Components|ParticleSystem|SetFloatParameter (CallFunction|GetIdleVFXComponent) "GameTime" (Utilities|Time|GetGameTimeinSeconds))
    (Variables|Default|SetEnableWeakPointEffect true)
    (Utilities|FlowControl|Delay 0.95)
    (Effects|Components|ParticleSystem|SetBoolParameter (CallFunction|GetIdleVFXComponent) "FadeBackIn" false)
    (Effects|Components|ParticleSystem|SetBoolParameter (CallFunction|GetIdleVFXComponent) "FlashBurst" false)))

(event FortWeapon|EventOnPlayImpactFX
  (params (HitResult :type Hit%20Result%20Structure%20%28by%20ref%29) (ImpactPhysicalSurface :type EPhysicalSurface%20Enum :default "SurfaceType_Default") (SpawnedPSC :type FXSystem%20Component%20Object%20Reference))
  (|CallParentFunction HitResult ImpactPhysicalSurface SpawnedPSC)
  (Variables|Default|SetImpact%20FX SpawnedPSC)
  (Effects|Components|ParticleSystem|SetBoolParameter (Variables|Default|GetImpact%20FX) "EnableWeakPoint" true)
  (if (Reactive|Frontend|IsCosmeticObjectDisplayedasPreview 0)
    (bind _section_index (Utilities|SetupSwingMontageSectionIndex (Game|GetInstigator))
      (:Out
        (Effects|Components|ParticleSystem|SetBoolParameter (Variables|Default|GetImpact%20FX) "IsWeakPoint" (== _section_index 2))
        (Rendering|Material|SetScalarParameterValueonMaterials (Variables|Visual|GetWeaponMesh) "BP_GetTime" (Utilities|Time|GetGameTimeinSeconds))))
    (elif (Math|Boolean|ORBoolean (Utilities|Enum|Equal(Enum) (Collision|GetSurfaceType HitResult) "SurfaceType14") (Utilities|Enum|Equal(Enum) (Collision|GetSurfaceType HitResult) "SurfaceType15") (Utilities|Enum|Equal(Enum) (Collision|GetSurfaceType HitResult) "SurfaceType16") (Utilities|Enum|Equal(Enum) (Collision|GetSurfaceType HitResult) "SurfaceType12"))
      (CallFunction|TriggerReactivity)
      (CallFunction|ReplicateReactivity))))

(event Custom|ReplicateReactivity
  (CallFunction|TriggerReactivity))

(event Custom|TriggerReactivity
  (Effects|Components|ParticleSystem|SetBoolParameter (Variables|Default|GetImpact%20FX) "IsWeakPoint" true)
  (Effects|Components|ParticleSystem|SetBoolParameter (CallFunction|GetIdleVFXComponent) "FlashBurst" true))

(fn ConstructionScript
  (|CallParentFunction)
  (Components|Activation|SetAutoActivate (CallFunction|GetIdlePSC) false))
