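; MotorMonth pickaxe cosmetic Blueprint that drives its deploy, idle, swing-trail and electric impact effects.
; Begin play, cosmetic init and visibility changes switch off idle VFX, then through separate DoOnce gates spawn a Niagara deploy burst and restore idle after a short delay.
; Swings pass montage section index and front-end preview state to the trail.
; Impacts toggle body or standard electric discharge emitters depending on whether a pawn was hit.
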(event EventBeginPlay
  (bind _returnvalue (CallFunction|GetIdleVFXComponent))
  (|CallParentFunction)
  (Utilities|FlowControl|Sequence
    (:then_0
      (Components|Activation|Activate (CallFunction|GetIdlePSC) false))
    (:then_1
      (bind _asgalileo_ferry_axe_anim_bp (Utilities|Casting|CastToGalileo_Ferry_Axe_AnimBP (Components|SkeletalMesh|GetAnimInstance (Variables|Visual|GetWeaponMesh))))
      (Delegates|BindEventtoFrontEndIdle _asgalileo_ferry_axe_anim_bp (delegate Custom|FrontEndIdle)))
    (:then_2
      (Utilities|FlowControl|DelayUntilNextTick)
      (Components|Activation|Deactivate _returnvalue)
      (Utilities|FlowControl|DoOnce false
        (:Completed
          (Niagara|SpawnSystemAttached "/BRCosmetics/Weapons/FORT_Melee/Pickaxe_MotorMonth/FX/NS_Pickaxe_MotorMonth_Deploy.NS_Pickaxe_MotorMonth_Deploy" (Variables|Visual|GetWeaponMesh) "root" "0, 0, 0" "0, 0, 0" "KeepRelativeOffset" true)
          (Utilities|FlowControl|RetriggerableDelay 0.68)
          (Components|Activation|Activate _returnvalue false))))))

(event Custom|FrontEndIdle
  (Utilities|FlowControl|DoOnce false
    (:Completed
      (Utilities|FlowControl|Delay 0.3)
      (Components|Activation|Activate (CallFunction|GetIdlePSC) false))))

(event EventEndPlay
  (params (EndPlayReason :type EEndPlayReason%20Enum :default "Destroyed"))
  (|CallParentFunction EndPlayReason)
  (Utilities|FlowControl|DoOnce false
    (:Completed
      (Utilities|FlowControl|Delay 0.3)
      (Components|Activation|Activate (CallFunction|GetIdlePSC) false))))

(event EventOnWeaponVisibilityChanged
  (params (bVisible :type Boolean) (bSetForLocalControllerOnly :type Boolean))
  (Utilities|FlowControl|DelayUntilNextTick)
  (Components|Activation|Deactivate (CallFunction|GetIdleVFXComponent))
  (Utilities|FlowControl|DoOnce false
    (:Completed
      (Niagara|SpawnSystemAttached "/BRCosmetics/Weapons/FORT_Melee/Pickaxe_MotorMonth/FX/NS_Pickaxe_MotorMonth_Deploy.NS_Pickaxe_MotorMonth_Deploy" (Variables|Visual|GetWeaponMesh) "root" "0, 0, 0" "0, 0, 0" "KeepRelativeOffset" true)
      (Utilities|FlowControl|RetriggerableDelay 0.68)
      (Components|Activation|Activate (CallFunction|GetIdleVFXComponent) false))))

(event EventOnInitWeaponCosmetics
  (bind _returnvalue (CallFunction|GetIdleVFXComponent))
  (|CallParentFunction)
  (Utilities|FlowControl|DelayUntilNextTick)
  (Components|Activation|Deactivate _returnvalue)
  (Utilities|FlowControl|DoOnce false
    (:Completed
      (Niagara|SpawnSystemAttached "/BRCosmetics/Weapons/FORT_Melee/Pickaxe_MotorMonth/FX/NS_Pickaxe_MotorMonth_Deploy.NS_Pickaxe_MotorMonth_Deploy" (Variables|Visual|GetWeaponMesh) "root" "0, 0, 0" "0, 0, 0" "KeepRelativeOffset" true)
      (Utilities|FlowControl|RetriggerableDelay 0.68)
      (Components|Activation|Activate _returnvalue false))))

(event FortWeapon|EventOnPlayImpactFX
  (params (HitResult :type Hit%20Result%20Structure%20%28by%20ref%29) (ImpactPhysicalSurface :type EPhysicalSurface%20Enum :default "SurfaceType_Default") (SpawnedPSC :type FXSystem%20Component%20Object%20Reference))
  (bind (_bblockinghit _binitialoverlap _time _distance _location _impactpoint _normal _impactnormal _physmat _hitactor) (Collision|BreakHitResult HitResult))
  (bind (_asfort_pawn _bsuccess) (Utilities|Casting|CastToFortPawn _hitactor))
  (|CallParentFunction HitResult ImpactPhysicalSurface SpawnedPSC)
  (Effects|Components|ParticleSystem|SetActorParameter SpawnedPSC "BoneSocketActor" _hitactor)
  (if _bsuccess
    (Effects|Components|ParticleSystem|SetEmitterEnable SpawnedPSC "Electric Discharge" false)
    (Effects|Components|ParticleSystem|SetEmitterEnable SpawnedPSC "Electric Discharge_Body" true)
    (else
      (Effects|Components|ParticleSystem|SetEmitterEnable SpawnedPSC "Electric Discharge" true)
      (Effects|Components|ParticleSystem|SetEmitterEnable SpawnedPSC "Electric Discharge_Body" false))))

(event EventSwingLeft
  (bind _returnvalue (CallFunction|GetAnimTrailVFXComponent))
  ; SwingSectionIndex Trails
  (|CallParentFunction)
  (Effects|Components|ParticleSystem|SetBoolParameter (CallFunction|GetAnimTrailVFXComponent) "bIsFrontEndPreview" (Reactive|Frontend|IsCosmeticObjectDisplayedinFrontendContext 0))
  (bind _section_index (Utilities|SetupSwingMontageSectionIndex (Variables|Actor|GetInstigator))
    (:Out
      (Effects|Components|ParticleSystem|SetIntParameter _returnvalue "SwingSectionIndex" _section_index))
    (:Cast%20Failed%20Out
      (Effects|Components|ParticleSystem|SetIntParameter _returnvalue "SwingSectionIndex" -1))))

(event EventSwingRight
  (bind _returnvalue (CallFunction|GetAnimTrailVFXComponent))
  ; SwingSectionIndex Trails
  (|CallParentFunction)
  (Effects|Components|ParticleSystem|SetBoolParameter (CallFunction|GetAnimTrailVFXComponent) "bIsFrontEndPreview" (Reactive|Frontend|IsCosmeticObjectDisplayedinFrontendContext 0))
  (bind _section_index (Utilities|SetupSwingMontageSectionIndex (Variables|Actor|GetInstigator))
    (:Out
      (Effects|Components|ParticleSystem|SetIntParameter _returnvalue "SwingSectionIndex" _section_index))
    (:Cast%20Failed%20Out
      (Effects|Components|ParticleSystem|SetIntParameter _returnvalue "SwingSectionIndex" -1))))

(fn ConstructionScript
  (|CallParentFunction)
  (Components|Activation|SetAutoActivate (CallFunction|GetIdlePSC) false))
