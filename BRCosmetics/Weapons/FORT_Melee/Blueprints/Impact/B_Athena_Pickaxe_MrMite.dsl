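; Mr. Mite pickaxe Blueprint that makes the weapon's idle VFX and sounds react to hits.
; The idle effect is deactivated shortly after equip, then activated on impact with HasHitThing set.
; A timeline restarts on each hit to drive ReactiveCurve, shutting the effect off when finished.
; DoOnce gates the activation-on sound, while a timeline audio event calls TriggerOff and plays the activation-off sound.
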(event EventBeginPlay)

(event EventTick
  (params (DeltaSeconds :type Float%20%28single-precision%29)))

(event FortWeapon|EventOnEquippedAndAttached
  (|CallParentFunction)
  (Utilities|FlowControl|Delay 0.2)
  (Components|Activation|Deactivate (CallFunction|GetIdleVFXComponent)))

(event FortWeapon|EventOnPlayImpactFX
  (params (HitResult :type Hit%20Result%20Structure%20%28by%20ref%29) (ImpactPhysicalSurface :type EPhysicalSurface%20Enum :default "SurfaceType_Default") (SpawnedPSC :type FXSystem%20Component%20Object%20Reference))
  (bind _returnvalue (CallFunction|GetIdleVFXComponent))
  (|CallParentFunction HitResult ImpactPhysicalSurface SpawnedPSC)
  (Effects|Components|ParticleSystem|SetBoolParameter _returnvalue "HasHitThing" true)
  (Components|Activation|Activate _returnvalue false)
  (Utilities|FlowControl|Sequence
    (:then_0
      (bind (_direction _delay) (Timeline|Timeline PlayFromStart)
        (:Update
          (Effects|Components|ParticleSystem|SetFloatParameter _returnvalue "ReactiveCurve" _delay))
        (:Finished
          (Effects|Components|ParticleSystem|SetBoolParameter _returnvalue "HasHitThing" false)
          (Components|Activation|Deactivate _returnvalue))
        (:AudioEvent
          (CallFunction|TriggerOff)
          (SoundLibrary|PlaySoundSimple (WorldSubsystems|GetSoundLibrarySubsystem) (Game|GetInstigator) (Utilities|Struct|MakeSimpleSoundSettings "(TagName=\"SoundLibrary.Pickaxe.Activate.Off\")" true false true false 1.0 1.0 0.0 0 0 "None" (Variables|Visual|GetWeaponMesh)) :!class /Script/SoundLibrary.SoundLibrarySubsystem))))
    (:then_1
      (Utilities|FlowControl|DoOnce false
        (:Completed
          (SoundLibrary|PlaySoundSimple (WorldSubsystems|GetSoundLibrarySubsystem) (Game|GetInstigator) (Utilities|Struct|MakeSimpleSoundSettings "(TagName=\"SoundLibrary.Pickaxe.Activate.On\")" true false true false 1.0 1.0 0.0 0 0 "None" (Variables|Visual|GetWeaponMesh)) :!class /Script/SoundLibrary.SoundLibrarySubsystem))))))

(event Custom|TriggerOff
  (Utilities|FlowControl|DoOnce false
    (:Completed
      (SoundLibrary|PlaySoundSimple (WorldSubsystems|GetSoundLibrarySubsystem) (Game|GetInstigator) (Utilities|Struct|MakeSimpleSoundSettings "(TagName=\"SoundLibrary.Pickaxe.Activate.On\")" true false true false 1.0 1.0 0.0 0 0 "None" (Variables|Visual|GetWeaponMesh)) :!class /Script/SoundLibrary.SoundLibrarySubsystem))))

(fn ConstructionScript
  (|CallParentFunction))
