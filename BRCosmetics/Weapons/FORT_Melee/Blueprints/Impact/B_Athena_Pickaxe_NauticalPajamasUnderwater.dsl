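; Cosmetic Blueprint for the Nautical Pajamas Underwater pickaxe that adapts its idle VFX to frontend display.
; When weapon cosmetics initialize, it runs the parent logic, then checks whether the object is displayed in a frontend context.
; It passes that result to the idle VFX component as the bIsFrontEndPreview bool parameter.
; The construction script only calls its parent.
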(event EventOnInitWeaponCosmetics
  (|CallParentFunction)
  (Effects|Components|ParticleSystem|SetBoolParameter (CallFunction|GetIdleVFXComponent) "bIsFrontEndPreview" (Reactive|Frontend|IsCosmeticObjectDisplayedinFrontendContext self)))

(fn ConstructionScript
  (|CallParentFunction))
