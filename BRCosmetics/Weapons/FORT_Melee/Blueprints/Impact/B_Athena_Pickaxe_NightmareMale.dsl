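; Cosmetic Blueprint for the NightmareMale pickaxe, handling idle VFX and swing-aware impact effects.
; Idle VFX start when the weapon becomes visible, and swing events record left or right to orient impacts along the hit normal.
; Impact Niagara systems receive facing, direction and a weak-point flag from surface type.
; In the frontend, a four-step hit counter replaces the weak-point check, feeding the effect and choosing the side for later parameters.
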(event EventOnWeaponVisibilityChanged
  (params (bVisible :type Boolean) (bSetForLocalControllerOnly :type Boolean))
  (if bVisible
    (Components|Activation|Activate (CallFunction|GetIdleVFXComponent) false)
    (Variables|Default|SetFEHitCounter 0)))

(event EventOnInitWeaponCosmetics
  (|CallParentFunction)
  (Components|Activation|Deactivate (CallFunction|GetSwingPSC))
  (Components|Activation|Deactivate (CallFunction|GetIdleVFXComponent)))

(event EventSwingLeft
  (|CallParentFunction)
  (Variables|Default|SetRight false))

(event EventSwingLeftEnd
  (|CallParentFunction)
  (Variables|Default|SetRight false))

(event EventSwingRight
  (|CallParentFunction)
  (Variables|Default|SetRight true))

(event EventSwingRightEnd
  (|CallParentFunction)
  (Variables|Default|SetRight true))

(event FortWeapon|EventOnPlayImpactFX
  (params (HitResult :type Hit%20Result%20Structure%20%28by%20ref%29) (ImpactPhysicalSurface :type EPhysicalSurface%20Enum :default "SurfaceType_Default") (SpawnedPSC :type FXSystem%20Component%20Object%20Reference))
  (bind (_bblockinghit _binitialoverlap _time _distance _location _impactpoint _normal) (Collision|BreakHitResult HitResult))
  (|CallParentFunction HitResult ImpactPhysicalSurface SpawnedPSC)
  (Effects|Components|ParticleSystem|SetBoolParameter SpawnedPSC "bIsFrontEndPreview" (Reactive|Frontend|IsCosmeticObjectDisplayedinFrontendContext 0))
  (bind _asniagara_particle_system_component (Utilities|Casting|CastToNiagaraComponent SpawnedPSC))
  (Transformation|SetWorldRotation SpawnedPSC (Math|Rotator|CombineRotators (Math|Rotator|MakeRotfromX _normal) (select (Variables|Default|GetRight) (Variables|Default|GetRight%20Rotation) (Variables|Default|GetLeft%20Rotation))) false false)
  (Niagara|SetNiagaraVariable(Vector3) _asniagara_particle_system_component "EffectFacing" _normal)
  (Utilities|Casting|CastToFortGameStateFrontEnd (Game|GetGameState)
    (:then
      (bind _result (Math|Integer|IncrementInt (Variables|Default|GetFEHitCounter))
        (:%20%20
          (Niagara|SetNiagaraVariable(Int32) _asniagara_particle_system_component "FEHitNumber" (Math|Integer|%25%28Integer%29 _result 4))
          (Variables|Default|SetRight (< (Math|Integer|%25%28Integer%29 _result 4) 2)))))
    (:CastFailed
      (if (Math|Boolean|ORBoolean (Utilities|Enum|Equal(Enum) ImpactPhysicalSurface "SurfaceType14") (Utilities|Enum|Equal(Enum) ImpactPhysicalSurface "SurfaceType15") (Utilities|Enum|Equal(Enum) ImpactPhysicalSurface "SurfaceType16"))
        (Niagara|SetNiagaraVariable(Float) _asniagara_particle_system_component "WeakPointImpact" 1.0)
        (else
          (Niagara|SetNiagaraVariable(Float) _asniagara_particle_system_component "WeakPointImpact" 0.0)))))
  (if (Variables|Default|GetRight)
    (Niagara|SetNiagaraVariable(Float) _asniagara_particle_system_component "DirectionalAngle" 0.0)
    (else
      (Niagara|SetNiagaraVariable(Float) _asniagara_particle_system_component "DirectionalAngle" 1.0))))

(fn ConstructionScript
  (|CallParentFunction)
  (Components|Activation|SetAutoActivate (CallFunction|GetIdlePSC) false))
