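; Cosmetic weapon Blueprint for the OrinChai pickaxe that drives its swing visual effects.
; On equip, it binds the swing Niagara component to the owning pawn's skeletal mesh.
; Both swing events check the active montage section; on "Swing1" they spawn a thrust effect at the weapon handle and hide the regular swing VFX, showing it otherwise.
; Construction turns off auto-activation of the idle particle component.
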(event FortWeapon|EventOnEquippedAndAttached
  (|CallParentFunction)
  (bind _asfort_pawn (Utilities|Casting|CastToFortPawn (Game|GetInstigator)))
  (Utilities|IsValid (Variables|/Script/Engine.Character|GetMesh _asfort_pawn)
    (:Is%20Valid
      (bind _asniagara_particle_system_component (Utilities|Casting|CastToNiagaraComponent (CallFunction|GetSwingVFXComponent))
        (:then
          (Niagara|SetNiagaraSkeletalMeshComponent _asniagara_particle_system_component "Pawn" (Variables|/Script/Engine.Character|GetMesh _asfort_pawn)))
        (:CastFailed
          (Development|PrintString "CAST FAILED - Niagara Component"))))))

(event EventSwingRight
  (bind _instigator (Variables|Actor|GetInstigator))
  (bind _asfort_pawn (Utilities|Casting|CastToFortPawn _instigator))
  (bind _mesh (Variables|/Script/Engine.Character|GetMesh _asfort_pawn))
  (bind _returnvalue (Components|SkeletalMesh|GetAnimInstance _mesh))
  (bind _returnvalue_2 (Animation|Montage|GetCurrentActiveMontage _returnvalue))
  (bind _returnvalue_3 (Animation|Montage|MontageGetCurrentSection _returnvalue _returnvalue_2))
  (bind _returnvalue_4 (== _returnvalue_3 "Swing1"))
  (|CallParentFunction)
  (Networking|CanExecuteCosmeticEvents
    (:True
      (Utilities|FlowControl|Sequence
        (:then_0
          (if _returnvalue_4
            (Effects|SpawnFXSystemAttached "/BRCosmetics/Weapons/FORT_Melee/Pickaxe_OrinChai/FX/NS_Pickaxe_OrinChai_Thrust.NS_Pickaxe_OrinChai_Thrust" (Variables|Visual|GetWeaponMesh) "Melee_TwoHanded_Handle")))
        (:then_1
          (Development|SetHiddeninGame (CallFunction|GetSwingVFXComponent) _returnvalue_4 false))))))

(event EventSwingLeft
  (bind _instigator (Variables|Actor|GetInstigator))
  (bind _asfort_pawn (Utilities|Casting|CastToFortPawn _instigator))
  (bind _mesh (Variables|/Script/Engine.Character|GetMesh _asfort_pawn))
  (bind _returnvalue (Components|SkeletalMesh|GetAnimInstance _mesh))
  (bind _returnvalue_2 (Animation|Montage|GetCurrentActiveMontage _returnvalue))
  (bind _returnvalue_3 (Animation|Montage|MontageGetCurrentSection _returnvalue _returnvalue_2))
  (bind _returnvalue_4 (== _returnvalue_3 "Swing1"))
  (|CallParentFunction)
  (Networking|CanExecuteCosmeticEvents
    (:True
      (Utilities|FlowControl|Sequence
        (:then_0
          (if _returnvalue_4
            (Effects|SpawnFXSystemAttached "/BRCosmetics/Weapons/FORT_Melee/Pickaxe_OrinChai/FX/NS_Pickaxe_OrinChai_Thrust.NS_Pickaxe_OrinChai_Thrust" (Variables|Visual|GetWeaponMesh) "Melee_TwoHanded_Handle")))
        (:then_1
          (Development|SetHiddeninGame (CallFunction|GetSwingVFXComponent) _returnvalue_4 false))))))

(fn ConstructionScript
  (|CallParentFunction)
  (Components|Activation|SetAutoActivate (CallFunction|GetIdlePSC) false))
