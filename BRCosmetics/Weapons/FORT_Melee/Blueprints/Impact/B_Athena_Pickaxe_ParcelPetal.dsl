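; Pickaxe weapon Blueprint for the ParcelPetal cosmetic that briefly turns off its idle VFX whenever weapon fire effects play.
; On fire FX it calls the parent handler, deactivates the idle VFX component, then reactivates it without reset after a short delay.
; The delay is retriggerable, so repeated swings keep the idle effect off until they stop.
; BeginPlay and Tick are empty, and the construction script only calls the parent.
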(event EventBeginPlay)

(event EventTick
  (params (DeltaSeconds :type Float%20%28single-precision%29)))

(event EventOnPlayWeaponFireFX
  (params (bPersistentFire :type Boolean) (bSecondaryFire :type Boolean))
  (bind _returnvalue (CallFunction|GetIdleVFXComponent))
  (|CallParentFunction bPersistentFire bSecondaryFire)
  (Components|Activation|Deactivate _returnvalue)
  (Utilities|FlowControl|RetriggerableDelay 0.8)
  (Components|Activation|Activate _returnvalue false))

(fn ConstructionScript
  (|CallParentFunction))
