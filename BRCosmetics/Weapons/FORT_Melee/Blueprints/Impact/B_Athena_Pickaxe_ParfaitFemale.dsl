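; Parfait Female pickaxe Blueprint that adds timed, hit-reactive "powers" effects to a cosmetic variant.
; On equip it flags whether the Particle2 variant is active; if so, impacts on three special surface types (weakpoints) trigger powers through a replication event.
; Powers toggle a trail particle bool, held by a retriggerable delay, and other impacts pass that state to impact FX.
; Preview impacts always trigger powers; PlayAudio is noted as moved to Niagara.
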(event FortWeapon|EventOnEquippedAndAttached
  ;[1 1 1] Sets reactivity once qeuipped in game
  (|CallParentFunction)
  (CallFunction|SetReactivity))

(event Custom|TriggerReactivity
  ;[1 1 1] Trigger event chain when bool is set to true.
  (CallFunction|PowerActive)
  (Utilities|FlowControl|RetriggerableDelay (Variables|Default|GetDuration))
  (CallFunction|PowerInactive))

(event Custom|SetReactivity
  ;[1 1 1] Verifies whether it's a reactive variant or not
  (bind _asfort_player_pawn (Utilities|Casting|CastToFortPlayerPawn (Game|GetInstigator)))
  (bind (_index _bindexisvalid) (CosmeticLoadout|GetVariantChannelActiveIndexforSlotActor "(TagName=\"LoadoutSlot.Pickaxe\")" _asfort_player_pawn "(TagName=\"Cosmetics.Variant.Channel.Particle\")" (Utilities|Array|MakeArray "(TagName=\"Cosmetics.Variant.Property.Particle2\")") true false))
  (Variables|Default|SetbIsReactiveVariant _bindexisvalid))

(event Custom|FrontendReactivity
  ;[1 1 1] Trigger event chain when bool is set to true.
  (CallFunction|PowerActive)
  (Utilities|FlowControl|RetriggerableDelay (Variables|Default|GetDuration))
  (CallFunction|PowerInactive))

(event Custom|ReplicateReactivity
  ;[1 1 1] Replication
  (CallFunction|TriggerReactivity))

(event FortWeapon|EventOnPlayImpactFX
  (params (HitResult :type Hit%20Result%20Structure%20%28by%20ref%29) (ImpactPhysicalSurface :type EPhysicalSurface%20Enum :default "SurfaceType_Default") (SpawnedPSC :type FXSystem%20Component%20Object%20Reference))
  ;[1 1 1] Gets the weakpoints and sets makes this the condition for the event send.
  (|CallParentFunction HitResult ImpactPhysicalSurface SpawnedPSC)
  (if (Reactive|Frontend|IsCosmeticObjectDisplayedasPreview 0)
    (CallFunction|FrontendReactivity)
    (else
      (if (Variables|Default|GetbIsReactiveVariant)
        (if (Math|Boolean|ORBoolean (Utilities|Enum|Equal(Enum) (Collision|GetSurfaceType HitResult) "SurfaceType14") (Utilities|Enum|Equal(Enum) (Collision|GetSurfaceType HitResult) "SurfaceType15") (Utilities|Enum|Equal(Enum) (Collision|GetSurfaceType HitResult) "SurfaceType16"))
          ;[1 1 1] Replication
          (CallFunction|ReplicateReactivity)
          (else
            ;[1 1 1] Initial bool state is set here. Bool needs to be kept at true during the event chain.
            (Effects|Components|ParticleSystem|SetBoolParameter SpawnedPSC "PowersActive" (Variables|Default|GetPowersActive))))))))

(event Custom|PowerActive
  ;[1 1 1] Grabs idlefx/trail/impact fx and switches the bool between the 2 bool states based on the event sends.
  (bind _output_get (Variables|Default|SetPowersActive true))
  (Effects|Components|ParticleSystem|SetBoolParameter (CallFunction|GetAnimTrailsPSC) "PowersActive" _output_get))

(event Custom|PowerInactive
  ;[1 1 1] Grabs idlefx/trail/impact fx and switches the bool between the 2 bool states based on the event sends.
  (bind _output_get (Variables|Default|SetPowersActive false))
  (Effects|Components|ParticleSystem|SetBoolParameter (CallFunction|GetAnimTrailsPSC) "PowersActive" _output_get))

(fn PlayAudio
  ;[1 1 1] Moved to idle NS effect
  (if (Variables|Default|GetPowersActive)
    (bind _impactaudio (Variables|Default|GetImpactAudio)
      (:then
        (Audio|Components|Audio|FadeOut _impactaudio 0.3 0.0 "Linear" :!class /Script/Engine.AudioComponent))
      (:else
        _))
    (bind _returnvalue (Audio|SpawnSoundAttached (Variables|Default|GetWeaponFXSound) (Variables|Visual|GetWeaponMesh)))
    (Variables|Default|SetImpactAudio _returnvalue)))

(fn ConstructionScript
  (|CallParentFunction))
