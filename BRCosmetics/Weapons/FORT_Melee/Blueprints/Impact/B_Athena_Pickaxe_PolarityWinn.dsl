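; Pickaxe weapon cosmetic Blueprint for Athena PolarityWinn that drives its idle and swing-trail VFX.
; In frontend or preview, a timeline ramps up, holds and reverses to feed the idle effect's EnableFrontendFakeAudioLoop parameter, faking audio reactivity.
; Idle VFX auto-activation is off; BeginPlay, visibility changes, the anim Blueprint's FrontEndIdle delegate and EndPlay control it.
; Left and right swings pass the montage section index to the trail VFX, or -1 on cast failure.
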(event Custom|OnFakeLoopTimelineUpdate
  (params (TransitionValue :type Float%20%28single-precision%29))
  (Effects|Components|ParticleSystem|SetFloatParameter (CallFunction|GetIdleVFXComponent) "EnableFrontendFakeAudioLoop" TransitionValue))

(event EventOnInitWeaponCosmetics
  (|CallParentFunction)
  (if (or (Reactive|Frontend|IsCosmeticObjectDisplayedinFrontendContext 0) (Reactive|Frontend|IsCosmeticObjectDisplayedasPreview 0))
    (Variables|Default|SetBeginLoop true)
    (bind (_direction _in) (Timeline|Timeline PlayFromStart)
      (:Update
        _)
      (:Finished
        (if (Variables|Default|GetBeginLoop)
          (Variables|Default|SetBeginLoop false)
          (Utilities|FlowControl|Delay (Variables|Default|GetFakeAmplitudeOnLength))
          (Timeline|Timeline ReverseFromEnd))
        (Utilities|FlowControl|Delay (Variables|Default|GetFakeAmplitudeOffLength))))
    (CallFunction|OnFakeLoopTimelineUpdate _in)))

(event EventBeginPlay
  (|CallParentFunction)
  (Utilities|FlowControl|Sequence
    (:then_0
      (Components|Activation|Activate (CallFunction|GetIdleVFXComponent) false))
    (:then_1
      (bind _asgalileo_ferry_axe_anim_bp (Utilities|Casting|CastToGalileo_Ferry_Axe_AnimBP (Components|SkeletalMesh|GetAnimInstance (Variables|Visual|GetWeaponMesh))))
      (Delegates|BindEventtoFrontEndIdle _asgalileo_ferry_axe_anim_bp (delegate Custom|FrontEndIdle)))))

(event Custom|FrontEndIdle
  (Utilities|FlowControl|DoOnce false
    (:Completed
      (Utilities|FlowControl|Delay 0.3)
      (Components|Activation|Activate (CallFunction|GetIdleVFXComponent) false))))

(event EventEndPlay
  (params (EndPlayReason :type EEndPlayReason%20Enum :default "Destroyed"))
  (|CallParentFunction EndPlayReason)
  (Utilities|FlowControl|DoOnce false
    (:Completed
      (Utilities|FlowControl|Delay 0.3)
      (Components|Activation|Activate (CallFunction|GetIdleVFXComponent) false))))

(event EventOnWeaponVisibilityChanged
  (params (bVisible :type Boolean) (bSetForLocalControllerOnly :type Boolean))
  (if bVisible
    (Components|Activation|Activate (CallFunction|GetIdleVFXComponent) false)
    (else
      (Components|Activation|Deactivate (CallFunction|GetIdleVFXComponent)))))

(event EventSwingLeft
  (bind _returnvalue (CallFunction|GetAnimTrailVFXComponent))
  ; SwingSectionIndex Trails
  (|CallParentFunction)
  (bind _section_index (Utilities|SetupSwingMontageSectionIndex (Variables|Actor|GetInstigator))
    (:Out
      (Effects|Components|ParticleSystem|SetIntParameter _returnvalue "SwingSectionIndex" _section_index))
    (:Cast%20Failed%20Out
      (Effects|Components|ParticleSystem|SetIntParameter _returnvalue "SwingSectionIndex" -1))))

(event EventSwingRight
  (bind _returnvalue (CallFunction|GetAnimTrailVFXComponent))
  ; SwingSectionIndex Trails
  (|CallParentFunction)
  (bind _section_index (Utilities|SetupSwingMontageSectionIndex (Variables|Actor|GetInstigator))
    (:Out
      (Effects|Components|ParticleSystem|SetIntParameter _returnvalue "SwingSectionIndex" _section_index))
    (:Cast%20Failed%20Out
      (Effects|Components|ParticleSystem|SetIntParameter _returnvalue "SwingSectionIndex" -1))))

(fn ConstructionScript
  (|CallParentFunction)
  (Components|Activation|SetAutoActivate (CallFunction|GetIdlePSC) false))
