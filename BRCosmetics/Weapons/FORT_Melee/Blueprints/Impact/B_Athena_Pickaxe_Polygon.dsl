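; Reactive Athena pickaxe Blueprint (Polygon) that upgrades its look as the player gets kills.
; Kill updates at 1, 3 and 5 trigger Stages 1-3, which set the weapon mesh's SuperLevel material parameter.
; Each upgrade also stamps SuperLevelAnim with current game time, presumably to time a material-driven transition.
; In frontend display, a looping timer previews the stages in sequence from Stage 0 through a switch counter.
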(event FortWeapon|EventOnEquippedAndAttached
  ;[1 1 1] Front end preview trigger
  (|CallParentFunction)
  (if (Reactive|Frontend|IsCosmeticObjectDisplayedinFrontendContext self)
    (Utilities|Time|SetTimerbyEvent (EventDispatchers|CreateEvent :!event Custom|FrontEnd) 3.0 true)))

(event Custom|FrontEnd
  (switch int (Variables|Default|GetFrontendSwitch)
    (:0
      (CallFunction|Stage0))
    (:1
      (CallFunction|Stage1))
    (:2
      (CallFunction|Stage2))
    (:3
      (CallFunction|Stage3)
      (Variables|Default|SetFrontendSwitch 0)))
  (Math|Integer|IncrementInt (Variables|Default|GetFrontendSwitch)))

(event EventUpdateBasedOnKills
  (params (Watched%20Kills :type Integer))
  ;[1 1 1] Gameplay reactivity
  (|CallParentFunction Watched%20Kills)
  (switch int Watched%20Kills
    (:1
      (CallFunction|Stage1))
    (:3
      (CallFunction|Stage2))
    (:5
      (CallFunction|Stage3))))

(event Custom|Stage%200
  (Rendering|Material|SetScalarParameterValueonMaterials (Variables|Visual|GetWeaponMesh) "SuperLevel" 0.0))

(event Custom|Stage%201
  (bind _weaponmesh (Variables|Visual|GetWeaponMesh))
  (Rendering|Material|SetScalarParameterValueonMaterials _weaponmesh "SuperLevel" 1.0)
  (Rendering|Material|SetScalarParameterValueonMaterials _weaponmesh "SuperLevelAnim" (Utilities|Time|GetGameTimeinSeconds)))

(event Custom|Stage%202
  (bind _weaponmesh (Variables|Visual|GetWeaponMesh))
  (Rendering|Material|SetScalarParameterValueonMaterials _weaponmesh "SuperLevel" 2.0)
  (Rendering|Material|SetScalarParameterValueonMaterials _weaponmesh "SuperLevelAnim" (Utilities|Time|GetGameTimeinSeconds)))

(event Custom|Stage%203
  (bind _weaponmesh (Variables|Visual|GetWeaponMesh))
  (Rendering|Material|SetScalarParameterValueonMaterials _weaponmesh "SuperLevel" 3.0)
  (Rendering|Material|SetScalarParameterValueonMaterials _weaponmesh "SuperLevelAnim" (Utilities|Time|GetGameTimeinSeconds)))

(fn ConstructionScript
  (|CallParentFunction))
