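; Cosmetic Blueprint for the ProdigyFire pickaxe that reddens the blade over consecutive weak-spot hits, then decays it.
; Weak-spot impacts relay through server/client events to a hit counter selecting timeline curves for mesh material parameters, only for the first particle variant.
; The resulting alpha feeds Niagara "Transition" on idle, swing and impact FX.
; Preview/frontend impacts use a duplicated reactivity path, and kills separately pulse an orange emissive glow.
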(event EventSwingLeft
  (|CallParentFunction)
  (Variables|Default|SetRight%20? false)
  (bind _asniagara_particle_system_component (Utilities|Casting|CastToNiagaraComponent (CallFunction|GetSwingVFXComponent)))
  (Niagara|SetNiagaraVariable(Float) _asniagara_particle_system_component "Transition" (Variables|Default|GetFX_TransitionAlpha)))

(event EventSwingRight
  (|CallParentFunction)
  (Variables|Default|SetRight%20? true)
  (bind _asniagara_particle_system_component (Utilities|Casting|CastToNiagaraComponent (CallFunction|GetSwingVFXComponent)))
  (Niagara|SetNiagaraVariable(Float) _asniagara_particle_system_component "Transition" (Variables|Default|GetFX_TransitionAlpha)))

(event Custom|UpdateBasedOnKills_1
  (params (Watched%20Kills :type Integer))
  ;[1 1 1] GlowOnKill
  (bind _output_get (Variables|Default|SetKillCount (Math|Conversions|ToFloat(Integer) Watched%20Kills)))
  (if (> _output_get 0.0)
    (bind (_direction _lerp_alpha) (Timeline|Timeline_0 PlayFromStart)
      (:Update
        (Rendering|Material|SetVectorParameterValueonMaterials (Variables|Visual|GetWeaponMesh) "EmissiveMultiplier" (Math|Conversions|ToVector(LinearColor) (Math|Color|LinearColor*Float "(R=1.000000,G=0.072205,B=0.000000,A=1.000000)" (Math|Float|Lerp 0.0 (Variables|Default|GetGlowIntensity) _lerp_alpha))))))))

(event EventOnInitWeaponCosmetics
  (|CallParentFunction)
  (Components|Activation|Deactivate (CallFunction|GetSwingPSC)))

(event FortWeapon|EventOnPlayImpactFX
  (params (HitResult :type Hit%20Result%20Structure%20%28by%20ref%29) (ImpactPhysicalSurface :type EPhysicalSurface%20Enum :default "SurfaceType_Default") (SpawnedPSC :type FXSystem%20Component%20Object%20Reference))
  ;[1 1 1] GlowOnWeakSpotHit
  (|CallParentFunction HitResult ImpactPhysicalSurface SpawnedPSC)
  (Utilities|FlowControl|Sequence
    (:then_0
      (bind _asniagara_particle_system_component (Utilities|Casting|CastToNiagaraComponent SpawnedPSC))
      (Niagara|SetNiagaraVariable(Float) _asniagara_particle_system_component "Transition" (Variables|Default|GetFX_TransitionAlpha)))
    (:then_1
      (Variables|Default|SetIsHitWeakPoint (Math|Boolean|ORBoolean (Utilities|Enum|Equal(Enum) ImpactPhysicalSurface "SurfaceType14") (Utilities|Enum|Equal(Enum) ImpactPhysicalSurface "SurfaceType15") (Utilities|Enum|Equal(Enum) ImpactPhysicalSurface "SurfaceType16")))
      (if (Variables|Default|GetIsHitWeakPoint)
        (CallFunction|CEFXServer)))
    (:then_2
      (if (or (Reactive|Frontend|IsCosmeticObjectDisplayedasPreview 0) (Reactive|Frontend|IsCosmeticObjectDisplayedinFrontendContext 0))
        (CallFunction|FrontEndReactivity)))))

(event Custom|CE_FX_Client
  (CallFunction|WeakSpotBladeReactivity))

(event Custom|WeakSpotBladeReactivity
  (bind _weaponmesh (Variables|Visual|GetWeaponMesh))
  (bind _weaponmesh_2 (Variables|Visual|GetWeaponMesh))
  (bind _redfxisactive (Variables|Default|GetRedFXIsActive))
  (bind _hitdecayanimvalue (Variables|Default|GetHitDecayAnimValue))
  ;[1 0.5 0.5] Make Blade Red on Hit (Progressively, 5-steps)
  (Variables|Default|SetPickaxeHitCount (+ (Variables|Default|GetPickaxeHitCount) 1.0))
  (bind _index (Class|FortCosmeticFunctionLibrary|%28Deprecated%29GetVariantChannelActiveIndexForSlot "(TagName=\"LoadoutSlot.Pickaxe\")" (CallFunction|GetCosmeticData) "(TagName=\"Cosmetics.Variant.Channel.Particle\")" (Utilities|Array|MakeArray "(TagName=\"Cosmetics.Variant.Property.Particle1\")" "(TagName=\"Cosmetics.Variant.Property.Particle2\")" "(TagName=\"Cosmetics.Variant.Property.Particle3\")")))
  (if (== _index 0)
    (bind (_direction _transitionamounthit01 _transitionamounthit02 _transitionamounthit03 _transitionamounthit04 _transitionamounthit05 _transitionamountforglow _redglowmultiplier) (Timeline|PickAxeHitTimeline PlayFromStart)
      (:Update
        (switch int (Math|Float|Truncate (Variables|Default|GetPickaxeHitCount))
          (:Default
            (Rendering|Material|SetScalarParameterValueonMaterials _weaponmesh "RedGlowMultiplier" _redglowmultiplier)
            (Rendering|Material|SetScalarParameterValueonMaterials _weaponmesh "TransitionAmount" _transitionamounthit05)
            (CallFunction|UpdateFX _transitionamounthit05)
            (Rendering|Material|SetScalarParameterValueonMaterials _weaponmesh "TransitionAmountForGlow" _transitionamountforglow))
          (:0
            (Rendering|Material|SetScalarParameterValueonMaterials _weaponmesh "TransitionAmount" 0.0)
            (CallFunction|UpdateFX 0.0)
            (Rendering|Material|SetScalarParameterValueonMaterials _weaponmesh "TransitionAmountForGlow" 0.0))
          (:1
            (Rendering|Material|SetScalarParameterValueonMaterials _weaponmesh "TransitionAmount" _transitionamounthit01)
            (Rendering|Material|SetScalarParameterValueonMaterials _weaponmesh "RedGlowMultiplier" _redglowmultiplier)
            (CallFunction|UpdateFX _transitionamounthit01))
          (:2
            (Rendering|Material|SetScalarParameterValueonMaterials _weaponmesh "TransitionAmount" _transitionamounthit02)
            (Rendering|Material|SetScalarParameterValueonMaterials _weaponmesh "RedGlowMultiplier" _redglowmultiplier)
            (CallFunction|UpdateFX _transitionamounthit02))
          (:3
            (Rendering|Material|SetScalarParameterValueonMaterials _weaponmesh "TransitionAmount" _transitionamounthit03)
            (Rendering|Material|SetScalarParameterValueonMaterials _weaponmesh "RedGlowMultiplier" _redglowmultiplier)
            (CallFunction|UpdateFX _transitionamounthit03))
          (:4
            (Rendering|Material|SetScalarParameterValueonMaterials _weaponmesh "TransitionAmount" _transitionamounthit04)
            (Rendering|Material|SetScalarParameterValueonMaterials _weaponmesh "RedGlowMultiplier" _redglowmultiplier)
            (CallFunction|UpdateFX _transitionamounthit04))
          (:5
            _))
        ;[1 0.5781 0.5781] Decay Red Blade
        (Variables|Default|SetDecayAmount 0.0)
        (Variables|Default|SetRedFXIsActive true)
        (Effects|SpawnFXSystemAttached "/BRCosmetics/Weapons/FORT_Melee/Pickaxe_ProdigyFire/FX/NS_Pickaxe_ProdigyFire_Activated.NS_Pickaxe_ProdigyFire_Activated" (Variables|Visual|GetWeaponMesh))
        (Utilities|FlowControl|Delay 0.05)
        (bind _output_get (Variables|Default|SetDecayAmount (+ (Variables|Default|GetDecayAmount) 0.05)))
        (if (> _output_get (+ 7.0 -4.8))
          _
          (else
            _))
        (Variables|Default|SetRedFXIsActive false)
        (bind (_direction_2 _hitdecay01 _hitdecay02 _hitdecay03 _hitdecay04 _hitdecay05) (Timeline|HitDecayedTimeline PlayFromStart)
          (:Update
            (switch int (Math|Float|Truncate (Variables|Default|GetPickaxeHitCount))
              (:Default
                (Variables|Default|SetHitDecayAnimValue _hitdecay05))
              (:0
                _)
              (:1
                (Variables|Default|SetHitDecayAnimValue _hitdecay01))
              (:2
                (Variables|Default|SetHitDecayAnimValue _hitdecay02))
              (:3
                (Variables|Default|SetHitDecayAnimValue _hitdecay03))
              (:4
                (Variables|Default|SetHitDecayAnimValue _hitdecay04))
              (:5
                _))
            (Variables|Default|SetDecayAmount 0.0)
            (if (not _redfxisactive)
              (Rendering|Material|SetScalarParameterValueonMaterials _weaponmesh_2 "TransitionAmount" _hitdecayanimvalue)
              (Rendering|Material|SetScalarParameterValueonMaterials _weaponmesh_2 "TransitionAmountForGlow" _hitdecayanimvalue)
              (CallFunction|UpdateFX (Variables|Default|GetHitDecayAnimValue))))
          (:Finished
            (if (not _redfxisactive)
              (Variables|Default|SetPickaxeHitCount 0.0)
              (Variables|Default|SetDecayAmount 0.0))))))))

(event Custom|CE_FX_Server
  (CallFunction|CEFXClient))

(event Custom|FrontEndReactivity
  (bind _weaponmesh (Variables|Visual|GetWeaponMesh))
  (bind _redfxisactive (Variables|Default|GetRedFXIsActive))
  (bind _weaponmesh_2 (Variables|Visual|GetWeaponMesh))
  (bind _hitdecayanimvalue (Variables|Default|GetHitDecayAnimValue))
  (bind _weaponmesh_3 (Variables|Visual|GetWeaponMesh))
  ;[0.581 0.7565 1] FrontEnd Blade Reactivity
  ;;[1 0.5 0.5] Make Blade Red on Hit (Progressively, 5-steps)
  (Variables|Default|SetPickaxeHitCount (+ (Variables|Default|GetPickaxeHitCount) 1.0))
  (bind _index (Class|FortCosmeticFunctionLibrary|%28Deprecated%29GetVariantChannelActiveIndexForSlot "(TagName=\"LoadoutSlot.Pickaxe\")" (CallFunction|GetCosmeticData) "(TagName=\"Cosmetics.Variant.Channel.Particle\")" (Utilities|Array|MakeArray "(TagName=\"Cosmetics.Variant.Property.Particle1\")" "(TagName=\"Cosmetics.Variant.Property.Particle2\")" "(TagName=\"Cosmetics.Variant.Property.Particle3\")")))
  (if (== _index 0)
    (bind (_direction _transitionamounthit01 _transitionamounthit02 _transitionamounthit03 _transitionamounthit04 _transitionamounthit05 _transitionamountforglow _redglowmultiplier) (Timeline|Timeline PlayFromStart)
      (:Update
        (switch int (Math|Float|Truncate (Variables|Default|GetPickaxeHitCount))
          (:Default
            (Rendering|Material|SetScalarParameterValueonMaterials _weaponmesh "RedGlowMultiplier" 0.1)
            (Rendering|Material|SetScalarParameterValueonMaterials _weaponmesh "TransitionAmount" 1.0)
            (CallFunction|UpdateFX _transitionamounthit05)
            (Rendering|Material|SetScalarParameterValueonMaterials _weaponmesh_3 "TransitionAmountForGlow" _transitionamountforglow))
          (:0
            (Rendering|Material|SetScalarParameterValueonMaterials _weaponmesh_3 "TransitionAmount" 0.0)
            (CallFunction|UpdateFX 0.0)
            (Rendering|Material|SetScalarParameterValueonMaterials _weaponmesh_3 "TransitionAmountForGlow" 0.0))
          (:1
            (Rendering|Material|SetScalarParameterValueonMaterials _weaponmesh_3 "TransitionAmount" _transitionamounthit01)
            (Rendering|Material|SetScalarParameterValueonMaterials _weaponmesh_3 "RedGlowMultiplier" _redglowmultiplier)
            (CallFunction|UpdateFX _transitionamounthit01))
          (:2
            (Rendering|Material|SetScalarParameterValueonMaterials _weaponmesh_3 "TransitionAmount" _transitionamounthit02)
            (Rendering|Material|SetScalarParameterValueonMaterials _weaponmesh_3 "RedGlowMultiplier" _redglowmultiplier)
            (CallFunction|UpdateFX _transitionamounthit02))
          (:3
            (Rendering|Material|SetScalarParameterValueonMaterials _weaponmesh_3 "TransitionAmount" _transitionamounthit03)
            (Rendering|Material|SetScalarParameterValueonMaterials _weaponmesh_3 "RedGlowMultiplier" _redglowmultiplier)
            (CallFunction|UpdateFX _transitionamounthit03))
          (:4
            (Rendering|Material|SetScalarParameterValueonMaterials _weaponmesh_3 "TransitionAmount" _transitionamounthit04)
            (Rendering|Material|SetScalarParameterValueonMaterials _weaponmesh_3 "RedGlowMultiplier" _redglowmultiplier)
            (CallFunction|UpdateFX _transitionamounthit04))
          (:5
            (Rendering|Material|SetScalarParameterValueonMaterials _weaponmesh_3 "RedGlowMultiplier" _redglowmultiplier)
            (Rendering|Material|SetScalarParameterValueonMaterials _weaponmesh_3 "TransitionAmount" _transitionamounthit05)))
        ;;[1 0.5781 0.5781] Decay Red Blade
        (Variables|Default|SetDecayAmount 0.0)
        (Variables|Default|SetRedFXIsActive true)
        (Effects|SpawnFXSystemAttached "/BRCosmetics/Weapons/FORT_Melee/Pickaxe_ProdigyFire/FX/NS_Pickaxe_ProdigyFire_Activated.NS_Pickaxe_ProdigyFire_Activated" (Variables|Visual|GetWeaponMesh))
        (Utilities|FlowControl|Delay 0.05)
        (bind _output_get (Variables|Default|SetDecayAmount (+ (Variables|Default|GetDecayAmount) 0.05)))
        (if (> _output_get (+ 9.1 -4.8))
          _
          (else
            _))
        (Variables|Default|SetRedFXIsActive false)
        (bind (_direction_2 _hitdecay01 _hitdecay02 _hitdecay03 _hitdecay04 _hitdecay05) (Timeline|Timeline_2 PlayFromStart)
          (:Update
            (switch int (Math|Float|Truncate (Variables|Default|GetPickaxeHitCount))
              (:Default
                (Variables|Default|SetHitDecayAnimValue _hitdecay05))
              (:0
                _)
              (:1
                (Variables|Default|SetHitDecayAnimValue _hitdecay01))
              (:2
                (Variables|Default|SetHitDecayAnimValue _hitdecay02))
              (:3
                (Variables|Default|SetHitDecayAnimValue _hitdecay03))
              (:4
                (Variables|Default|SetHitDecayAnimValue _hitdecay04))
              (:5
                _))
            (Variables|Default|SetDecayAmount 0.0)
            (if (not _redfxisactive)
              (Rendering|Material|SetScalarParameterValueonMaterials _weaponmesh_2 "TransitionAmount" _hitdecayanimvalue)
              (Rendering|Material|SetScalarParameterValueonMaterials _weaponmesh_2 "TransitionAmountForGlow" _hitdecayanimvalue)
              (CallFunction|UpdateFX (Variables|Default|GetHitDecayAnimValue))))
          (:Finished
            (if (not _redfxisactive)
              (Variables|Default|SetPickaxeHitCount 0.0)
              (Variables|Default|SetDecayAmount 0.0))))))))

(fn UpdateFX
  (params (Alpha :type Float%20%28double-precision%29))
  (bind _output_get (Variables|Default|SetFX_TransitionAlpha Alpha))
  (bind _asniagara_particle_system_component (Utilities|Casting|CastToNiagaraComponent (CallFunction|GetIdleVFXComponent)))
  (Niagara|SetNiagaraVariable(Float) _asniagara_particle_system_component "Transition" _output_get))

(fn ConstructionScript
  (|CallParentFunction))
