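; Cosmetic pickaxe weapon Blueprint (ProdigyHaughty) that glows progressively over five steps as it hits weak spots.
; Certain impact surfaces trigger a server-to-client event that increments a hit count; impacts in frontend preview also increment it.
; Timelines raise and decay mesh material glow and idle VFX color, then reset the count.
; The particle variant selects reactive, fixed full or zero levels for impact and swing-trail Niagara effects.
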(event EventBeginPlay
  (|CallParentFunction)
  (Utilities|FlowControl|Sequence
    (:then_0
      (Components|Activation|Activate (CallFunction|GetIdlePSC) false))
    (:then_1
      (bind _asgalileo_ferry_axe_anim_bp (Utilities|Casting|CastToGalileo_Ferry_Axe_AnimBP (Components|SkeletalMesh|GetAnimInstance (Variables|Visual|GetWeaponMesh))))
      (Delegates|BindEventtoFrontEndIdle _asgalileo_ferry_axe_anim_bp (delegate Custom|FrontEndIdle)))))

(event Custom|FrontEndIdle
  (Utilities|FlowControl|DoOnce false
    (:Completed
      (Utilities|FlowControl|Delay 0.3)
      (Components|Activation|Activate (CallFunction|GetIdlePSC) false))))

(event EventEndPlay
  (params (EndPlayReason :type EEndPlayReason%20Enum :default "Destroyed"))
  (|CallParentFunction EndPlayReason)
  (Utilities|FlowControl|DoOnce false
    (:Completed
      (Utilities|FlowControl|Delay 0.3)
      (Components|Activation|Activate (CallFunction|GetIdlePSC) false))))

(event EventOnWeaponVisibilityChanged
  (params (bVisible :type Boolean) (bSetForLocalControllerOnly :type Boolean))
  (if bVisible
    (Components|Activation|Activate (CallFunction|GetIdleVFXComponent) false)
    (else
      (Components|Activation|Deactivate (CallFunction|GetIdlePSC)))))

(event FortWeapon|EventOnPlayImpactFX
  (params (HitResult :type Hit%20Result%20Structure%20%28by%20ref%29) (ImpactPhysicalSurface :type EPhysicalSurface%20Enum :default "SurfaceType_Default") (SpawnedPSC :type FXSystem%20Component%20Object%20Reference))
  (|CallParentFunction HitResult ImpactPhysicalSurface SpawnedPSC)
  (Utilities|FlowControl|Sequence
    (:then_0
      (if (Math|Boolean|ORBoolean (Utilities|Enum|Equal(Enum) ImpactPhysicalSurface "SurfaceType14") (Utilities|Enum|Equal(Enum) ImpactPhysicalSurface "SurfaceType15") (Utilities|Enum|Equal(Enum) ImpactPhysicalSurface "SurfaceType16"))
        (CallFunction|CEFXServer)))
    (:then_1
      (bind _asniagara_particle_system_component (Utilities|Casting|CastToNiagaraComponent SpawnedPSC))
      (bind _index (Class|FortCosmeticFunctionLibrary|%28Deprecated%29GetVariantChannelActiveIndexForSlot "(TagName=\"LoadoutSlot.Pickaxe\")" (CallFunction|GetCosmeticData) "(TagName=\"Cosmetics.Variant.Channel.Particle\")" (Utilities|Array|MakeArray "(TagName=\"Cosmetics.Variant.Property.Particle1\")" "(TagName=\"Cosmetics.Variant.Property.Particle2\")" "(TagName=\"Cosmetics.Variant.Property.Particle3\")")))
      (if (== _index 0)
        (Effects|Components|ParticleSystem|SetFloatParameter _asniagara_particle_system_component "User.HitCount" (Math|Float|Clamp(Float) (Variables|Default|GetPickaxeHitCount) 0.0 5.0))
        (elif (== _index 1)
          (Effects|Components|ParticleSystem|SetFloatParameter _asniagara_particle_system_component "User.HitCount" 5.0)
          (else
            (Effects|Components|ParticleSystem|SetFloatParameter _asniagara_particle_system_component "User.HitCount" 0.0)))))
    (:then_2
      (if (or (Reactive|Frontend|IsCosmeticObjectDisplayedasPreview 0) (Reactive|Frontend|IsCosmeticObjectDisplayedinFrontendContext 0))
        (CallFunction|FrontEndReactivity)))))

(event Custom|CE_FX_Client
  (CallFunction|WeakSpotReactivity))

(event Custom|WeakSpotReactivity
  (bind _pickaxehitcount (Variables|Default|GetPickaxeHitCount))
  (bind _returnvalue (Math|Float|Truncate _pickaxehitcount))
  (bind _pickaxehitcount_2 (Variables|Default|GetPickaxeHitCount))
  (bind _returnvalue_2 (Math|Float|Truncate _pickaxehitcount_2))
  (bind _returnvalue_3 (Math|Integer|Clamp(Integer) _returnvalue_2 0 5))
  (bind _weaponmesh (Variables|Visual|GetWeaponMesh))
  (bind _weaponmesh_2 (Variables|Visual|GetWeaponMesh))
  (bind _redgemsareactive (Variables|Default|GetRedGemsAreActive))
  ;[0.4618 0.8693 1] Make Pickaxe Glow on Weak Spot Hit (Progressively, 5-steps)
  (Variables|Default|SetSetFloatForDecayTime 7.0)
  (Variables|Default|SetPickaxeHitCount (+ (Variables|Default|GetPickaxeHitCount) 1.0))
  (bind _index (Class|FortCosmeticFunctionLibrary|%28Deprecated%29GetVariantChannelActiveIndexForSlot "(TagName=\"LoadoutSlot.Pickaxe\")" (CallFunction|GetCosmeticData) "(TagName=\"Cosmetics.Variant.Channel.Particle\")" (Utilities|Array|MakeArray "(TagName=\"Cosmetics.Variant.Property.Particle1\")" "(TagName=\"Cosmetics.Variant.Property.Particle2\")" "(TagName=\"Cosmetics.Variant.Property.Particle3\")")))
  (if (== _index 0)
    (bind (_direction _transitionamounthit01 _transitionamounthit02 _transitionamounthit03 _transitionamounthit04 _transitionamounthit05 _glowmultiplier) (Timeline|Timeline PlayFromStart)
      (:Update
        (switch int _returnvalue_2
          (:Default
            (Rendering|Material|SetScalarParameterValueonMaterials _weaponmesh "EffectAmount" 5.0)
            (CallFunction|FXUpdateColors 5.0)
            (Rendering|Material|SetScalarParameterValueonMaterials _weaponmesh "OveralGlowMultiplier" (* _glowmultiplier _returnvalue_3)))
          (:0
            (Rendering|Material|SetScalarParameterValueonMaterials _weaponmesh "EffectAmount" 0.0)
            (Rendering|Material|SetScalarParameterValueonMaterials _weaponmesh "OveralGlowMultiplier" 0.0)
            (CallFunction|FXUpdateColors 0.0))
          (:1
            (Rendering|Material|SetScalarParameterValueonMaterials _weaponmesh "EffectAmount" _transitionamounthit01)
            (CallFunction|FXUpdateColors _transitionamounthit01)
            (Rendering|Material|SetScalarParameterValueonMaterials _weaponmesh "OveralGlowMultiplier" (* _glowmultiplier _returnvalue_3)))
          (:2
            (Rendering|Material|SetScalarParameterValueonMaterials _weaponmesh "EffectAmount" _transitionamounthit02)
            (CallFunction|FXUpdateColors _transitionamounthit02))
          (:3
            (Rendering|Material|SetScalarParameterValueonMaterials _weaponmesh "EffectAmount" _transitionamounthit03)
            (CallFunction|FXUpdateColors _transitionamounthit03))
          (:4
            (Rendering|Material|SetScalarParameterValueonMaterials _weaponmesh "EffectAmount" _transitionamounthit04)
            (CallFunction|FXUpdateColors _transitionamounthit04))
          (:5
            (Rendering|Material|SetScalarParameterValueonMaterials _weaponmesh "EffectAmount" _transitionamounthit05)
            (CallFunction|FXUpdateColors _transitionamounthit05)))
        ;[0.3062 0.4374 1] Decay Gems
        (Variables|Default|SetDecayAmount 0.0)
        (Variables|Default|SetRedGemsAreActive true)
        (Utilities|FlowControl|Delay 0.05)
        (bind _output_get (Variables|Default|SetDecayAmount (+ (Variables|Default|GetDecayAmount) 0.05)))
        (if (> _output_get (+ (Variables|Default|GetSetFloatForDecayTime) -5.3))
          _
          (else
            _))
        (Variables|Default|SetRedGemsAreActive false)
        (bind (_direction_2 _hitdecay_glowintensity _hitdecay01 _hitdecay02 _hitdecay03 _hitdecay04 _hitdecay05) (Timeline|Timeline_1 PlayFromStart)
          (:Update
            (switch int _returnvalue
              (:Default
                (Variables|Default|SetHitDecayAnimValue _hitdecay05))
              (:0
                _)
              (:1
                (Variables|Default|SetHitDecayAnimValue _hitdecay01))
              (:2
                (Variables|Default|SetHitDecayAnimValue _hitdecay02))
              (:3
                (Variables|Default|SetHitDecayAnimValue _hitdecay03))
              (:4
                (Variables|Default|SetHitDecayAnimValue _hitdecay04))
              (:5
                _))
            (Variables|Default|SetDecayAmount 0.0)
            (if (not _redgemsareactive)
              (Rendering|Material|SetScalarParameterValueonMaterials _weaponmesh_2 "EffectAmount" (Variables|Default|GetHitDecayAnimValue))
              (Rendering|Material|SetScalarParameterValueonMaterials _weaponmesh_2 "OveralGlowMultiplier" (* _hitdecay_glowintensity (Math|Integer|Clamp(Integer) _returnvalue 0 5)))
              (CallFunction|FXUpdateColors (Variables|Default|GetHitDecayAnimValue))))
          (:Finished
            (if (not _redgemsareactive)
              (Variables|Default|SetPickaxeHitCount 0.0)
              (Variables|Default|SetDecayAmount 0.0))))))))

(event Custom|FrontEndReactivity
  (bind _pickaxehitcount (Variables|Default|GetPickaxeHitCount))
  (bind _returnvalue (Math|Float|Truncate _pickaxehitcount))
  (bind _pickaxehitcount_2 (Variables|Default|GetPickaxeHitCount))
  (bind _returnvalue_2 (Math|Float|Truncate _pickaxehitcount_2))
  (bind _returnvalue_3 (Math|Integer|Clamp(Integer) _returnvalue_2 0 5))
  (bind _weaponmesh (Variables|Visual|GetWeaponMesh))
  (bind _weaponmesh_2 (Variables|Visual|GetWeaponMesh))
  (bind _redgemsareactive (Variables|Default|GetRedGemsAreActive))
  ;[0.4618 0.8693 1] Make Pickaxe Glow on Weak Spot Hit (Progressively, 5-steps)
  (Variables|Default|SetSetFloatForDecayTime 9.0)
  (Variables|Default|SetPickaxeHitCount (+ (Variables|Default|GetPickaxeHitCount) 1.0))
  (bind _index (Class|FortCosmeticFunctionLibrary|%28Deprecated%29GetVariantChannelActiveIndexForSlot "(TagName=\"LoadoutSlot.Pickaxe\")" (CallFunction|GetCosmeticData) "(TagName=\"Cosmetics.Variant.Channel.Particle\")" (Utilities|Array|MakeArray "(TagName=\"Cosmetics.Variant.Property.Particle1\")" "(TagName=\"Cosmetics.Variant.Property.Particle2\")" "(TagName=\"Cosmetics.Variant.Property.Particle3\")")))
  (if (== _index 0)
    (Timeline|Timeline PlayFromStart)))

(event Custom|CE_FX_Server
  (CallFunction|CEFXClient))

(event Custom|FX_UpdateColors
  (params (HitCount :type Float%20%28double-precision%29))
  ;[1 0.1557 0.2856] Update FX Colors depending on Hit Count
  (Variables|Default|SetFXTransitionAlpha (Math|Float|Clamp(Float) HitCount 0.0 5.0))
  (Effects|Components|ParticleSystem|SetFloatParameter (CallFunction|GetIdleVFXComponent) "User.HitCount" (Variables|Default|GetFXTransitionAlpha)))

(event EventSwingLeft
  (bind _returnvalue (CallFunction|GetAnimTrailVFXComponent))
  ;[1 0.1557 0.2856] Update FX Colors depending on Hit Count
  (|CallParentFunction)
  (bind _index (Class|FortCosmeticFunctionLibrary|%28Deprecated%29GetVariantChannelActiveIndexForSlot "(TagName=\"LoadoutSlot.Pickaxe\")" (CallFunction|GetCosmeticData) "(TagName=\"Cosmetics.Variant.Channel.Particle\")" (Utilities|Array|MakeArray "(TagName=\"Cosmetics.Variant.Property.Particle1\")" "(TagName=\"Cosmetics.Variant.Property.Particle2\")" "(TagName=\"Cosmetics.Variant.Property.Particle3\")")))
  (if (== _index 0)
    (Effects|Components|ParticleSystem|SetFloatParameter _returnvalue "User.HitCount" (Variables|Default|GetFXTransitionAlpha))
    (elif (== _index 1)
      (Effects|Components|ParticleSystem|SetFloatParameter _returnvalue "User.HitCount" 5.0)
      (else
        (Effects|Components|ParticleSystem|SetFloatParameter _returnvalue "User.HitCount" 0.0)))))

(event EventSwingRight
  (bind _returnvalue (CallFunction|GetAnimTrailVFXComponent))
  ;[1 0.1557 0.2856] Update FX Colors depending on Hit Count
  (|CallParentFunction)
  (bind _index (Class|FortCosmeticFunctionLibrary|%28Deprecated%29GetVariantChannelActiveIndexForSlot "(TagName=\"LoadoutSlot.Pickaxe\")" (CallFunction|GetCosmeticData) "(TagName=\"Cosmetics.Variant.Channel.Particle\")" (Utilities|Array|MakeArray "(TagName=\"Cosmetics.Variant.Property.Particle1\")" "(TagName=\"Cosmetics.Variant.Property.Particle2\")" "(TagName=\"Cosmetics.Variant.Property.Particle3\")")))
  (if (== _index 0)
    (Effects|Components|ParticleSystem|SetFloatParameter _returnvalue "User.HitCount" (Variables|Default|GetFXTransitionAlpha))
    (elif (== _index 1)
      (Effects|Components|ParticleSystem|SetFloatParameter _returnvalue "User.HitCount" 5.0)
      (else
        (Effects|Components|ParticleSystem|SetFloatParameter _returnvalue "User.HitCount" 0.0)))))

(fn ConstructionScript
  (|CallParentFunction)
  (Components|Activation|SetAutoActivate (CallFunction|GetIdlePSC) false))
