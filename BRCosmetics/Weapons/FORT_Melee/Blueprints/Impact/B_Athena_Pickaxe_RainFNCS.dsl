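; Cosmetic RainFNCS pickaxe Blueprint that visually reacts to weak-point hits, and to every impact in frontend or preview display.
; Both hit events call a shared ReactivityEvent, which restarts a timeline from the start.
; The timeline drives reactivity and dots-progress scalar parameters on the weapon mesh materials and blend floats on the anim trail VFX.
; The construction script lowers the idle Niagara effect's translucent sort priority.
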(event FortWeapon|EventOnPlayImpactFX
  (params (HitResult :type Hit%20Result%20Structure%20%28by%20ref%29) (ImpactPhysicalSurface :type EPhysicalSurface%20Enum :default "SurfaceType_Default") (SpawnedPSC :type FXSystem%20Component%20Object%20Reference))
  ;[1 0.2581 0.8991] Main Setup For Weak Spot Reactivity
  (|CallParentFunction HitResult ImpactPhysicalSurface SpawnedPSC)
  ;;[0.1768 0.291 1] Calls the reactivity on FrontEnd. Notice how we don't need to check if the hit was a "weak spot hit" on FrontEnd. Every hit there should be reactive.
  (if (or (Reactive|Frontend|IsCosmeticObjectDisplayedasPreview 0) (Reactive|Frontend|IsCosmeticObjectDisplayedinFrontendContext 0))
    (CallFunction|ReactivityEvent)))

(event EventOnPlayWeakpointHitFX
  ;[1 0.2581 0.8991] Main Setup For Weak Spot Reactivity
  (CallFunction|ReactivityEvent))

(event Custom|ReactivityEvent
  (bind _returnvalue (CallFunction|GetAnimTrailVFXComponent))
  (bind _weaponmesh (Variables|Visual|GetWeaponMesh))
  ;[0.4618 0.8693 1] The Actual Reactivity Logic
  ;; TA updates on weakpoint hit
  (bind (_direction _reactivity _dotsprogress) (Timeline|RainFNCS_Reactivity PlayFromStart)
    (:Update
      (Rendering|Material|SetScalarParameterValueonMaterials _weaponmesh "RainFNCS_Pickaxe_ReactivityProgress" _reactivity) ; The parameter you want to change goes here
      (Rendering|Material|SetScalarParameterValueonMaterials _weaponmesh "RainFNCS_Pickaxe_DotsProgress" _dotsprogress) ; The parameter you want to change goes here
      ;; FX updates on weakpoint hit
      (Effects|Components|ParticleSystem|SetFloatParameter _returnvalue "StateBlendOpacity" _reactivity)
      (Effects|Components|ParticleSystem|SetFloatParameter _returnvalue "StateBlend" _dotsprogress))))

(fn ConstructionScript
  (|CallParentFunction)
  (Rendering|SetTranslucentSortPriority (Variables|Default|GetNS_Pickaxe_RainFNCS_Idle) -1))
