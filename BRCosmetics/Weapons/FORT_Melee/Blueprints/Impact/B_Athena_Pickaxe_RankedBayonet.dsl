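; Cosmetic Blueprint for the Ranked Bayonet pickaxe that shows ranked progression in its effects and supports a non-ranked look.
; Impact and left or right swing events call the parent, then pass the VFX rank to the particle component's RankedProgression parameter.
; On visibility changes and frontend cosmetic initialization, a deprecated variant lookup checks the material channel and applies the default material when the Ranked Off stage is active.
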(event FortWeapon|EventOnPlayImpactFX
  (params (HitResult :type Hit%20Result%20Structure%20%28by%20ref%29) (ImpactPhysicalSurface :type EPhysicalSurface%20Enum :default "SurfaceType_Default") (SpawnedPSC :type FXSystem%20Component%20Object%20Reference))
  (|CallParentFunction HitResult ImpactPhysicalSurface SpawnedPSC)
  (Effects|Components|ParticleSystem|SetIntParameter SpawnedPSC "RankedProgression" (Variables|Default|GetVFX_Rank)))

(event EventSwingLeft
  (|CallParentFunction)
  (Effects|Components|ParticleSystem|SetIntParameter (CallFunction|GetSwingVFXComponent) "RankedProgression" (Variables|Default|GetVFX_Rank)))

(event EventSwingRight
  (|CallParentFunction)
  (Effects|Components|ParticleSystem|SetIntParameter (CallFunction|GetSwingVFXComponent) "RankedProgression" (Variables|Default|GetVFX_Rank)))

(event EventOnWeaponVisibilityChanged
  (params (bVisible :type Boolean) (bSetForLocalControllerOnly :type Boolean))
  (|CallParentFunction bVisible bSetForLocalControllerOnly)
  ; Check if the active variant has the Ranked Off tag to override the material with a default non-ranked version
  (bind _index (Class|FortCosmeticFunctionLibrary|%28Deprecated%29GetVariantChannelActiveIndexForSlot "(TagName=\"LoadoutSlot.Pickaxe\")" (CallFunction|GetCosmeticData) "(TagName=\"Cosmetics.Variant.Channel.Material\")" (Utilities|Array|MakeArray "(TagName=\"Cosmetics.Variant.Property.Stage1\")" "(TagName=\"Cosmetics.Variant.Property.Stage2\")")))
  (if (== _index 1)
    (Rendering|Material|SetMaterial (Variables|Visual|GetWeaponMesh) 0 (Variables|Default|GetDefault%20Material))))

(event EventOnInitWeaponCosmetics
  (|CallParentFunction)
  (if (Reactive|Frontend|IsCosmeticObjectDisplayedinFrontendContext 0)
    ; Check if the active variant has the Ranked Off tag to override the material with a default non-ranked version
    (bind _index (Class|FortCosmeticFunctionLibrary|%28Deprecated%29GetVariantChannelActiveIndexForSlot "(TagName=\"LoadoutSlot.Pickaxe\")" (CallFunction|GetCosmeticData) "(TagName=\"Cosmetics.Variant.Channel.Material\")" (Utilities|Array|MakeArray "(TagName=\"Cosmetics.Variant.Property.Stage1\")" "(TagName=\"Cosmetics.Variant.Property.Stage2\")")))
    (if (== _index 1)
      (Rendering|Material|SetMaterial (Variables|Visual|GetWeaponMesh) 0 (Variables|Default|GetDefault%20Material)))))

(fn ConstructionScript
  (|CallParentFunction))
