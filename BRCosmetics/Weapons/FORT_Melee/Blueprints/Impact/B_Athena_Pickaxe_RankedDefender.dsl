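; Pickaxe weapon Blueprint for the Ranked Defender cosmetic that sets its rank progress when shown as a preview or in the frontend.
; On weapon cosmetics init, after the parent call, it reads the pickaxe slot's Ranked variant channel through a deprecated lookup.
; The channel's custom data is parsed as an integer, converted to a Habanero rank and passed to SetCurrentProgress.
; BeginPlay and Tick are empty, and no in-match rank handling appears here.
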(event EventBeginPlay)

(event EventTick
  (params (DeltaSeconds :type Float%20%28single-precision%29)))

(event EventOnInitWeaponCosmetics
  (|CallParentFunction)
  (if (or (Reactive|Frontend|IsCosmeticObjectDisplayedasPreview 0) (Reactive|Frontend|IsCosmeticObjectDisplayedinFrontendContext 0))
    (bind _outvariantchannels (Class|FortCosmeticFunctionLibrary|%28Deprecated%29GetVariantChannelInfoForSlotTag "(TagName=\"LoadoutSlot.Pickaxe\")" (CallFunction|GetCosmeticData) "(TagName=\"Cosmetics.Variant.Channel.Ranked\")"))
    (Default|SetCurrentProgress (Class|FortCosmeticRankedVariant|GetHabaneroRankfromRankIndex (Utilities|String|StringToInteger (.CustomData (Utilities|Struct|BreakMcpVariantChannelInfo _outvariantchannels)))))))

(fn ConstructionScript
  (|CallParentFunction))
