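; B_Athena_Pickaxe_RankedFirth is a cosmetic Blueprint for the RankedFirth pickaxe that adjusts its glow for frontend display.
; When weapon cosmetics initialize, it runs the parent logic, then checks whether the cosmetic object is displayed in a frontend context.
; If so, it raises the DiamondEmissiveIntensity scalar parameter on the weapon mesh materials to a fixed value.
; The construction script only calls its parent.
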(event EventOnInitWeaponCosmetics
  (|CallParentFunction)
  (if (Reactive|Frontend|IsCosmeticObjectDisplayedinFrontendContext 0)
    (Rendering|Material|SetScalarParameterValueonMaterials (Variables|Visual|GetWeaponMesh) "DiamondEmissiveIntensity" 22.0)))

(fn ConstructionScript
  (|CallParentFunction))
