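; Blueprint for the Ranked Olympus melee pickaxe in Athena, extending a parent weapon class.
; On cosmetic initialization it runs the parent logic, then deactivates the particle component returned by GetSwingPSC, so the swing effect starts switched off.
; The left swing, right swing and construction script overrides only call their parent versions and add no logic of their own.
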(event EventSwingLeft
  (|CallParentFunction))

(event EventSwingRight
  (|CallParentFunction))

(event EventOnInitWeaponCosmetics
  (|CallParentFunction)
  (Components|Activation|Deactivate (CallFunction|GetSwingPSC)))

(fn ConstructionScript
  (|CallParentFunction))
