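; Ranked Scythe pickaxe cosmetic Blueprint that swaps in a non-ranked look when the Ranked Off variant is active.
; On weapon visibility change, and on cosmetic init in a frontend context, it queries the Material variant channel through a deprecated lookup.
; If the second stage is active, it sets the weapon mesh's first material slot to the default rank material.
; Swing events and construction only call the parent.
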(event EventSwingLeft
  (|CallParentFunction))

(event EventSwingRight
  (|CallParentFunction))

(event EventOnWeaponVisibilityChanged
  (params (bVisible :type Boolean) (bSetForLocalControllerOnly :type Boolean))
  (|CallParentFunction bVisible bSetForLocalControllerOnly)
  ; Check if the active variant has the Ranked Off tag to override the material with a default non-ranked version
  (bind _index (Class|FortCosmeticFunctionLibrary|%28Deprecated%29GetVariantChannelActiveIndexForSlot "(TagName=\"LoadoutSlot.Pickaxe\")" (CallFunction|GetCosmeticData) "(TagName=\"Cosmetics.Variant.Channel.Material\")" (Utilities|Array|MakeArray "(TagName=\"Cosmetics.Variant.Property.Stage1\")" "(TagName=\"Cosmetics.Variant.Property.Stage2\")")))
  (if (== _index 1)
    (Rendering|Material|SetMaterial (Variables|Visual|GetWeaponMesh) 0 (Variables|Rank|RankMaterials|GetDefault%20Material))))

(event EventOnInitWeaponCosmetics
  (|CallParentFunction)
  (if (Reactive|Frontend|IsCosmeticObjectDisplayedinFrontendContext 0)
    ; Check if the active variant has the Ranked Off tag to override the material with a default non-ranked version
    (bind _index (Class|FortCosmeticFunctionLibrary|%28Deprecated%29GetVariantChannelActiveIndexForSlot "(TagName=\"LoadoutSlot.Pickaxe\")" (CallFunction|GetCosmeticData) "(TagName=\"Cosmetics.Variant.Channel.Material\")" (Utilities|Array|MakeArray "(TagName=\"Cosmetics.Variant.Property.Stage1\")" "(TagName=\"Cosmetics.Variant.Property.Stage2\")")))
    (if (== _index 1)
      (Rendering|Material|SetMaterial (Variables|Visual|GetWeaponMesh) 0 (Variables|Rank|RankMaterials|GetDefault%20Material)))))

(fn ConstructionScript
  (|CallParentFunction))
