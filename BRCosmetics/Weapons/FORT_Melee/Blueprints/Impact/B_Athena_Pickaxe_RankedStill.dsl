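; Ranked pickaxe cosmetic Blueprint that tints impact, swing and idle VFX by player rank, unless the variant turns ranked styling off.
; After parent calls, each event reads the Material variant channel: Stage2 disables UseRankedColors, otherwise it enables it and passes VFX_Rank as RankedProgression.
; Visibility changes and cosmetic init, the latter only in frontend display, update idle VFX and restore the default mesh material when ranked is off.
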(event FortWeapon|EventOnPlayImpactFX
  (params (HitResult :type Hit%20Result%20Structure%20%28by%20ref%29) (ImpactPhysicalSurface :type EPhysicalSurface%20Enum :default "SurfaceType_Default") (SpawnedPSC :type FXSystem%20Component%20Object%20Reference))
  (|CallParentFunction HitResult ImpactPhysicalSurface SpawnedPSC)
  (bind _index (Class|FortCosmeticFunctionLibrary|%28Deprecated%29GetVariantChannelActiveIndexForSlot "(TagName=\"LoadoutSlot.Pickaxe\")" (CallFunction|GetCosmeticData) "(TagName=\"Cosmetics.Variant.Channel.Material\")" (Utilities|Array|MakeArray "(TagName=\"Cosmetics.Variant.Property.Stage1\")" "(TagName=\"Cosmetics.Variant.Property.Stage2\")")))
  (if (== _index 1)
    (Effects|Components|ParticleSystem|SetBoolParameter SpawnedPSC "UseRankedColors" false)
    (else
      (Effects|Components|ParticleSystem|SetBoolParameter SpawnedPSC "UseRankedColors" true)
      (Effects|Components|ParticleSystem|SetIntParameter SpawnedPSC "RankedProgression" (Variables|Default|GetVFX_Rank)))))

(event EventSwingLeft
  (bind _returnvalue (CallFunction|GetSwingVFXComponent))
  (|CallParentFunction)
  (bind _index (Class|FortCosmeticFunctionLibrary|%28Deprecated%29GetVariantChannelActiveIndexForSlot "(TagName=\"LoadoutSlot.Pickaxe\")" (CallFunction|GetCosmeticData) "(TagName=\"Cosmetics.Variant.Channel.Material\")" (Utilities|Array|MakeArray "(TagName=\"Cosmetics.Variant.Property.Stage1\")" "(TagName=\"Cosmetics.Variant.Property.Stage2\")")))
  (if (== _index 1)
    (Effects|Components|ParticleSystem|SetBoolParameter _returnvalue "UseRankedColors" false)
    (else
      (Effects|Components|ParticleSystem|SetBoolParameter _returnvalue "UseRankedColors" true)
      (Effects|Components|ParticleSystem|SetIntParameter _returnvalue "RankedProgression" (Variables|Default|GetVFX_Rank)))))

(event EventSwingRight
  (bind _returnvalue (CallFunction|GetSwingVFXComponent))
  (|CallParentFunction)
  (bind _index (Class|FortCosmeticFunctionLibrary|%28Deprecated%29GetVariantChannelActiveIndexForSlot "(TagName=\"LoadoutSlot.Pickaxe\")" (CallFunction|GetCosmeticData) "(TagName=\"Cosmetics.Variant.Channel.Material\")" (Utilities|Array|MakeArray "(TagName=\"Cosmetics.Variant.Property.Stage1\")" "(TagName=\"Cosmetics.Variant.Property.Stage2\")")))
  (if (== _index 1)
    (Effects|Components|ParticleSystem|SetBoolParameter _returnvalue "UseRankedColors" false)
    (else
      (Effects|Components|ParticleSystem|SetBoolParameter _returnvalue "UseRankedColors" true)
      (Effects|Components|ParticleSystem|SetIntParameter _returnvalue "RankedProgression" (Variables|Default|GetVFX_Rank)))))

(event EventOnWeaponVisibilityChanged
  (params (bVisible :type Boolean) (bSetForLocalControllerOnly :type Boolean))
  (bind _returnvalue (CallFunction|GetIdleVFXComponent))
  (|CallParentFunction bVisible bSetForLocalControllerOnly)
  ;[1 1 1] Check if the active variant has the Ranked Off tag to override the material with a default non-ranked version
  (bind _index (Class|FortCosmeticFunctionLibrary|%28Deprecated%29GetVariantChannelActiveIndexForSlot "(TagName=\"LoadoutSlot.Pickaxe\")" (CallFunction|GetCosmeticData) "(TagName=\"Cosmetics.Variant.Channel.Material\")" (Utilities|Array|MakeArray "(TagName=\"Cosmetics.Variant.Property.Stage1\")" "(TagName=\"Cosmetics.Variant.Property.Stage2\")")))
  (if (== _index 1)
    (Effects|Components|ParticleSystem|SetBoolParameter _returnvalue "UseRankedColors" false)
    (Rendering|Material|SetMaterial (Variables|Visual|GetWeaponMesh) 0 (Variables|Rank|RankMaterials|GetDefault%20Material))
    (else
      (Effects|Components|ParticleSystem|SetBoolParameter _returnvalue "UseRankedColors" true)
      (Effects|Components|ParticleSystem|SetIntParameter _returnvalue "RankedProgression" (Variables|Default|GetVFX_Rank)))))

(event EventOnInitWeaponCosmetics
  (bind _returnvalue (CallFunction|GetIdleVFXComponent))
  (|CallParentFunction)
  (if (Reactive|Frontend|IsCosmeticObjectDisplayedinFrontendContext 0)
    ;[1 1 1] Check if the active variant has the Ranked Off tag to override the material with a default non-ranked version
    (bind _index (Class|FortCosmeticFunctionLibrary|%28Deprecated%29GetVariantChannelActiveIndexForSlot "(TagName=\"LoadoutSlot.Pickaxe\")" (CallFunction|GetCosmeticData) "(TagName=\"Cosmetics.Variant.Channel.Material\")" (Utilities|Array|MakeArray "(TagName=\"Cosmetics.Variant.Property.Stage1\")" "(TagName=\"Cosmetics.Variant.Property.Stage2\")")))
    (if (== _index 1)
      (Effects|Components|ParticleSystem|SetBoolParameter _returnvalue "UseRankedColors" false)
      (Rendering|Material|SetMaterial (Variables|Visual|GetWeaponMesh) 0 (Variables|Rank|RankMaterials|GetDefault%20Material))
      (else
        (Effects|Components|ParticleSystem|SetBoolParameter _returnvalue "UseRankedColors" true)
        (Effects|Components|ParticleSystem|SetIntParameter _returnvalue "RankedProgression" (Variables|Default|GetVFX_Rank))))))

(fn ConstructionScript
  (|CallParentFunction))
