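; Ranked Tandem pickaxe Blueprint that pushes the weapon's rank into its visual effects.
; Backpack material setup sets the idle VFX "RankedProgression" parameter from the passed index and applies the default material to weapon mesh slot 0.
; Left and right swings and impact effects set the same parameter from the VFX_Rank variable on swing and spawned impact components.
; Each event calls its parent first, so the base rank logic lives in the parent class.
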(event Default|EventBackpackMaterialSetup
  (params (Index :type Integer))
  (|CallParentFunction Index)
  (Effects|Components|ParticleSystem|SetIntParameter (CallFunction|GetIdleVFXComponent) "RankedProgression" Index)
  (Rendering|Material|SetMaterial (Variables|Visual|GetWeaponMesh) 0 (Variables|Rank|RankMaterials|GetDefault%20Material)))

(event EventSwingLeft
  (|CallParentFunction)
  (Effects|Components|ParticleSystem|SetIntParameter (CallFunction|GetSwingVFXComponent) "RankedProgression" (Variables|Default|GetVFX_Rank)))

(event EventSwingRight
  (|CallParentFunction)
  (Effects|Components|ParticleSystem|SetIntParameter (CallFunction|GetSwingVFXComponent) "RankedProgression" (Variables|Default|GetVFX_Rank)))

(event FortWeapon|EventOnPlayImpactFX
  (params (HitResult :type Hit%20Result%20Structure%20%28by%20ref%29) (ImpactPhysicalSurface :type EPhysicalSurface%20Enum :default "SurfaceType_Default") (SpawnedPSC :type FXSystem%20Component%20Object%20Reference))
  (|CallParentFunction HitResult ImpactPhysicalSurface SpawnedPSC)
  (Effects|Components|ParticleSystem|SetIntParameter SpawnedPSC "RankedProgression" (Variables|Default|GetVFX_Rank)))

(fn ConstructionScript
  (|CallParentFunction))
