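; Cosmetic Blueprint for the Red Pepper pickaxe that animates a reveal effect when the weapon is equipped.
; On equip, it resets Reveal to zero, waits briefly, then plays the RedPepperEquip timeline to drive the value.
; UpdateRevealVisuals writes that scalar to the weapon mesh materials and, if valid, the idle VFX component.
; Cosmetic init also zeroes the reveal when displayed in a frontend context, before calling the parent.
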(event EventOnInitWeaponCosmetics
  (Utilities|FlowControl|Sequence
    (:then_0
      (if (Reactive|Frontend|IsCosmeticObjectDisplayedinFrontendContext self)
        (CallFunction|UpdateRevealVisuals 0.0)))
    (:then_1
      (|CallParentFunction))))

(event FortWeapon|EventOnEquippedAndAttached
  (|CallParentFunction)
  (CallFunction|UpdateRevealVisuals 0.0)
  (Utilities|FlowControl|Delay 0.22)
  (bind (_direction _reveal) (Timeline|RedPepperEquip PlayFromStart)
    (:Update
      (CallFunction|UpdateRevealVisuals _reveal))))

(event EventBeginPlay)

(event EventTick
  (params (DeltaSeconds :type Float%20%28single-precision%29)))

(fn UpdateRevealVisuals
  (params (ParameterValue :type Float%20%28single-precision%29))
  (bind _returnvalue (CallFunction|GetIdleVFXComponent))
  (Rendering|Material|SetScalarParameterValueonMaterials (Variables|Visual|GetWeaponMesh) "Reveal" ParameterValue)
  (Utilities|IsValid _returnvalue
    (:Is%20Valid
      (Effects|Components|ParticleSystem|SetFloatParameter _returnvalue "Reveal" ParameterValue))))

(fn ConstructionScript
  (|CallParentFunction))
