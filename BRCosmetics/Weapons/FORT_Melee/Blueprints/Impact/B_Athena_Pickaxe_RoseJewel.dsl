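; RoseJewel pickaxe cosmetic Blueprint that drives swing and impact-smoke effects.
; On equip, it feeds the owning pawn's mesh to the swing Niagara component's "Pawn" parameter.
; Impacts activate the idle VFX, enable its smoke parameter and start a timer to disable it; left or right swings cancel that timer.
; Cosmetic init and becoming visible also clear the smoke, and construction disables auto-activation on the idle particle component.
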(event FortWeapon|EventOnEquippedAndAttached
  (|CallParentFunction)
  (bind _asfort_pawn (Utilities|Casting|CastToFortPawn (Game|GetInstigator)))
  (Utilities|IsValid (Variables|/Script/Engine.Character|GetMesh _asfort_pawn)
    (:Is%20Valid
      (bind _asniagara_particle_system_component (Utilities|Casting|CastToNiagaraComponent (CallFunction|GetSwingVFXComponent))
        (:then
          (Niagara|SetNiagaraSkeletalMeshComponent _asniagara_particle_system_component "Pawn" (Variables|/Script/Engine.Character|GetMesh _asfort_pawn)))
        (:CastFailed
          (Development|PrintString "CAST FAILED - Niagara Component"))))))

(event FortWeapon|EventOnPlayImpactFX
  (params (HitResult :type Hit%20Result%20Structure%20%28by%20ref%29) (ImpactPhysicalSurface :type EPhysicalSurface%20Enum :default "SurfaceType_Default") (SpawnedPSC :type FXSystem%20Component%20Object%20Reference))
  (|CallParentFunction HitResult ImpactPhysicalSurface SpawnedPSC)
  (Components|Activation|Activate (CallFunction|GetIdleVFXComponent) true)
  (Effects|Components|ParticleSystem|SetBoolParameter (CallFunction|GetIdleVFXComponent) "bimpactSmoke" true)
  (bind _returnvalue (Utilities|Time|SetTimerbyEvent (delegate Custom|DeactivateSmoke) 5.0))
  (Variables|Default|SetSmokeTimer _returnvalue))

(event Custom|DeactivateSmoke
  (Effects|Components|ParticleSystem|SetBoolParameter (CallFunction|GetIdleVFXComponent) "bimpactSmoke" false))

(event Custom|ResetIdle
  (Components|Activation|Activate (CallFunction|GetIdleVFXComponent) true))

(event EventOnInitWeaponCosmetics
  (|CallParentFunction)
  (CallFunction|DeactivateSmoke))

(event EventOnWeaponVisibilityChanged
  (params (bVisible :type Boolean) (bSetForLocalControllerOnly :type Boolean))
  (|CallParentFunction bVisible bSetForLocalControllerOnly)
  (if bVisible
    (CallFunction|DeactivateSmoke)))

(event EventSwingLeft
  (bind _smoketimer (Variables|Default|GetSmokeTimer))
  (|CallParentFunction)
  (if (Utilities|Time|IsValidTimerHandle _smoketimer)
    (Utilities|Time|ClearandInvalidateTimerbyHandle _smoketimer)))

(event EventSwingRight
  (bind _smoketimer (Variables|Default|GetSmokeTimer))
  (|CallParentFunction)
  (if (Utilities|Time|IsValidTimerHandle _smoketimer)
    (Utilities|Time|ClearandInvalidateTimerbyHandle _smoketimer)))

(fn ConstructionScript
  (|CallParentFunction)
  (Components|Activation|SetAutoActivate (CallFunction|GetIdlePSC) false))
