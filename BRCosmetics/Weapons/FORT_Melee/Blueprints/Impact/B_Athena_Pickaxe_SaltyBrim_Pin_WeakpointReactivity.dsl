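; Salty Brim Pin pickaxe weapon Blueprint that makes its impact effect react to weakpoint hits.
; On impact it stores the spawned particle component, then sets its "isWeakpoint" parameter when the hit surface is in a configured list or the cosmetic is in preview or frontend.
; In-game hits also route through ServerWeakpointEvent to OnWeakpointEvent, which sets the same flag.
; Construction disables idle particle auto-activation; OnWeakPointHit is empty.
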(event FortWeapon|EventOnPlayImpactFX
  (params (HitResult :type Hit%20Result%20Structure%20%28by%20ref%29) (ImpactPhysicalSurface :type EPhysicalSurface%20Enum :default "SurfaceType_Default") (SpawnedPSC :type FXSystem%20Component%20Object%20Reference))
  (|CallParentFunction HitResult ImpactPhysicalSurface SpawnedPSC)
  (Variables|Default|SetImpact%20FX SpawnedPSC)
  ; Frontend Weak Point
  (if (or (Reactive|Frontend|IsCosmeticObjectDisplayedasPreview 0) (Reactive|Frontend|IsCosmeticObjectDisplayedinFrontendContext 0))
    (CallFunction|OnWeakpointEvent)
    (elif (Utilities|Array|ContainsItem (Variables|Default|GetImpactPhysicalSurface) ImpactPhysicalSurface)
      (Effects|Components|ParticleSystem|SetBoolParameter (Variables|Default|GetImpact%20FX) "isWeakpoint" true)
      (CallFunction|ServerWeakpointEvent))))

(event Custom|ServerWeakpointEvent
  (CallFunction|OnWeakpointEvent))

(event Custom|OnWeakpointEvent
  (Effects|Components|ParticleSystem|SetBoolParameter (Variables|Default|GetImpact%20FX) "isWeakpoint" true))

(fn OnWeakPointHit)

(fn ConstructionScript
  (|CallParentFunction)
  (Components|Activation|SetAutoActivate (CallFunction|GetIdlePSC) false))
