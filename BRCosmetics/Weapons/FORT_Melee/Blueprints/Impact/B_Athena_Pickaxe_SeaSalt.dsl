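; SeaSalt pickaxe cosmetic weapon Blueprint that adapts swing effects to the selected mesh variant.
; Weapon-cosmetic init, instigator set and equip events cache the loadout's active Stage1/Stage2 variant index, defaulting to 0.
; The final left swing sets a flag and, for variant index 1, attaches a punch Niagara system to the weapon mesh.
; A helper applies LoopDuration and Lifetime to main and offhand Niagara trails.
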(event EventOnInitWeaponCosmetics
  (|CallParentFunction)
  (bind _asfort_player_pawn (Utilities|Casting|CastToFortPlayerPawn (Game|GetInstigator)))
  (bind (_index _bindexisvalid) (CosmeticLoadout|GetVariantChannelActiveIndexforSlotActor "(TagName=\"LoadoutSlot.Pickaxe\")" _asfort_player_pawn "(TagName=\"Cosmetics.Variant.Channel.Mesh\")" (Utilities|Array|MakeArray "(TagName=\"Cosmetics.Variant.Property.Stage1\")" "(TagName=\"Cosmetics.Variant.Property.Stage2\")")))
  (Variables|Default|SetActiveVariantIndex (Math|Integer|SelectInt _index 0 _bindexisvalid)))

(event FortWeapon|EventOnInstigatorSet
  (|CallParentFunction)
  (bind _asfort_player_pawn (Utilities|Casting|CastToFortPlayerPawn (Game|GetInstigator)))
  (bind (_index _bindexisvalid) (CosmeticLoadout|GetVariantChannelActiveIndexforSlotActor "(TagName=\"LoadoutSlot.Pickaxe\")" _asfort_player_pawn "(TagName=\"Cosmetics.Variant.Channel.Mesh\")" (Utilities|Array|MakeArray "(TagName=\"Cosmetics.Variant.Property.Stage1\")" "(TagName=\"Cosmetics.Variant.Property.Stage2\")")))
  (Variables|Default|SetActiveVariantIndex (Math|Integer|SelectInt _index 0 _bindexisvalid)))

(event FortWeapon|EventOnEquippedAndAttached
  (|CallParentFunction)
  (bind _asfort_player_pawn (Utilities|Casting|CastToFortPlayerPawn (Game|GetInstigator)))
  (bind (_index _bindexisvalid) (CosmeticLoadout|GetVariantChannelActiveIndexforSlotActor "(TagName=\"LoadoutSlot.Pickaxe\")" _asfort_player_pawn "(TagName=\"Cosmetics.Variant.Channel.Mesh\")" (Utilities|Array|MakeArray "(TagName=\"Cosmetics.Variant.Property.Stage1\")" "(TagName=\"Cosmetics.Variant.Property.Stage2\")")))
  (Variables|Default|SetActiveVariantIndex (Math|Integer|SelectInt _index 0 _bindexisvalid)))

(event EventSwingLeft2
  ;[1 1 1] Final Swing
  (|CallParentFunction)
  (Variables|Default|SetLeft2%20Swing? true)
  (if (== (Variables|Default|GetActiveVariantIndex) 1)
    (Niagara|SpawnSystemAttached "/BRCosmetics/Athena/NoHome/NS_SSA_Punch_R.NS_SSA_Punch_R" (Variables|Visual|GetWeaponMesh) "None" "0, 0, 0" "0, 0, 0" "KeepRelativeOffset" true)))

(fn GetFXParamNames
  (return "LoopDuration"))

(fn SetParamsOnTrails
  (params (LoopDuration :type Float%20%28single-precision%29) (Lifetime :type Float%20%28single-precision%29))
  (for _array_element (Utilities|Array|MakeArray (CallFunction|GetAnimTrailsPSC) (CallFunction|GetAnimTrailsOffhandPSC))
    (bind _asniagara_particle_system_component (Utilities|Casting|CastToNiagaraComponent _array_element))
    (Niagara|SetNiagaraVariable(Float) _asniagara_particle_system_component "LoopDuration" LoopDuration)
    (Niagara|SetNiagaraVariable(Float) _asniagara_particle_system_component "Lifetime" Lifetime)))

(fn ConstructionScript
  (|CallParentFunction))
