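; Cosmetic Blueprint for the SharpFang pickaxe, whose gem glow builds through five levels with weak-point hits, or with impacts in the frontend.
; Impact FX route through server and client events, raising the idle Niagara level and, on one particle variant, spawning a flash.
; The first non-weak-point hit registers an elimination-stat watcher that maxes the glow.
; A retriggerable delay starts a timeline that steps the glow down; another variant skips the effect.
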(event FortWeapon|EventOnPlayImpactFX
  (params (HitResult :type Hit%20Result%20Structure%20%28by%20ref%29) (ImpactPhysicalSurface :type EPhysicalSurface%20Enum :default "SurfaceType_Default") (SpawnedPSC :type FXSystem%20Component%20Object%20Reference))
  (|CallParentFunction HitResult ImpactPhysicalSurface SpawnedPSC)
  (if (Reactive|Frontend|IsCosmeticObjectDisplayedinFrontendContext 0)
    (if (not (== (Niagara|GetNiagaraSystemAsset (Utilities|Casting|CastToNiagaraComponent (CallFunction|GetIdleVFXComponent))) "/BRCosmetics/Effects/Niagara/Systems/Blanks/NS_Blank_Pickaxe_Idle_MainHand.NS_Blank_Pickaxe_Idle_MainHand"))
      (CallFunction|GemReactivity))
    (else
      (bind _index (Class|FortCosmeticFunctionLibrary|%28Deprecated%29GetVariantChannelActiveIndexForSlot "(TagName=\"LoadoutSlot.Pickaxe\")" (CallFunction|GetCosmeticData) "(TagName=\"Cosmetics.Variant.Channel.Particle\")" (Utilities|Array|MakeArray "(TagName=\"Cosmetics.Variant.Property.Particle1\")" "(TagName=\"Cosmetics.Variant.Property.Particle2\")" "(TagName=\"Cosmetics.Variant.Property.Particle3\")")))
      (bind _output_get (Variables|Default|SetCosmetic%20Loadout%20Int _index))
      (if (not (== _output_get 2))
        ;[1 1 1] WeakPoint impact
        (bind _output_get_2 (Variables|Default|SetCurrent%20Impact%20Physical%20Surface ImpactPhysicalSurface))
        (bind _output_get_3 (Variables|Default|SetIs%20hit%20weak%20point (Utilities|Array|ContainsItem (Variables|Default|GetWeakPoint%20Physical%20Surface) _output_get_2)))
        (Effects|Components|ParticleSystem|SetBoolParameter (Variables|Default|GetImpact%20FX) "Is WeakPoint" _output_get_3)
        (CallFunction|CEFXServer ImpactPhysicalSurface)))))

(event Custom|CE_FX_Server
  (params (Impact%20Physical%20Surface :type EPhysicalSurface%20Enum :default "SurfaceType_Default"))
  (CallFunction|CEFXClient Impact%20Physical%20Surface))

(event Custom|GemReactivity
  ;[1 1 1] Gem Glow Logic
  (Math|Integer|IncrementInt (Variables|Default|Getins)
    (:%20%20
      (bind _output_get (Variables|Default|Setins (Variables|Default|Getins)))
      (Development|PrintString (Utilities|String|ToString(Integer) (Math|Integer|Clamp(Integer) _output_get 0 5)))
      (bind _index (Class|FortCosmeticFunctionLibrary|%28Deprecated%29GetVariantChannelActiveIndexForSlot "(TagName=\"LoadoutSlot.Pickaxe\")" (CallFunction|GetCosmeticData) "(TagName=\"Cosmetics.Variant.Channel.Particle\")" (Utilities|Array|MakeArray "(TagName=\"Cosmetics.Variant.Property.Particle1\")" "(TagName=\"Cosmetics.Variant.Property.Particle2\")" "(TagName=\"Cosmetics.Variant.Property.Particle3\")")))
      (bind _output_get_2 (Variables|Default|SetCosmetic%20Loadout%20Int _index))
      (if (not (== _output_get_2 2))
        (Effects|Components|ParticleSystem|SetIntParameter (CallFunction|GetIdleVFXComponent) "WeakPoint int" (Math|Integer|Clamp(Integer) _output_get 0 5))
        (if (not (> _output_get 5))
          (bind _index_2 (Class|FortCosmeticFunctionLibrary|%28Deprecated%29GetVariantChannelActiveIndexForSlot "(TagName=\"LoadoutSlot.Pickaxe\")" (CallFunction|GetCosmeticData) "(TagName=\"Cosmetics.Variant.Channel.Particle\")" (Utilities|Array|MakeArray "(TagName=\"Cosmetics.Variant.Property.Particle1\")" "(TagName=\"Cosmetics.Variant.Property.Particle2\")" "(TagName=\"Cosmetics.Variant.Property.Particle3\")")))
          (bind _output_get_3 (Variables|Default|SetCosmetic%20Loadout%20Int _index_2))
          (if (not (!= _output_get_3 0))
            (bind _returnvalue (Effects|SpawnFXSystemAttached "/BRCosmetics/Weapons/FORT_Melee/Pickaxe_SharpFang/FX/NS_Pickaxe_SharpFang_Flash.NS_Pickaxe_SharpFang_Flash" (Variables|Visual|GetWeaponMesh) "b_AR_Root"))
            (Effects|Components|ParticleSystem|SetIntParameter _returnvalue "WeakPoint int" (Math|Integer|Clamp(Integer) _output_get 0 5))))
        (Utilities|FlowControl|RetriggerableDelay 6.0)
        (Components|Timeline|Play (Variables|B_Athena_Pickaxe_SharpFang|GetTimeline_0))
        (Timeline|Timeline_0 SetNewTime :NewTime (Utilities|Select :Index (Variables|Default|Getins) :Option%200 0.73 :Option%201 0.58 :Option%202 0.43 :Option%203 0.28 :Option%204 0.13 :Option%205 0.0)
          (:NewTrack_0
            (bind _output_get_4 (Variables|Default|Setins 5))
            (Effects|Components|ParticleSystem|SetIntParameter (CallFunction|GetIdleVFXComponent) "WeakPoint int" _output_get_4))
          (:NewTrack_1
            (bind _output_get_5 (Variables|Default|Setins 4))
            (Effects|Components|ParticleSystem|SetIntParameter (CallFunction|GetIdleVFXComponent) "WeakPoint int" _output_get_5))
          (:NewTrack_2
            (bind _output_get_6 (Variables|Default|Setins 3))
            (Effects|Components|ParticleSystem|SetIntParameter (CallFunction|GetIdleVFXComponent) "WeakPoint int" _output_get_6))
          (:NewTrack_3
            (bind _output_get_7 (Variables|Default|Setins 2))
            (Effects|Components|ParticleSystem|SetIntParameter (CallFunction|GetIdleVFXComponent) "WeakPoint int" _output_get_7))
          (:NewTrack_4
            (bind _output_get_8 (Variables|Default|Setins 1))
            (Effects|Components|ParticleSystem|SetIntParameter (CallFunction|GetIdleVFXComponent) "WeakPoint int" _output_get_8))
          (:NewTrack_5
            (bind _output_get_9 (Variables|Default|Setins 0))
            (Effects|Components|ParticleSystem|SetIntParameter (CallFunction|GetIdleVFXComponent) "WeakPoint int" _output_get_9)))))))

(event Custom|CE_FX_Client
  (params (Impact%20Physical%20Surface :type EPhysicalSurface%20Enum :default "SurfaceType_Default"))
  (bind _output_get (Variables|Default|SetCurrent%20Impact%20Physical%20Surface Impact%20Physical%20Surface))
  (Variables|Default|SetIs%20hit%20weak%20point (Utilities|Array|ContainsItem (Variables|Default|GetWeakPoint%20Physical%20Surface) _output_get))
  (if (Variables|Default|GetIs%20hit%20weak%20point)
    (CallFunction|GemReactivity)
    (else
      ;[1 1 1] Elimination
      (if true
        (bind _asfort_player_pawn (Utilities|Casting|CastToFortPlayerPawn (Variables|Actor|GetInstigator)))
        (Utilities|FlowControl|DoOnce false
          (:Completed
            (bind (_statname _statvalue) (Cosmetics|WatchCosmeticStat _asfort_player_pawn "(TagName=\"StatManager.AthenaKills\")")
              (:OnStatChanged
                (if (> _statvalue 0)
                  (bind _output_get_2 (Variables|Default|Setins 5))
                  (Development|PrintString (Utilities|String|ToString(Integer) _output_get_2))
                  ;[1 1 1] Gem Glow Logic
                  (bind _index (Class|FortCosmeticFunctionLibrary|%28Deprecated%29GetVariantChannelActiveIndexForSlot "(TagName=\"LoadoutSlot.Pickaxe\")" (CallFunction|GetCosmeticData) "(TagName=\"Cosmetics.Variant.Channel.Particle\")" (Utilities|Array|MakeArray "(TagName=\"Cosmetics.Variant.Property.Particle1\")" "(TagName=\"Cosmetics.Variant.Property.Particle2\")" "(TagName=\"Cosmetics.Variant.Property.Particle3\")")))
                  (bind _output_get_3 (Variables|Default|SetCosmetic%20Loadout%20Int _index))
                  (if (not (== _output_get_3 2))
                    (Effects|Components|ParticleSystem|SetIntParameter (CallFunction|GetIdleVFXComponent) "WeakPoint int" (Math|Integer|Clamp(Integer) (stale Custom|GemReactivity _output_get) 0 5))
                    (if (not (> (stale Custom|GemReactivity _output_get) 5))
                      (bind _index_2 (Class|FortCosmeticFunctionLibrary|%28Deprecated%29GetVariantChannelActiveIndexForSlot "(TagName=\"LoadoutSlot.Pickaxe\")" (CallFunction|GetCosmeticData) "(TagName=\"Cosmetics.Variant.Channel.Particle\")" (Utilities|Array|MakeArray "(TagName=\"Cosmetics.Variant.Property.Particle1\")" "(TagName=\"Cosmetics.Variant.Property.Particle2\")" "(TagName=\"Cosmetics.Variant.Property.Particle3\")")))
                      (bind _output_get_4 (Variables|Default|SetCosmetic%20Loadout%20Int _index_2))
                      (if (not (!= _output_get_4 0))
                        (bind _returnvalue (Effects|SpawnFXSystemAttached "/BRCosmetics/Weapons/FORT_Melee/Pickaxe_SharpFang/FX/NS_Pickaxe_SharpFang_Flash.NS_Pickaxe_SharpFang_Flash" (Variables|Visual|GetWeaponMesh) "b_AR_Root"))
                        (Effects|Components|ParticleSystem|SetIntParameter _returnvalue "WeakPoint int" (Math|Integer|Clamp(Integer) (stale Custom|GemReactivity _output_get) 0 5))))
                    (Utilities|FlowControl|RetriggerableDelay 6.0)
                    (Components|Timeline|Play (Variables|B_Athena_Pickaxe_SharpFang|GetTimeline_0))
                    (Timeline|Timeline_0 SetNewTime)))))))))))

(fn OnRep_CurrentImpactPhysicalSurface)

(fn ConstructionScript
  (|CallParentFunction))
