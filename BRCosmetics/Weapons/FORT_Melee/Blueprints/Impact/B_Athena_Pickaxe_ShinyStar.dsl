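; Cosmetic melee pickaxe Blueprint (ShinyStar) whose weapon material reacts to the player's eliminations.
; On equip, it binds the swing Niagara effect to the pawn's skeletal mesh and watches the AthenaKills stat to play a timeline driving the "ReactivityBoost" material parameter.
; Impact FX plays a separate reactivity timeline in frontend display, otherwise registers the stat watch again.
; The construction script disables auto-activation of the idle particle effect.
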(event FortWeapon|EventOnEquippedAndAttached
  (|CallParentFunction)
  (bind _asfort_pawn (Utilities|Casting|CastToFortPawn (Game|GetInstigator)))
  (Utilities|IsValid (Variables|/Script/Engine.Character|GetMesh _asfort_pawn)
    (:Is%20Valid
      (Utilities|FlowControl|Sequence
        (:then_0
          (bind _asniagara_particle_system_component (Utilities|Casting|CastToNiagaraComponent (CallFunction|GetSwingVFXComponent))
            (:then
              (Niagara|SetNiagaraSkeletalMeshComponent _asniagara_particle_system_component "Pawn" (Variables|/Script/Engine.Character|GetMesh _asfort_pawn)))
            (:CastFailed
              (Development|PrintString "CAST FAILED - Niagara Component"))))
        (:then_1
          (bind _asfort_player_pawn (Utilities|Casting|CastToFortPlayerPawn (Variables|Actor|GetInstigator)))
          (Cosmetics|WatchCosmeticStat _asfort_player_pawn "(TagName=\"StatManager.AthenaKills\")" 0 false
            (:OnStatChanged
              (bind (_direction _starsspeedboost) (Timeline|Ingame%20reactivity PlayFromStart)
                (:Update
                  (Rendering|Material|SetScalarParameterValueonMaterials (Variables|Visual|GetWeaponMesh) "ReactivityBoost" _starsspeedboost))))))) ; FORT-556265: Check stat for reactivity when equip Pickaxe
    )))

(event FortWeapon|EventOnPlayImpactFX
  (params (HitResult :type Hit%20Result%20Structure%20%28by%20ref%29) (ImpactPhysicalSurface :type EPhysicalSurface%20Enum :default "SurfaceType_Default") (SpawnedPSC :type FXSystem%20Component%20Object%20Reference))
  (|CallParentFunction HitResult ImpactPhysicalSurface SpawnedPSC)
  (if (Reactive|Frontend|IsCosmeticObjectDisplayedinFrontendContext 0)
    ;[0 0 0] FE cycling
    (bind (_direction _reactivity) (Timeline|FE%20reactivity PlayFromStart)
      (:Update
        (Rendering|Material|SetScalarParameterValueonMaterials (Variables|Visual|GetWeaponMesh) "ReactivityBoost" _reactivity)))
    (else
      (bind _asfort_player_pawn (Utilities|Casting|CastToFortPlayerPawn (Variables|Actor|GetInstigator)))
      (Cosmetics|WatchCosmeticStat _asfort_player_pawn "(TagName=\"StatManager.AthenaKills\")" 0 false
        (:OnStatChanged
          (Timeline|Ingame%20reactivity PlayFromStart))))))

(fn ConstructionScript
  (|CallParentFunction)
  (Components|Activation|SetAutoActivate (CallFunction|GetIdlePSC) false))
