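; Cosmetic weapon Blueprint for the SilentTempo pickaxe that orients its impact particle effect to match the current melee swing.
; On impact, it looks up the instigator pawn's current montage section in a section list and uses that index to pick one of four swing vectors for the effect's "Direction" parameter.
; It also deactivates the swing particle component on cosmetic init and sets left/right swing flags that are not read in this graph.
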(event FortWeapon|EventOnPlayImpactFX
  (params (HitResult :type Hit%20Result%20Structure%20%28by%20ref%29) (ImpactPhysicalSurface :type EPhysicalSurface%20Enum :default "SurfaceType_Default") (SpawnedPSC :type FXSystem%20Component%20Object%20Reference))
  (|CallParentFunction HitResult ImpactPhysicalSurface SpawnedPSC)
  (bind _asfort_pawn (Utilities|Casting|CastToFortPawn (Game|GetInstigator)))
  (Variables|Default|SetImpact%20Index (Utilities|Array|FindItem (Variables|Default|GetMontageSections) (Animation|Montage|MontageGetCurrentSection (Components|SkeletalMesh|GetAnimInstance (Variables|/Script/Engine.Character|GetMesh _asfort_pawn)) (Animation|Montage|GetCurrentActiveMontage (Components|SkeletalMesh|GetAnimInstance (Variables|/Script/Engine.Character|GetMesh _asfort_pawn))))))
  (Effects|Components|ParticleSystem|SetVectorParameter (Variables|Default|GetImpact%20FX) "Direction" (Utilities|Select :Index (Variables|Default|GetImpact%20Index) :Option%200 (Variables|Default|GetSwing%201%20Dir) :Option%201 (Variables|Default|GetSwing%202%20Dir) :Option%202 (Variables|Default|GetSwing%203%20Dir) :Option%203 (Variables|Default|GetSwing%204%20Dir))))

(event EventOnInitWeaponCosmetics
  (|CallParentFunction)
  (Components|Activation|Deactivate (CallFunction|GetSwingPSC)))

(event EventSwingLeft_Common
  (|CallParentFunction)
  (Variables|Default|SetRight? false))

(event EventSwingRight_Common
  (|CallParentFunction)
  (Variables|Default|SetRight? true))

(event EventSwingLeft
  (|CallParentFunction)
  (Variables|Default|SetLeft2%20Swing? false))

(event EventSwingRight
  (|CallParentFunction)
  (Variables|Default|SetLeft2%20Swing? false))

(fn ConstructionScript
  (|CallParentFunction))
