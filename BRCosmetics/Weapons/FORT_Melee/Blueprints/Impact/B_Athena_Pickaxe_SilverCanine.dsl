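; Silver Canine pickaxe Blueprint that adds cosmetic swing effects on top of its parent melee weapon logic.
; Both swing events set the Right flag and check whether the instigator's active montage section is "Swing2".
; When cosmetic events can run, that section spawns a jab FX on the weapon handle and hides the regular swing VFX, which other sections show.
; A helper function writes a "SwingIdx" integer audio parameter.
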(event EventSwingLeft
  (bind _returnvalue (Game|GetInstigator))
  (bind _asfort_pawn (Utilities|Casting|CastToFortPawn _returnvalue))
  (bind _mesh (Variables|/Script/Engine.Character|GetMesh _asfort_pawn))
  (bind _returnvalue_2 (Components|SkeletalMesh|GetAnimInstance _mesh))
  (bind _returnvalue_3 (Animation|Montage|GetCurrentActiveMontage _returnvalue_2))
  (bind _returnvalue_4 (Animation|Montage|MontageGetCurrentSection _returnvalue_2 _returnvalue_3))
  (bind _returnvalue_5 (== _returnvalue_4 "Swing2"))
  (|CallParentFunction)
  (Variables|Default|SetRight%20? false)
  (Networking|CanExecuteCosmeticEvents
    (:True
      (Utilities|FlowControl|Sequence
        (:then_0
          (if _returnvalue_5
            (Effects|SpawnFXSystemAttached (Variables|Default|GetFXSystemJab) (Variables|Visual|GetWeaponMesh) "Melee_TwoHanded_Handle")))
        (:then_1
          (Development|SetHiddeninGame (CallFunction|GetSwingVFXComponent) _returnvalue_5 false))))))

(event EventSwingRight
  (bind _returnvalue (Game|GetInstigator))
  (bind _asfort_pawn (Utilities|Casting|CastToFortPawn _returnvalue))
  (bind _mesh (Variables|/Script/Engine.Character|GetMesh _asfort_pawn))
  (bind _returnvalue_2 (Components|SkeletalMesh|GetAnimInstance _mesh))
  (bind _returnvalue_3 (Animation|Montage|GetCurrentActiveMontage _returnvalue_2))
  (bind _returnvalue_4 (Animation|Montage|MontageGetCurrentSection _returnvalue_2 _returnvalue_3))
  (bind _returnvalue_5 (== _returnvalue_4 "Swing2"))
  (|CallParentFunction)
  (Variables|Default|SetRight%20? true)
  (Networking|CanExecuteCosmeticEvents
    (:True
      (Utilities|FlowControl|Sequence
        (:then_0
          (if _returnvalue_5
            (Effects|SpawnFXSystemAttached (Variables|Default|GetFXSystemJab) (Variables|Visual|GetWeaponMesh) "Melee_TwoHanded_Handle")))
        (:then_1
          (Development|SetHiddeninGame (CallFunction|GetSwingVFXComponent) _returnvalue_5 false))))))

(fn SetSwingAudioParam
  (params (InInt :type Integer))
  (bind _returnvalue (AudioParameter|GetAudioParameterComponent))
  (Utilities|IsValid _returnvalue
    (:Is%20Valid
      (Audio|Parameter|SetIntegerParameter _returnvalue "SwingIdx" InInt :!class /Script/AudioExtensions.AudioParameterControllerInterface))))

(fn ConstructionScript
  (|CallParentFunction))
