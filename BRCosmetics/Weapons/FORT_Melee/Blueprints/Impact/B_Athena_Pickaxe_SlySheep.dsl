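; SlySheep dual-wield pickaxe Blueprint that drives its cosmetic Niagara effects, reusing SunbeamQuest equip and trail systems.
; On equip, it spawns trails on main and offhand meshes after a short retriggerable delay, unless player 0 moves fast, falls or jumps.
; Swing events record which swing is active, and SwingLeft2 also spawns trails on both meshes.
; Impact FX reads those flags to set SecondImpact and orient the effect from the hit normal.
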(event EventOnInitWeaponCosmetics
  (bind _returnvalue (CallFunction|GetIdleOffhandVFXComponent))
  ;[1 1 1] Cosmetic Init
  (Utilities|IsValid _returnvalue
    (:Is%20Valid
      (bind _asniagara_particle_system_component (Utilities|Casting|CastToNiagaraComponent _returnvalue))
      (Niagara|SetNiagaraVariable(Bool) _asniagara_particle_system_component "bIsOffhand" true))))

(event FortWeapon|EventOnEquippedAndAttached
  (|CallParentFunction)
  ;[1 1 1] Equip trail FX
  (bind _asplayer_pawn_athena (Utilities|Casting|CastToPlayerPawn_Athena (Game|GetPlayerCharacter 0)))
  (if (> (Variables|/Script/FortniteGame.FortPlayerPawnAthena|GetCurrentPawnSquaredSpeed _asplayer_pawn_athena) 8.1e+05)
    _
    (elif (or (AI|Components|NavMovement|IsFalling (Variables|/Script/Engine.Character|GetCharacterMovement _asplayer_pawn_athena)) (Variables|/Script/Engine.Character|GetbWasJumping _asplayer_pawn_athena))
      _
      (else
        (Utilities|FlowControl|RetriggerableDelay 0.09)
        (bind _returnvalue (Niagara|SpawnSystemAttached "/BRCosmetics/Weapons/FORT_Melee/Pickaxe_SunbeamQuest/FX/NS_Pickaxe_SunBeamQuest_Equip.NS_Pickaxe_SunBeamQuest_Equip" (Variables|Visual|GetWeaponMesh) "None" "0, 0, 0" "0, 0, 0" "KeepRelativeOffset" true false))
        (Niagara|SetNiagaraVariable(Float) _returnvalue "AnimTrailWidth" 1.7)
        (Niagara|SetNiagaraVariable(Bool) _returnvalue "bIsOffhand" false)
        (Components|Activation|SetActive _returnvalue true true)
        (bind _returnvalue_2 (Niagara|SpawnSystemAttached "/BRCosmetics/Weapons/FORT_Melee/Pickaxe_SunbeamQuest/FX/NS_Pickaxe_SunBeamQuest_Equip.NS_Pickaxe_SunBeamQuest_Equip" (Variables|Visual|GetWeaponMeshOffhand) "None" "0, 0, 0" "0, 0, 0" "KeepRelativeOffset" true false))
        (Niagara|SetNiagaraVariable(Float) _returnvalue_2 "AnimTrailWidth" 1.7)
        (Niagara|SetNiagaraVariable(Bool) _returnvalue_2 "bIsOffhand" true)
        (Components|Activation|SetActive _returnvalue_2 true true)))))

(event EventSwingLeft2
  (|CallParentFunction)
  (Variables|Default|SetLeft2%20Swing? true)
  ;[1 1 1] Final Swing
  (bind _returnvalue (Niagara|SpawnSystemAttached "/BRCosmetics/Weapons/FORT_Melee/Pickaxe_SunbeamQuest/FX/NS_Pickaxe_SunBeamQuest_AnimTrail.NS_Pickaxe_SunBeamQuest_AnimTrail" (Variables|Visual|GetWeaponMesh) "None" "0, 0, 0" "0, 0, 0" "KeepRelativeOffset" true false))
  (Niagara|SetNiagaraVariable(Float) _returnvalue "LoopDuration" 0.25)
  (Components|Activation|SetActive _returnvalue true true)
  (bind _returnvalue_2 (Niagara|SpawnSystemAttached "/BRCosmetics/Weapons/FORT_Melee/Pickaxe_SunbeamQuest/FX/NS_Pickaxe_SunBeamQuest_AnimTrail.NS_Pickaxe_SunBeamQuest_AnimTrail" (Variables|Visual|GetWeaponMeshOffhand) "None" "0, 0, 0" "0, 0, 0" "KeepRelativeOffset" true false))
  (Niagara|SetNiagaraVariable(Float) _returnvalue_2 "LoopDuration" 0.25)
  (Components|Activation|SetActive _returnvalue_2 true true))

(event FortWeapon|EventOnPlayImpactFX
  (params (HitResult :type Hit%20Result%20Structure%20%28by%20ref%29) (ImpactPhysicalSurface :type EPhysicalSurface%20Enum :default "SurfaceType_Default") (SpawnedPSC :type FXSystem%20Component%20Object%20Reference))
  (bind (_bblockinghit _binitialoverlap _time _distance _location _impactpoint _normal) (Collision|BreakHitResult HitResult))
  (|CallParentFunction HitResult ImpactPhysicalSurface SpawnedPSC)
  (bind _asniagara_particle_system_component (Utilities|Casting|CastToNiagaraComponent SpawnedPSC))
  (Utilities|FlowControl|Sequence
    (:then_0
      (Niagara|SetNiagaraVariable(Bool) _asniagara_particle_system_component "SecondImpact" (Variables|Default|GetLeft2%20Swing?)))
    (:then_1
      (Variables|Default|SetImpact%20FX_0 _asniagara_particle_system_component)
      (Transformation|SetWorldRotation SpawnedPSC (Math|Rotator|CombineRotators (Math|Rotator|MakeRotfromX _normal) (select (Variables|Default|GetRight?) (Variables|Default|GetRight%20Rotation) (Variables|Default|GetLeft%20Rotation))) false false)
      (Transformation|AddRelativeRotation SpawnedPSC (select (Variables|Default|GetLeft2%20Swing?) "-90.000000,0.000000,0.000000" (select (Variables|Default|GetRight2%20Swing%20?) "90.000000,0.000000,0.000000" "0, 0, 0")) false false))))

(event EventSwingLeft_Common
  (|CallParentFunction)
  (Variables|Default|SetRight? false)
  (Variables|Default|SetRight2%20Swing%20? false))

(event EventSwingRight_Common
  (|CallParentFunction)
  (Variables|Default|SetRight? true)
  (Variables|Default|SetRight2%20Swing%20? false))

(event EventSwingLeft
  (|CallParentFunction)
  (Variables|Default|SetLeft2%20Swing? false)
  (Variables|Default|SetRight2%20Swing%20? false))

(event EventSwingRight
  (|CallParentFunction)
  (Variables|Default|SetLeft2%20Swing? false)
  (Variables|Default|SetRight2%20Swing%20? false))

(event EventSwingRight2
  ;[1 1 1] Up swing blade Flare FX
  (|CallParentFunction)
  (Variables|Default|SetLeft2%20Swing? false)
  (Variables|Default|SetRight2%20Swing%20? true))

(fn ConstructionScript
  (|CallParentFunction))
