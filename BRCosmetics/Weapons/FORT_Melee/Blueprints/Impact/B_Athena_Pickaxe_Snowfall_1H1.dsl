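; Cosmetic weapon Blueprint for the one-handed Snowfall pickaxe that controls its idle and swing particle effects.
; Construction disables idle auto-activation, BeginPlay activates the idle effect, and cosmetic init deactivates the swing effect.
; Outside frontend display, visibility changes toggle the idle VFX and its PlayTimelineDeploy parameter.
; FrontEndIdle and EndPlay each reactivate idle once after a short delay; impact FX only calls the parent.
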(event EventBeginPlay
  (|CallParentFunction)
  (Components|Activation|Activate (CallFunction|GetIdlePSC) true))

(event Custom|FrontEndIdle
  (Utilities|FlowControl|DoOnce false
    (:Completed
      (Utilities|FlowControl|Delay 0.3)
      (Components|Activation|Activate (CallFunction|GetIdlePSC) true))))

(event EventEndPlay
  (params (EndPlayReason :type EEndPlayReason%20Enum :default "Destroyed"))
  (|CallParentFunction EndPlayReason)
  (Utilities|FlowControl|DoOnce false
    (:Completed
      (Utilities|FlowControl|Delay 0.3)
      (Components|Activation|Activate (CallFunction|GetIdlePSC) true))))

(event EventOnWeaponVisibilityChanged
  (params (bVisible :type Boolean) (bSetForLocalControllerOnly :type Boolean))
  (bind _returnvalue (CallFunction|GetIdleVFXComponent))
  (bind _returnvalue_2 (CallFunction|GetIdleVFXComponent))
  (|CallParentFunction bVisible bSetForLocalControllerOnly)
  (if (Reactive|Frontend|IsCosmeticObjectDisplayedinFrontendContext 0)
    _
    (elif bVisible
      (Components|Activation|Activate _returnvalue true)
      (Effects|Components|ParticleSystem|SetBoolParameter _returnvalue_2 "PlayTimelineDeploy" true)
      (else
        (Components|Activation|Deactivate _returnvalue)
        (Effects|Components|ParticleSystem|SetBoolParameter _returnvalue_2 "PlayTimelineDeploy" false)))))

(event EventOnInitWeaponCosmetics
  (|CallParentFunction)
  (Components|Activation|Deactivate (CallFunction|GetSwingPSC)))

(event FortWeapon|EventOnPlayImpactFX
  (params (HitResult :type Hit%20Result%20Structure%20%28by%20ref%29) (ImpactPhysicalSurface :type EPhysicalSurface%20Enum :default "SurfaceType_Default") (SpawnedPSC :type FXSystem%20Component%20Object%20Reference))
  (|CallParentFunction HitResult ImpactPhysicalSurface SpawnedPSC))

(fn ConstructionScript
  (|CallParentFunction)
  (Components|Activation|SetAutoActivate (CallFunction|GetIdlePSC) false))
