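; Pickaxe weapon Blueprint (SoilBlend) that varies its impact effect according to the swing that landed.
; Swing events set flags for left or right and for the second swing of each side.
; On impact it stores the Niagara component, aligns it to the hit normal plus a per-side rotation, and passes the second-swing flags as SecondImpact and RightSwing2.
; A separate function writes a SwingIdx integer to the audio parameter component.
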(event FortWeapon|EventOnPlayImpactFX
  (params (HitResult :type Hit%20Result%20Structure%20%28by%20ref%29) (ImpactPhysicalSurface :type EPhysicalSurface%20Enum :default "SurfaceType_Default") (SpawnedPSC :type FXSystem%20Component%20Object%20Reference))
  (bind (_bblockinghit _binitialoverlap _time _distance _location _impactpoint _normal) (Collision|BreakHitResult HitResult))
  (|CallParentFunction HitResult ImpactPhysicalSurface SpawnedPSC)
  (bind _asniagara_particle_system_component (Utilities|Casting|CastToNiagaraComponent SpawnedPSC))
  (Utilities|FlowControl|Sequence
    (:then_0
      (Variables|Default|SetImpact%20FX_0 _asniagara_particle_system_component)
      (Transformation|SetWorldRotation SpawnedPSC (Math|Rotator|CombineRotators (Math|Rotator|MakeRotfromX _normal) (select (Variables|Default|GetRight?) (Variables|Default|GetRight%20Rotation) (Variables|Default|GetLeft%20Rotation))) false false))
    (:then_1
      (Niagara|SetNiagaraVariable(Bool) _asniagara_particle_system_component "SecondImpact" (Variables|Default|GetLeft2%20Swing?)))
    (:then_2
      (Niagara|SetNiagaraVariable(Bool) _asniagara_particle_system_component "RightSwing2" (Variables|Default|GetRight%202%20Swing)))))

(event EventSwingLeft_Common
  (|CallParentFunction)
  (Variables|Default|SetRight? false))

(event EventSwingRight_Common
  (|CallParentFunction)
  (Variables|Default|SetRight? true))

(event EventSwingLeft
  (|CallParentFunction)
  (Variables|Default|SetLeft2%20Swing? false)
  (Variables|Default|SetRight%202%20Swing false))

(event EventSwingRight
  (|CallParentFunction)
  (Variables|Default|SetLeft2%20Swing? false)
  (Variables|Default|SetRight%202%20Swing false))

(event EventSwingRight2
  (|CallParentFunction)
  (Variables|Default|SetRight%202%20Swing true))

(event EventSwingLeft2
  (|CallParentFunction)
  (Variables|Default|SetRight%202%20Swing false)
  (Variables|Default|SetLeft2%20Swing? true))

(fn SetSwingAudioParam
  (params (InInt :type Integer))
  (bind _returnvalue (AudioParameter|GetAudioParameterComponent))
  (Utilities|IsValid _returnvalue
    (:Is%20Valid
      (Audio|Parameter|SetIntegerParameter _returnvalue "SwingIdx" InInt :!class /Script/AudioExtensions.AudioParameterControllerInterface))))

(fn ConstructionScript
  (|CallParentFunction))
