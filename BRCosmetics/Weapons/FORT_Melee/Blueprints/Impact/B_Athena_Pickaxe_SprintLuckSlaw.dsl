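; Reactive Fortnite pickaxe cosmetic Blueprint (SprintLuckSlaw) that temporarily powers up its materials and effects on impacts.
; Weak-point surface impacts relay a server event to the client, while frontend previews trigger the same cycle locally.
; PowerActive increments the idle VFX hit count, enables trail effects and sends time-stamped glow parameters to the mesh.
; A retriggerable delay runs PowerInactive after Duration, so repeated hits extend the powered state.
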(event FortWeapon|EventOnEquippedAndAttached
  ;[1 1 1] Sets reactivity once qeuipped in game
  (|CallParentFunction)
  (CallFunction|SetReactivity))

(event Custom|OnWeakPointHit_Server
  ;[0.9711 1 0] Send Signal from Server to Client
  (CallFunction|OnWeakPointHitClient))

(event Custom|SetReactivity
  ;[1 1 1] Verifies whether it's a reactive variant or not
  (Utilities|FlowControl|RetriggerableDelay 0.1)
  (bind _asfort_player_pawn (Utilities|Casting|CastToFortPlayerPawn (Game|GetInstigator)))
  (bind (_index _bindexisvalid) (Class|FortCosmeticFunctionLibrary|%28Deprecated%29GetVariantChannelActiveIndexForSlot "(TagName=\"LoadoutSlot.Pickaxe\")" (Variables|/Script/FortniteGame.FortPlayerPawn|GetCosmeticLoadout _asfort_player_pawn) "(TagName=\"Cosmetics.Variant.Channel.Particle\")" (Utilities|Array|MakeArray "(TagName=\"Cosmetics.Variant.Property.Particle1\")" "(TagName=\"Cosmetics.Variant.Property.Particle2\")" "(TagName=\"Cosmetics.Variant.Property.Particle3\")") true false))
  (Variables|Default|SetbIsReactiveVariant _bindexisvalid))

(event Custom|FrontendReactivity
  ;[1 1 1] Trigger event chain when bool is set to true.
  (CallFunction|PowerActive)
  (Utilities|FlowControl|RetriggerableDelay (Variables|Default|GetDuration))
  (CallFunction|PowerInactive))

(event Custom|OnWeakPointHit_Client
  ;[1 1 1] Trigger event chain when bool is set to true.
  (CallFunction|PowerActive)
  (Utilities|FlowControl|RetriggerableDelay (Variables|Default|GetDuration))
  (CallFunction|PowerInactive))

(event FortWeapon|EventOnPlayImpactFX
  (params (HitResult :type Hit%20Result%20Structure%20%28by%20ref%29) (ImpactPhysicalSurface :type EPhysicalSurface%20Enum :default "SurfaceType_Default") (SpawnedPSC :type FXSystem%20Component%20Object%20Reference))
  ;[1 1 1] Gets the weakpoints and sets makes this the condition for the event send.
  (|CallParentFunction HitResult ImpactPhysicalSurface SpawnedPSC)
  (if (or (Reactive|Frontend|IsCosmeticObjectDisplayedasPreview 0) (Reactive|Frontend|IsCosmeticObjectDisplayedinFrontendContext 0))
    (CallFunction|FrontendReactivity)
    (else
      (switch Utilities|FlowControl|Switch|SwitchonEPhysicalSurface ImpactPhysicalSurface
        (:SurfaceType12
          _)
        (:SurfaceType14
          _)
        (:SurfaceType15
          _)
        (:SurfaceType16
          _))
      (CallFunction|OnWeakPointHitServer)
      ;[1 1 1] Initial bool state is set here. Bool needs to be kept at true during the event chain.
      (Effects|Components|ParticleSystem|SetBoolParameter SpawnedPSC "PowersActive" (Variables|Default|GetPowersActive)))))

(event Custom|PowerActive
  (bind _weaponmesh (Variables|Visual|GetWeaponMesh))
  ;[1 1 1] Grabs idlefx/trail/impact fx and switches the bool between the 2 bool states based on the event sends.
  (bind _output_get (Variables|Default|SetPowersActive true))
  (bind _result (Math|Integer|IncrementInt (Variables|Default|GetWeakPointHitCount))
    (:%20%20
      (Effects|Components|ParticleSystem|SetIntParameter (CallFunction|GetIdleVFXComponent) "WeakPointHitCount" _result)
      (Effects|Components|ParticleSystem|SetBoolParameter (CallFunction|GetAnimTrailsPSC) "PowersActive" _output_get)
      (Rendering|Material|SetScalarParameterValueonMaterials _weaponmesh "GlowingStage" 1.0) ; ON
      (Rendering|Material|SetScalarParameterValueonMaterials _weaponmesh "Animation_Duration" 2.0)
      (Rendering|Material|SetScalarParameterValueonMaterials _weaponmesh "GameTimeFromBP" (Utilities|Time|GetGameTimeinSeconds)))))

(event Custom|PowerInactive
  (bind _weaponmesh (Variables|Visual|GetWeaponMesh))
  ;[1 1 1] Grabs idlefx/trail/impact fx and switches the bool between the 2 bool states based on the event sends.
  (bind _output_get (Variables|Default|SetPowersActive false))
  (Effects|Components|ParticleSystem|SetBoolParameter (CallFunction|GetAnimTrailsPSC) "PowersActive" _output_get)
  (Rendering|Material|SetScalarParameterValueonMaterials _weaponmesh "GlowingStage" 2.0) ; OFF
  (Rendering|Material|SetScalarParameterValueonMaterials _weaponmesh "Animation_Duration" 2.0)
  (Rendering|Material|SetScalarParameterValueonMaterials _weaponmesh "GameTimeFromBP" (Utilities|Time|GetGameTimeinSeconds)))

(fn PlayAudio
  ;[1 1 1] Moved to idle NS effect
  (if (Variables|Default|GetPowersActive)
    (bind _impactaudio (Variables|Default|GetImpactAudio)
      (:then
        (Audio|Components|Audio|FadeOut _impactaudio 0.3 0.0 "Linear" :!class /Script/Engine.AudioComponent))
      (:else
        _))
    (bind _returnvalue (Audio|SpawnSoundAttached (Variables|Default|GetWeaponFXSound) (Variables|Visual|GetWeaponMesh)))
    (Variables|Default|SetImpactAudio _returnvalue)))

(fn ConstructionScript
  (|CallParentFunction))
