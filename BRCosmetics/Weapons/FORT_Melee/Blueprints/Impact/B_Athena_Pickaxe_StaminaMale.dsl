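; Pickaxe weapon Blueprint for the StaminaMale cosmetic, handling its swing effects and front-end material look.
; Both swing events check the instigator's montage section; on "Swing1" they attach a first-attack effect to the handle and hide the anim trail, otherwise showing it.
; After a short BeginPlay delay, front-end display gets dynamic materials with an inverse LUT, noted as a temporary workaround.
; The construction script disables idle particle auto-activation.
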(event EventBeginPlay
  ;[1 1 1] Front End LUT, Remove after creating auto LUT tool
  (Utilities|FlowControl|Delay 0.1)
  (if (Reactive|Frontend|IsCosmeticObjectDisplayedinFrontendContext self)
    (for (_array_element _array_index) (Rendering|Material|GetMaterials (Variables|Visual|GetWeaponMesh))
      (bind _returnvalue (Rendering|Material|CreateDynamicMaterialInstance _array_element "None" "None" :!class /Script/Engine.KismetMaterialLibrary))
      (Rendering|Material|SetMaterial (Variables|Visual|GetWeaponMesh) _array_index _returnvalue)
      (Rendering|Material|SetTextureParameterValue _returnvalue "LUT Volume Texture" "/Game/Packages/Fortress_SharedMaterials/Anime_Master_Material/VT_InverseLUT_FE.VT_InverseLUT_FE")
      (Rendering|Material|SetScalarParameterValue _returnvalue "LUT Multiplier" 1.0 :!class /Script/Engine.MaterialInstanceDynamic)
      (Rendering|Material|SetScalarParameterValue _returnvalue "Mobile LUT Brightness" 1.0 :!class /Script/Engine.MaterialInstanceDynamic)
      (Rendering|Material|SetScalarParameterValue _returnvalue "use 33 Resolution" 0.0 :!class /Script/Engine.MaterialInstanceDynamic))))

(event EventSwingRight
  (bind _returnvalue (Game|GetInstigator))
  (bind _asfort_pawn (Utilities|Casting|CastToFortPawn _returnvalue))
  (bind _mesh (Variables|/Script/Engine.Character|GetMesh _asfort_pawn))
  (bind _returnvalue_2 (Components|SkeletalMesh|GetAnimInstance _mesh))
  (bind _returnvalue_3 (Animation|Montage|GetCurrentActiveMontage _returnvalue_2))
  (bind _returnvalue_4 (Animation|Montage|MontageGetCurrentSection _returnvalue_2 _returnvalue_3))
  (bind _returnvalue_5 (== _returnvalue_4 "Swing1"))
  (|CallParentFunction)
  (Networking|CanExecuteCosmeticEvents
    (:True
      (Utilities|FlowControl|Sequence
        (:then_0
          (if _returnvalue_5
            (Effects|SpawnFXSystemAttached (Variables|Default|GetFXFirstAttack) (Variables|Visual|GetWeaponMesh) "Melee_TwoHanded_Handle")))
        (:then_1
          (Development|SetHiddeninGame (CallFunction|GetAnimTrailVFXComponent) _returnvalue_5 false))))))

(event EventSwingLeft
  (bind _returnvalue (Game|GetInstigator))
  (bind _asfort_pawn (Utilities|Casting|CastToFortPawn _returnvalue))
  (bind _mesh (Variables|/Script/Engine.Character|GetMesh _asfort_pawn))
  (bind _returnvalue_2 (Components|SkeletalMesh|GetAnimInstance _mesh))
  (bind _returnvalue_3 (Animation|Montage|GetCurrentActiveMontage _returnvalue_2))
  (bind _returnvalue_4 (Animation|Montage|MontageGetCurrentSection _returnvalue_2 _returnvalue_3))
  (bind _returnvalue_5 (== _returnvalue_4 "Swing1"))
  (|CallParentFunction)
  (Networking|CanExecuteCosmeticEvents
    (:True
      (Utilities|FlowControl|Sequence
        (:then_0
          (if _returnvalue_5
            (Effects|SpawnFXSystemAttached (Variables|Default|GetFXFirstAttack) (Variables|Visual|GetWeaponMesh) "Melee_TwoHanded_Handle")))
        (:then_1
          (Development|SetHiddeninGame (CallFunction|GetAnimTrailVFXComponent) _returnvalue_5 false))))))

(fn ConstructionScript
  (|CallParentFunction)
  (Components|Activation|SetAutoActivate (CallFunction|GetIdlePSC) false))
