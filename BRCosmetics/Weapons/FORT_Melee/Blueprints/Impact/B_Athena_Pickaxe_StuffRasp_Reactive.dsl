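; Reactive dual-wield StuffRasp pickaxe Blueprint that changes mesh materials and idle VFX as the player swings.
; Right and left swings snap DissolveMask_Radius low on the main and offhand mesh respectively and flag both idle VFX components as swinging.
; Swing-end timelines animate that hand's radius back, then clear the flags.
; Impact FX resets NSImpact before calling the parent, and a helper passes duration and lifetime to both Niagara trails.
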(event Custom|CustomEvent
  (Default|CallOnStuffRaspEffectChanged))

(event EventHandlePostPickaxeandCharacterCustomizationInitilization
  (params (CustomBaseMesh :type Skeletal%20Mesh%20Component%20Object%20Reference) (CustomParts :type Array%20of%20Custom%20Character%20Part%20Object%20References) (CustomMeshParts :type Array%20of%20Skeletal%20Mesh%20Component%20Budgeted%20Object%20References)))

(event EventOnInitWeaponCosmetics
  (|CallParentFunction))

(event FortWeapon|EventOnInstigatorSet
  (|CallParentFunction))

(event FortWeapon|EventOnEquippedAndAttached
  (|CallParentFunction))

(event EventSwingRight
  ;[1 1 1] Final Swing
  (|CallParentFunction))

(event EventSwingLeft2
  ;[1 1 1] Final Swing
  (|CallParentFunction))

(event EventSwingLeft
  ;[1 1 1] Final Swing
  (|CallParentFunction))

(event EventSwingRight2
  ;[1 1 1] Final Swing
  (|CallParentFunction))

(event Custom|WeakpointHit)

(event FortWeapon|EventOnPlayImpactFX
  (params (HitResult :type Hit%20Result%20Structure%20%28by%20ref%29) (ImpactPhysicalSurface :type EPhysicalSurface%20Enum :default "SurfaceType_Default") (SpawnedPSC :type FXSystem%20Component%20Object%20Reference))
  ; On Weakpoint shader reactivity
  (Variables|Default|SetNSImpact 0)
  (|CallParentFunction HitResult ImpactPhysicalSurface SpawnedPSC))

(event EventSwingRight_Common
  ; SwingR
  (|CallParentFunction)
  (Variables|Default|Setswing_R? true)
  (Rendering|Material|SetScalarParameterValueonMaterials (Variables|Visual|GetWeaponMesh) "DissolveMask_Radius" -0.3)
  (Effects|Components|ParticleSystem|SetBoolParameter (CallFunction|GetIdleVFXComponent) "bisSwinging" true)
  (Effects|Components|ParticleSystem|SetBoolParameter (CallFunction|GetIdleOffhandVFXComponent) "bisSwinging" true))

(event EventSwingRightEnd
  ; SwingR
  (|CallParentFunction)
  (Variables|Default|Setswing_R? false)
  (bind (_direction _newtrack_0) (Timeline|Timeline PlayFromStart)
    (:Update
      (Rendering|Material|SetScalarParameterValueonMaterials (Variables|Visual|GetWeaponMesh) "DissolveMask_Radius" _newtrack_0))
    (:Finished
      (Rendering|Material|SetScalarParameterValueonMaterials (Variables|Visual|GetWeaponMesh) "DissolveMask_Radius" 1.0)
      (Effects|Components|ParticleSystem|SetBoolParameter (CallFunction|GetIdleVFXComponent) "bisSwinging" false)
      (Effects|Components|ParticleSystem|SetBoolParameter (CallFunction|GetIdleOffhandVFXComponent) "bisSwinging" false))))

(event EventSwingLeft_Common
  ; SwingL
  (|CallParentFunction)
  (Rendering|Material|SetScalarParameterValueonMaterials (Variables|Visual|GetWeaponMeshOffhand) "DissolveMask_Radius" -0.3)
  (Effects|Components|ParticleSystem|SetBoolParameter (CallFunction|GetIdleVFXComponent) "bisSwinging" true)
  (Effects|Components|ParticleSystem|SetBoolParameter (CallFunction|GetIdleOffhandVFXComponent) "bisSwinging" true))

(event EventSwingLeftEnd
  (bind _weaponmeshoffhand (Variables|Visual|GetWeaponMeshOffhand))
  ; SwingL
  (|CallParentFunction)
  (bind (_direction _newtrack_0) (Timeline|Timeline_1 PlayFromStart)
    (:Update
      (Rendering|Material|SetScalarParameterValueonMaterials _weaponmeshoffhand "DissolveMask_Radius" _newtrack_0))
    (:Finished
      (Rendering|Material|SetScalarParameterValueonMaterials _weaponmeshoffhand "DissolveMask_Radius" 1.0)
      (Effects|Components|ParticleSystem|SetBoolParameter (CallFunction|GetIdleVFXComponent) "bisSwinging" false)
      (Effects|Components|ParticleSystem|SetBoolParameter (CallFunction|GetIdleOffhandVFXComponent) "bisSwinging" false))))

(fn GetFXParamNames
  (return "LoopDuration"))

(fn SetParamsOnTrails
  (params (LoopDuration :type Float%20%28single-precision%29) (Lifetime :type Float%20%28single-precision%29))
  (for _array_element (Utilities|Array|MakeArray (CallFunction|GetAnimTrailsPSC) (CallFunction|GetAnimTrailsOffhandPSC))
    (bind _asniagara_particle_system_component (Utilities|Casting|CastToNiagaraComponent _array_element))
    (Niagara|SetNiagaraVariable(Float) _asniagara_particle_system_component "LoopDuration" LoopDuration)
    (Niagara|SetNiagaraVariable(Float) _asniagara_particle_system_component "Lifetime" Lifetime)))

(fn ConstructionScript
  (|CallParentFunction))
