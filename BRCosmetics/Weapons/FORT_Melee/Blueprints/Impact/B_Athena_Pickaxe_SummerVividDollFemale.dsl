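; Fortnite SummerVividDollFemale pickaxe Blueprint that cracks open doll layers as impacts accumulate, resetting on equip.
; Each impact increments HitCounter, which hides mesh elements through dynamic material scalars at three and six hits.
; ImpactHit spawns Niagara impact FX, plays blue, white or gold sounds by count, and adds crack sounds and DollHeadIndex at both thresholds.
; Fire FX updates the swing trail's DollHeadIndex when the counter equals either threshold.
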(event EventTick
  (params (DeltaSeconds :type Float%20%28single-precision%29)))

(event FortWeapon|EventOnPlayImpactFX
  (params (HitResult :type Hit%20Result%20Structure%20%28by%20ref%29) (ImpactPhysicalSurface :type EPhysicalSurface%20Enum :default "SurfaceType_Default") (SpawnedPSC :type FXSystem%20Component%20Object%20Reference))
  (bind _hitcounter (Variables|Default|GetHitCounter))
  ;[0.4125 0.4818 1] Primary - Inform Items Hiding after number of Hits
  (Variables|Default|SetPIckaxe%20Impact HitResult)
  (Variables|Default|SetFXSystemComponent SpawnedPSC)
  (|CallParentFunction HitResult ImpactPhysicalSurface SpawnedPSC)
  (Utilities|FlowControl|Sequence
    (:then_0
      (Math|Integer|IncrementInt (Variables|Default|GetHitCounter)))
    (:then_1
      (bind _returnvalue (Fortnite|GetDynamicMaterialInstance (Variables|Visual|GetWeaponMesh)))
      (Rendering|Material|SetScalarParameterValue _returnvalue "Hide Element 01" (Math|Float|SelectFloat 1.0 0.0 (>= _hitcounter 3)) :!class /Script/Engine.MaterialInstanceDynamic)
      (Rendering|Material|SetScalarParameterValue _returnvalue "Hide Element 02" (Math|Float|SelectFloat 1.0 0.0 (>= _hitcounter 6)) :!class /Script/Engine.MaterialInstanceDynamic))
    (:then_2
      (CallFunction|ImpactHit))))

(event Custom|ImpactHit
  (bind _pickaxe_impact (Variables|Default|GetPickaxe%20Impact))
  (bind (_bblockinghit _binitialoverlap _time _distance _location _impactpoint _normal _impactnormal) (Collision|BreakHitResult _pickaxe_impact))
  (bind _hitcounter (Variables|Default|GetHitCounter))
  (bind _impact_ref (Variables|Default|GetImpact_Ref))
  ;[1 1 1] Spawn an Impact FX at location
  (bind _returnvalue (Niagara|SpawnSystematLocation (Variables|Default|GetImpactFX) _location))
  (Variables|Default|SetImpact_Ref _returnvalue)
  (CallFunction|UpdateImpactSound _location)
  (Effects|Components|ParticleSystem|SetVectorParameter _impact_ref "ImpactNormal" _impactnormal)
  (if (== _hitcounter 6)
    (Effects|Components|ParticleSystem|SetIntParameter _impact_ref "DollHeadIndex" 1)
    (Audio|PlaySoundatLocation "/BRCosmetics/Sounds/PickAxes/SummerVividDollFemale/SC_Pickaxe_SummerVividDollFemale_CrackOpen_White.SC_Pickaxe_SummerVividDollFemale_CrackOpen_White" (Transformation|GetWorldLocation (Variables|Visual|GetWeaponMesh))) ; White Crack
    (elif (== _hitcounter 3)
      (Effects|Components|ParticleSystem|SetIntParameter _impact_ref "DollHeadIndex" 0)
      (Audio|PlaySoundatLocation "/BRCosmetics/Sounds/PickAxes/SummerVividDollFemale/SC_Pickaxe_SummerVividDollFemale_CrackOpen_Blue.SC_Pickaxe_SummerVividDollFemale_CrackOpen_Blue" (Transformation|GetWorldLocation (Variables|Visual|GetWeaponMesh))) ; Blue Crack
    )))

(event Custom|ResetImpact
  ;[1 1 1] Spawn an Impact FX at location
  (Utilities|FlowControl|DoOnce false))

(event EventOnPlayWeaponFireFX
  (params (bPersistentFire :type Boolean) (bSecondaryFire :type Boolean))
  (bind _returnvalue (CallFunction|GetSwingVFXComponent))
  (bind _hitcounter (Variables|Default|GetHitCounter))
  ;[1 1 1] Swing Trails
  (if (== _hitcounter 6)
    (Effects|Components|ParticleSystem|SetIntParameter _returnvalue "DollHeadIndex" 2)
    (elif (== _hitcounter 3)
      (Effects|Components|ParticleSystem|SetIntParameter _returnvalue "DollHeadIndex" 1))))

(event FortWeapon|EventOnEquippedAndAttached
  ;[1 1 1] Reset on Equip
  (|CallParentFunction)
  (Utilities|FlowControl|Sequence
    (:then_0
      (Variables|Default|SetHitCounter 0))
    (:then_1
      (bind _returnvalue (Fortnite|GetDynamicMaterialInstance (Variables|Visual|GetWeaponMesh)))
      (Rendering|Material|SetScalarParameterValue _returnvalue "Hide Element 01" 0.0 :!class /Script/Engine.MaterialInstanceDynamic)
      (Rendering|Material|SetScalarParameterValue _returnvalue "Hide Element 02" 0.0 :!class /Script/Engine.MaterialInstanceDynamic)
      (Rendering|Material|SetScalarParameterValue _returnvalue "Hide Element 03" 0.0 :!class /Script/Engine.MaterialInstanceDynamic)
      (Rendering|Material|SetScalarParameterValue _returnvalue "Hide Element 04" 0.0 :!class /Script/Engine.MaterialInstanceDynamic)
      (Effects|Components|ParticleSystem|SetIntParameter (CallFunction|GetSwingVFXComponent) "DollHeadIndex" 0))))

(event Custom|ResetFruitLogic
  ;[1 1 1] Reset on Equip
  (Utilities|FlowControl|Sequence
    (:then_0
      (Variables|Default|SetHitCounter 0))
    (:then_1
      (bind _returnvalue (Fortnite|GetDynamicMaterialInstance (Variables|Visual|GetWeaponMesh)))
      (Rendering|Material|SetScalarParameterValue _returnvalue "Hide Element 01" 0.0 :!class /Script/Engine.MaterialInstanceDynamic)
      (Rendering|Material|SetScalarParameterValue _returnvalue "Hide Element 02" 0.0 :!class /Script/Engine.MaterialInstanceDynamic)
      (Rendering|Material|SetScalarParameterValue _returnvalue "Hide Element 03" 0.0 :!class /Script/Engine.MaterialInstanceDynamic)
      (Rendering|Material|SetScalarParameterValue _returnvalue "Hide Element 04" 0.0 :!class /Script/Engine.MaterialInstanceDynamic)
      (Effects|Components|ParticleSystem|SetIntParameter (CallFunction|GetSwingVFXComponent) "DollHeadIndex" 0))))

(fn HandleFrontedHotnessLogic
  (if (== (Variables|Default|GetHitCounter) 0)
    (Variables|Default|SetHitCounter 2)
    (elif (> (Variables|Default|GetHitCounter) 9)
      (CallFunction|ResetFruitLogic) ; But we have to unhide the fruit!
      (Variables|Default|SetHitCounter 1))))

(fn UpdateImpactSound
  (params (Location :type Vector))
  (LocalVariables|SetImpactLocation Location)
  (if (< (Variables|Default|GetHitCounter) 3)
    (Audio|PlaySoundatLocation (Variables|Default|GetImpact_Sound_Blue) (LocalVariables|GetImpactLocation)) ; Blue Mesh Impacts
    (elif (< (Variables|Default|GetHitCounter) 6)
      (Audio|PlaySoundatLocation (Variables|Default|GetImpact_Sound_White) (LocalVariables|GetImpactLocation)) ; White Mesh Impacts
      (elif (> (Variables|Default|GetHitCounter) 6)
        (Audio|PlaySoundatLocation (Variables|Default|GetImpact_Sound_Gold) (LocalVariables|GetImpactLocation)) ; Gold Mesh Impacts
      ))))

(fn ConstructionScript
  (|CallParentFunction))
