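; Cosmetic Blueprint for the SuperNovaTaro Battle Royale pickaxe, controlling its idle, swing and impact effects.
; Idle VFX follows BeginPlay, weapon visibility and a delayed front-end idle event from the weapon's anim Blueprint, while swing VFX starts deactivated.
; Impacts set a green flag when the instigator wears the HighBeam body mesh.
; The current montage section selects one of four swing direction vectors for the impact effect.
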(event EventBeginPlay
  (|CallParentFunction)
  (Utilities|FlowControl|Sequence
    (:then_0
      (Components|Activation|Activate (CallFunction|GetIdlePSC) false))
    (:then_1
      (bind _asgalileo_ferry_axe_anim_bp (Utilities|Casting|CastToGalileo_Ferry_Axe_AnimBP (Components|SkeletalMesh|GetAnimInstance (Variables|Visual|GetWeaponMesh))))
      (Delegates|BindEventtoFrontEndIdle _asgalileo_ferry_axe_anim_bp (delegate Custom|FrontEndIdle)))))

(event Custom|FrontEndIdle
  (Utilities|FlowControl|DoOnce false
    (:Completed
      (Utilities|FlowControl|Delay 0.3)
      (Components|Activation|Activate (CallFunction|GetIdlePSC) false))))

(event EventEndPlay
  (params (EndPlayReason :type EEndPlayReason%20Enum :default "Destroyed"))
  (|CallParentFunction EndPlayReason)
  (Utilities|FlowControl|DoOnce false
    (:Completed
      (Utilities|FlowControl|Delay 0.3)
      (Components|Activation|Activate (CallFunction|GetIdlePSC) false))))

(event EventOnWeaponVisibilityChanged
  (params (bVisible :type Boolean) (bSetForLocalControllerOnly :type Boolean))
  (if bVisible
    (Components|Activation|Activate (CallFunction|GetIdleVFXComponent) false)
    (else
      (Components|Activation|Deactivate (CallFunction|GetIdlePSC)))))

(event EventOnInitWeaponCosmetics
  (|CallParentFunction)
  (Components|Activation|Deactivate (CallFunction|GetSwingPSC)))

(event FortWeapon|EventOnPlayImpactFX
  (params (HitResult :type Hit%20Result%20Structure%20%28by%20ref%29) (ImpactPhysicalSurface :type EPhysicalSurface%20Enum :default "SurfaceType_Default") (SpawnedPSC :type FXSystem%20Component%20Object%20Reference))
  (|CallParentFunction HitResult ImpactPhysicalSurface SpawnedPSC)
  (Networking|CanExecuteCosmeticEvents
    (:True
      (Utilities|FlowControl|Sequence
        (:then_0
          (bind _asplayer_pawn_athena (Utilities|Casting|CastToPlayerPawn_Athena (Game|GetInstigator)))
          (if (== (Variables|/Script/Engine.SkeletalMeshComponent|GetSkeletalMeshAsset (Customization|GetSkeletalMeshforPartType _asplayer_pawn_athena "Body" :!class /Script/FortniteGame.FortPlayerPawn)) "/Game/Characters/Player/Female/Medium/Bodies/F_MED_HighBeam/Meshes/F_MED_HighBeam.F_MED_HighBeam")
            (Effects|Components|ParticleSystem|SetBoolParameter SpawnedPSC "Is green" true)
            (else
              (Effects|Components|ParticleSystem|SetBoolParameter SpawnedPSC "Is green" false))))
        (:then_1
          (Variables|Default|SetImpactFX%20Index (Utilities|Array|FindItem (Variables|Default|GetMontageSections) (Animation|Montage|MontageGetCurrentSection (Components|SkeletalMesh|GetAnimInstance (Variables|/Script/Engine.Character|GetMesh (Utilities|Casting|CastToFortPawn (Game|GetInstigator)))) (Animation|Montage|GetCurrentActiveMontage (Components|SkeletalMesh|GetAnimInstance (Variables|/Script/Engine.Character|GetMesh (Utilities|Casting|CastToFortPawn (Game|GetInstigator))))))))
          (Effects|Components|ParticleSystem|SetVectorParameter (Variables|Default|GetImpact%20FX) "Direction" (Utilities|Select :Index (Variables|Default|GetImpactFX%20Index) :Option%200 (Variables|Default|GetSwing1%20Dir) :Option%201 (Variables|Default|GetSwing2%20Dir) :Option%202 (Variables|Default|GetSwing3%20Dir) :Option%203 (Variables|Default|GetSwing4%20Dir))))))))

(fn ConstructionScript
  (|CallParentFunction)
  (Components|Activation|SetAutoActivate (CallFunction|GetIdlePSC) false))
