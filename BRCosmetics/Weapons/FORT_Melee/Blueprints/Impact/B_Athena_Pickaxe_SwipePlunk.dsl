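; Pickaxe weapon Blueprint for the SwipePlunk cosmetic that controls its idle VFX component when the weapon is equipped and attached.
; After calling the parent handler, it activates that component and leaves it running outside frontend display contexts.
; In a frontend display context, it deactivates the effect after a short delay, then runs a second delay with nothing following it.
; BeginPlay and Tick are empty.
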(event EventBeginPlay)

(event EventTick
  (params (DeltaSeconds :type Float%20%28single-precision%29)))

(event FortWeapon|EventOnEquippedAndAttached
  (bind _returnvalue (CallFunction|GetIdleVFXComponent))
  (|CallParentFunction)
  (if (Reactive|Frontend|IsCosmeticObjectDisplayedinFrontendContext 0)
    (Components|Activation|Activate _returnvalue false)
    (Utilities|FlowControl|Delay 5.0)
    (Components|Activation|Deactivate _returnvalue)
    (Utilities|FlowControl|Delay 5.0)
    (else
      (Components|Activation|Activate _returnvalue false))))

(fn ConstructionScript
  (|CallParentFunction))
