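; Cosmetic weapon Blueprint for the SwissKale pickaxe that orients its impact particle effect to match the current melee swing.
; On impact, it looks up the instigator pawn's current montage section in a section list and uses that index to pick one of four swing direction vectors for the effect's Direction parameter.
; Swing events also record a left/right flag, unread here, and cosmetic initialization deactivates the swing particle component.
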(event FortWeapon|EventOnPlayImpactFX
  (params (HitResult :type Hit%20Result%20Structure%20%28by%20ref%29) (ImpactPhysicalSurface :type EPhysicalSurface%20Enum :default "SurfaceType_Default") (SpawnedPSC :type FXSystem%20Component%20Object%20Reference))
  (|CallParentFunction HitResult ImpactPhysicalSurface SpawnedPSC)
  (Variables|Default|SetIMpact%20FX%20Index (Utilities|Array|FindItem (Variables|Default|GetMontageSections) (Animation|Montage|MontageGetCurrentSection (Components|SkeletalMesh|GetAnimInstance (Variables|/Script/Engine.Character|GetMesh (Utilities|Casting|CastToFortPawn (Game|GetInstigator)))) (Animation|Montage|GetCurrentActiveMontage (Components|SkeletalMesh|GetAnimInstance (Variables|/Script/Engine.Character|GetMesh (Utilities|Casting|CastToFortPawn (Game|GetInstigator))))))))
  (Effects|Components|ParticleSystem|SetVectorParameter (Variables|Default|GetImpact%20FX) "Direction" (Utilities|Select :Index (Variables|Default|GetIMpact%20FX%20Index) :Option%200 (Variables|Default|GetSwing1%20Dir) :Option%201 (Variables|Default|GetSwing2%20Dir) :Option%202 (Variables|Default|GetSwing3%20Dir) :Option%203 (Variables|Default|GetSwing4%20Dir))))

(event EventSwingLeft
  (|CallParentFunction)
  (Variables|Default|SetRight%20? false))

(event EventSwingRight
  (|CallParentFunction)
  (Variables|Default|SetRight%20? true))

(event EventOnInitWeaponCosmetics
  (|CallParentFunction)
  (Components|Activation|Deactivate (CallFunction|GetSwingPSC)))

(fn ConstructionScript
  (|CallParentFunction))
