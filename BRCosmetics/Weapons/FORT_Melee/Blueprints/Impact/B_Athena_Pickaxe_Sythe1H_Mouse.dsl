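; Blueprint for a one-handed scythe-style "Mouse" pickaxe that animates a glowing reveal effect when equipped and attached.
; After the parent equip call, a RevealMask timeline drives Reveal parameters on the swirl, three spike material slots and glow overlay.
; A master emissive track sets EmissiveStrenght on the swirl, overlay and first spike slot only.
; Each timeline update calls CreateDynamicMaterialInstance again rather than caching the instances.
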(event FortWeapon|EventOnEquippedAndAttached
  (bind _sm_mouse_pickaxe_spike (Variables|Default|GetSM_Mouse_Pickaxe_Spike))
  (|CallParentFunction)
  (bind (_direction _spiralreveal _spikereveal_1 _spikereveal_2 _spikereveal_3 _masteremissive) (Timeline|RevealMask PlayFromStart :NewTime 1.0)
    (:Update
      ;[1 0 0.1057] Spiral
      (bind _returnvalue (Rendering|Material|CreateDynamicMaterialInstance (Variables|Default|GetSM_Mouse_Pickaxe_Swirl) 0 0 "None" :!class /Script/Engine.PrimitiveComponent))
      ;[0 0.9659 1] Spike
      (bind _returnvalue_2 (Rendering|Material|CreateDynamicMaterialInstance _sm_mouse_pickaxe_spike 0 0 "None" :!class /Script/Engine.PrimitiveComponent))
      (bind _returnvalue_3 (Rendering|Material|CreateDynamicMaterialInstance _sm_mouse_pickaxe_spike 1 0 "None" :!class /Script/Engine.PrimitiveComponent))
      (bind _returnvalue_4 (Rendering|Material|CreateDynamicMaterialInstance _sm_mouse_pickaxe_spike 2 0 "None" :!class /Script/Engine.PrimitiveComponent))
      ;[1 0.7801 0] Glowverlay
      (bind _returnvalue_5 (Rendering|Material|CreateDynamicMaterialInstance (Variables|Default|GetSM_Mouse_Pickaxe_Glowverlay) 0 0 "None" :!class /Script/Engine.PrimitiveComponent))
      ;[1 0.009102 0.2095] Spiral
      (Rendering|Material|SetScalarParameterValue _returnvalue "Reveal" _spiralreveal :!class /Script/Engine.MaterialInstanceDynamic)
      (Rendering|Material|SetScalarParameterValue _returnvalue "EmissiveStrenght" _masteremissive :!class /Script/Engine.MaterialInstanceDynamic)
      ;[0 0.9719 1] Spike
      (Rendering|Material|SetScalarParameterValue _returnvalue_2 "Reveal" _spikereveal_1 :!class /Script/Engine.MaterialInstanceDynamic)
      (Rendering|Material|SetScalarParameterValue _returnvalue_3 "Reveal" _spikereveal_3 :!class /Script/Engine.MaterialInstanceDynamic)
      (Rendering|Material|SetScalarParameterValue _returnvalue_4 "Reveal" _spikereveal_2 :!class /Script/Engine.MaterialInstanceDynamic)
      (Rendering|Material|SetScalarParameterValue _returnvalue_2 "EmissiveStrenght" _masteremissive :!class /Script/Engine.MaterialInstanceDynamic)
      ;[1 0.6284 0] Glowverlay
      (Rendering|Material|SetScalarParameterValue _returnvalue_5 "Reveal" _spiralreveal :!class /Script/Engine.MaterialInstanceDynamic)
      (Rendering|Material|SetScalarParameterValue _returnvalue_5 "EmissiveStrenght" _masteremissive :!class /Script/Engine.MaterialInstanceDynamic))))

(fn ConstructionScript
  (|CallParentFunction))
