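; Pickaxe Blueprint for the one-handed OysterKnock scythe, adding custom impact effects to hits.
; The impact FX event calls the parent, then relays through ServerImpact to ClientImpact.
; For four specific surface types or preview/frontend display, it spawns impact Niagara and stamps game time into weapon mesh custom primitive data.
; After a short delay, it attaches a Weakpoint Niagara system to the mesh, passing impact location and a frontend flag.
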(event EventBeginPlay)

(event EventTick
  (params (DeltaSeconds :type Float%20%28single-precision%29)))

(event Custom|ServerImpact
  (params (HitResult :type Hit%20Result%20Structure) (Impact%20Phys%20Surf :type EPhysicalSurface%20Enum :default "SurfaceType_Default") (SpawnPSC :type FXSystem%20Component%20Object%20Reference))
  (CallFunction|ClientImpact HitResult Impact%20Phys%20Surf SpawnPSC))

(event FortWeapon|EventOnPlayImpactFX
  (params (HitResult :type Hit%20Result%20Structure%20%28by%20ref%29) (ImpactPhysicalSurface :type EPhysicalSurface%20Enum :default "SurfaceType_Default") (SpawnedPSC :type FXSystem%20Component%20Object%20Reference))
  (|CallParentFunction HitResult ImpactPhysicalSurface SpawnedPSC)
  (CallFunction|ServerImpact HitResult ImpactPhysicalSurface SpawnedPSC))

(event Custom|ClientImpact
  (params (HitResult :type Hit%20Result%20Structure) (ImpactPhysSurf :type EPhysicalSurface%20Enum :default "SurfaceType_Default") (SpawnPSC :type FXSystem%20Component%20Object%20Reference))
  (bind (_bblockinghit _binitialoverlap _time _distance _location) (Collision|BreakHitResult HitResult))
  (if (Math|Boolean|ORBoolean (Utilities|Enum|Equal(Enum) (Collision|GetSurfaceType HitResult) "SurfaceType12") (Utilities|Enum|Equal(Enum) (Collision|GetSurfaceType HitResult) "SurfaceType14") (Utilities|Enum|Equal(Enum) (Collision|GetSurfaceType HitResult) "SurfaceType15") (Utilities|Enum|Equal(Enum) (Collision|GetSurfaceType HitResult) "SurfaceType16") (or (Reactive|Frontend|IsCosmeticObjectDisplayedasPreview 0) (Reactive|Frontend|IsCosmeticObjectDisplayedinFrontendContext 0)))
    (Niagara|SpawnSystematLocation "/BRCosmetics/Weapons/FORT_Melee/Pickaxe_OysterKnock/FX/NS_Pickaxe_OysterKnock_Impact.NS_Pickaxe_OysterKnock_Impact" _location)
    (Rendering|Material|SetCustomPrimitiveDataFloat (Variables|Visual|GetWeaponMesh) 2 (Utilities|Time|GetGameTimeinSeconds))
    (Utilities|FlowControl|Delay 0.1)
    (bind _returnvalue (Niagara|SpawnSystemAttached "/BRCosmetics/Weapons/FORT_Melee/Pickaxe_OysterKnock/FX/NS_Pickaxe_OysterKnock_Weakpoint.NS_Pickaxe_OysterKnock_Weakpoint" (Variables|Visual|GetWeaponMesh) ""))
    (bind _output_get (Variables|Default|SetSpawnedNS _returnvalue))
    (Effects|Components|ParticleSystem|SetVectorParameter (Variables|Default|GetSpawnedNS) "ImpactLocation" _location)
    (Niagara|SetNiagaraVariable(Bool) (Variables|Default|GetSpawnedNS) "bIsFrontEnd" (Reactive|Frontend|IsCosmeticObjectDisplayedinFrontendContext 0))))

(fn ConstructionScript
  (|CallParentFunction))
