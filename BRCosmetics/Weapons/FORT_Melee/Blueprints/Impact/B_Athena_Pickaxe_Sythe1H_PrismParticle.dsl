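; Prism Particle one-handed scythe pickaxe Blueprint whose glass vial progresses from normal through cracked and busted to broken, resetting on equip.
; A server-to-client custom event chain increments glassState and spawns a Niagara impact at the last hit.
; The state selects impact particle parameters and sounds, and drives vial visibility and cracks through dynamic materials.
; Swing events pass montage section and busted state to the trail; impact handling also checks preview and surface type.
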(event EventSwingRight
  (bind _returnvalue (CallFunction|GetSwingVFXComponent))
  ;[0 0 0] anim trail
  (|CallParentFunction)
  (Utilities|FlowControl|Sequence
    (:then_0
      (bind _section_index (Utilities|SetupSwingMontageSectionIndex (Variables|Actor|GetInstigator))
        (:Out
          (Effects|Components|ParticleSystem|SetIntParameter (CallFunction|GetSwingVFXComponent) "SwingSectionIndex" _section_index))))
    (:then_1
      ;; UpdateAnimTrailState
      (switch int (Variables|Default|GetglassState)
        (:0
          _)
        (:1
          _)
        (:2
          (Effects|Components|ParticleSystem|SetIntParameter _returnvalue "isBusted" 1 :!label _label) ; sets animtrail parameter
        )
        (:3
          (goto _label)))
      (Effects|Components|ParticleSystem|SetIntParameter _returnvalue "isBusted" 0) ; sets animtrail parameter
    )))

(event EventSwingLeft
  (bind _returnvalue (CallFunction|GetSwingVFXComponent))
  ;[0 0 0] anim trail
  (|CallParentFunction)
  (Utilities|FlowControl|Sequence
    (:then_0
      (bind _section_index (Utilities|SetupSwingMontageSectionIndex (Variables|Actor|GetInstigator))
        (:Out
          (Effects|Components|ParticleSystem|SetIntParameter (CallFunction|GetSwingVFXComponent) "SwingSectionIndex" _section_index))))
    (:then_1
      ;; UpdateAnimTrailState
      (switch int (Variables|Default|GetglassState)
        (:0
          _)
        (:1
          _)
        (:2
          (Effects|Components|ParticleSystem|SetIntParameter _returnvalue "isBusted" 1 :!label _label) ; sets animtrail parameter
        )
        (:3
          (goto _label)))
      (Effects|Components|ParticleSystem|SetIntParameter _returnvalue "isBusted" 0) ; sets animtrail parameter
    )))

(event FortWeapon|EventOnPlayImpactFX
  (params (HitResult :type Hit%20Result%20Structure%20%28by%20ref%29) (ImpactPhysicalSurface :type EPhysicalSurface%20Enum :default "SurfaceType_Default") (SpawnedPSC :type FXSystem%20Component%20Object%20Reference))
  (bind _spawned_psc (Variables|Default|GetSpawned%20PSC))
  (bind _spawned_psc_2 (Variables|Default|GetSpawned%20PSC))
  ;[0 0 0] IMPACT FX
  (|CallParentFunction HitResult ImpactPhysicalSurface SpawnedPSC)
  (Variables|Default|SetSpawned%20PSC SpawnedPSC)
  (if (not (Reactive|Frontend|IsCosmeticObjectDisplayedasPreview 0))
    (Utilities|FlowControl|Sequence
      (:then_0
        ;[0 0 0] directionality
        (bind _section_index (Utilities|SetupSwingMontageSectionIndex (Variables|Actor|GetInstigator))
          (:Out
            (Utilities|FlowControl|Sequence
              (:then_0
                (Effects|Components|ParticleSystem|SetIntParameter _spawned_psc_2 "SwingSectionIndex" _section_index))))
          (:Cast%20Failed%20Out
            (Effects|Components|ParticleSystem|SetIntParameter _spawned_psc_2 "SwingSectionIndex" -1))))
      (:then_1
        (if (not (Math|Boolean|ORBoolean (== ImpactPhysicalSurface 12) (== ImpactPhysicalSurface 14) (== ImpactPhysicalSurface 15) (== ImpactPhysicalSurface 16)))
          (switch int (Variables|Default|GetglassState)
            (:Default
              _)
            (:0
              (Utilities|FlowControl|Sequence
                (:then_0
                  ;; State 0 - Normal
                  (Effects|Components|ParticleSystem|SetFloatParameter (Variables|Default|GetSpawned%20PSC) "bustedState" 0.0) ; Normal Impact
                )
                (:then_1
                  (Audio|SpawnSoundAttached "/BRCosmetics/Sounds/PickAxes/PrismParticle/SC_PickaxeImpactEnemy_PrismParticle.SC_PickaxeImpactEnemy_PrismParticle" _spawned_psc))))
            (:1
              (Utilities|FlowControl|Sequence
                (:then_0
                  ;; State 1 - Cracked
                  (Effects|Components|ParticleSystem|SetFloatParameter (Variables|Default|GetSpawned%20PSC) "bustedState" 2.0) ; Cracked Impact
                  (CallFunction|UpdateVialState 1.0 1.0))
                (:then_1
                  (Audio|SpawnSoundAttached "/BRCosmetics/Sounds/PickAxes/PrismParticle/SC_PickaxeImpactEnemy_PrismParticle.SC_PickaxeImpactEnemy_PrismParticle" _spawned_psc))))
            (:2
              (Utilities|FlowControl|Sequence
                (:then_0
                  ;; State 2 - Busted
                  (Math|Integer|IncrementInt (Variables|Default|GetglassState)
                    (:%20%20
                      (Effects|Components|ParticleSystem|SetFloatParameter (Variables|Default|GetSpawned%20PSC) "bustedState" 1.0) ; Splash Impact
                      (CallFunction|UpdateVialState 0.0))) ; One-Shot Event
                )
                (:then_1
                  (Audio|SpawnSoundAttached "/BRCosmetics/Sounds/PickAxes/PrismParticle/SC_PickaxeImpactEnemy_PrismParticle_Cracked.SC_PickaxeImpactEnemy_PrismParticle_Cracked" _spawned_psc))))
            (:3
              _))
          (Utilities|FlowControl|Sequence
            (:then_0
              ;; State 3 - Broken
              (Effects|Components|ParticleSystem|SetFloatParameter (Variables|Default|GetSpawned%20PSC) "bustedState" 3.0) ; AfterSplash Impact
            )
            (:then_1
              (Audio|SpawnSoundAttached "/BRCosmetics/Sounds/PickAxes/PrismParticle/SC_PickaxeImpactEnemy_PrismParticle.SC_PickaxeImpactEnemy_PrismParticle" _spawned_psc)))))))
  (Components|Activation|Deactivate _spawned_psc)
  (CallFunction|CEFXServer))

(event Custom|ReactivityEvent
  (bind _spawned_psc (Variables|Default|GetSpawned%20PSC))
  ;[0 0 0] ReactivityEventNetworked
  (Math|Integer|IncrementInt (Variables|Default|GetglassState)
    (:%20%20
      (bind _returnvalue (Niagara|SpawnSystematLocation "/BRCosmetics/Weapons/FORT_Melee/Pickaxe_PrismParticle/FX/V2/NS_Pickaxe_PrismParticle_Impact.NS_Pickaxe_PrismParticle_Impact" (.ImpactPoint (Collision|BreakHitResult (stale FortWeapon|EventOnPlayImpactFX HitResult))) (Math|Conversions|RotationFromXVector (.ImpactNormal (Collision|BreakHitResult (stale FortWeapon|EventOnPlayImpactFX HitResult))))))
      (bind _output_get (Variables|Default|SetSpawned%20PSC _returnvalue))
      (Effects|Components|ParticleSystem|SetBoolParameter _output_get "bIsFrontEndPreview" (Reactive|Frontend|IsCosmeticObjectDisplayedinFrontendContext self))
      ;[0 0 0] IMPACT FX
      (switch int (Variables|Default|GetglassState)
        (:Default
          _)
        (:0
          (Utilities|FlowControl|Sequence
            (:then_0
              ;; State 0 - Normal
              (Effects|Components|ParticleSystem|SetFloatParameter (Variables|Default|GetSpawned%20PSC) "bustedState" 0.0) ; Normal Impact
            )
            (:then_1
              (Audio|SpawnSoundAttached "/BRCosmetics/Sounds/PickAxes/PrismParticle/SC_PickaxeImpactEnemy_PrismParticle.SC_PickaxeImpactEnemy_PrismParticle" _spawned_psc))))
        (:1
          (Utilities|FlowControl|Sequence
            (:then_0
              ;; State 1 - Cracked
              (Effects|Components|ParticleSystem|SetFloatParameter (Variables|Default|GetSpawned%20PSC) "bustedState" 2.0) ; Cracked Impact
              (CallFunction|UpdateVialState 1.0 1.0))
            (:then_1
              (Audio|SpawnSoundAttached "/BRCosmetics/Sounds/PickAxes/PrismParticle/SC_PickaxeImpactEnemy_PrismParticle.SC_PickaxeImpactEnemy_PrismParticle" _spawned_psc))))
        (:2
          (Utilities|FlowControl|Sequence
            (:then_0
              ;; State 2 - Busted
              (Math|Integer|IncrementInt (Variables|Default|GetglassState)
                (:%20%20
                  (Effects|Components|ParticleSystem|SetFloatParameter (Variables|Default|GetSpawned%20PSC) "bustedState" 1.0) ; Splash Impact
                  (CallFunction|UpdateVialState 0.0))) ; One-Shot Event
            )
            (:then_1
              (Audio|SpawnSoundAttached "/BRCosmetics/Sounds/PickAxes/PrismParticle/SC_PickaxeImpactEnemy_PrismParticle_Cracked.SC_PickaxeImpactEnemy_PrismParticle_Cracked" _spawned_psc))))
        (:3
          _))
      (Utilities|FlowControl|Sequence
        (:then_0
          ;; State 3 - Broken
          (Effects|Components|ParticleSystem|SetFloatParameter (Variables|Default|GetSpawned%20PSC) "bustedState" 3.0) ; AfterSplash Impact
        )
        (:then_1
          (Audio|SpawnSoundAttached "/BRCosmetics/Sounds/PickAxes/PrismParticle/SC_PickaxeImpactEnemy_PrismParticle.SC_PickaxeImpactEnemy_PrismParticle" _spawned_psc))))))

(event Custom|CE_FX_Server
  ;[0 0 0] RESET ON EQUIP
  (CallFunction|CEFXClient))

(event Custom|CE_FX_Client
  ;[0 0 0] RESET ON EQUIP
  (CallFunction|ReactivityEvent))

(event FortWeapon|EventOnEquippedAndAttached
  ;[0 0 0] RESET ON EQUIP
  (|CallParentFunction)
  (Variables|Default|SetglassState 0)
  (CallFunction|UpdateVialState 1.0))

(event Custom|ImpactAudioLogic
  (bind _spawned_psc (Variables|Default|GetSpawned%20PSC))
  (switch int (Variables|Default|GetglassState)
    (:Default
      _)
    (:0
      (Audio|PlaySoundatLocation "/BRCosmetics/Sounds/PickAxes/PrismParticle/SC_PickaxeImpactEnemy_PrismParticle.SC_PickaxeImpactEnemy_PrismParticle" (Transformation|GetWorldLocation _spawned_psc)))
    (:1
      (Audio|PlaySoundatLocation "/BRCosmetics/Sounds/PickAxes/PrismParticle/SC_PickaxeImpactEnemy_PrismParticle.SC_PickaxeImpactEnemy_PrismParticle" (Transformation|GetWorldLocation _spawned_psc)))
    (:2
      (Audio|PlaySoundatLocation "/BRCosmetics/Sounds/PickAxes/PrismParticle/SC_PickaxeImpactEnemy_PrismParticle_Cracked.SC_PickaxeImpactEnemy_PrismParticle_Cracked" (Transformation|GetWorldLocation _spawned_psc)))
    (:3
      _))
  (Audio|PlaySoundatLocation "/BRCosmetics/Sounds/PickAxes/PrismParticle/SC_PickaxeImpactEnemy_PrismParticle.SC_PickaxeImpactEnemy_PrismParticle" (Transformation|GetWorldLocation _spawned_psc)))

(fn UpdateVialState
  (params (Visibility :type Float%20%28double-precision%29) (IsCracked :type Float%20%28double-precision%29))
  (bind _weaponmesh (Variables|Visual|GetWeaponMesh))
  (Rendering|Material|SetScalarParameterValue (Class|BTechArtGlobalLibrary|GetDynamicMaterialInstance _weaponmesh) "Hide Element 01" (- 1.0 Visibility) :!class /Script/Engine.MaterialInstanceDynamic)
  (Rendering|Material|SetScalarParameterValue (Class|BTechArtGlobalLibrary|GetDynamicMaterialInstance _weaponmesh 2) "CrackProgress" IsCracked :!class /Script/Engine.MaterialInstanceDynamic))

(fn ConstructionScript
  (|CallParentFunction))
