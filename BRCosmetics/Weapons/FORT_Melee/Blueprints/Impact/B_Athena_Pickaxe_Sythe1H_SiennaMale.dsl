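; Weapon blueprint for the Sienna Male one-handed scythe pickaxe cosmetic that customizes its impact effect.
; On impact, after calling the parent, it samples the SwingFX socket location, waits one tick, and samples again to derive the slash direction.
; It rotates the spawned effect to that direction so scattered sparks follow the swing.
; It also passes the physical surface index for debris and a frontend-preview flag as effect parameters.
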(event FortWeapon|EventOnPlayImpactFX
  (params (HitResult :type Hit%20Result%20Structure%20%28by%20ref%29) (ImpactPhysicalSurface :type EPhysicalSurface%20Enum :default "SurfaceType_Default") (SpawnedPSC :type FXSystem%20Component%20Object%20Reference))
  (bind _weaponmesh (Variables|Visual|GetWeaponMesh))
  (bind _returnvalue_location (Math|Transform|BreakTransform (Transformation|GetSocketTransform _weaponmesh "SwingFX" "RTS_World")))
  (bind _returnvalue_location_2 (Math|Transform|BreakTransform (Transformation|GetSocketTransform _weaponmesh "SwingFX" "RTS_World")))
  (bind _impact_fx (Variables|Default|GetImpact%20FX))
  ;[1 0.7403 0.224] Match the direction of the slash with the Velocity of the scattered sparks.
  (|CallParentFunction HitResult ImpactPhysicalSurface SpawnedPSC)
  ;[1 1 1] Obtain slashing vector based on position information after 1 frame.
  (bind _output_get (Variables|Default|SetHit%20Location _returnvalue_location_2))
  (Utilities|FlowControl|DelayUntilNextTick)
  (bind _output_get_2 (Variables|Default|SetNext%20Hit%20Location _returnvalue_location))
  (bind _output_get_3 (Variables|Default|SetImpact%20FX SpawnedPSC))
  (Utilities|IsValid _output_get_3
    (:Is%20Valid
      (bind (_returnvalue_roll _returnvalue_pitch _returnvalue_yaw) (Math|Rotator|BreakRotator (Math|Conversions|RotationFromXVector (Math|Vector|GetUnitDirection(Vector) _output_get _output_get_2))))
      (Transformation|SetWorldRotation (Variables|Default|GetImpact%20FX) (Math|Rotator|MakeRotator _returnvalue_roll _returnvalue_pitch _returnvalue_yaw) false false)
      ;[1 1 1] For Physmat driven debris
      (Effects|Components|ParticleSystem|SetIntParameter _impact_fx "PhysicalSurface" (Utilities|Select :Index ImpactPhysicalSurface :SurfaceType_Default 0 :SurfaceType1 1 :SurfaceType2 2 :SurfaceType3 3 :SurfaceType4 4 :SurfaceType5 5 :SurfaceType6 6 :SurfaceType7 7 :SurfaceType8 8 :SurfaceType9 9 :SurfaceType10 10 :SurfaceType11 11 :SurfaceType12 12 :SurfaceType13 13 :SurfaceType14 14 :SurfaceType15 15 :SurfaceType16 16 :SurfaceType17 17 :SurfaceType18 18 :SurfaceType19 19 :SurfaceType20 20 :SurfaceType21 21 :SurfaceType22 22 :SurfaceType23 23 :SurfaceType24 24 :SurfaceType25 25 :SurfaceType26 26 :SurfaceType27 0 :SurfaceType28 0 :SurfaceType29 0 :SurfaceType30 0 :SurfaceType31 0 :SurfaceType32 0 :SurfaceType33 0 :SurfaceType34 0))
      ;[0.792 1 0.5376] Frontend Preview Setup
      (Effects|Components|ParticleSystem|SetBoolParameter _impact_fx "bIsFrontEndPreview" (Reactive|Frontend|IsCosmeticObjectDisplayedinFrontendContext self)))))

(event EventBeginPlay)

(event EventTick
  (params (DeltaSeconds :type Float%20%28single-precision%29)))

(fn ConstructionScript
  (|CallParentFunction))
