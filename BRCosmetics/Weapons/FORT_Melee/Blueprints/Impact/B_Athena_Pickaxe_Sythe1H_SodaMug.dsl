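; Cosmetic Blueprint for the SodaMug one-handed scythe pickaxe that drives its idle and swing particle effects.
; Init flags the idle VFX when shown in the frontend.
; Swing events toggle a swinging flag on the swing VFX and pass it a montage section index derived from the instigator.
; Impacts store the spawned effect and call WeakPointReactivity, which stamps current game time onto the idle VFX.
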(event EventOnInitWeaponCosmetics
  ;[0 0 0] fe
  (|CallParentFunction)
  (Effects|Components|ParticleSystem|SetBoolParameter (CallFunction|GetIdleVFXComponent) "bIsFrontEnd" (Reactive|Frontend|IsCosmeticObjectDisplayedinFrontendContext self)))

(event EventSwingLeft
  (|CallParentFunction)
  (Effects|Components|ParticleSystem|SetBoolParameter (CallFunction|GetSwingVFXComponent) "bisSwinging" true)
  (bind _section_index (Utilities|SetupSwingMontageSectionIndex (Variables|Actor|GetInstigator))
    (:Out
      (Effects|Components|ParticleSystem|SetIntParameter (CallFunction|GetSwingVFXComponent) "SwingSectionIndex" _section_index))))

(event EventSwingRight
  (|CallParentFunction)
  (Effects|Components|ParticleSystem|SetBoolParameter (CallFunction|GetSwingVFXComponent) "bisSwinging" true)
  (bind _section_index (Utilities|SetupSwingMontageSectionIndex (Variables|Actor|GetInstigator))
    (:Out
      (Effects|Components|ParticleSystem|SetIntParameter (CallFunction|GetSwingVFXComponent) "SwingSectionIndex" _section_index))))

(event Custom|WeakPointReactivity
  (Effects|Components|ParticleSystem|SetFloatParameter (CallFunction|GetIdleVFXComponent) "GameTime" (Utilities|Time|GetGameTimeinSeconds)))

(event EventSwingLeftEnd
  (|CallParentFunction)
  (Effects|Components|ParticleSystem|SetBoolParameter (CallFunction|GetSwingVFXComponent) "bisSwinging" false))

(event EventSwingRightEnd
  (|CallParentFunction)
  (Effects|Components|ParticleSystem|SetBoolParameter (CallFunction|GetSwingVFXComponent) "bisSwinging" false))

(event FortWeapon|EventOnPlayImpactFX
  (params (HitResult :type Hit%20Result%20Structure%20%28by%20ref%29) (ImpactPhysicalSurface :type EPhysicalSurface%20Enum :default "SurfaceType_Default") (SpawnedPSC :type FXSystem%20Component%20Object%20Reference))
  (Variables|Default|SetCurrentImpact SpawnedPSC)
  (if (Reactive|Frontend|IsCosmeticObjectDisplayedinFrontendContext 0)
    (CallFunction|WeakPointReactivity)
    (else
      (switch Utilities|FlowControl|Switch|SwitchonEPhysicalSurface ImpactPhysicalSurface
        (:SurfaceType12
          _)
        (:SurfaceType14
          _)
        (:SurfaceType15
          _)
        (:SurfaceType16
          _))
      (CallFunction|WeakPointReactivity))))

(fn ConstructionScript
  (|CallParentFunction))
