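; Scythe pickaxe cosmetic weapon Blueprint that makes its idle, swing and impact effects react to weak point hits.
; On impact it stores the spawned effect, then checks whether the struck surface is in its configured surface list.
; A match stamps the impact effect and calls ServerEvent, which runs OnWeakPointHit to write game time into "ReactivityTime" on all three effects.
; Frontend and preview displays treat every impact as a weak point hit, skipping the surface check.
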(event FortWeapon|EventOnPlayImpactFX
  (params (HitResult :type Hit%20Result%20Structure%20%28by%20ref%29) (ImpactPhysicalSurface :type EPhysicalSurface%20Enum :default "SurfaceType_Default") (SpawnedPSC :type FXSystem%20Component%20Object%20Reference))
  (|CallParentFunction HitResult ImpactPhysicalSurface SpawnedPSC)
  (Variables|Default|SetImpact%20FX SpawnedPSC)
  ; Frontend Weak Point
  (if (or (Reactive|Frontend|IsCosmeticObjectDisplayedasPreview 0) (Reactive|Frontend|IsCosmeticObjectDisplayedinFrontendContext 0))
    (CallFunction|OnWeakPointHit)
    (elif (Utilities|Array|ContainsItem (Variables|Default|GetImpactPhysicalSurface) ImpactPhysicalSurface)
      (Effects|Components|ParticleSystem|SetFloatParameter (Variables|Default|GetImpact%20FX) "ReactivityTime" (Utilities|Time|GetGameTimeinSeconds))
      (CallFunction|ServerEvent))))

(event Custom|ServerEvent
  (CallFunction|OnWeakPointHit))

(event Custom|OnWeakPointHit
  (bind _returnvalue (Utilities|Time|GetGameTimeinSeconds))
  (Effects|Components|ParticleSystem|SetFloatParameter (CallFunction|GetIdleVFXComponent) "ReactivityTime" _returnvalue)
  (Effects|Components|ParticleSystem|SetFloatParameter (CallFunction|GetSwingVFXComponent) "ReactivityTime" _returnvalue)
  (Effects|Components|ParticleSystem|SetFloatParameter (Variables|Default|GetImpact%20FX) "ReactivityTime" _returnvalue))

(fn ConstructionScript
  (|CallParentFunction))
