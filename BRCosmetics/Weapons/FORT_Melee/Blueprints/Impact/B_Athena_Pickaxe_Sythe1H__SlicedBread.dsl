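; Sliced Bread one-handed scythe pickaxe cosmetic that builds an impact stage in its VFX on weak-point hits.
; Impacts feed stage and weak-point parameters to impact FX, then pass the surface through server and client events.
; Weak-point hits, or frontend display, increment the idle VFX stage; a retriggerable delay resets it after hits stop.
; Initialization zeroes the stage and offsets translucency sorting to layer main and rear idle effects.
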(event FortWeapon|EventOnPlayImpactFX
  (params (HitResult :type Hit%20Result%20Structure%20%28by%20ref%29) (ImpactPhysicalSurface :type EPhysicalSurface%20Enum :default "SurfaceType_Default") (SpawnedPSC :type FXSystem%20Component%20Object%20Reference))
  (|CallParentFunction HitResult ImpactPhysicalSurface SpawnedPSC)
  (Effects|Components|ParticleSystem|SetIntParameter SpawnedPSC "BP_ImpactStage" (Variables|Default|GetImpact%20Stage))
  ;[1 1 1] WeakPoint impact
  (bind _output_get (Variables|Default|SetCurrent%20Impact%20Physical%20Surface ImpactPhysicalSurface))
  (bind _output_get_2 (Variables|Default|SetIs%20hit%20weak%20point (or (Utilities|Array|ContainsItem (Variables|Default|GetWeakPoint%20Physical%20Surface) _output_get) (Reactive|Frontend|IsCosmeticObjectDisplayedinFrontendContext 0))))
  (Effects|Components|ParticleSystem|SetBoolParameter (Variables|Default|GetImpact%20FX) "Is WeakPoint" _output_get_2)
  (CallFunction|CEFXServer ImpactPhysicalSurface))

(event EventHandlePostCosmeticInitialization
  (|CallParentFunction)
  (Rendering|SetTranslucencySortDistanceOffset (CallFunction|GetIdleVFXComponent) -1e+02)
  (Rendering|SetTranslucencySortDistanceOffset (Variables|Default|GetNS_Pickaxe_SlicedBread_Idle_Rear) 1e+02)
  (Variables|Default|SetImpact%20Stage 0))

(event Custom|CE_FX_Server
  (params (Impact%20Physical%20Surface :type EPhysicalSurface%20Enum :default "SurfaceType_Default"))
  (CallFunction|CEFXClient Impact%20Physical%20Surface))

(event Custom|CE_FX_Client
  (params (Impact%20Physical%20Surface :type EPhysicalSurface%20Enum :default "SurfaceType_Default"))
  (bind _returnvalue (CallFunction|GetIdleVFXComponent))
  (bind _output_get (Variables|Default|SetCurrent%20Impact%20Physical%20Surface Impact%20Physical%20Surface))
  (Variables|Default|SetIs%20hit%20weak%20point (Utilities|Array|ContainsItem (Variables|Default|GetWeakPoint%20Physical%20Surface) _output_get))
  (if (or (Variables|Default|GetIs%20hit%20weak%20point) (Reactive|Frontend|IsCosmeticObjectDisplayedinFrontendContext 0))
    (bind _output_get_2 (Variables|Default|SetImpact%20Stage (+ (Variables|Default|GetImpact%20Stage) 1)))
    (Effects|Components|ParticleSystem|SetIntParameter _returnvalue "BP_ImpactStage" _output_get_2)
    (Effects|Components|ParticleSystem|SetBoolParameter _returnvalue "hasImpacted" true)
    (Utilities|FlowControl|RetriggerableDelay 6.0)
    (bind _output_get_3 (Variables|Default|SetImpact%20Stage 0))
    (Effects|Components|ParticleSystem|SetBoolParameter _returnvalue "hasImpacted" false)
    (Effects|Components|ParticleSystem|SetIntParameter _returnvalue "BP_ImpactStage" _output_get_3)))

(event EventBeginPlay)

(event EventTick
  (params (DeltaSeconds :type Float%20%28single-precision%29)))

(fn ConstructionScript
  (|CallParentFunction))
