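; TopicProud pickaxe cosmetic Blueprint that makes the weapon's look react to successive weakpoint hits.
; An Index counter rises on weakpoint hits, or any impact in frontend display, driving "ReactivityIndex" on mesh materials and swing VFX.
; Those hits also spawn a Niagara burst flagged for frontend, with weakpoint hits reusing the impact event's HitResult for location.
; Init and unequip events zero the counter and update materials and idle VFX instead.
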(event EventOnInitWeaponCosmetics
  (bind _index (Variables|Default|GetIndex))
  (bind _returnvalue (Math|Conversions|ToFloat(Integer) _index))
  (|CallParentFunction)
  (Variables|Default|SetIndex 0)
  (Rendering|Material|SetScalarParameterValueonMaterials (Variables|Visual|GetWeaponMesh) "ReactivityIndex" _returnvalue)
  (Effects|Components|ParticleSystem|SetIntParameter (CallFunction|GetIdleVFXComponent) "ReactivityIndex" (Math|Float|Truncate _returnvalue)))

(event FortWeapon|EventOnPlayImpactFX
  (params (HitResult :type Hit%20Result%20Structure%20%28by%20ref%29) (ImpactPhysicalSurface :type EPhysicalSurface%20Enum :default "SurfaceType_Default") (SpawnedPSC :type FXSystem%20Component%20Object%20Reference))
  (bind _index (Variables|Default|GetIndex))
  (bind _returnvalue (Math|Conversions|ToFloat(Integer) _index))
  (bind (_bblockinghit _binitialoverlap _time _distance _location) (Collision|BreakHitResult HitResult))
  (|CallParentFunction HitResult ImpactPhysicalSurface SpawnedPSC)
  (if (Reactive|Frontend|IsCosmeticObjectDisplayedinFrontendContext 0)
    (Variables|Default|SetIndex (+ (Variables|Default|GetIndex) 1))
    (Rendering|Material|SetScalarParameterValueonMaterials (Variables|Visual|GetWeaponMesh) "ReactivityIndex" _returnvalue)
    (bind _returnvalue_2 (Niagara|SpawnSystematLocation "/BRCosmetics/Weapons/FORT_Melee/Pickaxe_TopicProud/FX/NS_Pickaxe_TopicProud_WeakpointImpact.NS_Pickaxe_TopicProud_WeakpointImpact" _location))
    (Effects|Components|ParticleSystem|SetIntParameter (CallFunction|GetSwingVFXComponent) "ReactivityIndex" (Math|Float|Truncate _returnvalue))
    (Niagara|SetNiagaraVariable(Bool) _returnvalue_2 "bIsFrontEnd" (Reactive|Frontend|IsCosmeticObjectDisplayedinFrontendContext 0))))

(event EventOnPlayWeakpointHitFX
  (bind _index (Variables|Default|GetIndex))
  (bind _returnvalue (Math|Conversions|ToFloat(Integer) _index))
  (|CallParentFunction)
  (Variables|Default|SetIndex (+ (Variables|Default|GetIndex) 1))
  (Rendering|Material|SetScalarParameterValueonMaterials (Variables|Visual|GetWeaponMesh) "ReactivityIndex" _returnvalue)
  (bind _returnvalue_2 (Niagara|SpawnSystematLocation "/BRCosmetics/Weapons/FORT_Melee/Pickaxe_TopicProud/FX/NS_Pickaxe_TopicProud_WeakpointImpact.NS_Pickaxe_TopicProud_WeakpointImpact" (.Location (Collision|BreakHitResult (stale FortWeapon|EventOnPlayImpactFX HitResult)))))
  (Effects|Components|ParticleSystem|SetIntParameter (CallFunction|GetSwingVFXComponent) "ReactivityIndex" (Math|Float|Truncate _returnvalue))
  (Niagara|SetNiagaraVariable(Bool) _returnvalue_2 "bIsFrontEnd" (Reactive|Frontend|IsCosmeticObjectDisplayedinFrontendContext 0)))

(event EventBeginPlay)

(event EventTick
  (params (DeltaSeconds :type Float%20%28single-precision%29)))

(event FortWeapon|EventOnLocalUnEquip
  (bind _index (Variables|Default|GetIndex))
  (bind _returnvalue (Math|Conversions|ToFloat(Integer) _index))
  (|CallParentFunction)
  (Variables|Default|SetIndex 0)
  (Rendering|Material|SetScalarParameterValueonMaterials (Variables|Visual|GetWeaponMesh) "ReactivityIndex" _returnvalue)
  (Effects|Components|ParticleSystem|SetIntParameter (CallFunction|GetIdleVFXComponent) "ReactivityIndex" (Math|Float|Truncate _returnvalue)))

(event FortWeapon|EventOnUnEquippedAndDetached
  (bind _index (Variables|Default|GetIndex))
  (bind _returnvalue (Math|Conversions|ToFloat(Integer) _index))
  (|CallParentFunction)
  (Variables|Default|SetIndex 0)
  (Rendering|Material|SetScalarParameterValueonMaterials (Variables|Visual|GetWeaponMesh) "ReactivityIndex" _returnvalue)
  (Effects|Components|ParticleSystem|SetIntParameter (CallFunction|GetIdleVFXComponent) "ReactivityIndex" (Math|Float|Truncate _returnvalue)))

(fn ConstructionScript
  (|CallParentFunction))
