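; Athena pickaxe cosmetic Blueprint (Troops) that drives an engine-revving idle effect and swing-aligned impact FX.
; Becoming visible, common left/right swings and impacts pulse "isReving" on the idle VFX, using a retriggerable delay to clear it.
; Swing events record a direction index, which selects a rotation offset combined with the hit normal to orient the spawned impact effect.
; That effect also receives physical surface and frontend-preview parameters; construction disables idle particle auto-activation.
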(event EventOnWeaponVisibilityChanged
  (params (bVisible :type Boolean) (bSetForLocalControllerOnly :type Boolean))
  (|CallParentFunction bVisible bSetForLocalControllerOnly)
  (if bVisible
    (CallFunction|RevingEngine)))

(event Custom|Reving%20Engine
  (bind _returnvalue (CallFunction|GetIdleVFXComponent))
  (Effects|Components|ParticleSystem|SetBoolParameter _returnvalue "isReving" true)
  (Utilities|FlowControl|RetriggerableDelay 0.2)
  (Effects|Components|ParticleSystem|SetBoolParameter _returnvalue "isReving" false))

(event FortWeapon|EventOnPlayImpactFX
  (params (HitResult :type Hit%20Result%20Structure%20%28by%20ref%29) (ImpactPhysicalSurface :type EPhysicalSurface%20Enum :default "SurfaceType_Default") (SpawnedPSC :type FXSystem%20Component%20Object%20Reference))
  (bind (_bblockinghit _binitialoverlap _time _distance _location _impactpoint _normal) (Collision|BreakHitResult HitResult))
  (CallFunction|RevingEngine)
  (Utilities|IsValid SpawnedPSC
    (:Is%20Valid
      (bind _output_get (Variables|Default|SetImpact%20FX SpawnedPSC))
      (Utilities|IsValid _output_get
        (:Is%20Valid
          (Transformation|SetWorldRotation (Variables|Default|GetImpact%20FX) (Math|Transform|TransformRotation (Transformation|GetWorldTransform (Variables|Default|GetImpact%20FX)) (Math|Rotator|CombineRotators (Math|Transform|InverseTransformRotation (Transformation|GetWorldTransform _output_get) (Math|Rotator|MakeRotfromX _normal)) (Utilities|Select :Index (Variables|Default|GetSwingDirection) :Option%200 (Variables|Default|GetRightSwing) :Option%201 (Variables|Default|GetLeft%20Swing) :Option%202 (Variables|Default|GetUp%20Swing) :Option%203 (Variables|Default|GetDown%20Swing)))) false false)
          (Effects|Components|ParticleSystem|SetIntParameter (Variables|Default|GetImpact%20FX) "PhysicalSurface" (Utilities|Select :Index ImpactPhysicalSurface :SurfaceType_Default 0 :SurfaceType1 1 :SurfaceType2 2 :SurfaceType3 3 :SurfaceType4 4 :SurfaceType5 5 :SurfaceType6 6 :SurfaceType7 7 :SurfaceType8 8 :SurfaceType9 9 :SurfaceType10 10 :SurfaceType11 11 :SurfaceType12 12 :SurfaceType13 13 :SurfaceType14 14 :SurfaceType15 15 :SurfaceType16 16 :SurfaceType17 17 :SurfaceType18 18 :SurfaceType19 19 :SurfaceType20 20 :SurfaceType21 21 :SurfaceType22 22 :SurfaceType23 23 :SurfaceType24 24 :SurfaceType25 25 :SurfaceType26 26 :SurfaceType27 0 :SurfaceType28 0 :SurfaceType29 0 :SurfaceType30 0 :SurfaceType31 0 :SurfaceType32 0 :SurfaceType33 0 :SurfaceType34 0))
          (Effects|Components|ParticleSystem|SetBoolParameter (Variables|Default|GetImpact%20FX) "bIsFrontEndPreview" (Reactive|Frontend|IsCosmeticObjectDisplayedinFrontendContext self)))))))

(event EventSwingRight
  (|CallParentFunction)
  (Variables|Default|SetSwingDirection 0))

(event EventSwingLeft
  (|CallParentFunction)
  (Variables|Default|SetSwingDirection 1))

(event EventSwingRight2
  (|CallParentFunction)
  (Variables|Default|SetSwingDirection 2))

(event EventSwingLeft2
  (|CallParentFunction)
  (Variables|Default|SetSwingDirection 3))

(event EventSwingLeft_Common
  (|CallParentFunction)
  (CallFunction|RevingEngine))

(event EventSwingRight_Common
  (|CallParentFunction)
  (CallFunction|RevingEngine))

(fn ConstructionScript
  (|CallParentFunction)
  (Components|Activation|SetAutoActivate (CallFunction|GetIdlePSC) false))
