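; Impact Blueprint for the TrustFan pickaxe cosmetic, adding a reactive effect on weak-point hits.
; A hit stamps game time into the weapon mesh material's "Timestamp" and enables idle VFX "ReactivityOn" for four seconds, then clears bisWeakPointHit.
; On cosmetic init, TriggerReactive plays the same sequence if frontend reactivity should display.
; Impact effects copy bisWeakPointHit, which is never set true here, into their "ReactivityOn" parameter.
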(event EventBeginPlay)

(event EventTick
  (params (DeltaSeconds :type Float%20%28single-precision%29)))

(event Custom|TriggerReactive
  (Utilities|FlowControl|Sequence
    (:then_0
      (Rendering|Material|SetScalarParameterValueonMaterials (Visual|GetWeaponMesh) "Timestamp" (Utilities|Time|GetGameTimeinSeconds))
      (Effects|Components|ParticleSystem|SetBoolParameter (CallFunction|GetIdleVFXComponent) "ReactivityOn" true)
      (Utilities|FlowControl|Delay 4.0)
      (Effects|Components|ParticleSystem|SetBoolParameter (CallFunction|GetIdleVFXComponent) "ReactivityOn" false)
      (Variables|Default|SetbisWeakPointHit false))))

(event EventOnPlayWeakpointHitFX
  (Utilities|FlowControl|Sequence
    (:then_0
      (Rendering|Material|SetScalarParameterValueonMaterials (Visual|GetWeaponMesh) "Timestamp" (Utilities|Time|GetGameTimeinSeconds))
      (Effects|Components|ParticleSystem|SetBoolParameter (CallFunction|GetIdleVFXComponent) "ReactivityOn" true)
      (Utilities|FlowControl|Delay 4.0)
      (Effects|Components|ParticleSystem|SetBoolParameter (CallFunction|GetIdleVFXComponent) "ReactivityOn" false)
      (Variables|Default|SetbisWeakPointHit false))))

(event EventOnInitWeaponCosmetics
  (|CallParentFunction)
  (if (Reactive|Frontend|ShouldDisplayFrontendReactivity 0 "Gameplay")
    (CallFunction|TriggerReactive)
    (Utilities|FlowControl|Delay 1.0)))

(event FortWeapon|EventOnPlayImpactFX
  (params (HitResult :type Hit%20Result%20Structure%20%28by%20ref%29) (ImpactPhysicalSurface :type EPhysicalSurface%20Enum :default "SurfaceType_Default") (SpawnedPSC :type FXSystem%20Component%20Object%20Reference))
  (|CallParentFunction HitResult ImpactPhysicalSurface SpawnedPSC)
  (Effects|Components|ParticleSystem|SetBoolParameter SpawnedPSC "ReactivityOn" (Variables|Default|GetbisWeakPointHit)))

(fn ConstructionScript
  (|CallParentFunction))
