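; Cosmetic Blueprint for the Vertigo pickaxe that links its swing trail VFX to the wielding player's character mesh.
; On instigator set and on equip, it waits for the MiscOrTail cosmetic part and feeds that mesh to the trail Niagara component's SkeletalMesh parameter.
; It adds trail and idle VFX to the pawn highlight set so friendly outlines avoid drawing over the tentacles.
; Swings pass the montage section index to the trail; idle particle auto-activation is disabled.
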(event FortWeapon|EventOnInstigatorSet
  (|CallParentFunction)
  (bind _asfort_player_pawn (Utilities|Casting|CastToFortPlayerPawn (Game|GetInstigator)))
  (bind (_pawn _part _meshpart) (Cosmetics|WaitForCosmeticSpecificPartApplied _asfort_player_pawn true "MiscOrTail")
    (:OnPartApplied
      (Utilities|IsValid _meshpart
        (:Is%20Valid
          (bind _asniagara_particle_system_component (Utilities|Casting|CastToNiagaraComponent (CallFunction|GetAnimTrailVFXComponent)))
          (Niagara|SetNiagaraSkeletalMeshComponent _asniagara_particle_system_component "SkeletalMesh" _meshpart))))))

(event FortWeapon|EventOnEquippedAndAttached
  (|CallParentFunction)
  (bind _asfort_player_pawn (Utilities|Casting|CastToFortPlayerPawn (Game|GetInstigator)))
  (bind (_pawn _part _meshpart) (Cosmetics|WaitForCosmeticSpecificPartApplied _asfort_player_pawn true "MiscOrTail")
    (:OnPartApplied
      (Utilities|IsValid _meshpart
        (:Is%20Valid
          (bind _asniagara_particle_system_component (Utilities|Casting|CastToNiagaraComponent (CallFunction|GetAnimTrailVFXComponent)))
          (Niagara|SetNiagaraSkeletalMeshComponent _asniagara_particle_system_component "SkeletalMesh" _meshpart))))))

(event EventOnInitWeaponCosmetics
  ;[1 1 1] Prevents friendly outlines to spawn above the tentacle vfx
  (|CallParentFunction)
  (bind _asfort_player_pawn_athena (Utilities|Casting|CastToFortPlayerPawnAthena (Game|GetInstigator)))
  (Class|FortPlayerPawnAthena|AddComponenttoPawnHighlightSet _asfort_player_pawn_athena (CallFunction|GetAnimTrailVFXComponent))
  (Class|FortPlayerPawnAthena|AddComponenttoPawnHighlightSet _asfort_player_pawn_athena (CallFunction|GetIdleVFXComponent)))

(event EventOnPlayWeaponFireFX
  (params (bPersistentFire :type Boolean) (bSecondaryFire :type Boolean))
  (bind _returnvalue (CallFunction|GetAnimTrailVFXComponent))
  (|CallParentFunction bPersistentFire bSecondaryFire)
  (bind _section_index (Utilities|SetupSwingMontageSectionIndex (Variables|Actor|GetInstigator))
    (:Out
      (Effects|Components|ParticleSystem|SetIntParameter _returnvalue "SwingSectionIndex" _section_index))
    (:Cast%20Failed%20Out
      (Effects|Components|ParticleSystem|SetIntParameter _returnvalue "SwingSectionIndex" -1))))

(event FortWeapon|EventOnPlayImpactFX
  (params (HitResult :type Hit%20Result%20Structure%20%28by%20ref%29) (ImpactPhysicalSurface :type EPhysicalSurface%20Enum :default "SurfaceType_Default") (SpawnedPSC :type FXSystem%20Component%20Object%20Reference))
  (|CallParentFunction HitResult ImpactPhysicalSurface SpawnedPSC))

(fn ConstructionScript
  (|CallParentFunction)
  (Components|Activation|SetAutoActivate (CallFunction|GetIdlePSC) false))
