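; Pickaxe weapon Blueprint for the WarmShadeWagon cosmetic that customizes melee impact effects.
; On impact FX playback, it stores the spawned effect component, calls the parent, switches on physical surface with empty listed cases, and sets the WeakPointHit particle parameter true.
; A server-named custom event then calls a client-named one that sets the same parameter.
; BeginPlay, Tick and the OnRep functions are empty.
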(event EventBeginPlay)

(event EventTick
  (params (DeltaSeconds :type Float%20%28single-precision%29)))

(event FortWeapon|EventOnPlayImpactFX
  (params (HitResult :type Hit%20Result%20Structure%20%28by%20ref%29) (ImpactPhysicalSurface :type EPhysicalSurface%20Enum :default "SurfaceType_Default") (SpawnedPSC :type FXSystem%20Component%20Object%20Reference))
  (Variables|Default|SetImpactFX SpawnedPSC)
  (|CallParentFunction HitResult ImpactPhysicalSurface SpawnedPSC)
  (switch Utilities|FlowControl|Switch|SwitchonEPhysicalSurface ImpactPhysicalSurface
    (:SurfaceType12
      _)
    (:SurfaceType14
      _)
    (:SurfaceType15
      _)
    (:SurfaceType16
      _))
  (Effects|Components|ParticleSystem|SetBoolParameter (Variables|Default|GetImpactFX) "WeakPointHit" true)
  (CallFunction|CEFXServer))

(event Custom|CE_FX_Server
  (CallFunction|CEFXClient))

(event Custom|CE_FX_Client
  (Effects|Components|ParticleSystem|SetBoolParameter (Variables|Default|GetImpactFX) "WeakPointHit" true))

(fn OnRep_ImpactFX)

(fn OnRep_IsWeakPoint)

(fn OnRep_PhysicalSurface)

(fn ConstructionScript
  (|CallParentFunction))
