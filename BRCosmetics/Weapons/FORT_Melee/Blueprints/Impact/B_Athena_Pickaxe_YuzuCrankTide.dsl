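; Blueprint for the YuzuCrankTide pickaxe cosmetic, adding custom impact FX and a material reactivity hook.
; Impact FX events route through ServerImpact to ClientImpact, a noted workaround for replicating the correct hit effect.
; ClientImpact spawns the weak-spot Niagara system for certain surface types or frontend/preview display, passing bIsFrontEnd; both inner branches are effectively identical.
; KillReactivity backdates the weapon material's Timestamp parameter, with no caller here.
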(event Custom|KillReactivity
  (params (AditionalTime :type Float%20%28double-precision%29 :default 2.0))
  ; Reactivity
  (Rendering|Material|SetScalarParameterValue (Class|BTechArtGlobalLibrary|GetDynamicMaterialInstance (Variables|Visual|GetWeaponMesh)) "Timestamp" (- (Utilities|Time|GetGameTimeinSeconds) AditionalTime) :!class /Script/Engine.MaterialInstanceDynamic))

(event EventOnWeaponVisibilityChanged
  (params (bVisible :type Boolean) (bSetForLocalControllerOnly :type Boolean))
  (|CallParentFunction bVisible bSetForLocalControllerOnly))

(event Custom|ServerImpact
  (params (Hit%20Result :type Hit%20Result%20Structure) (Impact%20Physical%20Surface :type EPhysicalSurface%20Enum :default "SurfaceType_Default") (Spawn%20PSC :type FXSystem%20Component%20Object%20Reference))
  (CallFunction|ClientImpact Hit%20Result Impact%20Physical%20Surface Spawn%20PSC))

(event FortWeapon|EventOnPlayImpactFX
  (params (HitResult :type Hit%20Result%20Structure%20%28by%20ref%29) (ImpactPhysicalSurface :type EPhysicalSurface%20Enum :default "SurfaceType_Default") (SpawnedPSC :type FXSystem%20Component%20Object%20Reference))
  (|CallParentFunction HitResult ImpactPhysicalSurface SpawnedPSC)
  (CallFunction|ServerImpact HitResult ImpactPhysicalSurface SpawnedPSC))

(event Custom|ClientImpact
  (params (Hit%20Result :type Hit%20Result%20Structure) (Impact%20Physical%20Surface :type EPhysicalSurface%20Enum :default "SurfaceType_Default") (Spawned%20PSC :type FXSystem%20Component%20Object%20Reference))
  (bind _returnvalue (Reactive|Frontend|IsCosmeticObjectDisplayedinFrontendContext 0))
  (bind _returnvalue_2 (Reactive|Frontend|IsCosmeticObjectDisplayedasPreview 0))
  (bind (_bblockinghit _binitialoverlap _time _distance _location) (Collision|BreakHitResult Hit%20Result))
  (bind _returnvalue_3 (or _returnvalue_2 _returnvalue))
  ; Workaround Solution to get correct Hit Effect to spawn and replicate across clients
  (Variables|Default|SetSpawned%20PSC Spawned%20PSC)
  (if (Math|Boolean|ORBoolean (Utilities|Enum|Equal(Enum) (Collision|GetSurfaceType Hit%20Result) "SurfaceType12") (Utilities|Enum|Equal(Enum) (Collision|GetSurfaceType Hit%20Result) "SurfaceType14") (Utilities|Enum|Equal(Enum) (Collision|GetSurfaceType Hit%20Result) "SurfaceType15") (Utilities|Enum|Equal(Enum) (Collision|GetSurfaceType Hit%20Result) "SurfaceType16") _returnvalue_3)
    (if _returnvalue_3
      (bind _returnvalue_4 (Niagara|SpawnSystematLocation "/BRCosmetics/Weapons/FORT_Melee/Pickaxe_YuzuCrankTide/FX/NS_Pickaxe_YuzuCrankTide_WeakSpotImpact.NS_Pickaxe_YuzuCrankTide_WeakSpotImpact" _location))
      (Variables|Default|SetSpawned%20PSC _returnvalue_4)
      (Niagara|SetNiagaraVariable(Bool) _returnvalue_4 "bIsFrontEnd" _returnvalue_3)
      (else
        (bind _returnvalue_5 (Niagara|SpawnSystematLocation "/BRCosmetics/Weapons/FORT_Melee/Pickaxe_YuzuCrankTide/FX/NS_Pickaxe_YuzuCrankTide_WeakSpotImpact.NS_Pickaxe_YuzuCrankTide_WeakSpotImpact" _location))
        (Variables|Default|SetSpawned%20PSC _returnvalue_5)
        (Niagara|SetNiagaraVariable(Bool) _returnvalue_5 "bIsFrontEnd" _returnvalue_3)))))

(fn ConstructionScript
  (|CallParentFunction))
