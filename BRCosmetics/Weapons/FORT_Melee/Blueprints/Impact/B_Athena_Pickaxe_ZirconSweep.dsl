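; Cosmetic Blueprint for the ZirconSweep pickaxe that drives variant-dependent swing and impact VFX.
; On equip, the mesh variant stage sets dual-wield and alternate-color flags, and the swing Niagara component binds to the pawn mesh.
; Single-wield swings read the montage section; "Swing1" spawns handle-attached, color-aware FX and hides the anim trail.
; Dual-wield impacts force SwingSectionIndex to zero, and construction disables idle FX auto-activation.
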(event FortWeapon|EventOnEquippedAndAttached
  (|CallParentFunction)
  (bind _asplayer_pawn_athena (Utilities|Casting|CastToPlayerPawn_Athena (Variables|Actor|GetInstigator)))
  (bind _index (CosmeticLoadout|GetVariantChannelActiveIndexforSlotActor "(TagName=\"LoadoutSlot.Pickaxe\")" _asplayer_pawn_athena "(TagName=\"Cosmetics.Variant.Channel.Mesh\")" (Utilities|Array|MakeArray "(TagName=\"Cosmetics.Variant.Property.Stage1\")" "(TagName=\"Cosmetics.Variant.Property.Stage2\")" "(TagName=\"Cosmetics.Variant.Property.Stage3\")" "(TagName=\"Cosmetics.Variant.Property.Stage4\")") false false))
  (switch int _index
    (:0
      (Variables|Default|SetIsDifferentColor false)
      (Variables|Default|SetDualWield false))
    (:1
      (Variables|Default|SetIsDifferentColor false)
      (Variables|Default|SetDualWield true))
    (:2
      (Variables|Default|SetIsDifferentColor true)
      (Variables|Default|SetDualWield false))
    (:3
      (Variables|Default|SetIsDifferentColor true)
      (Variables|Default|SetDualWield true)))
  (bind _asfort_pawn (Utilities|Casting|CastToFortPawn (Variables|Actor|GetInstigator)))
  (bind _output_get (Variables|Default|SetSkeletal%20Mesh%20Cast (Variables|/Script/Engine.Character|GetMesh _asfort_pawn)))
  (Utilities|IsValid _output_get
    (:Is%20Valid
      (bind _asniagara_particle_system_component (Utilities|Casting|CastToNiagaraComponent (CallFunction|GetSwingVFXComponent)))
      (Niagara|SetNiagaraSkeletalMeshComponent _asniagara_particle_system_component "Pawn" (Variables|Default|GetSkeletal%20Mesh%20Cast)))))

(event EventSwingLeft
  (bind _current_swing (Variables|Default|GetCurrent%20Swing))
  (bind _returnvalue (== _current_swing "Swing1"))
  (|CallParentFunction)
  (if (not (Variables|Default|GetDualWield))
    (Variables|Default|SetCurrent%20Swing (Animation|Montage|MontageGetCurrentSection (Components|SkeletalMesh|GetAnimInstance (Variables|Default|GetSkeletal%20Mesh%20Cast)) (Animation|Montage|GetCurrentActiveMontage (Components|SkeletalMesh|GetAnimInstance (Variables|Default|GetSkeletal%20Mesh%20Cast)))))
    (Networking|CanExecuteCosmeticEvents
      (:True
        (Utilities|FlowControl|Sequence
          (:then_0
            (if _returnvalue
              (bind _returnvalue_2 (Effects|SpawnFXSystemAttached (Variables|Default|GetSpawn%20Swing%20FX) (Variables|Visual|GetWeaponMesh) "Melee_TwoHanded_Handle"))
              (Effects|Components|ParticleSystem|SetBoolParameter _returnvalue_2 "Variant" (Variables|Default|GetIsDifferentColor))))
          (:then_1
            (Development|SetHiddeninGame (CallFunction|GetAnimTrailVFXComponent) _returnvalue false)))))))

(event EventSwingLeft2
  (bind _current_swing (Variables|Default|GetCurrent%20Swing))
  (bind _returnvalue (== _current_swing "Swing1"))
  (|CallParentFunction)
  (if (not (Variables|Default|GetDualWield))
    (Variables|Default|SetCurrent%20Swing (Animation|Montage|MontageGetCurrentSection (Components|SkeletalMesh|GetAnimInstance (Variables|Default|GetSkeletal%20Mesh%20Cast)) (Animation|Montage|GetCurrentActiveMontage (Components|SkeletalMesh|GetAnimInstance (Variables|Default|GetSkeletal%20Mesh%20Cast)))))
    (Networking|CanExecuteCosmeticEvents
      (:True
        (Utilities|FlowControl|Sequence
          (:then_0
            (if _returnvalue
              (bind _returnvalue_2 (Effects|SpawnFXSystemAttached (Variables|Default|GetSpawn%20Swing%20FX) (Variables|Visual|GetWeaponMesh) "Melee_TwoHanded_Handle"))
              (Effects|Components|ParticleSystem|SetBoolParameter _returnvalue_2 "Variant" (Variables|Default|GetIsDifferentColor))))
          (:then_1
            (Development|SetHiddeninGame (CallFunction|GetAnimTrailVFXComponent) _returnvalue false)))))))

(event EventSwingRight
  (bind _current_swing (Variables|Default|GetCurrent%20Swing))
  (bind _returnvalue (== _current_swing "Swing1"))
  (|CallParentFunction)
  (if (not (Variables|Default|GetDualWield))
    (Variables|Default|SetCurrent%20Swing (Animation|Montage|MontageGetCurrentSection (Components|SkeletalMesh|GetAnimInstance (Variables|Default|GetSkeletal%20Mesh%20Cast)) (Animation|Montage|GetCurrentActiveMontage (Components|SkeletalMesh|GetAnimInstance (Variables|Default|GetSkeletal%20Mesh%20Cast)))))
    (Networking|CanExecuteCosmeticEvents
      (:True
        (Utilities|FlowControl|Sequence
          (:then_0
            (if _returnvalue
              (bind _returnvalue_2 (Effects|SpawnFXSystemAttached (Variables|Default|GetSpawn%20Swing%20FX) (Variables|Visual|GetWeaponMesh) "Melee_TwoHanded_Handle"))
              (Effects|Components|ParticleSystem|SetBoolParameter _returnvalue_2 "Variant" (Variables|Default|GetIsDifferentColor))))
          (:then_1
            (Development|SetHiddeninGame (CallFunction|GetAnimTrailVFXComponent) _returnvalue false)))))))

(event EventSwingRight2
  (bind _current_swing (Variables|Default|GetCurrent%20Swing))
  (bind _returnvalue (== _current_swing "Swing1"))
  (|CallParentFunction)
  (if (not (Variables|Default|GetDualWield))
    (Variables|Default|SetCurrent%20Swing (Animation|Montage|MontageGetCurrentSection (Components|SkeletalMesh|GetAnimInstance (Variables|Default|GetSkeletal%20Mesh%20Cast)) (Animation|Montage|GetCurrentActiveMontage (Components|SkeletalMesh|GetAnimInstance (Variables|Default|GetSkeletal%20Mesh%20Cast)))))
    (Networking|CanExecuteCosmeticEvents
      (:True
        (Utilities|FlowControl|Sequence
          (:then_0
            (if _returnvalue
              (bind _returnvalue_2 (Effects|SpawnFXSystemAttached (Variables|Default|GetSpawn%20Swing%20FX) (Variables|Visual|GetWeaponMesh) "Melee_TwoHanded_Handle"))
              (Effects|Components|ParticleSystem|SetBoolParameter _returnvalue_2 "Variant" (Variables|Default|GetIsDifferentColor))))
          (:then_1
            (Development|SetHiddeninGame (CallFunction|GetAnimTrailVFXComponent) _returnvalue false)))))))

(event FortWeapon|EventOnPlayImpactFX
  (params (HitResult :type Hit%20Result%20Structure%20%28by%20ref%29) (ImpactPhysicalSurface :type EPhysicalSurface%20Enum :default "SurfaceType_Default") (SpawnedPSC :type FXSystem%20Component%20Object%20Reference))
  (|CallParentFunction HitResult ImpactPhysicalSurface SpawnedPSC)
  (if (Variables|Default|GetDualWield)
    (Effects|Components|ParticleSystem|SetIntParameter (Variables|Default|GetImpact%20FX) "SwingSectionIndex" 0)))

(fn ConstructionScript
  (|CallParentFunction)
  (Components|Activation|SetAutoActivate (CallFunction|GetIdlePSC) false))
