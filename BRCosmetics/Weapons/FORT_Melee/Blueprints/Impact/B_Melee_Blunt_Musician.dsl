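; Cosmetic Blueprint for the Musician blunt melee weapon that sets up audio-reactive analysis of the game's Fortnite_Final submix.
; On equip and attach, it registers that submix with the FortAudioAnalysisSubsystem using its AnalyzerData and sets bSubmixRegistered.
; Local unequip, unequip-and-detach and EndPlay each check that flag before unregistering.
; All paths run only where cosmetic events can execute, and this graph does not show what consumes the analysis.
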(event FortWeapon|EventOnEquippedAndAttached
  ; Audio reactive setup for Fortnite_Final Submix - Client only, unregisters on UnEquipped and End Play
  (Networking|CanExecuteCosmeticEvents
    (:True
      (AudioAnalysis|RegisterSubmix (WorldSubsystems|GetFortAudioAnalysisSubsystem) "/Game/Sounds/SubMixes/Fortnite_Final.Fortnite_Final" (Variables|Default|GetAnalyzerData))
      (Variables|Default|SetbSubmixRegistered true))))

(event FortWeapon|EventOnLocalUnEquip
  ; Audio reactive setup for Fortnite_Final Submix - Client only, unregisters on UnEquipped and End Play
  (Networking|CanExecuteCosmeticEvents
    (:True
      (if (Variables|Default|GetbSubmixRegistered)
        (AudioAnalysis|UnRegisterSubmix (WorldSubsystems|GetFortAudioAnalysisSubsystem) "/Game/Sounds/SubMixes/Fortnite_Final.Fortnite_Final")
        (Variables|Default|SetbSubmixRegistered false)))))

(event FortWeapon|EventOnUnEquippedAndDetached
  ; Audio reactive setup for Fortnite_Final Submix - Client only, unregisters on UnEquipped and End Play
  (Networking|CanExecuteCosmeticEvents
    (:True
      (if (Variables|Default|GetbSubmixRegistered)
        (AudioAnalysis|UnRegisterSubmix (WorldSubsystems|GetFortAudioAnalysisSubsystem) "/Game/Sounds/SubMixes/Fortnite_Final.Fortnite_Final")
        (Variables|Default|SetbSubmixRegistered false)))))

(event EventEndPlay
  (params (EndPlayReason :type EEndPlayReason%20Enum :default "Destroyed"))
  ; Audio reactive setup for Fortnite_Final Submix - Client only, unregisters on UnEquipped and End Play
  (Networking|CanExecuteCosmeticEvents
    (:True
      (if (Variables|Default|GetbSubmixRegistered)
        (AudioAnalysis|UnRegisterSubmix (WorldSubsystems|GetFortAudioAnalysisSubsystem) "/Game/Sounds/SubMixes/Fortnite_Final.Fortnite_Final")
        (Variables|Default|SetbSubmixRegistered false)))))

(fn ConstructionScript
  (|CallParentFunction))
