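; Impact Blueprint for the Loofah female one-handed dual-wield melee cosmetic, adjusting hit-counter logic for fruit visuals.
; HandleFrontedHotnessLogic sets both main and offhand hit counters to 2 when both are zero.
; When both counters exceed 8, it calls ResetFruitLogic, noted as unhiding the fruit, then restarts both at 1.
; The construction script only calls its parent.
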(fn HandleFrontedHotnessLogic
  (if (and (== (Variables|Default|GetHitCounter) (Variables|Default|GetHitCounter_Offhand)) (== (Variables|Default|GetHitCounter_Offhand) 0))
    (bind _output_get (Variables|Default|SetHitCounter 2))
    (Variables|Default|SetHitCounter_Offhand _output_get)
    (elif (and (> (Variables|Default|GetHitCounter) 8) (> (Variables|Default|GetHitCounter_Offhand) 8))
      (CallFunction|ResetFruitLogic) ; But we have to unhide the fruit!
      (bind _output_get_2 (Variables|Default|SetHitCounter 1))
      (Variables|Default|SetHitCounter_Offhand _output_get_2))))

(fn ConstructionScript
  (|CallParentFunction))
