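; Dual-wield Squishy pickaxe Blueprint that customizes impact effects and sound.
; On each impact FX event it stores the hit, calls the parent, then activates idle VFX and deactivates swing VFX for the current main or off hand.
; It also plays one of eight player-impact cues at the impact point, cycling in order through a wrapping counter, with no actor-type check in this graph.
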(event FortWeapon|EventOnPlayImpactFX
  (params (HitResult :type Hit%20Result%20Structure%20%28by%20ref%29) (ImpactPhysicalSurface :type EPhysicalSurface%20Enum :default "SurfaceType_Default") (SpawnedPSC :type FXSystem%20Component%20Object%20Reference))
  (bind _returnvalue (CallFunction|GetSwingVFXComponent))
  (bind _returnvalue_2 (CallFunction|GetSwingOffhandVFXComponent))
  (bind (_bblockinghit _binitialoverlap _time _distance _location _impactpoint) (Collision|BreakHitResult (Variables|Default|GetPIckaxe%20Impact)))
  (Variables|Default|SetPIckaxe%20Impact HitResult)
  (|CallParentFunction HitResult ImpactPhysicalSurface SpawnedPSC)
  (Utilities|FlowControl|Sequence
    (:then_0
      (switch Utilities|FlowControl|Switch|SwitchonEFortDualWieldSwingState (CallFunction|GetCurrentSwingState)
        (:MainHand
          (Components|Activation|Activate (CallFunction|GetIdleVFXComponent) false))
        (:OffHand
          (Components|Activation|Activate (CallFunction|GetIdleOffhandVFXComponent) false))))
    (:then_1
      (switch Utilities|FlowControl|Switch|SwitchonEFortDualWieldSwingState (CallFunction|GetCurrentSwingState)
        (:MainHand
          (Utilities|IsValid _returnvalue
            (:Is%20Valid
              (if (Components|Activation|IsActive _returnvalue)
                (Components|Activation|Deactivate _returnvalue)))))
        (:OffHand
          (Utilities|IsValid _returnvalue_2
            (:Is%20Valid
              (if (Components|Activation|IsActive _returnvalue_2)
                (Components|Activation|Deactivate _returnvalue_2)))))))
    (:then_2
      ;[1 1 1] Play special sound for pickaxe when hitting actors
      (Audio|PlaySoundatLocation (Utilities|Select :Index (Variables|Default|GetPixkaxe%20Impact%20Count) :Option%200 "/BRCosmetics/Sounds/PickAxes/Squishy/Pickaxe_Squishy_Impact_Player_01_Cue.Pickaxe_Squishy_Impact_Player_01_Cue" :Option%201 "/BRCosmetics/Sounds/PickAxes/Squishy/Pickaxe_Squishy_Impact_Player_02_Cue.Pickaxe_Squishy_Impact_Player_02_Cue" :Option%202 "/BRCosmetics/Sounds/PickAxes/Squishy/Pickaxe_Squishy_Impact_Player_03_Cue.Pickaxe_Squishy_Impact_Player_03_Cue" :Option%203 "/BRCosmetics/Sounds/PickAxes/Squishy/Pickaxe_Squishy_Impact_Player_04_Cue.Pickaxe_Squishy_Impact_Player_04_Cue" :Option%204 "/BRCosmetics/Sounds/PickAxes/Squishy/Pickaxe_Squishy_Impact_Player_05_Cue.Pickaxe_Squishy_Impact_Player_05_Cue" :Option%205 "/BRCosmetics/Sounds/PickAxes/Squishy/Pickaxe_Squishy_Impact_Player_06_Cue.Pickaxe_Squishy_Impact_Player_06_Cue" :Option%206 "/BRCosmetics/Sounds/PickAxes/Squishy/Pickaxe_Squishy_Impact_Player_07_Cue.Pickaxe_Squishy_Impact_Player_07_Cue" :Option%207 "/BRCosmetics/Sounds/PickAxes/Squishy/Pickaxe_Squishy_Impact_Player_08_Cue.Pickaxe_Squishy_Impact_Player_08_Cue") _impactpoint)
      (bind _result (Math|Integer|IncrementInt (Variables|Default|GetPixkaxe%20Impact%20Count))
        (:%20%20
          (bind _output_get (Variables|Default|SetPixkaxe%20Impact%20Count _result))
          (if (>= _output_get 8)
            (Variables|Default|SetPixkaxe%20Impact%20Count 0)))))))

(fn ConstructionScript
  (|CallParentFunction))
