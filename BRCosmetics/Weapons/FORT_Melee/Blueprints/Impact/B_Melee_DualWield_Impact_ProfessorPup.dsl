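; Dual-wield melee impact Blueprint for the ProfessorPup cosmetic, which hides each weapon's idle particle effect while that hand swings.
; Each swing event runs the parent logic, then deactivates the idle particle component: offhand for left swings, main for right swings.
; A retriggerable 1.2-second delay in each event then reactivates that component without a reset, so repeating the same swing restarts that event's countdown.
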(event EventSwingLeft
  (bind _returnvalue (CallFunction|GetIdleOffhandPSC))
  (|CallParentFunction)
  (Components|Activation|Deactivate _returnvalue)
  (Utilities|FlowControl|RetriggerableDelay 1.2)
  (Components|Activation|Activate _returnvalue false))

(event EventSwingLeft2
  (bind _returnvalue (CallFunction|GetIdleOffhandPSC))
  (|CallParentFunction)
  (Components|Activation|Deactivate _returnvalue)
  (Utilities|FlowControl|RetriggerableDelay 1.2)
  (Components|Activation|Activate _returnvalue false))

(event EventSwingRight
  (bind _returnvalue (CallFunction|GetIdlePSC))
  (|CallParentFunction)
  (Components|Activation|Deactivate _returnvalue)
  (Utilities|FlowControl|RetriggerableDelay 1.2)
  (Components|Activation|Activate _returnvalue false))

(event EventSwingRight2
  (bind _returnvalue (CallFunction|GetIdlePSC))
  (|CallParentFunction)
  (Components|Activation|Deactivate _returnvalue)
  (Utilities|FlowControl|RetriggerableDelay 1.2)
  (Components|Activation|Activate _returnvalue false))

(fn ConstructionScript
  (|CallParentFunction))
