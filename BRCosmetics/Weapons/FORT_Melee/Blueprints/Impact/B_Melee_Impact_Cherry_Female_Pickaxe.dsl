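; Cosmetic weapon Blueprint for the Cherry female pickaxe that drives its Niagara idle effects in the frontend and on swings.
; On init, a frontend display activates the frontend idle system and a timeline feeds its Frontend_Alpha; otherwise that system is deactivated.
; Outside the frontend, each fire event activates the gameplay idle system, animates its light falloff and torso alpha by timeline, then deactivates it after a retriggerable delay, keeping the frontend effect off.
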(event EventOnInitWeaponCosmetics
  (|CallParentFunction)
  (if (Reactive|Frontend|IsCosmeticObjectDisplayedinFrontendContext (Variables|Visual|GetWeaponMesh))
    (bind (_direction _frontend_alpha) (Timeline|FrontEnd PlayFromStart)
      (:Update
        (Niagara|SetNiagaraVariable(Float) (Variables|Default|GetNS_FrontEnd_Pickaxe_Cherry_Idle) "Frontend_Alpha" _frontend_alpha))
      (:Active
        (Components|Activation|SetActive (Variables|Default|GetNS_FrontEnd_Pickaxe_Cherry_Idle) true true)))
    (else
      (Components|Activation|Deactivate (Variables|Default|GetNS_FrontEnd_Pickaxe_Cherry_Idle)))))

(event EventOnPlayWeaponFireFX
  (params (bPersistentFire :type Boolean) (bSecondaryFire :type Boolean))
  (bind _ns_pickaxe_cherry_idle (Variables|Default|GetNS_Pickaxe_Cherry_Idle))
  (|CallParentFunction bPersistentFire bSecondaryFire)
  (if (not (Reactive|Frontend|IsCosmeticObjectDisplayedinFrontendContext (Variables|Visual|GetWeaponMesh)))
    (Utilities|FlowControl|Sequence
      (:then_0
        (Components|Activation|Activate _ns_pickaxe_cherry_idle false)
        (Utilities|FlowControl|Sequence
          (:then_0
            (bind (_direction _light_falloff _torso_alpha) (Timeline|Light PlayFromStart)
              (:Update
                (Niagara|SetNiagaraVariable(Float) (Variables|Default|GetNS_Pickaxe_Cherry_Idle) "Light_falloff" _light_falloff)
                (Niagara|SetNiagaraVariable(Float) (Variables|Default|GetNS_Pickaxe_Cherry_Idle) "Torso_Alpha" _torso_alpha))))
          (:then_1
            (Utilities|FlowControl|RetriggerableDelay 0.95)
            (Components|Activation|Deactivate _ns_pickaxe_cherry_idle))))
      (:then_1
        (Components|Activation|Deactivate (Variables|Default|GetNS_FrontEnd_Pickaxe_Cherry_Idle))))))

(fn ConstructionScript)
