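; Cosmetic Blueprint for the Cube Ninja Male pickaxe that drives its idle Niagara effect in frontend display and during melee swings.
; On BeginPlay in a frontend context, it activates the effect and sets a fixed spline width.
; Outside the frontend, each weapon fire FX event activates the effect and restarts a timeline that animates its size, two alphas and dissolve through Niagara user variables.
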(event EventBeginPlay
  (bind _ns_pickaxe_cube_ninja_male_idle (Variables|Default|GetNS_Pickaxe_Cube_Ninja_Male_Idle))
  (|CallParentFunction)
  (if (Reactive|Frontend|IsCosmeticObjectDisplayedinFrontendContext (Variables|Visual|GetWeaponMesh))
    (Components|Activation|Activate _ns_pickaxe_cube_ninja_male_idle false)
    (Niagara|SetNiagaraVariable(Float) _ns_pickaxe_cube_ninja_male_idle "CubeninjaFX_SplineWidth" 0.6)))

(event EventOnPlayWeaponFireFX
  (params (bPersistentFire :type Boolean) (bSecondaryFire :type Boolean))
  (bind _ns_pickaxe_cube_ninja_male_idle (Variables|Default|GetNS_Pickaxe_Cube_Ninja_Male_Idle))
  (|CallParentFunction bPersistentFire bSecondaryFire)
  (if (not (Reactive|Frontend|IsCosmeticObjectDisplayedinFrontendContext (Variables|Visual|GetWeaponMesh)))
    (Components|Activation|Activate _ns_pickaxe_cube_ninja_male_idle false)
    (bind (_direction _cubeninjafx_size _cubeninjafx_alpha _cubeninjafx_dissolve _cubeninjafx_alpha2) (Timeline|Timeline PlayFromStart)
      (:Update
        (Utilities|FlowControl|Sequence
          (:then_0
            (Niagara|SetNiagaraVariable(Float) _ns_pickaxe_cube_ninja_male_idle "CubeninjaFX_Size" _cubeninjafx_size))
          (:then_1
            (Niagara|SetNiagaraVariable(Float) _ns_pickaxe_cube_ninja_male_idle "CubeninjaFX_Alpha" _cubeninjafx_alpha))
          (:then_2
            (Niagara|SetNiagaraVariable(Float) _ns_pickaxe_cube_ninja_male_idle "CubeninjaFX_Dissolve" _cubeninjafx_dissolve))
          (:then_3
            (Niagara|SetNiagaraVariable(Float) _ns_pickaxe_cube_ninja_male_idle "CubeninjaFX_Alpha2" _cubeninjafx_alpha2)))))))

(fn ConstructionScript
  (|CallParentFunction))
