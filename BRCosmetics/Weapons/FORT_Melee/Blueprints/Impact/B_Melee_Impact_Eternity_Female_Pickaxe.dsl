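; Melee impact Blueprint for the Eternity Female pickaxe cosmetic, controlling its sparkle Niagara effect when the weapon fire FX play.
; On each fire FX event it calls the parent, switches off the sparkles component, then turns it back on after a short retriggerable delay.
; Repeated swings restart the delay, so the sparkles stay off until swinging pauses.
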(event EventOnPlayWeaponFireFX
  (params (bPersistentFire :type Boolean) (bSecondaryFire :type Boolean))
  (|CallParentFunction bPersistentFire bSecondaryFire)
  (Components|Activation|Deactivate (Variables|Default|GetNS_PIckaxe_Eternity_Female_sparkles))
  (Utilities|FlowControl|RetriggerableDelay 0.5)
  (Components|Activation|Activate (Variables|Default|GetNS_PIckaxe_Eternity_Female_sparkles) false))

(fn ConstructionScript)
