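; Majesty pickaxe Blueprint that reflects MajestyConsolidatedQuest progress on the weapon mesh.
; Quest updates keep the highest achieved count, write it and the start/end levels to material parameters, and refresh the Corruption cosmetic slider variant.
; A gated range check leads to OnActualStatChange, which stamps game time on the materials and spawns an attached Niagara burst.
; A repeating timer re-applies the values as a workaround, and quest updates clear it after a delay.
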(event Custom|HackyWorkarundRefreshTimerGuy
  (Utilities|FlowControl|Sequence
    (:then_0
      (CallFunction|InitializeVariable)
      (Rendering|Material|SetScalarParameterValueonMaterials (Variables|Visual|GetWeaponMesh) "SL_CurrentLevel" (Math|Conversions|ToFloat(Integer) (Variables|Default|GetSL_LastValue)))
      (Utilities|FlowControl|DoOnce false
        (:Completed
          (bind _returnvalue (Utilities|Time|SetTimerbyEvent (EventDispatchers|CreateEvent :!event Custom|HackyWorkarundRefreshTimerGuy) 0.1 true))
          (Variables|Default|SetSL_Refresh_Time_Handler _returnvalue))))
    (:then_1
      (Utilities|FlowControl|Gate true
        (:Exit
          (if (and (> (Math|Integer|Max(Integer) (stale OnCustomUpdateObjective(MajestyConsolidatedQuest) AchievedCount) (Variables|Default|GetSL_LastValue)) (Variables|Default|GetSL_StartLevel)) (< (Math|Integer|Max(Integer) (stale OnCustomUpdateObjective(MajestyConsolidatedQuest) AchievedCount) (Variables|Default|GetSL_LastValue)) (Variables|Default|GetSL_EndLevel)))
            (CallFunction|OnActualStatChange)))))))

(event Custom|ResetSafetyGate
  (Utilities|FlowControl|DoOnce false
    (:Completed
      (bind _returnvalue (Utilities|Time|SetTimerbyEvent (EventDispatchers|CreateEvent :!event Custom|HackyWorkarundRefreshTimerGuy) 0.1 true))
      (Variables|Default|SetSL_Refresh_Time_Handler _returnvalue))))

(event EventOnWeaponVisibilityChanged
  (params (bVisible :type Boolean) (bSetForLocalControllerOnly :type Boolean))
  (|CallParentFunction bVisible bSetForLocalControllerOnly)
  (bind _asfort_player_pawn (Utilities|Casting|CastToFortPlayerPawn (Variables|Actor|GetInstigator)))
  (Utilities|FlowControl|DoOnce false))

(event OnCustomUpdateObjective(MajestyConsolidatedQuest)
  (params (PlayerController :type Fort%20Player%20Controller%20Object%20Reference) (AchievedCount :type Integer) (RequiredCount :type Integer))
  (bind _sl_lastvalue (Variables|Default|GetSL_LastValue))
  (bind _returnvalue (Math|Integer|Max(Integer) AchievedCount _sl_lastvalue))
  (Utilities|FlowControl|Sequence
    (:then_0
      (Utilities|FlowControl|Sequence
        (:then_0
          (Utilities|FlowControl|Delay 1.0)
          (Utilities|Time|ClearandInvalidateTimerbyHandle (Variables|Default|GetSL_Refresh_Time_Handler)))
        (:then_1
          (Variables|Default|SetSL_LastValue _returnvalue)
          (Utilities|FlowControl|Sequence
            (:then_0
              (CallFunction|InitializeVariable)
              (Rendering|Material|SetScalarParameterValueonMaterials (Variables|Visual|GetWeaponMesh) "SL_CurrentLevel" (Math|Conversions|ToFloat(Integer) (Variables|Default|GetSL_LastValue)))
              (Utilities|FlowControl|DoOnce false
                (:Completed
                  (bind _returnvalue_2 (Utilities|Time|SetTimerbyEvent (EventDispatchers|CreateEvent :!event Custom|HackyWorkarundRefreshTimerGuy) 0.1 true))
                  (Variables|Default|SetSL_Refresh_Time_Handler _returnvalue_2))))
            (:then_1
              (Utilities|FlowControl|Gate true
                (:Exit
                  (if (and (> _returnvalue (Variables|Default|GetSL_StartLevel)) (< _returnvalue (Variables|Default|GetSL_EndLevel)))
                    (CallFunction|OnActualStatChange)))))))))
    (:then_1
      (CustomCharacter|UpdateLockedFloatSliderVariant (Variables|Default|GetCostmeticItemDefinition) PlayerController "(TagName=\"Cosmetics.Variant.Channel.Corruption\")"))))

(event Custom|OnActualStatChange
  (Rendering|Material|SetScalarParameterValueonMaterials (Variables|Visual|GetWeaponMesh) "SL_GameTimeFromBP" (Utilities|Time|GetGameTimeinSeconds))
  (Effects|SpawnFXSystemAttached "/BRCosmetics/Characters/Player/Male/Medium/Bodies/M_MED_Majesty/FX/NS_Majesty_Pickaxe_Trigger.NS_Majesty_Pickaxe_Trigger" (Variables|Visual|GetWeaponMesh) "b_AR_Root"))

(fn InitializeVariable
  (bind _weaponmesh (Variables|Visual|GetWeaponMesh))
  (Rendering|Material|SetScalarParameterValueonMaterials _weaponmesh "SL_StartAtLevel" (Math|Conversions|ToFloat(Integer) (Variables|Default|GetSL_StartLevel)))
  (Rendering|Material|SetScalarParameterValueonMaterials _weaponmesh "SL_StopAtLevel" (Math|Conversions|ToFloat(Integer) (Variables|Default|GetSL_EndLevel))))

(fn ConstructionScript)
