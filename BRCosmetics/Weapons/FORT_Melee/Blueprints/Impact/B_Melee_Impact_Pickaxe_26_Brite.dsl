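; Brite pickaxe Blueprint that makes the weapon's unicorn horn glow and grow as the watched kill count rises.
; On each kill update it calls the parent, stores the count, then maps it through separate clamped threshold ranges to 0–1.
; One value drives the HornGlowINT material parameter on the weapon mesh.
; The other sets HornScale on that mesh's ANimBP_UnicornHorn anim instance, an asset stored under Pickaxe_29.
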(event EventUpdateBasedOnKills
  (params (Watched%20Kills :type Integer))
  (|CallParentFunction Watched%20Kills)
  (Variables|Default|SetKillCount (Math|Conversions|ToFloat(Integer) Watched%20Kills))
  (Utilities|FlowControl|Sequence
    (:then_0
      ;[1 1 1] Glow
      (Rendering|Material|SetScalarParameterValueonMaterials (Variables|Visual|GetWeaponMesh) "HornGlowINT" (Math|Float|MapRangeClamped (Variables|Default|GetKillCount) (Variables|Default|GetGlowKillThresholdMin) (Variables|Default|GetGlowKillThresholdMax) 0.0 1.0)))
    (:then_1
      ;[1 1 1] Horn Grow
      (bind _asanim_bp_unicorn_horn (Utilities|Casting|CastToANimBP_UnicornHorn (Components|SkeletalMesh|GetAnimInstance (Variables|Visual|GetWeaponMesh))))
      (Variables|/BRCosmetics/Weapons/FORT_Melee/Pickaxe_29/Meshes/ANimBP_UnicornHorn.ANimBP_UnicornHorn_C|SetHornScale (Math|Float|MapRangeClamped (Variables|Default|GetKillCount) (Variables|Default|GetHornScaleKillThresholdMin) (Variables|Default|GetHornScaleKillThresholdMax) 0.0 1.0) _asanim_bp_unicorn_horn))))

(fn ConstructionScript
  (|CallParentFunction))
