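; Cosmetic weapon Blueprint for the Barbarian melee pickaxe, adding a brief particle and emissive effect when the weapon fires.
; Its construction script creates a dynamic material instance on the weapon mesh, and BeginPlay switches the idle particle system off.
; Each fire FX event activates that particle system and halves EmissiveMultiplier, then restores it and deactivates the particles after a short retriggerable delay, so repeated swings extend the effect.
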(event EventBeginPlay
  (|CallParentFunction)
  (Components|Activation|Deactivate (CallFunction|GetIdlePSC)))

(event EventOnPlayWeaponFireFX
  (params (bPersistentFire :type Boolean) (bSecondaryFire :type Boolean))
  (|CallParentFunction bPersistentFire bSecondaryFire)
  (Components|Activation|Activate (CallFunction|GetIdlePSC) false)
  (Rendering|Material|SetVectorParameterValue (Variables|Default|GetMID) "EmissiveMultiplier" "(R=0.500000,G=0.500000,B=0.500000,A=1.000000)" :!class /Script/Engine.MaterialInstanceDynamic)
  (Utilities|FlowControl|RetriggerableDelay 0.32)
  (Rendering|Material|SetVectorParameterValue (Variables|Default|GetMID) "EmissiveMultiplier" "(R=1.000000,G=1.000000,B=1.000000,A=1.000000)" :!class /Script/Engine.MaterialInstanceDynamic)
  (Components|Activation|Deactivate (CallFunction|GetIdlePSC)))

(fn ConstructionScript
  (|CallParentFunction)
  (bind _returnvalue (Rendering|Material|CreateDynamicMaterialInstance (Variables|Visual|GetWeaponMesh) 0 0 "None" :!class /Script/Engine.PrimitiveComponent))
  (Variables|Default|SetMID _returnvalue))
