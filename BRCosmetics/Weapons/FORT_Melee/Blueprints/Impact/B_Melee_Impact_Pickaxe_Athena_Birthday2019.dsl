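; Birthday 2019 pickaxe Blueprint that toggles its cosmetic particle effects when the weapon plays fire FX.
; After the parent call, each fire event deactivates the reload-smoke component, calls BackToLife and restarts a one-second retriggerable delay.
; BackToLife and the delay's completion each reactivate the smoke, then the idle particles 0.05 seconds later.
; Both reactivation paths sit behind DoOnce gates with no visible reset, so each runs only once.
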(event EventOnPlayWeaponFireFX
  (params (bPersistentFire :type Boolean) (bSecondaryFire :type Boolean))
  (bind _p_pistol_reloadsmoke (Variables|Default|GetP_Pistol_ReloadSmoke))
  (|CallParentFunction bPersistentFire bSecondaryFire)
  (Utilities|FlowControl|Sequence
    (:then_0
      (Components|Activation|Deactivate _p_pistol_reloadsmoke)
      (CallFunction|BacktoLife))
    (:then_1
      (Utilities|FlowControl|RetriggerableDelay 1.0)
      (Utilities|FlowControl|DoOnce false
        (:Completed
          (Components|Activation|Activate _p_pistol_reloadsmoke false)
          (Utilities|FlowControl|Delay 0.05)
          (Components|Activation|Activate (CallFunction|GetIdlePSC) false))))))

(event Custom|BackToLife
  (Utilities|FlowControl|DoOnce false
    (:Completed
      (Components|Activation|Activate (Variables|Default|GetP_Pistol_ReloadSmoke) false)
      (Utilities|FlowControl|Delay 0.05)
      (Components|Activation|Activate (CallFunction|GetIdlePSC) false))))

(fn ConstructionScript
  (|CallParentFunction))
