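; Melee weapon Blueprint for the Black Widow pickaxe that makes it glow red and plays its idle particle effect on impact.
; Idle particles stay off through construction, cosmetic init and visibility changes.
; On impact it activates them and boosts the mesh's red EmissiveMultiplier.
; After a short delay it deactivates the particles, and a timeline drives the emissive between base and boosted red.
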(event EventOnInitWeaponCosmetics
  (Components|Activation|Deactivate (CallFunction|GetIdlePSC)))

(event EventOnWeaponVisibilityChanged
  (params (bVisible :type Boolean) (bSetForLocalControllerOnly :type Boolean))
  (|CallParentFunction bVisible bSetForLocalControllerOnly)
  (Components|Activation|Deactivate (CallFunction|GetIdlePSC)))

(event FortWeapon|EventOnPlayImpactFX
  (params (HitResult :type Hit%20Result%20Structure%20%28by%20ref%29) (ImpactPhysicalSurface :type EPhysicalSurface%20Enum :default "SurfaceType_Default") (SpawnedPSC :type FXSystem%20Component%20Object%20Reference))
  (bind _returnvalue (CallFunction|GetIdlePSC))
  (|CallParentFunction HitResult ImpactPhysicalSurface SpawnedPSC)
  (Components|Activation|Activate _returnvalue false)
  (Rendering|Material|SetVectorParameterValueonMaterials (Variables|Visual|GetWeaponMesh) "EmissiveMultiplier" "35.000000, 0, 0")
  (Utilities|FlowControl|Delay 2.0)
  (Components|Activation|Deactivate _returnvalue)
  (bind (_direction _emissive) (Timeline|Timeline_0 PlayFromStart)
    (:Update
      (Rendering|Material|SetVectorParameterValueonMaterials (Variables|Visual|GetWeaponMesh) "EmissiveMultiplier" (Math|Conversions|ToVector(LinearColor) (Math|Color|Lerp(LinearColor) "(R=1.000000,G=0.000000,B=0.000000,A=1.000000)" "(R=35.000000,G=0.000000,B=0.000000,A=1.000000)" _emissive))))))

(fn ConstructionScript
  (|CallParentFunction)
  (Components|Activation|SetAutoActivate (CallFunction|GetIdlePSC) false))
