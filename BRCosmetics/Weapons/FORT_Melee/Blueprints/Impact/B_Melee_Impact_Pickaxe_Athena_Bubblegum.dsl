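; Impact Blueprint for the Bubblegum pickaxe that adds hit-driven bubble pop effects and sounds to the parent impact FX.
; The construction script attaches the BubblePop particle component to the weapon mesh at dyn_bubble.
; Each impact stores the hit, calls the parent, briefly pulses SuccessfulHit, and advances a counter that fires the pop every fourth hit.
; A separate counter cycles through four impact cues in order, played at the impact point.
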(event FortWeapon|EventOnPlayImpactFX
  (params (HitResult :type Hit%20Result%20Structure%20%28by%20ref%29) (ImpactPhysicalSurface :type EPhysicalSurface%20Enum :default "SurfaceType_Default") (SpawnedPSC :type FXSystem%20Component%20Object%20Reference))
  (bind (_bblockinghit _binitialoverlap _time _distance _location _impactpoint) (Collision|BreakHitResult (Variables|Default|GetPickaxe%20Impact)))
  (bind _hitcounter (Variables|Default|GetHitCounter))
  ;[1 1 1] Impact FX
  (Variables|Default|SetPickaxe%20Impact HitResult)
  (|CallParentFunction HitResult ImpactPhysicalSurface SpawnedPSC)
  (Utilities|FlowControl|Sequence
    (:then_0
      (Variables|Default|SetSuccessfulHit true)
      (Utilities|FlowControl|Delay 0.1)
      (Variables|Default|SetSuccessfulHit false))
    (:then_1
      (if (== _hitcounter 3)
        (Variables|Default|SetHitCounter 0)
        (Components|Activation|Activate (Variables|Default|GetBubblePop_PSC) true)
        (else
          (Variables|Default|SetHitCounter (+ _hitcounter 1)))))
    (:then_2
      (Audio|PlaySoundatLocation (Utilities|Select :Index (Variables|Default|GetPickaxe%20Impact%20Count) :Option%200 "/BRCosmetics/Sounds/PickAxes/Bubblegum/Pickaxe_Bubblegum_Impact_Player_01_Cue.Pickaxe_Bubblegum_Impact_Player_01_Cue" :Option%201 "/BRCosmetics/Sounds/PickAxes/Bubblegum/Pickaxe_Bubblegum_Impact_Player_02_Cue.Pickaxe_Bubblegum_Impact_Player_02_Cue" :Option%202 "/BRCosmetics/Sounds/PickAxes/Bubblegum/Pickaxe_Bubblegum_Impact_Player_03_Cue.Pickaxe_Bubblegum_Impact_Player_03_Cue" :Option%203 "/BRCosmetics/Sounds/PickAxes/Bubblegum/Pickaxe_Bubblegum_Impact_Player_04_Cue.Pickaxe_Bubblegum_Impact_Player_04_Cue") _impactpoint)
      (bind _result (Math|Integer|IncrementInt (Variables|Default|GetPickaxe%20Impact%20Count))
        (:%20%20
          (bind _output_get (Variables|Default|SetPickaxe%20Impact%20Count _result))
          (if (>= _output_get 4)
            (Variables|Default|SetPickaxe%20Impact%20Count 0)))))))

(fn ConstructionScript
  (|CallParentFunction)
  (Transformation|AttachComponentToComponent (Variables|Default|GetBubblePop_PSC) (Variables|Visual|GetWeaponMesh) "dyn_bubble" "KeepWorld" "SnapToTarget" "KeepWorld" true))
