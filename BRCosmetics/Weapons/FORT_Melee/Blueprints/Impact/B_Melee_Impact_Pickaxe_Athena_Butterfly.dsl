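; Cosmetic weapon Blueprint for the Butterfly pickaxe that drives its idle, swing trail and reappearance effects.
; On equip, visibility change or swing, it turns on the star trail and hides the idle FX.
; Retriggerable one-second delays then stop the trail and restore the idle FX with an explosion, appear sound and timeline-driven "scaleFX" ramp.
; The swing particles also start shortly after equip or visibility change and stop on local unequip.
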(event EventBeginPlay
  (|CallParentFunction)
  (Components|Activation|Activate (CallFunction|GetIdlePSC) true))

(event EventOnWeaponVisibilityChanged
  (params (bVisible :type Boolean) (bSetForLocalControllerOnly :type Boolean))
  (bind _p_butterflypickaxe_trailstar (Variables|Default|GetP_ButterflyPickaxe_TrailStar))
  (|CallParentFunction bVisible false)
  (Utilities|FlowControl|Sequence
    (:then_0
      (Utilities|FlowControl|Delay 0.2)
      (Components|Activation|Activate (CallFunction|GetSwingPSC) true))
    (:then_1
      (Utilities|FlowControl|Sequence
        (:then_1
          (Utilities|FlowControl|RetriggerableDelay 1.0)
          (Components|Activation|Activate (Variables|Default|GetP_ButterflyPickaxe_MagicExplosion) true)
          (Components|Activation|Activate (Variables|Default|GetP_ButterflyPickaxe_IdleFX) true)
          (Audio|SpawnSoundAttached "/BRCosmetics/Sounds/PickAxes/Butterfly/Pickaxe_Butterfly_Appear_Cue.Pickaxe_Butterfly_Appear_Cue" (Variables|Default|GetP_ButterflyPickaxe_MagicExplosion))
          (bind (_direction _newtrack_0) (Timeline|Timeline_0 PlayFromStart)
            (:Update
              (Effects|Components|ParticleSystem|SetVectorParameter (CallFunction|GetIdlePSC) "scaleFX" (Math|Conversions|ToVector(Float) _newtrack_0)))))
        (:then_2
          (Components|Activation|Activate _p_butterflypickaxe_trailstar true)
          (Components|Activation|Deactivate (Variables|Default|GetP_ButterflyPickaxe_IdleFX))
          (Utilities|FlowControl|RetriggerableDelay 1.0)
          (Components|Activation|Deactivate _p_butterflypickaxe_trailstar))))))

(event EventOnPlayWeaponFireFX
  (params (bPersistentFire :type Boolean) (bSecondaryFire :type Boolean))
  (bind _p_butterflypickaxe_trailstar (Variables|Default|GetP_ButterflyPickaxe_TrailStar))
  (|CallParentFunction bPersistentFire bSecondaryFire)
  (Utilities|FlowControl|Sequence
    (:then_1
      (Utilities|FlowControl|RetriggerableDelay 1.0)
      (Components|Activation|Activate (Variables|Default|GetP_ButterflyPickaxe_MagicExplosion) true)
      (Components|Activation|Activate (Variables|Default|GetP_ButterflyPickaxe_IdleFX) true)
      (Audio|SpawnSoundAttached "/BRCosmetics/Sounds/PickAxes/Butterfly/Pickaxe_Butterfly_Appear_Cue.Pickaxe_Butterfly_Appear_Cue" (Variables|Default|GetP_ButterflyPickaxe_MagicExplosion))
      (Timeline|Timeline_0 PlayFromStart))
    (:then_2
      (Components|Activation|Activate _p_butterflypickaxe_trailstar true)
      (Components|Activation|Deactivate (Variables|Default|GetP_ButterflyPickaxe_IdleFX))
      (Utilities|FlowControl|RetriggerableDelay 1.0)
      (Components|Activation|Deactivate _p_butterflypickaxe_trailstar))))

(event FortWeapon|EventOnEquippedAndAttached
  (bind _p_butterflypickaxe_trailstar (Variables|Default|GetP_ButterflyPickaxe_TrailStar))
  (|CallParentFunction)
  (Utilities|FlowControl|Sequence
    (:then_0
      (Utilities|FlowControl|Sequence
        (:then_1
          (Utilities|FlowControl|RetriggerableDelay 1.0)
          (Components|Activation|Activate (Variables|Default|GetP_ButterflyPickaxe_MagicExplosion) true)
          (Components|Activation|Activate (Variables|Default|GetP_ButterflyPickaxe_IdleFX) true)
          (Audio|SpawnSoundAttached "/BRCosmetics/Sounds/PickAxes/Butterfly/Pickaxe_Butterfly_Appear_Cue.Pickaxe_Butterfly_Appear_Cue" (Variables|Default|GetP_ButterflyPickaxe_MagicExplosion))
          (Timeline|Timeline_0 PlayFromStart))
        (:then_2
          (Components|Activation|Activate _p_butterflypickaxe_trailstar true)
          (Components|Activation|Deactivate (Variables|Default|GetP_ButterflyPickaxe_IdleFX))
          (Utilities|FlowControl|RetriggerableDelay 1.0)
          (Components|Activation|Deactivate _p_butterflypickaxe_trailstar))))
    (:then_1
      (Utilities|FlowControl|Delay 0.2)
      (Components|Activation|Activate (CallFunction|GetSwingPSC) true))))

(event FortWeapon|EventOnLocalUnEquip
  (|CallParentFunction)
  (Components|Activation|Deactivate (CallFunction|GetSwingPSC)))

(fn ConstructionScript
  (|CallParentFunction))
