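; Melee weapon Blueprint for the Cardboard Crew pickaxe that tracks impacts to flag its boxes as broken.
; Each impact stores the hit result and increments a counter, and equipping resets the counter and both break flags.
; The third hit sets BreakBoxA, activates the swing particle component and spawns a pooled burst at the impact, while the sixth sets BreakBoxB.
; On impact, the swing component is repositioned if the entire party is in the frontend.
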(event FortWeapon|EventOnPlayImpactFX
  (params (HitResult :type Hit%20Result%20Structure%20%28by%20ref%29) (ImpactPhysicalSurface :type EPhysicalSurface%20Enum :default "SurfaceType_Default") (SpawnedPSC :type FXSystem%20Component%20Object%20Reference))
  (bind (_bblockinghit _binitialoverlap _time _distance _location) (Collision|BreakHitResult (Variables|Default|GetPickaxe%20Impact)))
  (bind _hitcounter (Variables|Default|GetHitCounter))
  ;[1 1 1] qq
  (Variables|Default|SetPickaxe%20Impact HitResult)
  (|CallParentFunction HitResult ImpactPhysicalSurface SpawnedPSC)
  (if (PartyContext|IsEntirePartyinFrontEnd (BlueprintContext|GetFortPartyContext))
    (Transformation|SetRelativeLocation (Variables|Default|GetP_CardboardCrew_Pickaxe_Swing_at3) "-2.069641,26.817804,51.714413" false false))
  (Utilities|FlowControl|Sequence
    (:then_0
      ;;[1 0.1455 0.8004] Increment the Hit Counter
      (Variables|Default|SetHitCounter (+ (Variables|Default|GetHitCounter) 1)))
    (:then_1
      ;;[0.4125 0.4818 1] Inform Boxes Breaking after number of Hits
      (if (== _hitcounter 3)
        (Variables|Default|SetBreakBoxA true)
        (Components|Activation|Activate (Variables|Default|GetP_CardboardCrew_Pickaxe_Swing_at3) false)
        (Effects|SpawnEmitteratLocation :EmitterTemplate (Variables|Default|GetCardboardCrew%20Particle%20System) :Location _location :Scale "1.000000,1.000000,1.000000" :bAutoDestroy true :PoolingMethod "AutoRelease" :bAutoActivateSystem true)))
    (:then_2
      (if (== _hitcounter 6)
        (Variables|Default|SetBreakBoxB true)))))

(event FortWeapon|EventOnEquippedAndAttached
  ;[1 1 1] Reset on Equip
  (|CallParentFunction)
  (Variables|Default|SetHitCounter 0)
  (Variables|Default|SetBreakBoxA false)
  (Variables|Default|SetBreakBoxB false))

(fn ConstructionScript
  (|CallParentFunction))
