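; Cosmetic Blueprint for the Cowgirl pickaxe that drives material parameters on its FX mesh from player motion and swings.
; A looping timer started on cosmetics init acts as a pseudo-tick, gated to cosmetic-capable contexts.
; It smooths the instigating pawn's scaled velocity into FlameDrift and writes it to "SkirtDrag".
; Left and right swings restart a shared timeline driving "SwingFade" to fade the idle steam FX.
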(event EventOnInitWeaponCosmetics
  (|CallParentFunction)
  (Utilities|Time|SetTimerbyEvent (delegate Custom|PseudoTick) 0.05 true))

(event Custom|PseudoTick
  (bind _returnvalue (Game|GetInstigator))
  (bind (_asplayer_pawn_athena _bsuccess) (Utilities|Casting|CastToPlayerPawn_Athena _returnvalue))
  (Networking|CanExecuteCosmeticEvents
    (:True
      (if _bsuccess
        (bind _output_get (Variables|Default|SetFlameDrift (Math|Interpolation|VInterpTo (Variables|Default|GetFlameDrift) (* (Transformation|GetVelocity _asplayer_pawn_athena) (Variables|Default|GetScale%20Velocity)) (Utilities|Time|GetWorldDeltaSeconds) 2.5)))
        (Rendering|Material|SetVectorParameterValueonMaterials (Variables|Default|GetFX_Mesh) "SkirtDrag" _output_get)))))

(event EventSwingLeft
  ;[0.5772 0.783 0.9531 0.5] Fade Idle Steam FX when swinging
  (|CallParentFunction)
  (bind (_direction _fade) (Timeline|FadeTimeline PlayFromStart)
    (:Update
      (Rendering|Material|SetScalarParameterValueonMaterials (Variables|Default|GetFX_Mesh) "SwingFade" _fade))))

(event EventSwingRight
  ;[0.5772 0.783 0.9531 0.5] Fade Idle Steam FX when swinging
  (|CallParentFunction)
  (Timeline|FadeTimeline PlayFromStart))

(fn ConstructionScript
  (|CallParentFunction))
