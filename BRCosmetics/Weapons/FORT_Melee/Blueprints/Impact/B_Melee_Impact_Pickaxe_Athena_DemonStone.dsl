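; Blueprint for the DemonStone melee pickaxe, adding custom swing and impact effects on top of its parent class.
; On weapon fire it disables the "Lava" emitter on the idle particle component, then re-enables it after a retriggerable one-second delay, so repeated swings keep it off.
; On impact it calls the parent, then spawns the DemonStone impact particle at the hit point, oriented along the impact normal.
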(event Custom|CE_IdleEffectRespawn
  (bind _returnvalue (CallFunction|GetIdlePSC))
  (Effects|Components|ParticleSystem|SetEmitterEnable _returnvalue "Lava" false)
  (Utilities|FlowControl|RetriggerableDelay 1.0)
  (Effects|Components|ParticleSystem|SetEmitterEnable _returnvalue "Lava" true))

(event EventOnPlayWeaponFireFX
  (params (bPersistentFire :type Boolean) (bSecondaryFire :type Boolean))
  (|CallParentFunction bPersistentFire bSecondaryFire)
  (CallFunction|CEIdleEffectRespawn))

(event FortWeapon|EventOnPlayImpactFX
  (params (HitResult :type Hit%20Result%20Structure%20%28by%20ref%29) (ImpactPhysicalSurface :type EPhysicalSurface%20Enum :default "SurfaceType_Default") (SpawnedPSC :type FXSystem%20Component%20Object%20Reference))
  (bind (_bblockinghit _binitialoverlap _time _distance _location _impactpoint _normal _impactnormal) (Collision|BreakHitResult HitResult))
  (|CallParentFunction HitResult ImpactPhysicalSurface SpawnedPSC)
  (Effects|SpawnEmitteratLocation "/BRCosmetics/Weapons/FORT_Melee/Pickaxe_DemonStone/FX/P_Pickaxe_DemonStone_Impact.P_Pickaxe_DemonStone_Impact" _impactpoint (Math|Rotator|MakeRotfromX _impactnormal)))

(fn ConstructionScript
  (|CallParentFunction))
