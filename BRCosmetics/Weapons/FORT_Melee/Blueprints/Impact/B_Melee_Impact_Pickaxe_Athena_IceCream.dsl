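; Impact Blueprint for the Ice Cream pickaxe cosmetic that pauses its idle particle effect while the weapon's fire effects play.
; On each weapon fire FX event it runs the parent behavior, then deactivates the idle particle system component.
; A 0.8 second retriggerable delay restores the idle effect afterward, so repeated swings keep extending the pause until swinging stops.
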(event EventOnPlayWeaponFireFX
  (params (bPersistentFire :type Boolean) (bSecondaryFire :type Boolean))
  (bind _returnvalue (CallFunction|GetIdlePSC))
  (|CallParentFunction bPersistentFire bSecondaryFire)
  (Components|Activation|Deactivate _returnvalue)
  (Utilities|FlowControl|RetriggerableDelay 0.8)
  (Components|Activation|Activate _returnvalue false))

(fn ConstructionScript
  (|CallParentFunction))
