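; Athena melee weapon Blueprint for the Log pickaxe cosmetic, adding swing trail and impact effects.
; On equip, it attaches two trail components to the player pawn mesh's root.
; Right and left swings activate trail 01 and 02 respectively.
; On impact, it stores the hit result, spawns the log impact emitter oriented along the impact normal, then deactivates an active trail, checking 02 only when 01 is valid.
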(event FortWeapon|EventOnEquippedAndAttached
  (|CallParentFunction)
  (bind _asplayer_pawn_athena (Utilities|Casting|CastToPlayerPawn_Athena (Game|GetInstigator)))
  (Transformation|AttachComponentToComponent (Variables|Default|GetP_Melee_Trail_01) (Variables|/Script/Engine.Character|GetMesh _asplayer_pawn_athena) "root" "SnapToTarget" "SnapToTarget" "KeepRelative" true)
  (Transformation|AttachComponentToComponent (Variables|Default|GetP_Melee_Trail_02) (Variables|/Script/Engine.Character|GetMesh _asplayer_pawn_athena) "root" "SnapToTarget" "SnapToTarget" "KeepRelative" true))

(event EventSwingLeft
  (|CallParentFunction)
  (Components|Activation|Activate (Variables|Default|GetP_Melee_Trail_02) true))

(event EventSwingRight
  (|CallParentFunction)
  (Components|Activation|Activate (Variables|Default|GetP_Melee_Trail_01) true))

(event FortWeapon|EventOnPlayImpactFX
  (params (HitResult :type Hit%20Result%20Structure%20%28by%20ref%29) (ImpactPhysicalSurface :type EPhysicalSurface%20Enum :default "SurfaceType_Default") (SpawnedPSC :type FXSystem%20Component%20Object%20Reference))
  (bind (_bblockinghit _binitialoverlap _time _distance _location _impactpoint _normal _impactnormal) (Collision|BreakHitResult (Variables|Default|GetPickaxe%20Impact)))
  (bind _p_melee_trail_01 (Variables|Default|GetP_Melee_Trail_01))
  (bind _p_melee_trail_02 (Variables|Default|GetP_Melee_Trail_02))
  (bind _output_get (Variables|Default|SetPickaxe%20Impact HitResult))
  (|CallParentFunction _output_get ImpactPhysicalSurface SpawnedPSC)
  (Utilities|FlowControl|Sequence
    (:then_0
      (Effects|SpawnEmitteratLocation "/BRCosmetics/Weapons/FORT_Melee/Pickaxe_Log/FX/P_Log_Pickaxe_Impact.P_Log_Pickaxe_Impact" _location (Math|Conversions|RotationFromXVector _impactnormal)))
    (:then_1
      (Utilities|IsValid _p_melee_trail_01
        (:Is%20Valid
          (if (Components|Activation|IsActive _p_melee_trail_01)
            (Components|Activation|Deactivate (Variables|Default|GetP_Melee_Trail_01))
            (else
              _))
          (Utilities|IsValid _p_melee_trail_02
            (:Is%20Valid
              (if (Components|Activation|IsActive _p_melee_trail_02)
                (Components|Activation|Deactivate _p_melee_trail_02)))))
        (:Is%20Not%20Valid
          _)))))

(fn ConstructionScript
  (|CallParentFunction))
