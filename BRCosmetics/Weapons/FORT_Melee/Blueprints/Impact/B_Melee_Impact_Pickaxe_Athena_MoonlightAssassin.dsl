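; Melee pickaxe Blueprint for the Moonlight Assassin cosmetic that toggles an idle mist effect around weapon fire.
; On fire FX, it calls the parent, prints a leftover "Hello" debug string, and disables the idle particle system's MovingMist emitter.
; A retriggerable delay re-enables the mist one second after the latest trigger.
; Left and right swing events call their parents with false, ignoring the incoming first-swing flag.
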(event EventOnPlayWeaponFireFX
  (params (bPersistentFire :type Boolean) (bSecondaryFire :type Boolean))
  (bind _returnvalue (CallFunction|GetIdlePSC))
  (|CallParentFunction bPersistentFire bSecondaryFire)
  (Development|PrintString "Hello" true true "(R=1.000000,G=0.781470,B=0.000000,A=1.000000)")
  (Effects|Components|ParticleSystem|SetEmitterEnable _returnvalue "MovingMist" false)
  (Utilities|FlowControl|RetriggerableDelay 1.0)
  (Effects|Components|ParticleSystem|SetEmitterEnable _returnvalue "MovingMist" true))

(event InterfacePlayerPawn|EventMeleeSwingLeft
  (params (First%20Left :type Boolean))
  (|CallParentFunction false))

(event InterfacePlayerPawn|EventMeleeSwingRight
  (params (First%20Right :type Boolean))
  (|CallParentFunction false))

(fn ConstructionScript
  (|CallParentFunction))
