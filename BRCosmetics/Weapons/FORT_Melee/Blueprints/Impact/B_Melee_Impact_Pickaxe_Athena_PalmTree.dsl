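; Impact effect Blueprint for the Palm Tree pickaxe melee cosmetic in Battle Royale.
; When the weapon fire FX event plays, it runs the parent logic, then activates its P_Pickaxe_Palmtree component with a reset and enables shadow casting on it.
; The construction script only calls the parent implementation.
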(event EventOnPlayWeaponFireFX
  (params (bPersistentFire :type Boolean) (bSecondaryFire :type Boolean))
  (bind _p_pickaxe_palmtree (Variables|Default|GetP_Pickaxe_Palmtree))
  (|CallParentFunction bPersistentFire bSecondaryFire)
  (Components|Activation|Activate _p_pickaxe_palmtree true)
  (Rendering|SetCastShadow _p_pickaxe_palmtree true))

(fn ConstructionScript
  (|CallParentFunction))
