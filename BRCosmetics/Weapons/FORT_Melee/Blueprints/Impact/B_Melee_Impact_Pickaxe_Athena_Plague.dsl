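; Plague pickaxe weapon cosmetic whose emissive glow and glow particles charge up as the storm's next shrink approaches.
; It binds to the safe zone indicator's state changes, retrying the lookup on a timer if none is found.
; In the Starting state, a looping timer maps server time to a squared ramp toward the charged color.
; A rate-scaled timeline smooths each color step and reverses to fade the glow when reset.
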(event EventOnInitWeaponCosmetics
  ;[1 1 1] On Begin Play - Find safe zone actor in world - Only on locally owned pickaxes so we aren't doing this on other clients
  (|CallParentFunction)
  (bind (_safezoneactorfound _safezonestate) (CallFunction|GetSafeZoneActor))
  (if _safezoneactorfound
    (CallFunction|SafeZoneStateChange _safezonestate)
    (else
      (bind _returnvalue (Utilities|Time|SetTimerbyEvent (delegate Custom|SearchForSafeZoneActor) 15.0 true))
      (Variables|Default|SetSearchForSafeZoneTimer _returnvalue))))

(event Custom|RestartSearchForSafeZoneActor
  ;[1 1 1] On Begin Play - Find safe zone actor in world - Only on locally owned pickaxes so we aren't doing this on other clients
  (bind _returnvalue (Utilities|Time|SetTimerbyEvent (delegate Custom|SearchForSafeZoneActor) 15.0 true))
  (Variables|Default|SetSearchForSafeZoneTimer _returnvalue))

(event EventOnWeaponVisibilityChanged
  (params (bVisible :type Boolean) (bSetForLocalControllerOnly :type Boolean))
  (bind (_currentcolortl_r _currentcolortl_g _currentcolortl_b _currentcolortl_a) (Math|Color|BreakColor (Variables|GetCurrentColorTL_R)))
  ;[1 1 1] On Begin Play - Find safe zone actor in world - Only on locally owned pickaxes so we aren't doing this on other clients
  (|CallParentFunction bVisible false)
  (CallFunction|SetParticlesActive)
  (Effects|Components|ParticleSystem|SetFloatParameter (Variables|Default|GetP_Plague_Glow) "Alpha" _currentcolortl_a))

(event Custom|SearchForSafeZoneActor
  ;[1 1 1] On Begin Play - Find safe zone actor in world - Only on locally owned pickaxes so we aren't doing this on other clients
  (bind (_safezoneactorfound _safezonestate) (CallFunction|GetSafeZoneActor))
  (if _safezoneactorfound
    (Utilities|Time|ClearandInvalidateTimerbyHandle (Variables|Default|GetSearchForSafeZoneTimer))
    (CallFunction|SafeZoneStateChange _safezonestate)))

(event Custom|UpdateMaterialParameters
  (bind (_currentcolortl_r _currentcolortl_g _currentcolortl_b _currentcolortl_a) (Math|Color|BreakColor (Variables|GetCurrentColorTL_R)))
  ;[1 1 1] Update pickaxe material params
  (CallFunction|SetSafeZoneTimeParameters)
  (Variables|Default|SetSafeZoneLERP (Math|Float|Power (Math|Float|MapRangeClamped (CallFunction|GetServerTime) (Variables|Default|GetSafeZoneStartShrinkTime) (Variables|Default|GetSafeZoneFinishShrinkTime) 0.0 1.0) 2.0))
  (Variables|Default|SetColorA (Variables|Default|GetColorB))
  (Variables|Default|SetColorB (Math|Color|Lerp(LinearColor) "(R=0.000000,G=0.000000,B=0.000000,A=0.000000)" (Variables|Default|GetChargedColor) (Variables|Default|GetSafeZoneLERP)))
  (if (not (Components|Timeline|IsPlaying (Variables|B_Melee_Impact_Pickaxe_Athena_Plague|GetLERP_TL)))
    (Components|Timeline|SetPlayRate (Variables|B_Melee_Impact_Pickaxe_Athena_Plague|GetLERP_TL) (Utilities|Reciprocal (* (CallFunction|GetLengthBetweenColorValues) (Variables|Default|GetLengthBetweenColorValuesTimeScale))))
    (CallFunction|SetParticlesActive)
    (bind (_direction _lerp) (Timeline|LERP_TL Play)
      (:Update
        (bind _output_get (Variables|Default|SetCurrentColorTL (Math|Color|Lerp(LinearColor) (Variables|Default|GetColorA) (Variables|Default|GetColorB) _lerp)))
        (Rendering|Material|SetVectorParameterValueonMaterials (Variables|Visual|GetWeaponMesh) "EmissiveMultiplier" (Math|Conversions|ToVector(LinearColor) _output_get))
        (Effects|Components|ParticleSystem|SetFloatParameter (Variables|Default|GetP_Plague_Glow) "Alpha" _currentcolortl_a))
      (:Finished
        (switch Utilities|FlowControl|Switch|SwitchonETimelineDirection _direction
          (:Backward
            (CallFunction|SetParticlesActive)))))))

(event Custom|ResetGlowToBlack
  (bind (_currentcolortl_r _currentcolortl_g _currentcolortl_b _currentcolortl_a) (Math|Color|BreakColor (Variables|GetCurrentColorTL_R)))
  ;[1 1 1] Update pickaxe material params
  (Variables|Default|SetColorA)
  (Components|Timeline|SetPlayRate (Variables|B_Melee_Impact_Pickaxe_Athena_Plague|GetLERP_TL) (Utilities|Reciprocal (* (CallFunction|GetLengthBetweenColorValues) (Variables|Default|GetLengthBetweenColorValuesTimeScale))))
  (CallFunction|SetParticlesActive)
  (Timeline|LERP_TL Reverse))

(event Custom|SafeZoneStateChange
  (params (SafeZoneState :type EFortSafeZoneState%20Enum))
  ;[1 1 1] Safe zone state has changed
  (Variables|Default|SetSafeZoneState SafeZoneState)
  (bind _safezoneactorfound (CallFunction|GetSafeZoneActor))
  (if _safezoneactorfound
    (switch Utilities|FlowControl|Switch|SwitchonEFortSafeZoneState (Variables|Default|GetSafeZoneState)
      (:None
        _)
      (:Starting
        (if (not (Utilities|Time|IsTimerActivebyHandle (Variables|Default|GetMaterialParametersTimer)))
          (Variables|Default|SetSafeZoneStartShrinkTime (CallFunction|GetServerTime))
          (Variables|Default|SetColorB)
          (bind _returnvalue (Utilities|Time|SetTimerbyEvent (EventDispatchers|CreateEvent :!event Custom|UpdateMaterialParameters) (Variables|Default|GetMaterialParametersTimerLoop) true))
          (Variables|Default|SetMaterialParametersTimer _returnvalue)))
      (:Holding
        _)
      (:Shrinking
        _))
    (Components|Timeline|Stop (Variables|B_Melee_Impact_Pickaxe_Athena_Plague|GetLERP_TL))
    (Utilities|Time|ClearandInvalidateTimerbyHandle (Variables|Default|GetMaterialParametersTimer))
    (CallFunction|ResetGlowtoBlack)
    (else
      (CallFunction|RestartSearchforSafeZoneActor))))

(fn GetLengthBetweenColorValues
  (bind (_hsv_r _hsv_g _hsv_b) (Math|Color|BreakColor (Math|Color|RGBtoHSV(Vector) (Variables|Default|GetColorB))))
  (bind (_hsv_r_2 _hsv_g_2 _hsv_b_2) (Math|Color|BreakColor (Math|Color|RGBtoHSV(Vector) (Variables|Default|GetColorA))))
  (return (Math|Float|Absolute(Float) (- _hsv_b_2 _hsv_b))))

(fn GetSafeZoneActor
  (bind _returnvalue (SafeZone|GetSafeZoneIndicator))
  (bind _safezoneactorvalid (LocalVariables|GetSafeZoneActorValid))
  (Variables|Default|GetSafeZoneIndicator
    (:then
      (LocalVariables|SetSafeZoneActorValid true))
    (:else
      (bind _assafe_zone_indicator (Utilities|Casting|CastToSafeZoneIndicator _returnvalue)
        (:then
          (bind _output_get (Variables|Default|SetSafeZoneIndicator _assafe_zone_indicator))
          (LocalVariables|SetSafeZoneActorValid true)
          (Default|BindEventtoOnSafeZoneStateChangeDispatcher _output_get (EventDispatchers|CreateEvent :!event SafeZoneStateChange)))
        (:CastFailed
          (Development|PrintString (Utilities|String|Append "SAFEZONE INDICATOR FAILED" " - " (Utilities|GetDisplayName _returnvalue)))
          (LocalVariables|SetSafeZoneActorValid false)
          (return _safezoneactorvalid "None")))))
  (return _safezoneactorvalid (Variables|/Game/Athena/SafeZone/SafeZoneIndicator.SafeZoneIndicator_C|GetSafeZoneState _output_get)))

(fn GetServerTime
  (return (GameState|GetServerWorldTimeSeconds (Utilities|Casting|CastToAthena_GameState (Game|GetGameState)))))

(fn IsPickaxeLocallyOwned
  (bind _weaponmesh (Variables|Visual|GetWeaponMesh))
  (bind _returnvalue (Components|GetOwner _weaponmesh))
  (bind _instigator (Variables|/Script/Engine.Actor|GetInstigator _returnvalue))
  (Utilities|IsValid _weaponmesh
    (:Is%20Valid
      (Utilities|IsValid _returnvalue
        (:Is%20Valid
          (Utilities|IsValid _instigator
            (:Is%20Valid
              (return (Pawn|IsLocallyControlled _instigator)))
            (:Is%20Not%20Valid
              _)))
        (:Is%20Not%20Valid
          _))
      (return false))
    (:Is%20Not%20Valid
      _)))

(fn SetParticlesActive
  (bind (_currentcolortl_r _currentcolortl_g _currentcolortl_b _currentcolortl_a) (Math|Color|BreakColor (Variables|GetCurrentColorTL_R)))
  (Components|Activation|SetActive (Variables|Default|GetP_Plague_Glow) (> _currentcolortl_a 0.01) false))

(fn SetSafeZoneTimeParameters
  (bind _safezoneindicator (Variables|Default|GetSafeZoneIndicator)
    (:then
      (Variables|Default|SetSafeZoneFinishShrinkTime (Variables|/Script/FortniteGame.FortSafeZoneIndicator|GetSafeZoneStartShrinkTime _safezoneindicator)))
    (:else)))

(fn ConstructionScript
  (|CallParentFunction))
