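; Melee pickaxe impact Blueprint for the QuarrelMale cosmetic, adding a glow effect to the axe blade when the weapon plays its fire FX.
; After the parent fire FX event, it checks that cosmetic events can run and that GetUseAnimTrailsPSC returns true.
; It then writes the current game time into custom primitive data slot 0 on the weapon mesh, a timestamp the material can use to drive the glow.
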(event EventOnPlayWeaponFireFX
  (params (bPersistentFire :type Boolean) (bSecondaryFire :type Boolean))
  (|CallParentFunction bPersistentFire bSecondaryFire)
  ;[1 1 1] Glow axe blade
  (Networking|CanExecuteCosmeticEvents
    (:True
      (if (CallFunction|GetUseAnimTrailsPSC)
        (Rendering|Material|SetCustomPrimitiveDataFloat (Variables|Visual|GetWeaponMesh) 0 (Utilities|Time|GetGameTimeinSeconds))))))

(fn ConstructionScript
  (|CallParentFunction))
