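; Impact effect Blueprint for the Star Wand pickaxe cosmetic, played when the melee weapon hits something.
; On the impact FX event it runs the parent behavior, then activates, with reset, a Niagara component named for the star's on-impact effect.
; The construction script only defers to the parent.
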(event FortWeapon|EventOnPlayImpactFX
  (params (HitResult :type Hit%20Result%20Structure%20%28by%20ref%29) (ImpactPhysicalSurface :type EPhysicalSurface%20Enum :default "SurfaceType_Default") (SpawnedPSC :type FXSystem%20Component%20Object%20Reference))
  (|CallParentFunction HitResult ImpactPhysicalSurface SpawnedPSC)
  (Components|Activation|SetActive (Variables|Default|GetNS_Pickaxe_StarWand_Impact_OnStar) true true))

(fn ConstructionScript
  (|CallParentFunction))
