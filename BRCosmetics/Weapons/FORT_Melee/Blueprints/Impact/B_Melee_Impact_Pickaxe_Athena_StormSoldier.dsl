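; Storm Soldier pickaxe impact Blueprint that adds a swing trail effect to the melee weapon's fire FX.
; When weapon fire FX play, it calls the parent, then begins trails on the SwingFX1 particle component between the FX_Swing_1 and FX_Swing_2 sockets.
; After a short delay it ends the trails and prints a leftover debug "Hello" string.
; The construction script only calls the parent.
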(event EventOnPlayWeaponFireFX
  (params (bPersistentFire :type Boolean) (bSecondaryFire :type Boolean))
  (bind _swingfx1 (Variables|Default|GetSwingFX1))
  (|CallParentFunction bPersistentFire bSecondaryFire)
  (Effects|Particles|Trails|BeginTrails _swingfx1 "FX_Swing_1" "FX_Swing_2" "ETrailWidthMode_FromCentre" 1.0)
  (Utilities|FlowControl|Delay 0.3)
  (Effects|Particles|Trails|EndTrails _swingfx1)
  (Development|PrintString "Hello"))

(fn ConstructionScript
  (|CallParentFunction))
