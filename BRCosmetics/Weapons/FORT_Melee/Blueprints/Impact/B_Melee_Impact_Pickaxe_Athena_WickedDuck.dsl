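; Weapon Blueprint for the WickedDuck Athena pickaxe that extends its parent's melee impact effects.
; On each impact FX event it stores the hit result in Pickaxe Impact, then calls the parent implementation.
; It then pulses SuccessfulHit true for a brief delay before clearing it; nothing in this graph reads either variable.
; The construction script only calls the parent.
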(event FortWeapon|EventOnPlayImpactFX
  (params (HitResult :type Hit%20Result%20Structure%20%28by%20ref%29) (ImpactPhysicalSurface :type EPhysicalSurface%20Enum :default "SurfaceType_Default") (SpawnedPSC :type FXSystem%20Component%20Object%20Reference))
  ;[1 1 1] Impact FX
  (Variables|Default|SetPickaxe%20Impact HitResult)
  (|CallParentFunction HitResult ImpactPhysicalSurface SpawnedPSC)
  (Utilities|FlowControl|Sequence
    (:then_0
      (Variables|Default|SetSuccessfulHit true)
      (Utilities|FlowControl|Delay 0.1)
      (Variables|Default|SetSuccessfulHit false))
    (:then_1)))

(fn ConstructionScript
  (|CallParentFunction))
