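; Cosmetic Blueprint for the Witch melee pickaxe that manages its idle particle effect around swings.
; On weapon cosmetic init it sets the idle particle component to absolute rotation, so the effect does not rotate with its parent.
; On fire FX it deactivates that component, then reactivates it with a reset after a 0.3-second retriggerable delay, so repeated swings keep it off.
; Both events call the parent implementation first and validity-check the component.
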(event EventOnInitWeaponCosmetics
  (bind _returnvalue (CallFunction|GetIdlePSC))
  (|CallParentFunction)
  (Utilities|IsValid _returnvalue
    (:Is%20Valid
      (Variables|/Script/Engine.SceneComponent|SetbAbsoluteRotation true _returnvalue))))

(event EventOnPlayWeaponFireFX
  (params (bPersistentFire :type Boolean) (bSecondaryFire :type Boolean))
  (bind _returnvalue (CallFunction|GetIdlePSC))
  (|CallParentFunction bPersistentFire bSecondaryFire)
  (Utilities|IsValid _returnvalue
    (:Is%20Valid
      (Components|Activation|SetActive _returnvalue false false)
      (Utilities|FlowControl|RetriggerableDelay 0.3)
      (Utilities|IsValid _returnvalue
        (:Is%20Valid
          (Components|Activation|SetActive _returnvalue true true))))))

(fn ConstructionScript
  (|CallParentFunction))
