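; Weapon Blueprint for the Dark Bomber pickaxe that makes its horns glow and grow as the watched kill count rises.
; On each kill update it stores the count, then maps it between separate min/max thresholds for glow and scale.
; Glow drives the HornGlowINT material scalar on the weapon mesh, while growth sets HornScale on the pickaxe's anim Blueprint through a cast.
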(event EventOnWeaponVisibilityChanged
  (params (bVisible :type Boolean) (bSetForLocalControllerOnly :type Boolean))
  (|CallParentFunction bVisible false))

(event EventUpdateBasedOnKills
  (params (Watched%20Kills :type Integer))
  (|CallParentFunction Watched%20Kills)
  (Variables|Default|SetKillCount (Math|Conversions|ToFloat(Integer) Watched%20Kills))
  (Utilities|FlowControl|Sequence
    (:then_0
      ;[1 1 1] Horn Glow
      (Rendering|Material|SetScalarParameterValueonMaterials (Variables|Visual|GetWeaponMesh) "HornGlowINT" (Math|Float|MapRangeClamped (Variables|Default|GetKillCount) (Variables|Default|GetGlowKillThresholdMin) (Variables|Default|GetGlowKillThresholdMax) 0.0 6.0)))
    (:then_1
      ;[1 1 1] Horn Grow
      (bind _assk_pickaxe_dark_bomber_skeleton_anim_bp (Utilities|Casting|CastToSK_Pickaxe_DarkBomber_Skeleton_AnimBP (Components|SkeletalMesh|GetAnimInstance (Variables|Visual|GetWeaponMesh)))
        (:then
          (Variables|/BRCosmetics/Weapons/FORT_Melee/Pickaxe_Dark_Bomber/Meshes/SK_Pickaxe_DarkBomber_Skeleton_AnimBP.SK_Pickaxe_DarkBomber_Skeleton_AnimBP_C|SetHornScale (Math|Float|MapRangeClamped (Variables|Default|GetKillCount) (Variables|Default|GetHornScaleKillThresholdMin) (Variables|Default|GetHornScaleKillThresholdMax) 0.0 1.0) _assk_pickaxe_dark_bomber_skeleton_anim_bp))
        (:CastFailed
          (Development|PrintString "Cast Failed to AnimBP"))))))

(fn ConstructionScript
  (|CallParentFunction))
