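; Evil Cowboy pickaxe cosmetic Blueprint that upgrades its look through three stages (0-2) during a Battle Royale match.
; A watched cosmetic stat and, from a set game phase, players-left thresholds trigger one-way upgrades, deferred by timer until equip if the pickaxe is stowed.
; Each stage applies parameter maps to a dynamic material and swaps idle, trail and swing particles, choosing base or Ghost sets from the loadout variant.
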(event Custom|Stage%200
  (bind _returnvalue (CallFunction|GetIdleVFXComponent))
  (bind _returnvalue_2 (CallFunction|GetSwingVFXComponent))
  (bind _current_variant_index (Variables|VariantController|GetCurrent%20Variant%20Index))
  (bind _stage_0_material_colors_ghost (Variables|MaterialChanges|GhostVariant|GetStage%200%20Material%20Colors_GHOST))
  (bind _stage_0_material_colors (Variables|MaterialChanges|GetStage%200%20Material%20Colors))
  (bind _returnvalue_3 (Utilities|Select :Index _current_variant_index :Option%200 _stage_0_material_colors :Option%201 _stage_0_material_colors_ghost))
  (bind _current_variant_index_2 (Variables|VariantController|GetCurrent%20Variant%20Index))
  (bind _stage_0_material_scalars_ghost (Variables|MaterialChanges|GhostVariant|GetStage%200%20Material%20Scalars_GHOST))
  (bind _stage_0_material_scalars (Variables|MaterialChanges|GetStage%200%20Material%20Scalars))
  (bind _returnvalue_4 (Utilities|Select :Index _current_variant_index_2 :Option%200 _stage_0_material_scalars :Option%201 _stage_0_material_scalars_ghost))
  (bind _current_variant_index_3 (Variables|VariantController|GetCurrent%20Variant%20Index))
  (bind _stage_0_material_textures_ghost (Variables|MaterialChanges|GhostVariant|GetStage%200%20Material%20Textures_GHOST))
  (bind _stage_0_material_textures (Variables|MaterialChanges|GetStage%200%20Material%20Textures))
  (bind _returnvalue_5 (Utilities|Select :Index _current_variant_index_3 :Option%200 _stage_0_material_textures :Option%201 _stage_0_material_textures_ghost))
  ;[0.8109 1 0.7423] STAGE 0
  (Utilities|FlowControl|Sequence
    (:then_0
      ;;[1 1 1] Stage 0 Material Changes
      (Utilities|FlowControl|Sequence
        (:then_0
          (bind _keys (Utilities|Map|Keys _returnvalue_3))
          (if (CallFunction|ArrayIsValid _keys)
            (for _array_element _keys
              (if (Utilities|Name|Equal(Name) _array_element "None")
                (Development|PrintString "What are you thinking?  A Variable called None.  Really?!")
                (else
                  (Rendering|Material|SetVectorParameterValue (Variables|MaterialChanges|GetMID) _array_element (Utilities|Map|Find _returnvalue_3 _array_element) :!class /Script/Engine.MaterialInstanceDynamic))))))
        (:then_1
          (bind _keys_2 (Utilities|Map|Keys _returnvalue_4))
          (if (CallFunction|ArrayIsValid _keys_2)
            (for _array_element_2 _keys_2
              (if (Utilities|Name|Equal(Name) _array_element_2 "None")
                (Development|PrintString "What are you thinking?  A Variable called None.  Really?!")
                (else
                  (Rendering|Material|SetScalarParameterValue (Variables|MaterialChanges|GetMID) _array_element_2 (Utilities|Map|Find _returnvalue_4 _array_element_2) :!class /Script/Engine.MaterialInstanceDynamic))))))
        (:then_2
          (bind _keys_3 (Utilities|Map|Keys _returnvalue_5))
          (if (CallFunction|ArrayIsValid _keys_3)
            (for _array_element_3 _keys_3
              (if (Utilities|Name|Equal(Name) _array_element_3 "None")
                (Development|PrintString "What are you thinking?  A Variable called None.  Really?!")
                (else
                  (Rendering|Material|SetTextureParameterValue (Variables|MaterialChanges|GetMID) _array_element_3 (Utilities|Map|Find _returnvalue_5 _array_element_3)))))))))
    (:then_1
      ;;[1 1 1] Idle Effects
      (Utilities|IsValid _returnvalue
        (:Is%20Valid
          (Components|Activation|Deactivate _returnvalue))
        (:Is%20Not%20Valid
          _))
      (bind _ascascade_particle_system_component (Utilities|Casting|CastToParticleSystemComponent _returnvalue)
        (:then
          (Effects|Components|ParticleSystem|SetTemplate _ascascade_particle_system_component (Utilities|Select :Index (Variables|VariantController|GetCurrent%20Variant%20Index) :Option%200 (Variables|VFX|GetStage0_IdleEffect) :Option%201 (Variables|VFX|GhostVariant|GetStage0_IdleEffect_GHOST)))
          (Components|Activation|Activate _ascascade_particle_system_component true))
        (:CastFailed
          (Development|PrintString "CAST FAILED - STAGE 0 - IDLE"))))
    (:then_2
      ;;[1 1 1] AnimTrails
      (bind _asfort_player_pawn (Utilities|Casting|CastToFortPlayerPawn (Game|GetInstigator)))
      (Animation|AnimTrailsNotify _asfort_player_pawn false)
      (Animation|AnimTrailsDisable _asfort_player_pawn)
      (bind _ascascade_particle_system_component_2 (Utilities|Casting|CastToParticleSystemComponent (CallFunction|GetAnimTrailVFXComponent)))
      (Effects|Components|ParticleSystem|SetTemplate _ascascade_particle_system_component_2 (Utilities|Select :Index (Variables|VariantController|GetCurrent%20Variant%20Index) :Option%200 (Variables|VFX|GetStage0_AnimTrail) :Option%201 (Variables|VFX|GhostVariant|GetStage0_AnimTrail_GHOST))))
    (:then_3
      ;;[1 1 1] Swing Effect
      (Utilities|IsValid _returnvalue_2
        (:Is%20Valid
          (Components|Activation|Deactivate _returnvalue_2))
        (:Is%20Not%20Valid
          _))
      (bind _ascascade_particle_system_component_3 (Utilities|Casting|CastToParticleSystemComponent _returnvalue_2)
        (:then
          (Effects|Components|ParticleSystem|SetTemplate _ascascade_particle_system_component_3 (Utilities|Select :Index (Variables|VariantController|GetCurrent%20Variant%20Index) :Option%200 (Variables|VFX|GetStage0_SwingEffect) :Option%201 (Variables|VFX|GhostVariant|GetStage0_SwingEffect_GHOST)))
          (Components|Activation|Activate _ascascade_particle_system_component_3 true))
        (:CastFailed
          (Development|PrintString "CAST FAILED - STAGE 0 - SWING"))))))

(event EventOnWeaponVisibilityChanged
  (params (bVisible :type Boolean) (bSetForLocalControllerOnly :type Boolean))
  (bind _returnvalue (CallFunction|GetIdleVFXComponent))
  ;[1 1 1] Turn off idle effect when unequipping pickaxe
  (|CallParentFunction bVisible false)
  (Utilities|IsValid _returnvalue
    (:Is%20Valid
      (if bVisible
        (Components|Activation|Activate _returnvalue false)
        (else
          (Components|Activation|Deactivate _returnvalue))))))

(event EventOnInitWeaponCosmetics
  (bind _instigator (Variables|Actor|GetInstigator))
  (bind (_asplayer_pawn_athena _bsuccess) (Utilities|Casting|CastToPlayerPawn_Athena _instigator))
  (|CallParentFunction)
  (bind _returnvalue (Rendering|Material|CreateDynamicMaterialInstance (Variables|Visual|GetWeaponMesh) 0 (Rendering|Material|GetMaterial (Variables|Visual|GetWeaponMesh) 0) "None" :!class /Script/Engine.PrimitiveComponent))
  (Variables|MaterialChanges|SetMID _returnvalue)
  (CallFunction|CosmeticLoadoutForVariantSetup)
  (Default|BindEventtoResOut (Utilities|Casting|CastToPlayerPawn_Athena (Variables|Actor|GetInstigator)) (delegate Custom|Holder's%20Death))
  (if (Variables|DEBUG|GetAllow%20FULL%20DEBUG)
    ;[1 1 1] DEBUG
    (switch int (Variables|DEBUG|GetWhich%20Stage%20to%20Debug?)
      (:0
        (CallFunction|Stage0))
      (:1
        (CallFunction|Stage1))
      (:2
        (CallFunction|Stage2)))
    (else
      ;[1 1 1] For Cosmetic Stat Watching
      ;
      ; EVENT InitializeCosmeticStatTracking
      (if _bsuccess
        (Utilities|FlowControl|Sequence
          (:then_0
            (bind (_statname _statvalue) (Cosmetics|WatchCosmeticStat _asplayer_pawn_athena (Variables|StatControls|GetStat_01))
              (:OnStatChanged
                (CallFunction|CosmeticStatCheck _statvalue (Variables|StatControls|GetStat_01_Threshold_Stage%202) (Variables|StatControls|GetStat_01_Threshold_Stage%201) _statname))))
          (:then_1
            ;[1 1 1] We also immediately check the remaining players when we are created.
            (bind _asathena_game_state (Utilities|Casting|CastToAthena_GameState (Game|GetGameState)))
            (Game|AssignOnPlayersLeftChanged _asathena_game_state (delegate Custom|TrackPlayersLeft))
            (if (Math|Byte|Byte>=Byte (BattleRoyaleGamePhaseLogic|GetGamePhase (BattleRoyaleGamePhaseLogic|GetBattleRoyaleGamePhaseLogicComponent)) 3)
              (CallFunction|PlayersLeftAliveCheck (.Total (Utilities|Struct|BreakPlayersLeft (Class|FortGameStateAthena|GetPlayersLeft _asathena_game_state :!class /Script/FortniteGame.FortGameStateAthena))) (Variables|StatControls|GetPlayersLeft_Threshold_Stage%202) (Variables|StatControls|GetPlayersLeft_Threshold_Stage%201)))))
        (else
          ;;[1 0 0.3552 0.179 11] Lazily poll this event to make sure we have a pawn before cosmetic updates. As a just in case since I haven't done much debugging on the chain of logic for pawns->parts/pickaxeBPs->cosmetic events
          ;; -Nicholas.Lacelle
          (Utilities|Time|SetTimerbyEvent (delegate Custom|InitializeCosmeticStatTracking) 1.0))))))

(event Custom|Holder's%20Death
  (Utilities|Time|ClearandInvalidateTimerbyHandle (Variables|Default|GetStage%201%20Timer))
  (Utilities|Time|ClearandInvalidateTimerbyHandle (Variables|Default|GetStage%202%20Timer)))

(event Custom|InitializeCosmeticStatTracking
  (bind _instigator (Variables|Actor|GetInstigator))
  (bind (_asplayer_pawn_athena _bsuccess) (Utilities|Casting|CastToPlayerPawn_Athena _instigator))
  ;[1 1 1] For Cosmetic Stat Watching
  ;
  ; EVENT InitializeCosmeticStatTracking
  (if _bsuccess
    (Utilities|FlowControl|Sequence
      (:then_0
        (bind (_statname _statvalue) (Cosmetics|WatchCosmeticStat _asplayer_pawn_athena (Variables|StatControls|GetStat_01))
          (:OnStatChanged
            (CallFunction|CosmeticStatCheck _statvalue (Variables|StatControls|GetStat_01_Threshold_Stage%202) (Variables|StatControls|GetStat_01_Threshold_Stage%201) _statname))))
      (:then_1
        ;[1 1 1] We also immediately check the remaining players when we are created.
        (bind _asathena_game_state (Utilities|Casting|CastToAthena_GameState (Game|GetGameState)))
        (Game|AssignOnPlayersLeftChanged _asathena_game_state (delegate Custom|TrackPlayersLeft))
        (if (Math|Byte|Byte>=Byte (BattleRoyaleGamePhaseLogic|GetGamePhase (BattleRoyaleGamePhaseLogic|GetBattleRoyaleGamePhaseLogicComponent)) 3)
          (CallFunction|PlayersLeftAliveCheck (.Total (Utilities|Struct|BreakPlayersLeft (Class|FortGameStateAthena|GetPlayersLeft _asathena_game_state :!class /Script/FortniteGame.FortGameStateAthena))) (Variables|StatControls|GetPlayersLeft_Threshold_Stage%202) (Variables|StatControls|GetPlayersLeft_Threshold_Stage%201)))))
    (else
      ;;[1 0 0.3552 0.179 11] Lazily poll this event to make sure we have a pawn before cosmetic updates. As a just in case since I haven't done much debugging on the chain of logic for pawns->parts/pickaxeBPs->cosmetic events
      ;; -Nicholas.Lacelle
      (Utilities|Time|SetTimerbyEvent (delegate Custom|InitializeCosmeticStatTracking) 1.0))))

(event FortWeapon|EventOnEquippedAndAttached
  ;[1 1 1 1 12] EVENT OnEquip
  ; EVENT OnUnEquip
  ; Timer Handling for handling weapon swapping & FX
  (|CallParentFunction)
  (Variables|StatControls|SetPickaxeIsEquipped true)
  (if (Variables|StatControls|GetStage2_Delay)
    (bind _returnvalue (Utilities|Time|SetTimerbyEvent (EventDispatchers|CreateEvent :!event Custom|Stage%202) 0.05))
    (Variables|Default|SetStage%202%20Timer _returnvalue)
    (Variables|StatControls|SetStage2_Delay false)
    (Variables|StatControls|SetStage1_Delay false)
    (elif (Variables|StatControls|GetStage1_Delay)
      (bind _returnvalue_2 (Utilities|Time|SetTimerbyEvent (EventDispatchers|CreateEvent :!event Custom|Stage%201) 0.05))
      (Variables|Default|SetStage%201%20Timer _returnvalue_2)
      (Variables|StatControls|SetStage1_Delay false))))

(event FortWeapon|EventOnLocalUnEquip
  ;[1 1 1 1 12] EVENT OnEquip
  ; EVENT OnUnEquip
  ; Timer Handling for handling weapon swapping & FX
  (|CallParentFunction)
  (Variables|StatControls|SetPickaxeIsEquipped false))

(event FortWeapon|EventOnUnEquippedAndDetached
  ;[1 1 1 1 12] EVENT OnEquip
  ; EVENT OnUnEquip
  ; Timer Handling for handling weapon swapping & FX
  (|CallParentFunction)
  (Variables|StatControls|SetPickaxeIsEquipped false))

(event Custom|Stage%201
  (bind _returnvalue (CallFunction|GetIdleVFXComponent))
  (bind _returnvalue_2 (CallFunction|GetSwingVFXComponent))
  (bind _current_variant_index (Variables|VariantController|GetCurrent%20Variant%20Index))
  (bind _stage_1_material_colors_ghost (Variables|MaterialChanges|GhostVariant|GetStage%201%20Material%20Colors_GHOST))
  (bind _stage_1_material_colors (Variables|MaterialChanges|GetStage%201%20Material%20Colors))
  (bind _returnvalue_3 (Utilities|Select :Index _current_variant_index :Option%200 _stage_1_material_colors :Option%201 _stage_1_material_colors_ghost))
  (bind _current_variant_index_2 (Variables|VariantController|GetCurrent%20Variant%20Index))
  (bind _stage_1_material_scalars_ghost (Variables|MaterialChanges|GhostVariant|GetStage%201%20Material%20Scalars_GHOST))
  (bind _stage_1_material_scalars (Variables|MaterialChanges|GetStage%201%20Material%20Scalars))
  (bind _returnvalue_4 (Utilities|Select :Index _current_variant_index_2 :Option%200 _stage_1_material_scalars :Option%201 _stage_1_material_scalars_ghost))
  (bind _current_variant_index_3 (Variables|VariantController|GetCurrent%20Variant%20Index))
  (bind _stage_1_material_textures_ghost (Variables|MaterialChanges|GhostVariant|GetStage%201%20Material%20Textures_GHOST))
  (bind _stage_1_material_textures (Variables|MaterialChanges|GetStage%201%20Material%20Textures))
  (bind _returnvalue_5 (Utilities|Select :Index _current_variant_index_3 :Option%200 _stage_1_material_textures :Option%201 _stage_1_material_textures_ghost))
  ;[0.1378 1 0] STAGE 1
  (Utilities|FlowControl|Sequence
    (:then_0
      ;;[1 1 1] Stage 1 Material Changes
      (Utilities|FlowControl|Sequence
        (:then_0
          (bind _keys (Utilities|Map|Keys _returnvalue_3))
          (if (CallFunction|ArrayIsValid _keys)
            (for _array_element _keys
              (if (Utilities|Name|Equal(Name) _array_element "None")
                (Development|PrintString "What are you thinking?  A Variable called None.  Really?!")
                (else
                  (Rendering|Material|SetVectorParameterValue (Variables|MaterialChanges|GetMID) _array_element (Utilities|Map|Find _returnvalue_3 _array_element) :!class /Script/Engine.MaterialInstanceDynamic))))))
        (:then_1
          (bind _keys_2 (Utilities|Map|Keys _returnvalue_4))
          (if (CallFunction|ArrayIsValid _keys_2)
            (for _array_element_2 _keys_2
              (if (Utilities|Name|Equal(Name) _array_element_2 "None")
                (Development|PrintString "What are you thinking?  A Variable called None.  Really?!")
                (else
                  (Rendering|Material|SetScalarParameterValue (Variables|MaterialChanges|GetMID) _array_element_2 (Utilities|Map|Find _returnvalue_4 _array_element_2) :!class /Script/Engine.MaterialInstanceDynamic))))))
        (:then_2
          (bind _keys_3 (Utilities|Map|Keys _returnvalue_5))
          (if (CallFunction|ArrayIsValid _keys_3)
            (for _array_element_3 _keys_3
              (if (Utilities|Name|Equal(Name) _array_element_3 "None")
                (Development|PrintString "What are you thinking?  A Variable called None.  Really?!")
                (else
                  (Rendering|Material|SetTextureParameterValue (Variables|MaterialChanges|GetMID) _array_element_3 (Utilities|Map|Find _returnvalue_5 _array_element_3)))))))))
    (:then_1
      ;;[1 1 1] Transition Effect_Burst
      (Effects|SpawnEmitterAttached (Utilities|Select :Index (Variables|VariantController|GetCurrent%20Variant%20Index) :Option%200 (Variables|VFX|GetStage%201%20Transition) :Option%201 (Variables|VFX|GhostVariant|GetStage%201%20Transition_GHOST)) (Variables|Visual|GetWeaponMesh)))
    (:then_2
      ;;[1 1 1] Idle Effects
      (Utilities|IsValid _returnvalue
        (:Is%20Valid
          (Components|Activation|Deactivate _returnvalue))
        (:Is%20Not%20Valid
          _))
      (bind _ascascade_particle_system_component (Utilities|Casting|CastToParticleSystemComponent _returnvalue)
        (:then
          (Effects|Components|ParticleSystem|SetTemplate _ascascade_particle_system_component (Utilities|Select :Index (Variables|VariantController|GetCurrent%20Variant%20Index) :Option%200 (Variables|VFX|GetStage1_IdleEffect) :Option%201 (Variables|VFX|GhostVariant|GetStage1_IdleEffect_GHOST)))
          (Components|Activation|Activate _ascascade_particle_system_component true))
        (:CastFailed
          (Development|PrintString "CAST FAILED - STAGE 1 - IDLE"))))
    (:then_3
      ;;[1 1 1] AnimTrails
      (bind _asfort_player_pawn (Utilities|Casting|CastToFortPlayerPawn (Game|GetInstigator)))
      (Animation|AnimTrailsNotify _asfort_player_pawn false)
      (Animation|AnimTrailsDisable _asfort_player_pawn)
      (bind _ascascade_particle_system_component_2 (Utilities|Casting|CastToParticleSystemComponent (CallFunction|GetAnimTrailVFXComponent)))
      (Effects|Components|ParticleSystem|SetTemplate _ascascade_particle_system_component_2 (Utilities|Select :Index (Variables|VariantController|GetCurrent%20Variant%20Index) :Option%200 (Variables|VFX|GetStage1_AnimTrail) :Option%201 (Variables|VFX|GhostVariant|GetStage1_AnimTrail_GHOST))))
    (:then_4
      ;;[1 1 1] Swing Effect
      (Utilities|IsValid _returnvalue_2
        (:Is%20Valid
          (Components|Activation|Deactivate _returnvalue_2))
        (:Is%20Not%20Valid
          _))
      (bind _ascascade_particle_system_component_3 (Utilities|Casting|CastToParticleSystemComponent _returnvalue_2)
        (:then
          (Effects|Components|ParticleSystem|SetTemplate _ascascade_particle_system_component_3 (Utilities|Select :Index (Variables|VariantController|GetCurrent%20Variant%20Index) :Option%200 (Variables|VFX|GetStage1_SwingEffect) :Option%201 (Variables|VFX|GhostVariant|GetStage1_SwingEffect_GHOST)))
          (Components|Activation|Activate _ascascade_particle_system_component_3 true))
        (:CastFailed
          (Development|PrintString "CAST FAILED - STAGE 1 - SWING"))))))

(event Custom|TrackPlayersLeft
  (params (NumPlayersLeft :type Integer))
  ;[1 1 1] We also immediately check the remaining players when we are created.
  (if (Math|Byte|Byte>=Byte (BattleRoyaleGamePhaseLogic|GetGamePhase (BattleRoyaleGamePhaseLogic|GetBattleRoyaleGamePhaseLogicComponent)) 3)
    (CallFunction|PlayersLeftAliveCheck NumPlayersLeft (Variables|StatControls|GetPlayersLeft_Threshold_Stage%202) (Variables|StatControls|GetPlayersLeft_Threshold_Stage%201))
    (else
      (Development|PrintString "Blocked attempt to check TotalPlayers in match for a cosmetic upgrade, since we have not left the starting island."))))

(event Custom|Stage%202
  (bind _returnvalue (CallFunction|GetIdleVFXComponent))
  (bind _returnvalue_2 (CallFunction|GetSwingVFXComponent))
  (bind _current_variant_index (Variables|VariantController|GetCurrent%20Variant%20Index))
  (bind _stage_2_material_scalars_ghost (Variables|MaterialChanges|GhostVariant|GetStage%202%20Material%20Scalars_GHOST))
  (bind _stage_2_material_scalars (Variables|MaterialChanges|GetStage%202%20Material%20Scalars))
  (bind _returnvalue_3 (Utilities|Select :Index _current_variant_index :Option%200 _stage_2_material_scalars :Option%201 _stage_2_material_scalars_ghost))
  (bind _current_variant_index_2 (Variables|VariantController|GetCurrent%20Variant%20Index))
  (bind _stage_2_material_colors_ghost (Variables|MaterialChanges|GhostVariant|GetStage%202%20Material%20Colors_GHOST))
  (bind _stage_2_material_colors (Variables|MaterialChanges|GetStage%202%20Material%20Colors))
  (bind _returnvalue_4 (Utilities|Select :Index _current_variant_index_2 :Option%200 _stage_2_material_colors :Option%201 _stage_2_material_colors_ghost))
  (bind _current_variant_index_3 (Variables|VariantController|GetCurrent%20Variant%20Index))
  (bind _stage_2_material_textures_ghost (Variables|MaterialChanges|GhostVariant|GetStage%202%20Material%20Textures_GHOST))
  (bind _stage_2_material_textures (Variables|MaterialChanges|GetStage%202%20Material%20Textures))
  (bind _returnvalue_5 (Utilities|Select :Index _current_variant_index_3 :Option%200 _stage_2_material_textures :Option%201 _stage_2_material_textures_ghost))
  ;[0 1 1] STAGE 2
  (Utilities|FlowControl|Sequence
    (:then_0
      ;;[1 1 1] Stage 2 Material Changes
      (Utilities|FlowControl|Sequence
        (:then_0
          (bind _keys (Utilities|Map|Keys _returnvalue_4))
          (if (CallFunction|ArrayIsValid _keys)
            (for _array_element _keys
              (if (Utilities|Name|Equal(Name) _array_element "None")
                (Development|PrintString "What are you thinking?  A Variable called None.  Really?!")
                (else
                  (Rendering|Material|SetVectorParameterValue (Variables|MaterialChanges|GetMID) _array_element (Utilities|Map|Find _returnvalue_4 _array_element) :!class /Script/Engine.MaterialInstanceDynamic))))))
        (:then_1
          (bind _keys_2 (Utilities|Map|Keys _returnvalue_3))
          (if (CallFunction|ArrayIsValid _keys_2)
            (for _array_element_2 _keys_2
              (if (Utilities|Name|Equal(Name) _array_element_2 "None")
                (Development|PrintString "What are you thinking?  A Variable called None.  Really?!")
                (else
                  (Rendering|Material|SetScalarParameterValue (Variables|MaterialChanges|GetMID) _array_element_2 (Utilities|Map|Find _returnvalue_3 _array_element_2) :!class /Script/Engine.MaterialInstanceDynamic))))))
        (:then_2
          (bind _keys_3 (Utilities|Map|Keys _returnvalue_5))
          (if (CallFunction|ArrayIsValid _keys_3)
            (for _array_element_3 _keys_3
              (if (Utilities|Name|Equal(Name) _array_element_3 "None")
                (Development|PrintString "What are you thinking?  A Variable called None.  Really?!")
                (else
                  (Rendering|Material|SetTextureParameterValue (Variables|MaterialChanges|GetMID) _array_element_3 (Utilities|Map|Find _returnvalue_5 _array_element_3)))))))))
    (:then_1
      ;;[1 1 1] Transition Effect_Burst
      (Effects|SpawnEmitterAttached (Utilities|Select :Index (Variables|VariantController|GetCurrent%20Variant%20Index) :Option%200 (Variables|VFX|GetStage%202%20Transition) :Option%201 (Variables|VFX|GhostVariant|GetStage%202%20Transition_GHOST)) (Variables|Visual|GetWeaponMesh)))
    (:then_2
      ;;[1 1 1] Idle Effects
      (Utilities|IsValid _returnvalue
        (:Is%20Valid
          (Components|Activation|Deactivate _returnvalue))
        (:Is%20Not%20Valid
          _))
      (bind _ascascade_particle_system_component (Utilities|Casting|CastToParticleSystemComponent _returnvalue)
        (:then
          (Effects|Components|ParticleSystem|SetTemplate _ascascade_particle_system_component (Utilities|Select :Index (Variables|VariantController|GetCurrent%20Variant%20Index) :Option%200 (Variables|VFX|GetStage2_IdleEffect) :Option%201 (Variables|VFX|GhostVariant|GetStage2_IdleEffect_GHOST)))
          (Components|Activation|Activate _ascascade_particle_system_component true))
        (:CastFailed
          (Development|PrintString "CAST FAILED - STAGE 2 - IDLE"))))
    (:then_3
      ;;[1 1 1] AnimTrails
      (bind _asfort_player_pawn (Utilities|Casting|CastToFortPlayerPawn (Game|GetInstigator)))
      (Animation|AnimTrailsNotify _asfort_player_pawn false)
      (Animation|AnimTrailsDisable _asfort_player_pawn)
      (bind _ascascade_particle_system_component_2 (Utilities|Casting|CastToParticleSystemComponent (CallFunction|GetAnimTrailVFXComponent)))
      (Effects|Components|ParticleSystem|SetTemplate _ascascade_particle_system_component_2 (Utilities|Select :Index (Variables|VariantController|GetCurrent%20Variant%20Index) :Option%200 (Variables|VFX|GetStage2_AnimTrail) :Option%201 (Variables|VFX|GhostVariant|GetStage2_AnimTrail_GHOST))))
    (:then_4
      ;;[1 1 1] Swing Effect
      (Utilities|IsValid _returnvalue_2
        (:Is%20Valid
          (Components|Activation|Deactivate _returnvalue_2))
        (:Is%20Not%20Valid
          _))
      (bind _ascascade_particle_system_component_3 (Utilities|Casting|CastToParticleSystemComponent _returnvalue_2)
        (:then
          (Effects|Components|ParticleSystem|SetTemplate _ascascade_particle_system_component_3 (Utilities|Select :Index (Variables|VariantController|GetCurrent%20Variant%20Index) :Option%200 (Variables|VFX|GetStage2_SwingEffect) :Option%201 (Variables|VFX|GhostVariant|GetStage2_SwingEffect_GHOST)))
          (Components|Activation|Activate _ascascade_particle_system_component_3 true))
        (:CastFailed
          (Development|PrintString "CAST FAILED - STAGE 2 - SWING"))))))

(fn ArrayIsValid
  (params (Name%20Array :type Array%20of%20Names))
  (return (> (Utilities|Array|Length Name%20Array) 0)))

(fn CosmeticLoadoutforVariantSetup
  (bind _index (CosmeticLoadout|GetVariantChannelActiveIndexforSlotActor "(TagName=\"LoadoutSlot.Pickaxe\")" (Utilities|Casting|CastToPlayerPawn_Athena (Variables|Actor|GetInstigator)) "(TagName=\"Cosmetics.Variant.Channel.Particle\")" (Utilities|Array|MakeArray "(TagName=\"Cosmetics.Variant.Property.Particle1\")" "(TagName=\"Cosmetics.Variant.Property.Particle2\")")))
  (bind _output_get (Variables|VariantController|SetCurrent%20Variant%20Index _index))
  (Development|PrintString (Utilities|String|ToString(Integer) _output_get)))

(fn CosmeticStatCheck
  (params (Value%20to%20Check :type Integer) (Stage%202%20Threshold :type Integer) (Stage%201%20Threshold :type Integer) (DEBUG%20-%20Stat%20Checked :type Name))
  (bind _pickaxeisequipped (Variables|StatControls|GetPickaxeIsEquipped))
  (bind _allow_stat_debug (Variables|DEBUG|GetAllow%20STAT%20DEBUG))
  (bind _allow_stat_debug_2 (Variables|DEBUG|GetAllow%20STAT%20DEBUG))
  (Utilities|FlowControl|Sequence
    (:then_0
      (if (Variables|DEBUG|GetAllow%20STAT%20DEBUG)
        (Development|PrintString (Utilities|String|ToString(Name) DEBUG%20-%20Stat%20Checked))))
    (:then_1
      (if (and (<= Stage%202%20Threshold Value%20to%20Check) (not (Variables|StatControls|GetAleady%20Fired_Stage%202)))
        (if _pickaxeisequipped
          (CallFunction|Stage2)
          (else
            (Variables|StatControls|SetStage1_Delay false)
            (Variables|StatControls|SetStage2_Delay true)))
        (Variables|StatControls|SetAleady%20Fired_Stage%202 true)
        (Variables|StatControls|SetAleady%20Fired_Stage%201 true)
        (if _allow_stat_debug
          (Development|PrintString "STAGE 2 HAS FIRED"))
        (elif (and (<= Stage%201%20Threshold Value%20to%20Check) (not (Variables|StatControls|GetAleady%20Fired_Stage%201)))
          (if _pickaxeisequipped
            (CallFunction|Stage1)
            (else
              (Variables|StatControls|SetStage1_Delay true)
              (Variables|StatControls|SetStage2_Delay false)))
          (Variables|StatControls|SetAleady%20Fired_Stage%201 true)
          (if _allow_stat_debug
            (Development|PrintString "STAGE 1 HAS FIRED"))
          (elif (Variables|StatControls|GetAleady%20Fired_Stage%201)
            (if (Variables|StatControls|GetAleady%20Fired_Stage%202)
              (if _allow_stat_debug_2
                (Development|PrintString "STAGE 2 ALREADY FIRED!"))
              (elif _allow_stat_debug_2
                (Development|PrintString "STAGE 1 ALREADY FIRED!")))
            (else
              (CallFunction|Stage0)
              (if (Variables|DEBUG|GetAllow%20STAT%20DEBUG)
                (Development|PrintString "STILL IN STAGE 0")))))))))

(fn IsNightTime
  (return (or (>= (TimeOfDay|GetTimeOfDay :!class /Script/FortniteGame.FortKismetLibrary) 22.0) (<= (TimeOfDay|GetTimeOfDay :!class /Script/FortniteGame.FortKismetLibrary) 3.0))))

(fn PlayersLeftAliveCheck
  (params (Value%20to%20Check :type Integer) (Stage%202%20Threshold :type Integer) (Stage%201%20Threshold :type Integer) (DEBUG%20-%20Stat%20Checked :type Name))
  (bind _pickaxeisequipped (Variables|StatControls|GetPickaxeIsEquipped))
  (bind _allow_stat_debug (Variables|DEBUG|GetAllow%20STAT%20DEBUG))
  (bind _allow_stat_debug_2 (Variables|DEBUG|GetAllow%20STAT%20DEBUG))
  (Utilities|FlowControl|Sequence
    (:then_0
      (if (Variables|DEBUG|GetAllow%20STAT%20DEBUG)
        (Development|PrintString (Utilities|String|ToString(Name) DEBUG%20-%20Stat%20Checked))
        (Development|PrintString (Utilities|String|Append "PlayersLeft = " (Utilities|String|ToString(Integer) Value%20to%20Check)))))
    (:then_1
      (if (and (>= Stage%202%20Threshold Value%20to%20Check) (not (Variables|StatControls|GetAleady%20Fired_Stage%202)))
        (if _pickaxeisequipped
          (CallFunction|Stage2)
          (else
            (Variables|StatControls|SetStage2_Delay true)))
        (Variables|StatControls|SetAleady%20Fired_Stage%202 true)
        (Variables|StatControls|SetAleady%20Fired_Stage%201 true)
        (if _allow_stat_debug
          (Development|PrintString "STAGE 2 HAS FIRED"))
        (elif (and (>= Stage%201%20Threshold Value%20to%20Check) (not (Variables|StatControls|GetAleady%20Fired_Stage%201)))
          (if _pickaxeisequipped
            (CallFunction|Stage1)
            (else
              (Variables|StatControls|SetStage1_Delay true)))
          (Variables|StatControls|SetAleady%20Fired_Stage%201 true)
          (if _allow_stat_debug
            (Development|PrintString "STAGE 1 HAS FIRED"))
          (elif (Variables|StatControls|GetAleady%20Fired_Stage%201)
            (if (Variables|StatControls|GetAleady%20Fired_Stage%202)
              (if _allow_stat_debug_2
                (Development|PrintString "STAGE 2 ALREADY FIRED!"))
              (elif _allow_stat_debug_2
                (Development|PrintString "STAGE 1 ALREADY FIRED!")))
            (else
              (CallFunction|Stage0)
              (if (Variables|DEBUG|GetAllow%20STAT%20DEBUG)
                (Development|PrintString "STILL IN STAGE 0")))))))))

(fn ConstructionScript
  (|CallParentFunction))
