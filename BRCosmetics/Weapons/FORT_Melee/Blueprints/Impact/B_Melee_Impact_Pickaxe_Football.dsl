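; Impact blueprint for the Football pickaxe melee weapon cosmetic, playing its impact sound when the pickaxe hits.
; It overrides the play-impact-FX event, runs the parent logic first, then plays the Football impact cue at the hit location.
; The same cue plays regardless of the physical surface struck.
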(event FortWeapon|EventOnPlayImpactFX
  (params (HitResult :type Hit%20Result%20Structure%20%28by%20ref%29) (ImpactPhysicalSurface :type EPhysicalSurface%20Enum :default "SurfaceType_Default") (SpawnedPSC :type FXSystem%20Component%20Object%20Reference))
  (bind (_bblockinghit _binitialoverlap _time _distance _location) (Collision|BreakHitResult HitResult))
  (|CallParentFunction HitResult ImpactPhysicalSurface SpawnedPSC)
  (Audio|PlaySoundatLocation "/BRCosmetics/Sounds/PickAxes/Football/PickAxe_Impact_Default_Football_Cue.PickAxe_Impact_Default_Football_Cue" _location))

(fn ConstructionScript
  (|CallParentFunction))
