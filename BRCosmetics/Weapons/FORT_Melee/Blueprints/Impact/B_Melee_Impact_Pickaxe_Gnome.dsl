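; Gnome pickaxe weapon Blueprint that drives a light component and emissive glow by time of day.
; The construction script creates a dynamic material on the weapon mesh, then sets glow, light auto-activation and timeline position from the current hour, lit outside 8 to 20.
; Day-phase change events use DoOnce gates to toggle the light and play or reverse a timeline that animates the emissive color.
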(event EventTODCheck
  (params (CurrentDayPhase :type EFortDayPhase%20Enum :default "Morning") (PreviousDayPhase :type EFortDayPhase%20Enum :default "Morning") (bAtCreation :type Boolean))
  (bind _p_light (Variables|Default|GetP_Light))
  (|CallParentFunction CurrentDayPhase PreviousDayPhase bAtCreation)
  (switch Utilities|FlowControl|Switch|SwitchonEFortDayPhase CurrentDayPhase
    (:Morning
      _)
    (:Day
      _)
    (:Evening
      _)
    (:Night
      (Utilities|FlowControl|Sequence
        (:then_0
          (Utilities|FlowControl|DoOnce false :!label _label_2
            (:Completed
              (Components|Activation|Deactivate _p_light)
              (Timeline|Timeline_1 Play))))
        (:then_1
          (Utilities|FlowControl|DoOnce false :!label _label
            (:Completed
              (Components|Activation|Activate _p_light false)
              (bind (_direction _thecolor) (Timeline|Timeline_1 Reverse)
                (:Update
                  (Rendering|Material|SetVectorParameterValue (Variables|Default|GetLight_MID) "EmissiveMultiplier" _thecolor :!class /Script/Engine.MaterialInstanceDynamic)))))))))
  (Utilities|FlowControl|Sequence
    (:then_0
      (goto _label Reset))
    (:then_1
      (goto _label_2))))

(fn ConstructionScript
  (bind _p_light (Variables|Default|GetP_Light))
  (bind _timeline_1 (Variables|B_Melee_Impact_Pickaxe_Gnome|GetTimeline_1))
  (|CallParentFunction)
  (bind _returnvalue (Rendering|Material|CreateDynamicMaterialInstance (Variables|Visual|GetWeaponMesh) 0 (Rendering|Material|GetMaterial (Variables|Visual|GetWeaponMesh) 0) "None" :!class /Script/Engine.PrimitiveComponent))
  (bind _output_get (Variables|Default|SetLight_MID _returnvalue))
  (if (Math|Float|InRange(Float) (TimeOfDay|GetTimeOfDay :!class /Script/FortniteGame.FortKismetLibrary) 8.0 2e+01)
    (Rendering|Material|SetVectorParameterValue :self _output_get :ParameterName "EmissiveMultiplier" :!class /Script/Engine.MaterialInstanceDynamic)
    (Components|Activation|SetAutoActivate _p_light false)
    (Components|Timeline|SetNewTime _timeline_1 1.0)
    (else
      (Rendering|Material|SetVectorParameterValue _output_get "EmissiveMultiplier" (Variables|Default|GetLightEmissiveColor) :!class /Script/Engine.MaterialInstanceDynamic)
      (Components|Activation|SetAutoActivate _p_light true)
      (Components|Timeline|SetNewTime _timeline_1 0.0))))
