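; Hippie pickaxe melee weapon Blueprint that adds a custom impact sound and petal coloring to hit effects.
; On impact it calls the parent handler, then plays the Hippie impact sound at the hit location.
; It loops over the ColorChoice array, setting the active particle system's "Petal Color" from a random index up to the current one, so the last pick wins.
; That loop also enables particle shadow casting when detail mode is above zero.
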(event FortWeapon|EventOnPlayImpactFX
  (params (HitResult :type Hit%20Result%20Structure%20%28by%20ref%29) (ImpactPhysicalSurface :type EPhysicalSurface%20Enum :default "SurfaceType_Default") (SpawnedPSC :type FXSystem%20Component%20Object%20Reference))
  (bind (_bblockinghit _binitialoverlap _time _distance _location) (Collision|BreakHitResult HitResult))
  (bind _colorchoice (Variables|Default|GetColorChoice))
  (|CallParentFunction HitResult ImpactPhysicalSurface SpawnedPSC)
  (Audio|PlaySoundatLocation "/BRCosmetics/Sounds/PickAxes/Hippie/PA_Hippie_Gen_Impact.PA_Hippie_Gen_Impact" _location)
  (for (_ _array_index) _colorchoice
    (if (Components|Activation|IsActive SpawnedPSC)
      (Effects|Components|ParticleSystem|SetVectorParameter SpawnedPSC "Petal Color" (Math|Conversions|ToVector(LinearColor) (Utilities|Array|Get(acopy) _colorchoice (Math|Random|RandomIntegerinRange 0 _array_index))))
      (if (> (Rendering|GetRenderingDetailMode) 0)
        (Rendering|SetCastShadow SpawnedPSC true)))))

(fn ConstructionScript
  (|CallParentFunction))
