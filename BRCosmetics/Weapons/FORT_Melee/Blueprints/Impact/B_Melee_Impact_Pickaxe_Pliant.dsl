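; Blueprint for the Pliant pickaxe melee weapon that tags its mesh so wrap materials can identify it as the Pliant pickaxe.
; On equip and attach, it calls the parent handler, then sets custom primitive data slot 0 on the weapon mesh to the Pliant pickaxe ID.
; That slot distinguishes the Pliant body, backpack and pickaxe from anything else using the wrap.
; BeginPlay, Tick and the construction script add no logic of their own.
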(event EventBeginPlay)

(event EventTick
  (params (DeltaSeconds :type Float%20%28single-precision%29)))

(event FortWeapon|EventOnEquippedAndAttached
  (|CallParentFunction)
  ;[1 1 1] Pliant Specific Custom Data for Wrap Identification
  (Rendering|Material|SetCustomPrimitiveDataFloat (Variables|Visual|GetWeaponMesh) 0 3.0) ; 0 is default (I.E. anything else using the wrap, 1 is Pliant Body, 2 is Pliant Backpack, 3 is Pliant Pickaxe
)

(fn ConstructionScript
  (|CallParentFunction))
