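; Impact Blueprint for the PunkDevilSummer pickaxe cosmetic in Fortnite, adding a kill-reactive visual to the melee weapon.
; When the watched kill count updates, it calls the parent logic, then writes the current game time in seconds to the "PD_GameTimeFromBP" scalar on the weapon mesh's materials.
; This timestamp gives the material a reference time for each kill update; the construction script only calls the parent.
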(event EventUpdateBasedOnKills
  (params (Watched%20Kills :type Integer))
  (|CallParentFunction Watched%20Kills)
  (Rendering|Material|SetScalarParameterValueonMaterials (Variables|Visual|GetWeaponMesh) "PD_GameTimeFromBP" (Utilities|Time|GetGameTimeinSeconds)))

(fn ConstructionScript
  (|CallParentFunction))
