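; Cosmetic Blueprint for the female Rays pickaxe, handling its light-blade reveal when equipped.
; On equip it caches the particle variant index from the player's loadout, hides the blade and plays the reveal, briefly delayed in frontend displays.
; The reveal restarts idle VFX, plays an equip montage and runs a timeline that drives a shared Reveal parameter on mesh materials and particles.
; The timeline's finish check for an opaque material swap ends without any action.
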(event FortWeapon|EventOnEquippedAndAttached
  (|CallParentFunction)
  (Utilities|FlowControl|Sequence
    (:then_0
      (bind _asfort_pawn (Utilities|Casting|CastToFortPawn (Game|GetInstigator)))
      (bind _asfort_player_pawn (Utilities|Casting|CastToFortPlayerPawn _asfort_pawn))
      (bind (_index _bindexisvalid _bfoundchannel) (CosmeticLoadout|GetVariantChannelActiveIndexforSlotActor "(TagName=\"LoadoutSlot.Pickaxe\")" _asfort_player_pawn "(TagName=\"Cosmetics.Variant.Channel.Particle\")" (Utilities|Array|MakeArray "(TagName=\"Cosmetics.Variant.Property.Particle1\")" "(TagName=\"Cosmetics.Variant.Property.Particle2\")") false false))
      (if (and _bindexisvalid _bfoundchannel)
        (Variables|Default|SetVariantIndex _index)))
    (:then_1
      ;[1 0.8991 0.2581] Set light blade material as hidden
      (CallFunction|SetRevealT 0.0)
      (if (Reactive|Frontend|IsCosmeticObjectDisplayedinFrontendContext 0)
        (Utilities|FlowControl|Delay 0.1)
        (else
          _))
      (CallFunction|PlayRevealSequence))))

(event Custom|PlayRevealSequence
  (bind _weaponmesh (Variables|Visual|GetWeaponMesh))
  (bind _returnvalue (Rendering|Material|GetMaterials _weaponmesh))
  (bind _returnvalue_2 (CallFunction|GetIdleVFXComponent))
  (Components|Activation|Deactivate _returnvalue_2)
  (Components|Activation|SetActive _returnvalue_2 true true)
  (Animation|Montage|PlayMontage (Variables|Visual|GetWeaponMesh) "/BRCosmetics/Weapons/FORT_Melee/Pickaxe_Rays_Female/Animation/Rays_Female_Pickaxe_Equip_Anim_Montage.Rays_Female_Pickaxe_Equip_Anim_Montage")
  (bind (_direction _materialreveal) (Timeline|OnEquippedTimeline PlayFromStart)
    (:Update
      (CallFunction|SetRevealT _materialreveal))
    (:Finished
      ;[1 0 0.2517] Set an Opaque version of the blade material for most of the gameplay
      (if (Utilities|Array|IsValidIndex _returnvalue 1)
        (bind _asmaterial_instance_dynamic (Utilities|Casting|CastToMaterialInstanceDynamic (Utilities|Array|Get(acopy) _returnvalue 1)))
        (if (== (Variables|/Script/Engine.MaterialInstance|GetParent _asmaterial_instance_dynamic) (Utilities|ResolveSoftReference (Variables|LightBlade|GetLightBladeMat)))
          _)))))

(fn GetAffiliationsToShowFor
  (return "Friendly"))

(fn IsLightVariant
  ;[1 1 1] todo
  (return (or (== (Variables|Default|GetVariantIndex) 1) (Reactive|Frontend|IsCosmeticObjectDisplayedinFrontendContext 0))))

(fn Melee_Effect_Color
  (return "0, 0, 0"))

(fn SetRevealT
  (params (T :type Float%20%28double-precision%29))
  (bind _returnvalue (CallFunction|GetIdleVFXComponent))
  (bind _weaponmesh (Variables|Visual|GetWeaponMesh))
  (Rendering|Material|SetScalarParameterValueonMaterials _weaponmesh "Reveal" T)
  (Rendering|Material|SetScalarParameterValueonMaterials _weaponmesh "MaxBrightness" 0.8)
  (Rendering|Material|SetScalarParameterValueonMaterials _weaponmesh "MinBrightness" 0.4)
  (Utilities|IsValid _returnvalue
    (:Is%20Valid
      (Effects|Components|ParticleSystem|SetFloatParameter _returnvalue "Reveal" T))))

(fn ShouldShowSoundIndicator
  (params (PlayerController :type Fort%20Player%20Controller%20Object%20Reference))
  (return false))

(fn ConstructionScript
  (|CallParentFunction))
