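; Cosmetic Blueprint for the dual-wield SirWolf Claws pickaxe, driving its Niagara equip, trail and impact effects.
; Equip spawns effects on both hands, flagging offhand systems with bIsOffhand.
; Swing events set SwingIndex and trail timing, with the final swing adding an extra trail.
; Impacts rotate from hit normal and SwingIndex and receive surface type; the final swing spawns a second offset impact, while frontend previews use an omnidirectional system.
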(event EventOnInitWeaponCosmetics
  (bind _returnvalue (CallFunction|GetIdleOffhandVFXComponent))
  (|CallParentFunction)
  ;[1 0.3164 0] On Equipped
  (Utilities|IsValid _returnvalue
    (:Is%20Valid
      (bind _asniagara_particle_system_component (Utilities|Casting|CastToNiagaraComponent _returnvalue))
      (Niagara|SetNiagaraVariable(Bool) _asniagara_particle_system_component "bIsOffhand" true))))

(event FortWeapon|EventOnEquippedAndAttached
  (bind _returnvalue (CallFunction|GetIdleOffhandVFXComponent))
  ;[1 0.3164 0] On Equipped
  (|CallParentFunction)
  (Utilities|FlowControl|Sequence
    (:then_0
      (Utilities|IsValid _returnvalue
        (:Is%20Valid
          (bind _asniagara_particle_system_component (Utilities|Casting|CastToNiagaraComponent _returnvalue))
          (Niagara|SetNiagaraVariable(Bool) _asniagara_particle_system_component "bIsOffhand" true))))
    (:then_1
      ;;[1 1 1] Equip FX
      (bind _returnvalue_2 (Niagara|SpawnSystemAttached "/BRCosmetics/Weapons/FORT_Melee/Pickaxe_Sirwolf/FX/NS_SirWolf_Pickaxe_Equip.NS_SirWolf_Pickaxe_Equip" (Variables|Visual|GetWeaponMesh) "None" "0, 0, 0" "0, 0, 0" "SnapToTarget"))
      (Niagara|SetNiagaraVariable(Bool) _returnvalue_2 "bIsOffhand" false)
      (bind _returnvalue_3 (Niagara|SpawnSystemAttached "/BRCosmetics/Weapons/FORT_Melee/Pickaxe_Sirwolf/FX/NS_SirWolf_Pickaxe_Equip.NS_SirWolf_Pickaxe_Equip" (Variables|Visual|GetWeaponMeshOffhand) "None" "0, 0, 0" "0, 0, 0" "SnapToTarget"))
      (Niagara|SetNiagaraVariable(Bool) _returnvalue_3 "bIsOffhand" true))))

(event EventSwingLeft
  (Variables|Default|SetSwingIndex 1)
  (|CallParentFunction)
  (CallFunction|SetParamsonTrails 0.2 0.175))

(event EventSwingRight
  (Variables|Default|SetSwingIndex 0)
  (|CallParentFunction)
  (CallFunction|SetParamsonTrails 0.2 0.175))

(event EventSwingRight2
  ;[0 1 0.09617] FX - Second swing right needs longer trail, short loop dur
  (|CallParentFunction)
  (Variables|Default|SetSwingIndex 2)
  (CallFunction|SetParamsonTrails 0.25 0.175))

(event FortWeapon|EventOnPlayImpactFX
  (params (HitResult :type Hit%20Result%20Structure%20%28by%20ref%29) (ImpactPhysicalSurface :type EPhysicalSurface%20Enum :default "SurfaceType_Default") (SpawnedPSC :type FXSystem%20Component%20Object%20Reference))
  (bind (_returnvalue_location _returnvalue_rotation) (Math|Transform|BreakTransform (Transformation|GetSocketTransform (Variables|Visual|GetWeaponMesh) "FX_FrontEndImpact" "RTS_World")))
  (if (Reactive|Frontend|IsCosmeticObjectDisplayedinFrontendContext self)
    ;[1 0.04746 0.1371] Spawn frontend impact - omnidirectional
    (Components|Activation|Deactivate SpawnedPSC)
    (bind _returnvalue (Niagara|SpawnSystematLocation "/BRCosmetics/Weapons/FORT_Melee/Pickaxe_Sirwolf/FX/NS_SirWolf_PickaxeImpact_Omni.NS_SirWolf_PickaxeImpact_Omni" _returnvalue_location _returnvalue_rotation))
    (Effects|Components|ParticleSystem|SetBoolParameter SpawnedPSC "bIsFrontEndPreview" true)
    (Components|DestroyComponent SpawnedPSC)
    (else
      (CallFunction|SpawnImpacts SpawnedPSC HitResult ImpactPhysicalSurface))))

(event EventSwingLeft2
  (bind _returnvalue 0.125)
  (bind _returnvalue_2 0.18)
  ;[0.8072 0 1] Final swing
  (|CallParentFunction)
  (Variables|Default|SetSwingIndex 3)
  (Utilities|FlowControl|Sequence
    (:then_0
      (CallFunction|SetParamsonTrails _returnvalue_2 _returnvalue))
    (:then_1
      (bind _returnvalue_3 (Niagara|SpawnSystemAttached "/BRCosmetics/Weapons/FORT_Melee/Pickaxe_Sirwolf/FX/NS_SirWolf_PickaxeTrail.NS_SirWolf_PickaxeTrail" (Variables|Visual|GetWeaponMesh) "None" "0, 0, 0" "0, 0, 0" "SnapToTarget"))
      (CallFunction|SetParamsonNSCTrail _returnvalue_3 _returnvalue_2 _returnvalue))))

(event EventBeginPlay)

(event EventTick
  (params (DeltaSeconds :type Float%20%28single-precision%29)))

(fn GetAffiliationsToShowFor
  (return "Friendly"))

(fn SetParamsOnNSCTrail
  (params (Sys :type Niagara%20Particle%20System%20Component%20Object%20Reference) (LoopDuration :type Float%20%28single-precision%29) (Lifetime :type Float%20%28single-precision%29) (bOffhand :type Boolean))
  (Niagara|SetNiagaraVariable(Float) Sys "LoopDuration" LoopDuration)
  (Niagara|SetNiagaraVariable(Float) Sys "Lifetime" Lifetime)
  (Niagara|SetNiagaraVariable(Bool) Sys "bIsOffhand" bOffhand))

(fn SetParamsOnTrails
  (params (LoopDuration :type Float%20%28single-precision%29 :default 0.2) (TrailLifetime :type Float%20%28single-precision%29 :default 0.2))
  (bind _returnvalue (CallFunction|GetAnimTrailsOffhandPSC))
  (for _array_element (Utilities|Array|MakeArray (CallFunction|GetAnimTrailsPSC) _returnvalue)
    (bind _asniagara_particle_system_component (Utilities|Casting|CastToNiagaraComponent _array_element))
    (CallFunction|SetParamsonNSCTrail _asniagara_particle_system_component LoopDuration TrailLifetime (== _array_element _returnvalue))))

(fn ShouldShowSoundIndicator
  (params (PlayerController :type Fort%20Player%20Controller%20Object%20Reference))
  (return false))

(fn SpawnImpacts
  (params (Object :type Object%20Reference) (Hit :type Hit%20Result%20Structure%20%28by%20ref%29) (InputPin :type EPhysicalSurface%20Enum :default "SurfaceType_Default"))
  (bind (_bblockinghit _binitialoverlap _time _distance _location _impactpoint _normal) (Collision|BreakHitResult Hit))
  (bind _returnvalue (Math|Rotator|MakeRotfromX _normal))
  (bind _impact_fx (Variables|Default|GetImpact%20FX))
  (bind _returnvalue_2 (Transformation|GetWorldTransform _impact_fx))
  (bind _swingindex (Variables|Default|GetSwingIndex))
  (bind _returnvalue_3 (Utilities|Select :Index _swingindex :Option%200 "0.000000,0.000000,12.000000" :Option%201 "0.000000,0.000000,200.000000" :Option%202 "0.000000,0.000000,-60.000000" :Option%203 "0.000000,0.000000,90.000000"))
  (bind _returnvalue_4 (Math|Rotator|CombineRotators "0.000000,90.000000,0.000000" _returnvalue_3))
  (bind _self self)
  (bind _returnvalue_5 (Reactive|Frontend|IsCosmeticObjectDisplayedinFrontendContext _self))
  (bind _swingindex_2 (Variables|Default|GetSwingIndex))
  (bind _returnvalue_6 (== _swingindex_2 3))
  (if (not (Reactive|Frontend|IsCosmeticObjectDisplayedinFrontendContext self))
    (LocalVariables|Setimpact_type (Utilities|Select :Index InputPin :SurfaceType_Default 0 :SurfaceType1 1 :SurfaceType2 2 :SurfaceType3 3 :SurfaceType4 4 :SurfaceType5 5 :SurfaceType6 6 :SurfaceType7 7 :SurfaceType8 8 :SurfaceType9 9 :SurfaceType10 10 :SurfaceType11 11 :SurfaceType12 12 :SurfaceType13 13 :SurfaceType14 14 :SurfaceType15 15 :SurfaceType16 16 :SurfaceType17 17 :SurfaceType18 18 :SurfaceType19 19 :SurfaceType20 20 :SurfaceType21 21 :SurfaceType22 22 :SurfaceType23 23 :SurfaceType24 24 :SurfaceType25 25 :SurfaceType26 26 :SurfaceType27 0 :SurfaceType28 0 :SurfaceType29 0 :SurfaceType30 0 :SurfaceType31 0 :SurfaceType32 0 :SurfaceType33 0 :SurfaceType34 0))
    (bind _asniagara_particle_system_component (Utilities|Casting|CastToNiagaraComponent Object))
    ;[1 0 0.1103] From DualWield_Generic, needed to change rotations x4
    (bind _output_get (Variables|Default|SetImpact%20FX _asniagara_particle_system_component))
    (Utilities|FlowControl|Sequence
      (:then_0
        (Effects|Components|ParticleSystem|SetBoolParameter (Variables|Default|GetImpact%20FX) "SecondImpact" _returnvalue_6)
        (if _returnvalue_6
          (bind _returnvalue_7 (Niagara|SpawnSystematLocation (Niagara|GetNiagaraSystemAsset _asniagara_particle_system_component) (+ (.location _returnvalue_2) (* (Math|Vector|GetForwardVector (Math|Transform|TransformRotation _returnvalue_2 (Math|Rotator|CombineRotators (Math|Transform|InverseTransformRotation (Transformation|GetWorldTransform _output_get) _returnvalue) "0.000000,90.000000,0.000000"))) 4e+01)) (Math|Rotator|CombineRotators (Math|Transform|TransformRotation _returnvalue_2 (Math|Rotator|CombineRotators (Math|Transform|InverseTransformRotation (Transformation|GetWorldTransform _output_get) _returnvalue) _returnvalue_4)))))
          (Effects|Components|ParticleSystem|SetBoolParameter _returnvalue_7 "SecondImpact" true)
          (Effects|Components|ParticleSystem|SetIntParameter _returnvalue_7 "PhysicalSurface" (LocalVariables|Getimpact_type))
          (Effects|Components|ParticleSystem|SetBoolParameter _returnvalue_7 "bIsFrontEndPreview" _returnvalue_5)))
      (:then_1
        (Transformation|SetWorldRotation _impact_fx (Math|Transform|TransformRotation _returnvalue_2 (Math|Rotator|CombineRotators (Math|Transform|InverseTransformRotation (Transformation|GetWorldTransform _output_get) _returnvalue) _returnvalue_4)) false false)
        (Effects|Components|ParticleSystem|SetIntParameter (Variables|Default|GetImpact%20FX) "PhysicalSurface" (LocalVariables|Getimpact_type))
        ;[1 1 1] Front end
        (Effects|Components|ParticleSystem|SetBoolParameter (Variables|Default|GetImpact%20FX) "bIsFrontEndPreview" _returnvalue_5)))))

(fn ConstructionScript
  (|CallParentFunction))
