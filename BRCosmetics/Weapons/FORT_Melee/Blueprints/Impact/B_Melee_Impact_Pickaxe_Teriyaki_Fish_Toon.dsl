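; Pickaxe Blueprint for the Teriyaki Fish Toon melee weapon that escalates its impact effects and audio as hits accumulate.
; Each impact calls the parent effect, increments a hit counter and advances a phase variable at set hit thresholds.
; A custom event then spawns a light or heavy voice impact sound attached to the weapon mesh, based on that phase.
; Equipping resets the counter and phase.
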(event FortWeapon|EventOnPlayImpactFX
  (params (HitResult :type Hit%20Result%20Structure%20%28by%20ref%29) (ImpactPhysicalSurface :type EPhysicalSurface%20Enum :default "SurfaceType_Default") (SpawnedPSC :type FXSystem%20Component%20Object%20Reference))
  (bind _hitcounter (Variables|Default|GetHitCounter))
  ;[1 1 1] Impact FX
  (Variables|Default|SetPIckaxe%20Impact HitResult)
  (|CallParentFunction HitResult ImpactPhysicalSurface SpawnedPSC)
  (Utilities|FlowControl|Sequence
    (:then_0
      ;;[1 1 1] set Hit Counter variable at value of 1 to start, is 0 on onEquip
      (Variables|Default|SetHitCounter (+ (Variables|Default|GetHitCounter) 1)))
    (:then_1
      (if (== _hitcounter 1)
        (Variables|Default|SetPhaseSelection 1)))
    (:then_2
      (if (== _hitcounter 5)
        (Variables|Default|SetPhaseSelection 2)))
    (:then_3
      (if (== _hitcounter 9)
        (Variables|Default|SetPhaseSelection 3)))
    (:then_4
      (CallFunction|HandleCustomImpactAudio))))

(event FortWeapon|EventOnEquippedAndAttached
  ;[1 1 1] Impact FX
  ;;[1 1 1] Reset on Equip
  (|CallParentFunction)
  (Variables|Default|SetHitCounter 0)
  (Variables|Default|SetPhaseSelection 1))

(event Custom|Handle%20Custom%20Impact%20Audio
  ;[1 1 1] Switch between light / heavy VOX impact audio
  (Audio|SpawnSoundAttached (select (>= (Variables|Default|GetPhaseSelection) 2) (Variables|Default|GetSoundHeavyDamage) (Variables|Default|GetSoundLightDamage)) (Variables|Visual|GetWeaponMesh)))

(fn ConstructionScript
  (|CallParentFunction))
