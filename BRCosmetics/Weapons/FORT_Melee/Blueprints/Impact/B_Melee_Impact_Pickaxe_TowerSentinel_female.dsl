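; Impact Blueprint for the female Tower Sentinel pickaxe that animates its idle Niagara VFX when weapon fire FX play.
; After calling the parent handler, it skips the effect if the weapon mesh is a cosmetic displayed in the frontend.
; Otherwise, it plays a timeline from the start, feeding flame length, WPO interpolation, smoke count and flame alpha into the idle Niagara component every update.
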(event EventOnPlayWeaponFireFX
  (params (bPersistentFire :type Boolean) (bSecondaryFire :type Boolean))
  (|CallParentFunction bPersistentFire bSecondaryFire)
  (if (not (Reactive|Frontend|IsCosmeticObjectDisplayedinFrontendContext (Variables|Visual|GetWeaponMesh)))
    (bind _asniagara_particle_system_component (Utilities|Casting|CastToNiagaraComponent (CallFunction|GetIdleVFXComponent)))
    (bind (_direction _flame_length _interpolate _smoke _flame_alpha) (Timeline|length PlayFromStart)
      (:Update
        (Niagara|SetNiagaraVariable(Float) _asniagara_particle_system_component "Flame_Length" _flame_length)
        (Niagara|SetNiagaraVariable(Float) _asniagara_particle_system_component "Interpolate_WPO" _interpolate)
        (Niagara|SetNiagaraVariable(Float) _asniagara_particle_system_component "Smoke_number" _smoke)
        (Niagara|SetNiagaraVariable(Float) _asniagara_particle_system_component "Flame_alpha" _flame_alpha)))))

(fn ConstructionScript)
