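; Cosmetic Blueprint for the Wukong pickaxe that builds up a reactive Niagara idle effect as the pickaxe lands impacts.
; Each impact activates the effect, increments a stack count that selects one of three stages with timestamped emissive parameters, and restarts a timeline that drives opacity.
; The timeline's end or hiding the weapon clears stacks and deactivates the effect.
; In frontend display, init and BeginPlay run a delayed demo of the stacking.
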(event EventOnInitWeaponCosmetics
  (bind _returnvalue (Utilities|Time|GetGameTimeinSeconds))
  (bind _ns_pickaxe_wukong_idlefx (Variables|Default|GetNS_Pickaxe_WuKong_IdleFX))
  ; Initialize
  (|CallParentFunction)
  (Variables|Default|SetStacks 0)
  (Components|Activation|Deactivate (Variables|Default|GetNS_Pickaxe_WuKong_IdleFX))
  ; FE Reactivity
  (if (Reactive|Frontend|IsCosmeticObjectDisplayedinFrontendContext 0)
    (Utilities|FlowControl|Delay 1.0)
    (Utilities|FlowControl|Sequence
      (:then_0
        (Variables|Default|SetStacks 0)
        (Niagara|ReinitializeSystem _ns_pickaxe_wukong_idlefx)
        ; Reactivity
        (Components|Activation|Activate _ns_pickaxe_wukong_idlefx false)
        (Math|Integer|IncrementInt (Variables|Default|GetStacks)
          (:%20%20
            (Utilities|FlowControl|Sequence
              (:then_0
                _)
              (:then_1
                (bind (_direction _emissiveamount) (Timeline|ReactivityDuration PlayFromStart)
                  (:Update
                    _)
                  (:Finished
                    (Variables|Default|SetStacks 0)
                    (Components|Activation|Deactivate _ns_pickaxe_wukong_idlefx)
                    (if (Reactive|Frontend|IsCosmeticObjectDisplayedinFrontendContext 0)
                      _)))
                (Effects|Components|ParticleSystem|SetFloatParameter _ns_pickaxe_wukong_idlefx "Opacity" _emissiveamount)))
            (if (>= (Variables|Default|GetStacks) 4)
              (if (>= (Variables|Default|GetStacks) 7)
                ;; Stage 3
                (Effects|Components|ParticleSystem|SetFloatParameter _ns_pickaxe_wukong_idlefx "EmissiveStage3" _returnvalue)
                (Effects|Components|ParticleSystem|SetFloatParameter _ns_pickaxe_wukong_idlefx "Stage" 3.0)
                (else
                  ;; Stage 2
                  (Effects|Components|ParticleSystem|SetFloatParameter _ns_pickaxe_wukong_idlefx "EmissiveStage2" _returnvalue)
                  (Effects|Components|ParticleSystem|SetFloatParameter _ns_pickaxe_wukong_idlefx "Stage" 2.0)))
              (else
                ;; Stage 1
                (Effects|Components|ParticleSystem|SetFloatParameter _ns_pickaxe_wukong_idlefx "EmissiveStage1" (Utilities|Time|GetGameTimeinSeconds))
                (Effects|Components|ParticleSystem|SetFloatParameter _ns_pickaxe_wukong_idlefx "Stage" 1.0))))))
      (:then_1
        _))
    (Utilities|FlowControl|Delay 2.0)
    (Utilities|FlowControl|Sequence
      (:then_0
        (Variables|Default|SetStacks 4))
      (:then_1
        _))
    (Utilities|FlowControl|Delay 2.0)
    (Variables|Default|SetStacks 7)))

(event EventBeginPlay
  (bind _returnvalue (Utilities|Time|GetGameTimeinSeconds))
  (bind _ns_pickaxe_wukong_idlefx (Variables|Default|GetNS_Pickaxe_WuKong_IdleFX))
  ; Initialize
  (|CallParentFunction)
  (Variables|Default|SetStacks 0)
  (Components|Activation|Deactivate (Variables|Default|GetNS_Pickaxe_WuKong_IdleFX))
  ; FE Reactivity
  (if (Reactive|Frontend|IsCosmeticObjectDisplayedinFrontendContext 0)
    (Utilities|FlowControl|Delay 1.0)
    (Utilities|FlowControl|Sequence
      (:then_0
        (Variables|Default|SetStacks 0)
        (Niagara|ReinitializeSystem _ns_pickaxe_wukong_idlefx)
        ; Reactivity
        (Components|Activation|Activate _ns_pickaxe_wukong_idlefx false)
        (Math|Integer|IncrementInt (Variables|Default|GetStacks)
          (:%20%20
            (Utilities|FlowControl|Sequence
              (:then_0
                _)
              (:then_1
                (Timeline|ReactivityDuration PlayFromStart)))
            (if (>= (Variables|Default|GetStacks) 4)
              (if (>= (Variables|Default|GetStacks) 7)
                ;; Stage 3
                (Effects|Components|ParticleSystem|SetFloatParameter _ns_pickaxe_wukong_idlefx "EmissiveStage3" _returnvalue)
                (Effects|Components|ParticleSystem|SetFloatParameter _ns_pickaxe_wukong_idlefx "Stage" 3.0)
                (else
                  ;; Stage 2
                  (Effects|Components|ParticleSystem|SetFloatParameter _ns_pickaxe_wukong_idlefx "EmissiveStage2" _returnvalue)
                  (Effects|Components|ParticleSystem|SetFloatParameter _ns_pickaxe_wukong_idlefx "Stage" 2.0)))
              (else
                ;; Stage 1
                (Effects|Components|ParticleSystem|SetFloatParameter _ns_pickaxe_wukong_idlefx "EmissiveStage1" (Utilities|Time|GetGameTimeinSeconds))
                (Effects|Components|ParticleSystem|SetFloatParameter _ns_pickaxe_wukong_idlefx "Stage" 1.0))))))
      (:then_1
        _))
    (Utilities|FlowControl|Delay 2.0)
    (Utilities|FlowControl|Sequence
      (:then_0
        (Variables|Default|SetStacks 4))
      (:then_1
        _))
    (Utilities|FlowControl|Delay 2.0)
    (Variables|Default|SetStacks 7)))

(event FortWeapon|EventOnPlayImpactFX
  (params (HitResult :type Hit%20Result%20Structure%20%28by%20ref%29) (ImpactPhysicalSurface :type EPhysicalSurface%20Enum :default "SurfaceType_Default") (SpawnedPSC :type FXSystem%20Component%20Object%20Reference))
  (bind _returnvalue (Utilities|Time|GetGameTimeinSeconds))
  (bind _ns_pickaxe_wukong_idlefx (Variables|Default|GetNS_Pickaxe_WuKong_IdleFX))
  (|CallParentFunction HitResult ImpactPhysicalSurface SpawnedPSC)
  ; Reactivity
  (Components|Activation|Activate _ns_pickaxe_wukong_idlefx false)
  (Math|Integer|IncrementInt (Variables|Default|GetStacks)
    (:%20%20
      (Utilities|FlowControl|Sequence
        (:then_0
          _)
        (:then_1
          (Timeline|ReactivityDuration PlayFromStart)))
      (if (>= (Variables|Default|GetStacks) 4)
        (if (>= (Variables|Default|GetStacks) 7)
          ;; Stage 3
          (Effects|Components|ParticleSystem|SetFloatParameter _ns_pickaxe_wukong_idlefx "EmissiveStage3" _returnvalue)
          (Effects|Components|ParticleSystem|SetFloatParameter _ns_pickaxe_wukong_idlefx "Stage" 3.0)
          (else
            ;; Stage 2
            (Effects|Components|ParticleSystem|SetFloatParameter _ns_pickaxe_wukong_idlefx "EmissiveStage2" _returnvalue)
            (Effects|Components|ParticleSystem|SetFloatParameter _ns_pickaxe_wukong_idlefx "Stage" 2.0)))
        (else
          ;; Stage 1
          (Effects|Components|ParticleSystem|SetFloatParameter _ns_pickaxe_wukong_idlefx "EmissiveStage1" (Utilities|Time|GetGameTimeinSeconds))
          (Effects|Components|ParticleSystem|SetFloatParameter _ns_pickaxe_wukong_idlefx "Stage" 1.0))))))

(event EventOnWeaponVisibilityChanged
  (params (bVisible :type Boolean) (bSetForLocalControllerOnly :type Boolean))
  ; Disable Idle FX on Weapon swap
  (if (not bVisible)
    (Variables|Default|SetStacks 0)
    (Components|Activation|Deactivate (Variables|Default|GetNS_Pickaxe_WuKong_IdleFX))))

(fn ConstructionScript
  (|CallParentFunction))
