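; Impact Blueprint for the ZestFemale pickaxe that drives a reactive material effect from consecutive hits.
; After the parent impact logic, it counts hits and writes hit state and count to a Material Parameter Collection, skipping this during animation.
; At eight or more hits, a timeline drives the hit animation parameter, then resets the count.
; A delayed count snapshot compared in a timer callback clears the streak when hits stop.
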(event FortWeapon|EventOnPlayImpactFX
  (params (HitResult :type Hit%20Result%20Structure%20%28by%20ref%29) (ImpactPhysicalSurface :type EPhysicalSurface%20Enum :default "SurfaceType_Default") (SpawnedPSC :type FXSystem%20Component%20Object%20Reference))
  (bind _pixkaxe_impact_count (Variables|Default|GetPixkaxe%20Impact%20Count))
  (Variables|Default|SetPIckaxe%20Impact HitResult)
  (|CallParentFunction HitResult ImpactPhysicalSurface SpawnedPSC)
  (if (not (Variables|Default|GetIs%20playing%20animation))
    (Utilities|FlowControl|Sequence
      (:then_0
        ;[1 1 1] Play special sound for pickaxe when hitting actors
        ;[1 1 1] HIt_fx
        (bind _result (Math|Integer|IncrementInt (Variables|Default|GetPixkaxe%20Impact%20Count))
          (:%20%20
            (bind _output_get (Variables|Default|SetPixkaxe%20Impact%20Count _result))
            (Utilities|FlowControl|Delay (Variables|Default|GetTime))
            (Variables|Default|SetHItChange _output_get))))
      (:then_1
        _)
      (:then_2
        (if (!= (Variables|Default|GetPixkaxe%20Impact%20Count) 0)
          (Rendering|Material|SetScalarParameterValue "/BRCosmetics/Weapons/FORT_Melee/Pickaxe_Zest_Female/Materials/MPC_Pickaxe_zest_female.MPC_Pickaxe_zest_female" "Hiting?" 1.0 :!class /Script/Engine.KismetMaterialLibrary)
          (else
            (Rendering|Material|SetScalarParameterValue "/BRCosmetics/Weapons/FORT_Melee/Pickaxe_Zest_Female/Materials/MPC_Pickaxe_zest_female.MPC_Pickaxe_zest_female" "Hiting?" :!class /Script/Engine.KismetMaterialLibrary))))
      (:then_3
        _)
      (:then_4
        (if (not (== (Variables|Default|GetPixkaxe%20Impact%20Count) 0))
          (Utilities|Time|SetTimerbyEvent (delegate Custom|CustomEvent_0) (Variables|Default|GetTime)))))
    (Rendering|Material|SetScalarParameterValue "/BRCosmetics/Weapons/FORT_Melee/Pickaxe_Zest_Female/Materials/MPC_Pickaxe_zest_female.MPC_Pickaxe_zest_female" "Hit count" (Math|Conversions|ToFloat(Integer) _pixkaxe_impact_count) :!class /Script/Engine.KismetMaterialLibrary)
    (if (>= _pixkaxe_impact_count 8)
      (bind (_direction _hit_animation) (Timeline|Timeline_0 PlayFromStart)
        (:Update
          (Variables|Default|SetIs%20playing%20animation true)
          (Rendering|Material|SetScalarParameterValue "/BRCosmetics/Weapons/FORT_Melee/Pickaxe_Zest_Female/Materials/MPC_Pickaxe_zest_female.MPC_Pickaxe_zest_female" "Hit animation" _hit_animation :!class /Script/Engine.KismetMaterialLibrary))
        (:Finished
          (Rendering|Material|SetScalarParameterValue "/BRCosmetics/Weapons/FORT_Melee/Pickaxe_Zest_Female/Materials/MPC_Pickaxe_zest_female.MPC_Pickaxe_zest_female" "Hit animation" :!class /Script/Engine.KismetMaterialLibrary)
          (Variables|Default|SetPixkaxe%20Impact%20Count 0)
          (Rendering|Material|SetScalarParameterValue "/BRCosmetics/Weapons/FORT_Melee/Pickaxe_Zest_Female/Materials/MPC_Pickaxe_zest_female.MPC_Pickaxe_zest_female" "Hiting?" :!class /Script/Engine.KismetMaterialLibrary)
          (Variables|Default|SetIs%20playing%20animation false))))))

(event Custom|CustomEvent_0
  (if (== (Variables|Default|GetPixkaxe%20Impact%20Count) (Variables|Default|GetHItChange))
    (Rendering|Material|SetScalarParameterValue "/BRCosmetics/Weapons/FORT_Melee/Pickaxe_Zest_Female/Materials/MPC_Pickaxe_zest_female.MPC_Pickaxe_zest_female" "Hiting?" :!class /Script/Engine.KismetMaterialLibrary)
    (Variables|Default|SetPixkaxe%20Impact%20Count 0)
    (Rendering|Material|SetScalarParameterValue "/BRCosmetics/Weapons/FORT_Melee/Pickaxe_Zest_Female/Materials/MPC_Pickaxe_zest_female.MPC_Pickaxe_zest_female" "Hit count" :!class /Script/Engine.KismetMaterialLibrary)
    (Rendering|Material|SetScalarParameterValue "/BRCosmetics/Weapons/FORT_Melee/Pickaxe_Zest_Female/Materials/MPC_Pickaxe_zest_female.MPC_Pickaxe_zest_female" "Hit animation" :!class /Script/Engine.KismetMaterialLibrary)))

(fn ConstructionScript
  (|CallParentFunction))
