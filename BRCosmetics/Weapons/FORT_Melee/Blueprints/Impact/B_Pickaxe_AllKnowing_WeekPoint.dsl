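; Impact Blueprint for the AllKnowing pickaxe that spawns a Niagara effect on the weapon mesh's FX_Idle_01 socket for weak-point hits.
; On impact it relays the physical surface through CE_FX_Server to CE_FX_Client, which spawns the effect if that surface is in the weak-point surface list.
; In a frontend cosmetic preview, every impact also spawns the effect with bIsFrontEndPreview set.
; The OnRep for the current impact surface is empty.
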(event FortWeapon|EventOnPlayImpactFX
  (params (HitResult :type Hit%20Result%20Structure%20%28by%20ref%29) (ImpactPhysicalSurface :type EPhysicalSurface%20Enum :default "SurfaceType_Default") (SpawnedPSC :type FXSystem%20Component%20Object%20Reference))
  (bind _self self)
  (bind _returnvalue (Reactive|Frontend|IsCosmeticObjectDisplayedinFrontendContext _self))
  (|CallParentFunction HitResult ImpactPhysicalSurface SpawnedPSC)
  (Utilities|FlowControl|Sequence
    (:then_0
      (if _returnvalue
        (bind _returnvalue_2 (Effects|SpawnFXSystemAttached "/BRCosmetics/Weapons/FORT_Melee/Pickaxe_AllKnowing/FX/NS_Pickaxe_AllKnowing_WeekPoint.NS_Pickaxe_AllKnowing_WeekPoint" (Variables|Visual|GetWeaponMesh) "FX_Idle_01"))
        (Effects|Components|ParticleSystem|SetBoolParameter _returnvalue_2 "bIsFrontEndPreview" _returnvalue)))
    (:then_1
      (CallFunction|CEFXServer ImpactPhysicalSurface))))

(event Custom|CE_FX_Server
  (params (Impact%20Physical%20Surface :type EPhysicalSurface%20Enum :default "SurfaceType_Default"))
  (CallFunction|CEFXClient Impact%20Physical%20Surface))

(event Custom|CE_FX_Client
  (params (Impact%20Physical%20Surface :type EPhysicalSurface%20Enum :default "SurfaceType_Default"))
  (bind _output_get (Variables|Default|SetCurrent%20Impact%20Physical%20Surface Impact%20Physical%20Surface))
  (Variables|Default|SetIs%20hit%20weak%20point (Utilities|Array|ContainsItem (Variables|Default|GetWeakPoint%20Physical%20Surface) _output_get))
  (if (Variables|Default|GetIs%20hit%20weak%20point)
    (Effects|SpawnFXSystemAttached "/BRCosmetics/Weapons/FORT_Melee/Pickaxe_AllKnowing/FX/NS_Pickaxe_AllKnowing_WeekPoint.NS_Pickaxe_AllKnowing_WeekPoint" (Variables|Visual|GetWeaponMesh) "FX_Idle_01")))

(fn OnRep_CurrentImpactPhysicalSurface)

(fn ConstructionScript
  (|CallParentFunction))
