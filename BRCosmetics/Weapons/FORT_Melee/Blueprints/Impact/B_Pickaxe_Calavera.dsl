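; Calavera pickaxe cosmetic Blueprint that animates particle color, dissolve and weapon glow when impacts play.
; In frontend display, each impact runs a timeline driving idle VFX and mesh emissive/pulse parameters.
; Every impact also passes its surface type through CE_FX_Server to CE_FX_Client.
; There, only weak-point surfaces start two timelines: one for impact color and idle dissolve, another for mesh emissive and pulse.
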(event FortWeapon|EventOnPlayImpactFX
  (params (HitResult :type Hit%20Result%20Structure%20%28by%20ref%29) (ImpactPhysicalSurface :type EPhysicalSurface%20Enum :default "SurfaceType_Default") (SpawnedPSC :type FXSystem%20Component%20Object%20Reference))
  (bind _returnvalue (CallFunction|GetIdleVFXComponent))
  (bind _weaponmesh (Variables|Visual|GetWeaponMesh))
  ;[0 0 0] Game
  (|CallParentFunction HitResult ImpactPhysicalSurface SpawnedPSC)
  (Utilities|FlowControl|Sequence
    (:then_0
      ;[0 0 0] FE
      (if (Reactive|Frontend|IsCosmeticObjectDisplayedinFrontendContext self)
        (Effects|Components|ParticleSystem|SetFloatParameter (Variables|Default|GetImpact%20FX) "ColorProgress" 1.0)
        (bind (_direction _colorprogress _dissolveprogress) (Timeline|Timeline_FE PlayFromStart)
          (:Update
            (Effects|Components|ParticleSystem|SetFloatParameter _returnvalue "ColorProgress" _colorprogress)
            (Effects|Components|ParticleSystem|SetFloatParameter _returnvalue "DissolveProgress" _dissolveprogress)
            (Rendering|Material|SetScalarParameterValueonMaterials _weaponmesh "VibrantEmissiveAmount" _colorprogress)
            (Rendering|Material|SetScalarParameterValueonMaterials _weaponmesh "PulseVisibility" _colorprogress)))))
    (:then_1
      (CallFunction|CEFXServer ImpactPhysicalSurface))))

(event Custom|CE_FX_Server
  (params (Impact%20Physical%20Surface :type EPhysicalSurface%20Enum :default "SurfaceType_Default"))
  ;[0 0 0] Game
  (CallFunction|CEFXClient Impact%20Physical%20Surface))

(event Custom|CE_FX_Client
  (params (Impact%20Physical%20Surface :type EPhysicalSurface%20Enum :default "SurfaceType_Default"))
  (bind _weaponmesh (Variables|Visual|GetWeaponMesh))
  ;[0 0 0] Game
  (bind _output_get (Variables|Default|SetCurrent%20Impact%20Physical%20Surface Impact%20Physical%20Surface))
  (Variables|Default|SetIs%20hit%20weak%20point (Utilities|Array|ContainsItem (Variables|Default|GetWeakPoint%20Physical%20Surface) _output_get))
  (if (Variables|Default|GetIs%20hit%20weak%20point)
    (Utilities|FlowControl|Sequence
      (:then_0
        (bind (_direction _colorprogress _dissolveprogress) (Timeline|Timeline_Game PlayFromStart)
          (:Update
            (Effects|Components|ParticleSystem|SetFloatParameter (Variables|Default|GetImpact%20FX) "ColorProgress" _colorprogress)
            (Effects|Components|ParticleSystem|SetFloatParameter (CallFunction|GetIdleVFXComponent) "DissolveProgress" _dissolveprogress))))
      (:then_1
        (bind (_direction_2 _vibrantamount _pulsevisibility) (Timeline|Vibrant%20Amount PlayFromStart)
          (:Update
            (Rendering|Material|SetScalarParameterValueonMaterials _weaponmesh "VibrantEmissiveAmount" _vibrantamount)
            (Rendering|Material|SetScalarParameterValueonMaterials _weaponmesh "PulseVisibility" _pulsevisibility)))))))

(fn OnRep_CurrentImpactPhysicalSurface)

(fn ConstructionScript
  (|CallParentFunction))
