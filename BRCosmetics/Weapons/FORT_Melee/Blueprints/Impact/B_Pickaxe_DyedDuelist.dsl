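; Dyed Duelist pickaxe Blueprint that animates weapon material glow and SpellSpeed on impacts.
; Impact FX relays the surface through server and client custom events; weak-point hits raise a capped hit counter and play a timeline.
; The counter selects which timeline track drives SpellSpeed, and a delayed looping timer truncates it down one step per tick.
; In frontend cosmetic displays, a separate timeline drives both material parameters.
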(event FortWeapon|EventOnPlayImpactFX
  (params (HitResult :type Hit%20Result%20Structure%20%28by%20ref%29) (ImpactPhysicalSurface :type EPhysicalSurface%20Enum :default "SurfaceType_Default") (SpawnedPSC :type FXSystem%20Component%20Object%20Reference))
  (bind _weaponmesh (Variables|Visual|GetWeaponMesh))
  ;[0 0 0] Game
  (|CallParentFunction HitResult ImpactPhysicalSurface SpawnedPSC)
  ;[0 0 0] FE
  (Utilities|FlowControl|Sequence
    (:then_0
      (if (Reactive|Frontend|IsCosmeticObjectDisplayedinFrontendContext self)
        (bind (_direction _colorprogress _spellspeed) (Timeline|Timeline_FE PlayFromStart)
          (:Update
            (Rendering|Material|SetScalarParameterValueonMaterials _weaponmesh "VibrantEmissiveAmount" _colorprogress)
            (Rendering|Material|SetScalarParameterValueonMaterials _weaponmesh "SpellSpeed" _spellspeed)))))
    (:then_1
      (CallFunction|CEFXServer ImpactPhysicalSurface))))

(event Custom|CE_FX_Server
  (params (Impact%20Physical%20Surface :type EPhysicalSurface%20Enum :default "SurfaceType_Default"))
  ;[0 0 0] Game
  (CallFunction|CEFXClient Impact%20Physical%20Surface))

(event Custom|CE_FX_Client
  (params (Impact%20Physical%20Surface :type EPhysicalSurface%20Enum :default "SurfaceType_Default"))
  (bind _weaponmesh (Variables|Visual|GetWeaponMesh))
  ;[0 0 0] Game
  (bind _output_get (Variables|Default|SetCurrent%20Impact%20Physical%20Surface Impact%20Physical%20Surface))
  (Variables|Default|SetIs%20hit%20weak%20point (Utilities|Array|ContainsItem (Variables|Default|GetWeakPoint%20Physical%20Surface) _output_get))
  (if (Variables|Default|GetIs%20hit%20weak%20point)
    (if (> (Variables|Default|GetNumberOfHits) 4)
      (Variables|Default|SetNumberOfHits 5.0)
      (else
        (Variables|Default|SetNumberOfHits (+ (Variables|Default|GetNumberOfHits) 1.0))))
    (CallFunction|DecayEvent)
    (bind (_direction _vibrantamount _spellspeed01 _spellspeed02 _spellspeed03 _spellspeed04 _spellspeed05) (Timeline|Vibrant%20Amount PlayFromStart)
      (:Update
        (Rendering|Material|SetScalarParameterValueonMaterials _weaponmesh "VibrantEmissiveAmount" _vibrantamount)
        (switch int (Math|Float|Truncate (Variables|Default|GetNumberOfHits))
          (:0
            (Variables|Default|SetSpellSpeed 0.0))
          (:1
            (Variables|Default|SetSpellSpeed _spellspeed01))
          (:2
            (Variables|Default|SetSpellSpeed _spellspeed02))
          (:3
            (Variables|Default|SetSpellSpeed _spellspeed03))
          (:4
            (Variables|Default|SetSpellSpeed _spellspeed04))
          (:5
            (Variables|Default|SetSpellSpeed _spellspeed05)))
        (Rendering|Material|SetScalarParameterValueonMaterials _weaponmesh "SpellSpeed" (Variables|Default|GetSpellSpeed))))))

(event Custom|Decay%20Event
  ;[0 0 0] Decay
  (bind _returnvalue (Utilities|Time|SetTimerbyEvent (delegate Custom|DDDecay%20Event) 0.1 true false 7.0)))

(event Custom|DDDecay%20Event
  ;[0 0 0] Decay
  (if (> (Variables|Default|GetNumberOfHits) 0)
    (bind _output_get (Variables|Default|SetNumberOfHits (Math|Conversions|ToFloat(Integer) (Math|Float|Truncate (- (Variables|Default|GetNumberOfHits) 0.05)))))
    (else
      (Utilities|Time|PauseTimerbyHandle (stale Custom|Decay%20Event _returnvalue)))))

(fn OnRep_CurrentImpactPhysicalSurface)

(fn ConstructionScript
  (|CallParentFunction))
