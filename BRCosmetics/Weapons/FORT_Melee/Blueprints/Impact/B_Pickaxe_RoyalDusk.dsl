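; Royal Dusk pickaxe Blueprint that charges its gem color on weak-point hits, or on any impact in the frontend display.
; Impacts route through server and client events into a timeline that steps intensity up to a cap and drives curve-mapped weapon material parameters.
; A delayed looping timer, re-set during timeline updates, decays intensity.
; The color blend feeds Niagara StateBlend on idle VFX and, at each swing, on trail and swing VFX.
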(event Custom|UpdateColorProgress_Event
  (bind _countintensity (Variables|Default|GetCountIntensity))
  (bind _countintensity_2 (Variables|Default|GetCountIntensity))
  ;[0 0 0] Update Color Progress
  (if (Variables|Default|GetIsTimelineRunning)
    (if (not (> _countintensity_2 2))
      (Variables|Default|SetCountIntensity (+ _countintensity_2 1.0))
      (bind (_direction _gempulse _colorprogress _emissionintensity) (Timeline|Timeline PlayFromStart)
        (:Update
          (Variables|Default|SetIsTimelineRunning true)
          (Rendering|Material|SetScalarParameterValueonMaterials (Variables|Visual|GetWeaponMesh) "GemPulse" _gempulse)
          (Rendering|Material|SetScalarParameterValueonMaterials (Variables|Visual|GetWeaponMesh) "ColorProgress" (Math|Curves|GetFloatValue (Variables|Default|GetColorProogressCurve) (Math|Float|Clamp(Float) (+ _countintensity _colorprogress) 0.0 3.0)))
          (Rendering|Material|SetScalarParameterValueonMaterials (Variables|Visual|GetWeaponMesh) "PinkExpoent" (Math|Curves|GetFloatValue (Variables|Default|GetPinkExpoentCurve) (Math|Float|Clamp(Float) (+ _countintensity _colorprogress) 0.0 3.0)))
          (Rendering|Material|SetScalarParameterValueonMaterials (Variables|Visual|GetWeaponMesh) "CausticsProgress" (Math|Curves|GetFloatValue (Variables|Default|GetCausticsProgress) (Math|Float|Clamp(Float) (+ _countintensity _colorprogress) 0.0 3.0)))
          (Rendering|Material|SetScalarParameterValueonMaterials (Variables|Visual|GetWeaponMesh) "EmissionIntensity" _emissionintensity)
          (CallFunction|DecayEvent)
          (bind _output_get (Variables|Default|SetFX_ColorStateBlend _colorprogress))
          (bind _asniagara_particle_system_component (Utilities|Casting|CastToNiagaraComponent (CallFunction|GetIdleVFXComponent)))
          (Niagara|SetNiagaraVariable(Float) _asniagara_particle_system_component "StateBlend" _output_get))
        (:Finished
          (Variables|Default|SetIsTimelineRunning false)
          (if (> (Variables|Default|GetCountIntensity) 2)
            (Variables|Default|SetCountIntensity 3.0)
            (else
              (Variables|Default|SetCountIntensity (+ (Variables|Default|GetCountIntensity) 1.0)))))))
    (else
      (Timeline|Timeline PlayFromStart))))

(event Custom|Decay_Event
  ;[0 0 0] Getting back to normal
  (bind _returnvalue (Utilities|Time|SetTimerbyEvent (delegate Custom|RDecay_Event) 0.05 true false 5.0)))

(event Custom|RDecay_Event
  (bind _colorproogresscurve (Variables|Default|GetColorProogressCurve))
  ;[0 0 0] Getting back to normal
  (if (> (Variables|Default|GetCountIntensity) 0.01)
    (bind _output_get (Variables|Default|SetCountIntensity (Math|Float|Max(Float) (- (Variables|Default|GetCountIntensity) 0.05))))
    (Rendering|Material|SetScalarParameterValueonMaterials (Variables|Visual|GetWeaponMesh) "ColorProgress" (Math|Curves|GetFloatValue _colorproogresscurve _output_get))
    (bind _output_get_2 (Variables|Default|SetFX_ColorStateBlend (Math|Curves|GetFloatValue _colorproogresscurve _output_get)))
    (bind _asniagara_particle_system_component (Utilities|Casting|CastToNiagaraComponent (CallFunction|GetIdleVFXComponent)))
    (Niagara|SetNiagaraVariable(Float) _asniagara_particle_system_component "StateBlend" _output_get_2)
    (else
      (Utilities|Time|PauseTimerbyHandle (stale Custom|Decay_Event _returnvalue)))))

(event EventSwingLeft
  ;[0.02734 0.03276 0.1094] FX - Set Trails color on state
  (|CallParentFunction)
  (for _array_element (Utilities|Array|MakeArray (CallFunction|GetAnimTrailVFXComponent) (CallFunction|GetSwingVFXComponent))
    (bind _asniagara_particle_system_component (Utilities|Casting|CastToNiagaraComponent _array_element))
    (Niagara|SetNiagaraVariable(Float) _asniagara_particle_system_component "StateBlend" (Variables|Default|GetFX_ColorStateBlend))))

(event EventSwingRight
  ;[0.02734 0.03276 0.1094] FX - Set Trails color on state
  (|CallParentFunction)
  (for _array_element (Utilities|Array|MakeArray (CallFunction|GetAnimTrailVFXComponent) (CallFunction|GetSwingVFXComponent))
    (bind _asniagara_particle_system_component (Utilities|Casting|CastToNiagaraComponent _array_element))
    (Niagara|SetNiagaraVariable(Float) _asniagara_particle_system_component "StateBlend" (Variables|Default|GetFX_ColorStateBlend))))

(event FortWeapon|EventOnPlayImpactFX
  (params (HitResult :type Hit%20Result%20Structure%20%28by%20ref%29) (ImpactPhysicalSurface :type EPhysicalSurface%20Enum :default "SurfaceType_Default") (SpawnedPSC :type FXSystem%20Component%20Object%20Reference))
  ;[0 0 0] Game
  (|CallParentFunction HitResult ImpactPhysicalSurface SpawnedPSC)
  (Utilities|FlowControl|Sequence
    (:then_0
      ;[0 0 0] FE
      (if (Reactive|Frontend|IsCosmeticObjectDisplayedinFrontendContext self)
        (CallFunction|UpdateColorProgressEvent)))
    (:then_1
      (CallFunction|RoyalDuskFXServer ImpactPhysicalSurface))))

(event Custom|RoyalDusk_FX_Server
  (params (Impact%20Physical%20Surface :type EPhysicalSurface%20Enum :default "SurfaceType_Default"))
  ;[0 0 0] Game
  (CallFunction|RoyalDuskFXClient Impact%20Physical%20Surface))

(event Custom|RoyalDusk_FX_Client
  (params (Impact%20Physical%20Surface :type EPhysicalSurface%20Enum :default "SurfaceType_Default"))
  ;[0 0 0] Game
  (bind _output_get (Variables|Default|SetCurrent%20Impact%20Physical%20Surface_0 Impact%20Physical%20Surface))
  (Variables|Default|SetIs%20hit%20weak%20point_0 (Utilities|Array|ContainsItem (Variables|Default|GetWeakPoint%20Physical%20Surface_0) _output_get))
  (if (Variables|Default|GetIs%20hit%20weak%20point_0)
    (CallFunction|UpdateColorProgressEvent)))

(fn RewindingEvent)

(fn ConstructionScript
  (|CallParentFunction))
