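; B_Pickaxe_Silencer is a melee pickaxe weapon Blueprint for the Silencer pickaxe cosmetic, stored among the FORT_Melee impact pickaxes.
; It adds no logic of its own: the construction script only calls the parent, and a RewindingEvent function is present but empty.
; Behavior comes from the parent class.
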(fn RewindingEvent)

(fn ConstructionScript
  (|CallParentFunction))
